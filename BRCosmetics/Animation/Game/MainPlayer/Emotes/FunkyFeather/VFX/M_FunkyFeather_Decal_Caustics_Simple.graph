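# Translucent unlit deferred decal for the FunkyFeather emote VFX, drawing two layers of animated, colored caustics.
# Both layers sample world-space XY textures panned by time and warped by shared scrolling Perlin noise, with the second adding rotation and offset.
# Smoothstepped patterns within a radial mask blend each layer's shade and highlight colors, which are summed with decal color tint.
# Opacity combines the stronger mask, decal alpha and a normal-versus-orientation fade on angled surfaces.
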Material /BRCosmetics/Animation/Game/MainPlayer/Emotes/FunkyFeather/VFX/M_FunkyFeather_Decal_Caustics_Simple
graph {
 properties {
  bUsedWithNiagaraMeshParticles=true
  blendMode=BLEND_Translucent
  materialDomain=MD_DeferredDecal
  pixelDepthOffsetMode=PDOM_Legacy
  shadingModel=MSM_Unlit
  usedShadingModels=Unlit
 }
 nodes [
  n1 = DecalColor()
  n2 = VectorParameter() {group=Emissive_Pattern1, parameterName=EmissiveColor_Caustic_Shade, sortPriority=11, defaultValue=<rgba 0.2859 0 1 1>}
  n3 = VectorParameter() {group=Emissive_Pattern1, parameterName=EmissiveColor_Caustic_Highlight, sortPriority=10, defaultValue=<rgba 1 0.276 0.9821 1>}
  n4 = WorldPosition()
  n5 = NamedRerouteUsage()
  n6 = Add(
   A=Divide(
    A=ComponentMask(None=@n4.XY) {r=true}
    B=ScalarParameter() {defaultValue=200, group=Mask_Pattern1, parameterName=WorldSpace_TileU_CausticPattern1, sortPriority=10}
   )
   B=Multiply(A=ScalarParameter() {defaultValue=0.1, group=Emissive, parameterName=Time_Multi_uvOffsetU_Cpattern1}, B=@n5)
  )
  n7 = Add(
   A=Divide(
    A=ComponentMask(None=@n4.XY) {g=true}
    B=ScalarParameter() {defaultValue=200, group=Mask_Pattern1, parameterName=WorldSpace_TileV_CausticPattern1, sortPriority=11}
   ) {constB=100}
   B=Multiply(A=@n5, B=ScalarParameter() {defaultValue=0.1, group=Emissive, parameterName=Time_Multi_uvOffsetV_Cpattern1})
  )
  n8 = WorldPosition()
  n9 = AppendVector(
   A=Divide(
    A=ComponentMask(None=@n8.XY) {r=true}
    B=ScalarParameter() {defaultValue=300, group=uvDistort, parameterName=WorldSpace_uvDistort_TileU, sortPriority=1}
   ) {constB=100}
   B=Divide(
    A=ComponentMask(None=@n8.XY) {g=true}
    B=ScalarParameter() {defaultValue=300, group=uvDistort, parameterName=WorldSpace_uvDistort_TileV, sortPriority=2}
   ) {constB=100}
  )
  n10 = Multiply(
   A=ScalarParameter() {defaultValue=0.25, group=uvDistort, parameterName=Time_Multi_uvDistort, sortPriority=2}
   B=NamedRerouteDeclaration(None=Time()) {name=reroute_Time, nodeColor=<rgba 1 0.7266 0 1>}
  )
  n11 = ComponentMask(None=TextureSampleParameter2D(UVs=Add(A=@n9, B=@n10)) {group=uvDistort, parameterName=tex_uvDistort_noise, texture=/Game/Effects/Fort_Effects/Textures/Tile/T_PerlinRGB_01_M}.RGB) {g=true, r=true}
  n12 = Multiply(
   A=NamedRerouteDeclaration(None=Subtract(A=@n11) {constB=0.5}) {name=DistortTexture, nodeColor=<rgba 0.2891 1 0 1>}
   B=ScalarParameter() {defaultValue=0.1, group=uvDistort, parameterName=uvDistort_Multi, sortPriority=3}
  )
  n13 = TextureSampleParameter2D(UVs=Add(A=AppendVector(A=@n6, B=@n7), B=@n12)) {
   group=Emissive
   parameterName=Tex_Emissive_Pattern1
   samplerType=SAMPLERTYPE_LinearGrayscale
   sortPriority=1
   texture=/Game/Effects/Fort_Effects/Textures/Tile/T_Caustic_Wind_Diagonal
  }
  n14 = Subtract(A=TextureCoordinate() {uTiling=2, vTiling=2})
  n15 = SmoothStep(
   Min=ScalarParameter() {group=Alpha, parameterName=MaskRadial_SmoothStep_Min, sortPriority=10}
   Max=ScalarParameter() {defaultValue=1, group=Alpha, parameterName=MaskRadial_SmoothStep_Max, sortPriority=11}
   Value=OneMinus(None=Length(None=@n14) {is=class MaterialExpressionLength})
  )
  n16 = Multiply(
   A=SmoothStep(
    Min=ScalarParameter() {group=Mask_Pattern1, parameterName=Mask_Pattern1_SmoothStep_min}
    Max=ScalarParameter() {defaultValue=1, group=Mask_Pattern1, parameterName=Mask_Pattern1_SmoothStep_max, sortPriority=33}
    Value=@n13.R
   )
   B=@n15
  )
  n17 = VectorParameter() {group=Emissive_Pattern2, parameterName=EmissiveCaustic_Pattern2_Shade, sortPriority=11, defaultValue=<rgba 0 0.1261 1 1>}
  n18 = VectorParameter() {group=Emissive_Pattern2, parameterName=EmissiveCaustic_Pattern2_Highlight, sortPriority=10, defaultValue=<rgba 0 0.8508 1 1>}
  n19 = WorldPosition()
  n20 = Divide(
   A=ComponentMask(None=@n19.XY) {r=true}
   B=ScalarParameter() {defaultValue=300, group=Mask_Pattern2, parameterName=WorldSpace_TileU_CausticPattern2, sortPriority=10}
  ) {constB=100}
  n21 = NamedRerouteUsage()
  n22 = Multiply(A=ScalarParameter() {defaultValue=0.1, group=Emissive, parameterName=Time_Multi_uvOffsetU_Cpattern2}, B=@n21)
  n23 = Divide(
   A=ComponentMask(None=@n19.XY) {g=true}
   B=ScalarParameter() {defaultValue=300, group=Mask_Pattern2, parameterName=WorldSpace_TileV_CausticPattern2, sortPriority=11}
  ) {constB=100}
  n24 = Multiply(A=@n21, B=ScalarParameter() {defaultValue=-0.05, group=Emissive, parameterName=Time_Multi_uvOffsetV_Cpattern2})
  n25 = MaterialFunctionCall!(
   Angle=ScalarParameter() {group=Mask_Pattern2, parameterName=Rotate_Pattern2, sortPriority=40}
   UV/2DVector=AppendVector(A=Add(A=@n20, B=@n22), B=Add(A=@n23, B=@n24))
  ) {materialFunction=/Game/Characters/Player/Male/Medium/Bodies/M_MED_Faux/Materials/MF_2DRotate}
  n26 = MaterialFunctionCall(
   R=ScalarParameter() {group=Mask_Pattern2, parameterName=OffsetX_Pattern2, sortPriority=41}
   G=ScalarParameter() {group=Mask_Pattern2, parameterName=OffsetY_Pattern2, sortPriority=42}
  ) {materialFunction=&AppendMany}
  n27 = TextureSampleParameter2D(UVs=Add(A=@n12, B=Add(A=@n25.Result, B=@n26.RG))) {
   group=Emissive
   parameterName=Tex_Emissive_Pattern2
   samplerType=SAMPLERTYPE_LinearGrayscale
   sortPriority=2
   texture=&T_Caustic_Wind_Diagonal
  }
  n28 = Multiply(
   A=SmoothStep(
    Min=ScalarParameter() {group=Mask_Pattern2, parameterName=Mask_Pattern2_SmoothStep_min}
    Max=ScalarParameter() {defaultValue=1, group=Mask_Pattern2, parameterName=Mask_Pattern2_SmoothStep_max, sortPriority=33}
    Value=@n27.R
   )
   B=@n15
  )
  Add(
   A=Multiply(A=@n1.RGB, B=LinearInterpolate(A=@n2.RGB, B=@n3.RGB, Alpha=@n16)) {constA=1}
   B=Multiply(A=@n1.RGB, B=LinearInterpolate(A=@n17.RGB, B=@n18.RGB, Alpha=@n28)) {constA=1}
  )
  n29 = Abs(None=SceneTexture() {sceneTextureId=PPI_WorldNormal}.Color)
  n30 = Abs(None=ObjectOrientation())
  Multiply(A=Multiply(A=DecalColor().A, B=Max(A=@n16, B=@n28)), B=Saturate(None=DotProduct(A=@n29, B=@n30)))
  ScalarParameter() {parameterName=metalic}
  ScalarParameter() {defaultValue=0.5, parameterName=Roughness}
  ScalarParameter() {defaultValue=0.5, parameterName=Specular}
 ]
}
