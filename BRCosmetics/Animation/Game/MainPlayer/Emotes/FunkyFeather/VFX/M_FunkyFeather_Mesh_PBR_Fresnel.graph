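# Default Lit PBR material for Niagara mesh particles in the FunkyFeather emote VFX, with OasisShard textures as defaults and a Fresnel glow.
# Emissive blends an inner color to a particle-color-tinted rim via saturated Fresnel, divided by eye adaptation to offset exposure.
# Base color is desaturatable texture times Dynamic Parameter RGB, or a constant.
# Static switches select constants or SpecRoughMetal channels for metal, specular and roughness; specular and roughness both read green.
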Material /BRCosmetics/Animation/Game/MainPlayer/Emotes/FunkyFeather/VFX/M_FunkyFeather_Mesh_PBR_Fresnel
graph {
 properties {bAutomaticallySetUsageInEditor=false, bUsedWithNiagaraMeshParticles=true, usedShadingModels=Default Lit}
 nodes [
  n1 = ParticleColor()
  n2 = VectorParameter() {parameterName=Fresnel_Emissive_Color, defaultValue=<rgba 0.9626 0 1 1>}
  n3 = Fresnel(
   ExponentIn=ScalarParameter() {defaultValue=1, group=Fresnel, parameterName=Fresnel_ExpontIn, sortPriority=2}
   BaseReflectFractionIn=ScalarParameter() {defaultValue=0.04, group=Fresnel, parameterName=Fresnel_BaseReflect, sortPriority=1}
  ) {exponent=1}
  Divide(
   A=LinearInterpolate(
    A=VectorParameter() {group=Fresnel, parameterName=Emissive_InnerColor, defaultValue=<rgba 0 0 0 1>}.RGB
    B=Multiply(A=@n1.RGB, B=@n2.RGB)
    Alpha=Saturate(None=@n3)
   )
   B=EyeAdaptation().EyeAdaptation
  )
  n4 = TextureSampleParameter2D() {parameterName=BaseColor, texture=&/T_FunkyFeather_OasisShard_D}
  n5 = ScalarParameter() {parameterName=BaseColorDesaturate}
  StaticSwitchParameter(
   True=VectorParameter() {group=PBR, parameterName=BaseColor, sortPriority=2, defaultValue=<rgba 0.6939 0.4452 0.1301 1>}.RGB
   False=Multiply(A=DynamicParameter() {paramNames [ColorR,ColorG,ColorB,Param4]}.RGB, B=Desaturation(None=@n4.RGB, Fraction=@n5))
  ) {group=PBR, parameterName=switch_BaseColor_Constant, sortPriority=1}
  n6 = TextureSampleParameter2D() {
   parameterName=SpecRoughMetal
   samplerType=SAMPLERTYPE_LinearColor
   texture=&/T_FunkyFeather_OasisShard_S
  }
  StaticSwitchParameter(True=ScalarParameter() {defaultValue=1, group=PBR, parameterName=Metalic, sortPriority=6}, False=@n6.B) {group=PBR, parameterName=switch_Metal_Constant, sortPriority=3}
  StaticSwitchParameter(True=ScalarParameter() {defaultValue=0.5, group=PBR, parameterName=Specular, sortPriority=7}, False=@n6.G) {group=PBR, parameterName=switch_Sepcular_Constant, sortPriority=4}
  StaticSwitchParameter(True=ScalarParameter() {defaultValue=0.3, group=PBR, parameterName=Roughness, sortPriority=8}, False=@n6.G) {group=PBR, parameterName=switch_Roughness_Constant, sortPriority=5}
  TextureSampleParameter2D() {
   group=PBR
   parameterName=Normal
   samplerType=SAMPLERTYPE_Normal
   sortPriority=10
   texture=&/T_FunkyFeather_OasisShard_N_02
  }
 ]
}
