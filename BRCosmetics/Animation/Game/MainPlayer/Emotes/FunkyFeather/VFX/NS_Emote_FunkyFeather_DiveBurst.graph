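# Niagara one-shot burst for the FunkyFeather emote's dive, built from a single mesh-particle emitter named Rocks001.
# An instantaneous burst spawns short-lived, blue-tinted meshes in a squashed torus-like shape, with randomized scale, speed and spin across two mesh renderers.
# Velocity radiates from an origin offset below, then randomized gravity, age-increasing drag and rotational forces arc and tumble the pieces.
# A size curve pops them in quickly and shrinks them to zero.
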NiagaraSystem /BRCosmetics/Animation/Game/MainPlayer/Emotes/FunkyFeather/VFX/NS_Emote_FunkyFeather_DiveBurst
dependencies {
 usedDataInterfaces [&NiagaraDataInterfaceMeshRendererInfo, &NiagaraDataInterfaceCurve]
 usedDynamicInputs []
 usedModules [&SystemState]
}
emitters {Rocks001 {
  emitterUpdateScript [
   &EmitterState()
   &SpawnBurst_Instantaneous(Spawn Count=100, Use Spawn Probability=false, UseLoopCountLimit=false)
  ]
  particleSpawnScript [
   &InitializeParticle(
    Color=<rgba 0.1845 0.227 0.7084 1>
    Lifetime=&RandomRangeFloat(Minimum=0, Maximum=1)
    Mesh Renderer Array Visibility Mode=::NewEnumerator3
    Mesh Renderer Info=DataInterface() {propertyValues {MeshRenderer=&self:Rocks001_0.NiagaraMeshRendererProperties_0}}
    Mesh Scale Mode=::NewEnumerator1
    Mesh Uniform Scale Max=0.75
    Mesh Uniform Scale Min=0.25
    Position Offset=&RandomRangeVector(Minimum=<0 0 0>, Maximum=<0 0 0>)
    UsePositionOffset=false
    Write Material Random=false
   ) {module=InitializeParticle001}
   &InitialMeshOrientation(Use Rotation Vector=true)
   &ShapeLocation(
    Handle Radius=10
    Large Radius=30
    MaskSpawnGroup=false
    Non Uniform Scale=<0.5 1 1>
    Override Seed=false
    Shape Primitive=::NewEnumerator3
    UseOwnerScale=true
   )
   &AddVelocity(
    Module.Use Falloff Distance=false
    Origin Offset=<0 0 -50>
    Origin Offset Coordinate Space=::World
    Override Seed=false
    Velocity Mode=::NewEnumerator1
    Velocity Speed=&RandomRangeFloat(Minimum=200, Maximum=450)
   )
   &AddRotationalVelocity(Rotation Rate=&RandomRangeVector(Minimum=<-10 -10 -10>, Maximum=<10 10 10>))
  ]
  particleUpdateScript [
   &ParticleState()
   &GravityForce(Gravity=&RandomRangeVector(Minimum=<0 0 -600>, Maximum=<0 0 -980>))
   &MeshRotationForce(Rotation=&RandomRangeVector(Minimum=<-90 -90 -90>, Maximum=<90 90 90>))
   &Drag(
    Drag=&Lerp_Float(A=0.25, B=2, Alpha=$Particles.NormalizedAge)
    Rotational Drag=0.05
    Use Linear Drag=true
    Use Rotational Drag=true
   )
   &SolveForcesAndVelocity(Clamp Velocity=false, Limit Acceleration=false, Manually Enable Rotational Solver=true)
   &ScaleMeshSize(Scale Factor=&VectorFromFloat(Value=&FloatFromCurve(FloatCurve=DataInterface() {propertyValues {Curve {keys [
          {arriveTangent=5.6831, interpMode=RCIM_Cubic, leaveTangent=5.6831, tangentMode=RCTM_User, time=0, value=0}
          {arriveTangent=-0.0035, interpMode=RCIM_Cubic, leaveTangent=-0.0035, tangentMode=RCTM_User, time=0.25, value=1}
          {arriveTangent=0.0043, interpMode=RCIM_Cubic, leaveTangent=0.0043, tangentMode=RCTM_User, time=0.5, value=0.9932}
          <at 1 0>
         ]}}})))
  ]
  rendererClasses [&NiagaraMeshRendererProperties]
  renderers [
   {rendererClass=&NiagaraMeshRendererProperties, rendererIndex=0}
   {rendererClass=&NiagaraMeshRendererProperties, rendererIndex=1}
  ]
 }}
summary {
 systemName=NS_Emote_FunkyFeather_DiveBurst
 emitters [{emitterName=Rocks001, rendererClasses [&NiagaraMeshRendererProperties]}]
 userVariables []
}
