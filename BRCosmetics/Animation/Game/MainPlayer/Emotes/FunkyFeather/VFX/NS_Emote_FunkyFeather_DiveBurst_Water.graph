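# Niagara water-splash burst for the dive in the FunkyFeather emote, throwing out a few orange Duck particles and many purple Rocks as meshes with sprites.
# Rocks and Rocks001 are near-identical bursts, the former on GPU.
# Height relative to the simulation position fakes a water surface: sinking particles gain heavy drag and a -X push.
# water001 adds an expanding two-mesh ripple, its alpha reduced by an IsPlayerIndoors parameter.
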NiagaraSystem /BRCosmetics/Animation/Game/MainPlayer/Emotes/FunkyFeather/VFX/NS_Emote_FunkyFeather_DiveBurst_Water
dependencies {
 usedDataInterfaces [
  &NiagaraDataInterfaceMeshRendererInfo
  &NiagaraDataInterfaceCurve
  &NiagaraDataInterfaceArrayColor
 ]
 usedDynamicInputs []
 usedModules [&SystemState]
}
emitters {
 Duck {
  emitterUpdateScript [
   &EmitterState(Life Cycle Mode=::NewEnumerator1, Loop Behavior=::NewEnumerator1, Loop Duration=2, UseLoopDelay=false)
   &SpawnBurst_Instantaneous(Spawn Count=5, Use Spawn Probability=false, UseLoopCountLimit=false)
  ]
  particleSpawnScript [
   &InitializeParticle(
    Color=<rgba 1 0.3638 0 1>
    Lifetime=&RandomRangeFloat(Minimum=1, Maximum=2)
    Mesh Scale Mode=::NewEnumerator3
    Position Offset=&RandomRangeVector(Minimum=<0 0 0>, Maximum=<0 0 0>)
    Sprite Size Mode=::NewEnumerator1
    Uniform Sprite Size Max=60
    Uniform Sprite Size Min=30
    UsePositionOffset=false
    Write Material Random=false
   ) {module=InitializeParticle001}
   &InitialMeshOrientation(Use Rotation Vector=true)
   &ShapeLocation(
    Handle Radius=10
    Large Radius=30
    MaskSpawnGroup=false
    Non Uniform Scale=<1 0.75 1>
    Override Seed=false
    Shape Primitive=::NewEnumerator3
    UseOwnerScale=true
   )
   &AddVelocity(
    Module.Use Falloff Distance=false
    Origin Offset=<0 0 -40>
    Origin Offset Coordinate Space=::World
    Override Seed=false
    Velocity Mode=::NewEnumerator1
    Velocity Speed=&RandomRangeFloat(Minimum=200, Maximum=350)
   )
   &AddRotationalVelocity(Rotation Rate=&RandomRangeVector(Minimum=<-10 -10 -10>, Maximum=<10 10 10>))
  ]
  particleUpdateScript [
   &ParticleState()
   &GravityForce(Gravity=&RandomRangeVector(Minimum=<0 0 -600>, Maximum=<0 0 -980>))
   &MeshRotationForce(Rotation=&RandomRangeVector(Minimum=<-90 -90 -90>, Maximum=<90 90 90>))
   &ScaleSpriteSize(
    Uniform Curve Index=&Add_Float(A=$Particles.NormalizedAge, B=&RandomRangeFloat(Minimum=0, Maximum=0.5))
    Uniform Curve Sprite Scale=DataInterface() {propertyValues {Curve {
       defaultValue=0
       keys [<at 0 0 RCIM_Constant>,<at 0.2 0.25 RCIM_Constant>,<at 0.4 0.5 RCIM_Constant>,<at 0.6 1 RCIM_Constant>,<at 0.75 0.5 RCIM_Constant>,<at 0.9 0 RCIM_Constant>]
      }}}
   ) {module=ScaleSpriteSize001}
   &Drag(
    Drag=&Lerp_Float(
     A=&RandomRangeFloat(
      Minimum=&RandomRangeFloat(Minimum=30, Maximum=40)
      Maximum=&RandomRangeFloat(Minimum=40, Maximum=60)
     )
     B=&Lerp_Float(A=0.3, B=2, Alpha=$Particles.NormalizedAge)
     Alpha=&RemapRange(
      Input Value To Remap=&MakeFloatFromVector(
       Vector=&SubtractVectorFromPosition(Position=$Particles.Position, Vector=/Niagara/DynamicInputs/Helpers/SimulationPosition())
       Channel=::NewEnumerator2
      )
      Input Min=-20
      Input Max=-1
      Output Min=0
      Output Max=1
     )
    )
    Rotational Drag:NiagaraFloat=&RemapRange(
     Input Value To Remap=&MakeFloatFromVector(
      Vector=&SubtractVectorFromPosition(Position=$Particles.Position, Vector=&SimulationPosition())
      Channel=::NewEnumerator2
     )
     Input Min=-20
     Input Max=-1
     Output Min=0.15
     Output Max=0.05
    )
    Use Linear Drag=true
    Use Rotational Drag=true
   ) {module=Drag002}
   &LinearForce(Force=&Lerp_Vector(
     A=<0 0 0>
     B=&RandomRangeVector(Minimum=<-500 0 0>, Maximum=<-250 1 0>)
     Alpha=&RemapRange(
      Input Value To Remap=&MakeFloatFromVector(
       Vector=&SubtractVectorFromPosition(Position=$Particles.Position, Vector=&SimulationPosition())
       Channel=::NewEnumerator2
      )
      Input Min=-20
      Input Max=-1
      Output Min=1
      Output Max=0
     )
    ))
   &SolveForcesAndVelocity(Clamp Velocity=false, Limit Acceleration=false, Manually Enable Rotational Solver=true)
   &ScaleMeshSize(Scale Factor=&VectorFromFloat(Value=&FloatFromCurve(FloatCurve=DataInterface() {propertyValues {Curve {keys [
          {arriveTangent=5.6831, interpMode=RCIM_Cubic, leaveTangent=5.6831, tangentMode=RCTM_User, time=0, value=0}
          {arriveTangent=-0.0035, interpMode=RCIM_Cubic, leaveTangent=-0.0035, tangentMode=RCTM_User, time=0.25, value=1}
          {arriveTangent=0.0043, interpMode=RCIM_Cubic, leaveTangent=0.0043, tangentMode=RCTM_User, time=0.5, value=0.9932}
          <at 1 0>
         ]}}})))
   /SetVariables(
    Particles.CameraOffset_Halo=20
    Particles.Color_Halo:LinearColor=&Multiply_LinearColorByFloat(
     Linear Color=<rgba 1 0.5433 0 1>
     Float=&FloatFromCurve(FloatCurve=DataInterface() {propertyValues {Curve {defaultValue=1, keys [<at 0 0>,<at 0.05 1>,<at 0.1 0.5>,<at 0.9 0.5>,<at 0.95 1>,<at 1 0>]}}})
    )
   ) {isSetParameters=true}
  ]
  rendererClasses [&NiagaraMeshRendererProperties, &NiagaraSpriteRendererProperties]
  renderers [
   {rendererClass=&NiagaraMeshRendererProperties, rendererIndex=0}
   {rendererClass=&NiagaraSpriteRendererProperties, rendererIndex=1}
  ]
 }
 Rocks {
  simTarget=GPUComputeSim
  emitterUpdateScript [
   &EmitterState(Life Cycle Mode=::NewEnumerator1, Loop Behavior=::NewEnumerator1, Loop Duration=2, UseLoopDelay=false)
   &SpawnBurst_Instantaneous(Spawn Count=150, Use Spawn Probability=false, UseLoopCountLimit=false)
  ]
  particleSpawnScript [
   &InitializeParticle(
    Color=&SelectLinearColorFromArray(
     Array Sampling Mode=::NewEnumerator1
     Direct Array Index=&RandomRangeInt(Minimum=0, Maximum=2)
     Color Selection Array=DataInterface() {propertyValues {ColorData [<rgba 0.7187 0 0.8406 1>, <rgba 0.6286 0.1769 1 1>, <rgba 0.3717 0.1052 1 1>]}}
     Random Seed=?
    )
    Lifetime=&RandomRangeFloat(Minimum=0.75, Maximum=1.75)
    Mesh Renderer Array Visibility Mode=::NewEnumerator3
    Mesh Renderer Info=DataInterface() {propertyValues {MeshRenderer=&self:Rocks_0.NiagaraMeshRendererProperties_2}}
    Mesh Scale Mode=::NewEnumerator1
    Mesh Uniform Scale Max=0.75
    Mesh Uniform Scale Min=0.25
    Position Offset=&RandomRangeVector(Minimum=<0 0 0>, Maximum=<0 0 0>)
    Sprite Size Mode=::NewEnumerator1
    Uniform Sprite Size Max=60
    Uniform Sprite Size Min=30
    UsePositionOffset=false
    Write Material Random=false
   ) {module=InitializeParticle001}
   &InitialMeshOrientation(Use Rotation Vector=true)
   &ShapeLocation(
    Handle Radius=10
    Large Radius=30
    MaskSpawnGroup=false
    Non Uniform Scale=<1 0.75 1>
    Override Seed=false
    Shape Primitive=::NewEnumerator3
    UseOwnerScale=true
   )
   &AddVelocity(
    Module.Use Falloff Distance=false
    Origin Offset=<0 0 -40>
    Origin Offset Coordinate Space=::World
    Override Seed=false
    Velocity Mode=::NewEnumerator1
    Velocity Speed=&RandomRangeFloat(Minimum=200, Maximum=350)
   )
   &AddRotationalVelocity(Rotation Rate=&RandomRangeVector(Minimum=<-10 -10 -10>, Maximum=<10 10 10>))
  ]
  particleUpdateScript [
   &ParticleState()
   &GravityForce(Gravity=&RandomRangeVector(Minimum=<0 0 -600>, Maximum=<0 0 -980>))
   &LinearForce(Force=&Lerp_Vector(
     A=<0 0 0>
     B=&RandomRangeVector(Minimum=<-2500 0 0>, Maximum=<-1500 1 0>)
     Alpha=&RemapRange(
      Input Value To Remap=&MakeFloatFromVector(
       Vector=&SubtractVectorFromPosition(Position=$Particles.Position, Vector=&SimulationPosition())
       Channel=::NewEnumerator2
      )
      Input Min=-20
      Input Max=-1
      Output Min=1
      Output Max=0
     )
    ))
   &ScaleSpriteSize(
    Uniform Curve Index=&Add_Float(A=$Particles.NormalizedAge, B=&RandomRangeFloat(Minimum=0, Maximum=1))
    Uniform Curve Sprite Scale=DataInterface() {propertyValues {Curve {
       defaultValue=0
       keys [<at 0 0 RCIM_Constant>,<at 0.1 0.25 RCIM_Constant>,<at 0.2 0.5 RCIM_Constant>,<at 0.3 0.75 RCIM_Constant>,<at 0.4 1 RCIM_Constant>,<at 0.5 0.5 RCIM_Constant>,<at 0.6 0.25 RCIM_Constant>,<at 0.7 0 RCIM_Constant>]
      }}}
   ) {module=ScaleSpriteSize001}
   &ScaleSpriteSize(
    Uniform Curve Index=&Add_Float(A=$Particles.NormalizedAge, B=&RandomRangeFloat(Minimum=0, Maximum=1))
    Uniform Curve Sprite Scale=DataInterface() {propertyValues {Curve {defaultValue=0, keys [<at 0.35 0>,<at 0.4 1>,<at 0.6 1>,<at 0.7 0>]}}}
   )
   &MeshRotationForce(Rotation=&RandomRangeVector(Minimum=<-90 -90 -90>, Maximum=<90 90 90>))
   &Drag(
    Drag=&Lerp_Float(
     A=&RandomRangeFloat(
      Minimum=&RandomRangeFloat(Minimum=30, Maximum=40)
      Maximum=&RandomRangeFloat(Minimum=40, Maximum=60)
     )
     B=&Lerp_Float(A=0.3, B=2, Alpha=$Particles.NormalizedAge)
     Alpha=&RemapRange(
      Input Value To Remap=&MakeFloatFromVector(
       Vector=&SubtractVectorFromPosition(Position=$Particles.Position, Vector=&SimulationPosition())
       Channel=::NewEnumerator2
      )
      Input Min=-20
      Input Max=-1
      Output Min=0
      Output Max=1
     )
    )
    Rotational Drag:NiagaraFloat=&RemapRange(
     Input Value To Remap=&MakeFloatFromVector(
      Vector=&SubtractVectorFromPosition(Position=$Particles.Position, Vector=&SimulationPosition())
      Channel=::NewEnumerator2
     )
     Input Min=-20
     Input Max=-1
     Output Min=0.15
     Output Max=0.05
    )
    Use Linear Drag=true
    Use Rotational Drag=true
   ) {module=Drag002}
   &SolveForcesAndVelocity(Clamp Velocity=false, Limit Acceleration=false, Manually Enable Rotational Solver=true)
   &ScaleMeshSize(Scale Factor=&VectorFromFloat(Value=&FloatFromCurve(FloatCurve=DataInterface() {propertyValues {Curve {keys [
          {arriveTangent=5.6831, interpMode=RCIM_Cubic, leaveTangent=5.6831, tangentMode=RCTM_User, time=0, value=0}
          {arriveTangent=-0.0035, interpMode=RCIM_Cubic, leaveTangent=-0.0035, tangentMode=RCTM_User, time=0.25, value=1}
          {arriveTangent=0.0043, interpMode=RCIM_Cubic, leaveTangent=0.0043, tangentMode=RCTM_User, time=0.5, value=0.9932}
          <at 1 0>
         ]}}})))
   &DynamicMaterialParameters(
    Index 0 Linear Color:LinearColor=&SelectLinearColorFromArray(
     Array Sampling Mode=::NewEnumerator1
     Direct Array Index=&RandomRangeInt(Minimum=0, Maximum=2)
     Color Selection Array=DataInterface() {propertyValues {ColorData [<rgba 0.147 0.305 0.4072 1>, <rgba 0.1488 0.147 0.4072 1>, <rgba 0.0671 0.3445 0.724 1>]}}
     Random Seed=?
    )
    Write Parameter Index 0=::NewEnumerator5
   )
   &ScaleColor(Scale RGB=<2 2 2>, ScaleA=false, ScaleRGB=true)
  ]
  rendererClasses [&NiagaraMeshRendererProperties, &NiagaraSpriteRendererProperties]
  renderers [
   {rendererClass=&NiagaraMeshRendererProperties, rendererIndex=0}
   {rendererClass=&NiagaraSpriteRendererProperties, rendererIndex=1}
  ]
 }
 Rocks001 {
  emitterUpdateScript [
   &EmitterState(Life Cycle Mode=::NewEnumerator1, Loop Behavior=::NewEnumerator1, Loop Duration=2, UseLoopDelay=false)
   &SpawnBurst_Instantaneous(Spawn Count=150, Use Spawn Probability=false, UseLoopCountLimit=false)
  ]
  particleSpawnScript [
   &InitializeParticle(
    Color=&SelectLinearColorFromArray(
     Array Sampling Mode=::NewEnumerator1
     Direct Array Index=&RandomRangeInt(Minimum=0, Maximum=2)
     Color Selection Array=DataInterface() {propertyValues {ColorData [<rgba 0.7187 0 0.8406 1>, <rgba 0.6286 0.1769 1 1>, <rgba 0.3717 0.1052 1 1>]}}
     Random Seed=?
    )
    Lifetime=&RandomRangeFloat(Minimum=0.75, Maximum=1.75)
    Mesh Renderer Array Visibility Mode=::NewEnumerator3
    Mesh Renderer Info=DataInterface() {propertyValues {MeshRenderer=&self:Rocks001_0.NiagaraMeshRendererProperties_2}}
    Mesh Scale Mode=::NewEnumerator1
    Mesh Uniform Scale Max=0.75
    Mesh Uniform Scale Min=0.25
    Position Offset=&RandomRangeVector(Minimum=<0 0 0>, Maximum=<0 0 0>)
    Sprite Size Mode=::NewEnumerator1
    Uniform Sprite Size Max=60
    Uniform Sprite Size Min=30
    UsePositionOffset=false
    Write Material Random=false
   ) {module=InitializeParticle001}
   &InitialMeshOrientation(Use Rotation Vector=true)
   &ShapeLocation(
    Handle Radius=10
    Large Radius=30
    MaskSpawnGroup=false
    Non Uniform Scale=<1 0.75 1>
    Override Seed=false
    Shape Primitive=::NewEnumerator3
    UseOwnerScale=true
   )
   &AddVelocity(
    Module.Use Falloff Distance=false
    Origin Offset=<0 0 -40>
    Origin Offset Coordinate Space=::World
    Override Seed=false
    Velocity Mode=::NewEnumerator1
    Velocity Speed=&RandomRangeFloat(Minimum=200, Maximum=350)
   )
   &AddRotationalVelocity(Rotation Rate=&RandomRangeVector(Minimum=<-10 -10 -10>, Maximum=<10 10 10>))
  ]
  particleUpdateScript [
   &ParticleState()
   &GravityForce(Gravity=&RandomRangeVector(Minimum=<0 0 -600>, Maximum=<0 0 -980>))
   &LinearForce(Force=&Lerp_Vector(
     A=<0 0 0>
     B=&RandomRangeVector(Minimum=<-2500 0 0>, Maximum=<-1500 1 0>)
     Alpha=&RemapRange(
      Input Value To Remap=&MakeFloatFromVector(
       Vector=&SubtractVectorFromPosition(Position=$Particles.Position, Vector=&SimulationPosition())
       Channel=::NewEnumerator2
      )
      Input Min=-20
      Input Max=-1
      Output Min=1
      Output Max=0
     )
    ))
   &ScaleSpriteSize(
    Uniform Curve Index=&Add_Float(A=$Particles.NormalizedAge, B=&RandomRangeFloat(Minimum=0, Maximum=1))
    Uniform Curve Sprite Scale=DataInterface() {propertyValues {Curve {
       defaultValue=0
       keys [<at 0 0 RCIM_Constant>,<at 0.1 0.25 RCIM_Constant>,<at 0.2 0.5 RCIM_Constant>,<at 0.3 0.75 RCIM_Constant>,<at 0.4 1 RCIM_Constant>,<at 0.5 0.5 RCIM_Constant>,<at 0.6 0.25 RCIM_Constant>,<at 0.7 0 RCIM_Constant>]
      }}}
   ) {module=ScaleSpriteSize001}
   &ScaleSpriteSize(
    Uniform Curve Index=&Add_Float(A=$Particles.NormalizedAge, B=&RandomRangeFloat(Minimum=0, Maximum=1))
    Uniform Curve Sprite Scale=DataInterface() {propertyValues {Curve {defaultValue=0, keys [<at 0.35 0>,<at 0.4 1>,<at 0.6 1>,<at 0.7 0>]}}}
   )
   &MeshRotationForce(Rotation=&RandomRangeVector(Minimum=<-90 -90 -90>, Maximum=<90 90 90>))
   &Drag(
    Drag=&Lerp_Float(
     A=&RandomRangeFloat(Minimum=30, Maximum=40)
     B=&Lerp_Float(A=0.3, B=2, Alpha=$Particles.NormalizedAge)
     Alpha=&RemapRange(
      Input Value To Remap=&MakeFloatFromVector(
       Vector=&SubtractVectorFromPosition(Position=$Particles.Position, Vector=&SimulationPosition())
       Channel=::NewEnumerator2
      )
      Input Min=-20
      Input Max=-1
      Output Min=0
      Output Max=1
     )
    )
    Rotational Drag:NiagaraFloat=&RemapRange(
     Input Value To Remap=&MakeFloatFromVector(
      Vector=&SubtractVectorFromPosition(Position=$Particles.Position, Vector=&SimulationPosition())
      Channel=::NewEnumerator2
     )
     Input Min=-20
     Input Max=-1
     Output Min=0.15
     Output Max=0.05
    )
    Use Linear Drag=true
    Use Rotational Drag=true
   ) {module=Drag002}
   &SolveForcesAndVelocity(Clamp Velocity=false, Limit Acceleration=false, Manually Enable Rotational Solver=true)
   &ScaleMeshSize(Scale Factor=&VectorFromFloat(Value=&FloatFromCurve(FloatCurve=DataInterface() {propertyValues {Curve {keys [
          {arriveTangent=5.6831, interpMode=RCIM_Cubic, leaveTangent=5.6831, tangentMode=RCTM_User, time=0, value=0}
          {arriveTangent=-0.0035, interpMode=RCIM_Cubic, leaveTangent=-0.0035, tangentMode=RCTM_User, time=0.25, value=1}
          {arriveTangent=0.0043, interpMode=RCIM_Cubic, leaveTangent=0.0043, tangentMode=RCTM_User, time=0.5, value=0.9932}
          <at 1 0>
         ]}}})))
   &DynamicMaterialParameters(
    Index 0 Linear Color:LinearColor=&SelectLinearColorFromArray(
     Array Sampling Mode=::NewEnumerator1
     Direct Array Index=&RandomRangeInt(Minimum=0, Maximum=2)
     Color Selection Array=DataInterface() {propertyValues {ColorData [<rgba 0.147 0.305 0.4072 1>, <rgba 0.1488 0.147 0.4072 1>, <rgba 0.0671 0.3445 0.724 1>]}}
     Random Seed=?
    )
    Write Parameter Index 0=::NewEnumerator5
   )
   &ScaleColor(Scale RGB=<2 2 2>, ScaleA=false, ScaleRGB=true)
  ]
  rendererClasses [&NiagaraMeshRendererProperties, &NiagaraSpriteRendererProperties]
  renderers [
   {rendererClass=&NiagaraMeshRendererProperties, rendererIndex=0}
   {rendererClass=&NiagaraSpriteRendererProperties, rendererIndex=1}
  ]
 }
 water001 {
  emitterUpdateScript [
   &EmitterState(Life Cycle Mode=::NewEnumerator1, Loop Behavior=::NewEnumerator1, Loop Duration=1, UseLoopDelay=false)
   &SpawnRate(SpawnRate=1.5, Use Spawn Probability=false) {enabled=false}
   &SpawnBurst_Instantaneous(Use Spawn Probability=false, UseLoopCountLimit=false)
  ]
  particleSpawnScript [&InitializeParticle(
    Color=<rgba 1 0 1 1>
    Lifetime=0.6
    Mesh Scale=<3.5 2 1>
    Mesh Scale Mode=::NewEnumerator4
    UsePositionOffset=false
    Write Material Random=true
   )]
  particleUpdateScript [
   &ParticleState()
   &ScaleMeshSize(Scale Factor=&Multiply_VectorByFloat(
     Vector=<1 1 1>
     Float=&FloatFromCurve(FloatCurve=DataInterface() {propertyValues {Curve {keys [
          {arriveTangent=2.8204, interpMode=RCIM_Cubic, leaveTangent=2.8204, tangentMode=RCTM_User, time=0, value=0}
          {arriveTangent=0.0509, interpMode=RCIM_Cubic, leaveTangent=0.0509, tangentMode=RCTM_User, time=1, value=1}
         ]}}})
    ))
   &ScaleColor(
    Scale Alpha=&FloatFromCurve(
     FloatCurve=DataInterface() {propertyValues {Curve {keys [
         {arriveTangent=5.0354, time=0, value=0}
         {arriveTangent=1.4286, leaveTangent=1.4286, time=0.25, value=1}
         {arriveTangent=-1.3333, leaveTangent=-1.3333, time=0.6, value=1}
         <at 1 0>
        ]}}}
     Scale Curve=&Lerp_Float(A=1, B=0.03, Alpha=$NPC.FortniteNPC.IsPlayerIndoors)
    )
    ScaleA=true
    ScaleRGB=false
   )
   &OffsetPosition(Position Offset=&AddVectorToPosition(
     Position=&Lerp_Vector(
      A=<0 0 -110>
      B=<0 0 0>
      Alpha=&ClampFloat(Float=&Multiply_Float(A=$Emitter.Age, B=1.25))
     )
     Vector=<0 0 0>
    )) {enabled=false, module=OffsetPosition001}
   &CameraOffset(Camera Offset Amount=-25)
   /SetVariables(
    Particles.Linear Color_BlueRipple:LinearColor=&MakeLinearColorFromVectorAndFloat(
     "Vector (RGB)"=&MakeVectorFromLinearColorRGB(Color=<rgba 0 0.3813 1 1>)
     "Float (Alpha)"=&FloatFromCurve(
      FloatCurve=DataInterface() {propertyValues {Curve {
         defaultValue=1
         keys [
          <at 0 0>
          {arriveTangent=0.1251, arriveTangentWeight=0.0333, leaveTangent=0.1251, time=0.1, value=1}
          {arriveTangent=-0.0966, arriveTangentWeight=0.2167, leaveTangent=-0.0966, time=0.75, value=1}
          <at 1 0>
         ]
        }}}
      Scale Curve=&Lerp_Float(A=1, B=0.03, Alpha=$NPC.FortniteNPC.IsPlayerIndoors)
     )
    )
    Particles.Scale_BlueRipple:Vector3f=&Multiply_VectorByFloat(Vector=$Particles.Scale, Float=1.05)
   ) {isSetParameters=true}
  ]
  rendererClasses [&NiagaraMeshRendererProperties]
  renderers [
   {rendererClass=&NiagaraMeshRendererProperties, rendererIndex=0}
   {rendererClass=&NiagaraMeshRendererProperties, rendererIndex=1}
  ]
 }
}
summary {
 systemName=NS_Emote_FunkyFeather_DiveBurst_Water
 emitters [
  {emitterName=Duck, rendererClasses [&NiagaraMeshRendererProperties, &NiagaraSpriteRendererProperties]}
  {
   emitterName=Rocks
   simTarget=GPUComputeSim
   rendererClasses [&NiagaraMeshRendererProperties, &NiagaraSpriteRendererProperties]
  }
  {emitterName=water001, rendererClasses [&NiagaraMeshRendererProperties]}
  {emitterName=Rocks001, rendererClasses [&NiagaraMeshRendererProperties, &NiagaraSpriteRendererProperties]}
 ]
 userVariables []
}
