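# Niagara system for the FunkyFeather GroundSwim emote that circles persistent mesh particles around tilted, flattened rings.
# Each burst-spawned emitter sets positions every update via ShapeLocation, using normalized particle index plus scaled emitter age to travel around the ring.
# Inner and outer ring emitters rise from below the origin as a clamped age lerp completes.
# A two-particle emitter adds small spinning meshes at side offsets, and another ring emitter is disabled.
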NiagaraSystem /BRCosmetics/Animation/Game/MainPlayer/Emotes/FunkyFeather/VFX/NS_Emote_FunkyFeather_GroundSwim
dependencies {
 usedDataInterfaces [&NiagaraDataInterfaceMeshRendererInfo, &NiagaraDataInterfaceArrayFloat3]
 usedDynamicInputs []
 usedModules [&SystemState]
}
emitters {
 InfiniteParticleLifetime {
  emitterUpdateScript [
   &EmitterState(
    Inactive Response=::NewEnumerator1
    Life Cycle Mode=::NewEnumerator1
    Loop Behavior=::NewEnumerator1
    Loop Duration Mode=::NewEnumerator1
    UseLoopDelay=false
   )
   &SpawnBurst_Instantaneous(Spawn Count=10, Use Spawn Probability=false, UseLoopCountLimit=false)
  ]
  particleSpawnScript [
   &InitializeParticle(
    Color Mode=::NewEnumerator0
    Mesh Scale Mode=::NewEnumerator3
    Mesh Uniform Scale=&RandomRangeFloat(Minimum=1.7, Maximum=2.2)
    Position Mode=::NewEnumerator0
    Write Material Random=true
   )
   &InitialMeshOrientation(
    Mesh Orientation Mode=::NewEnumerator1
    Rotation=&RandomRangeVector(Minimum=<0 0 0>, Maximum=<0 0 1>)
    Rotation Coordinate Space=::NewEnumerator2
    Use Rotation Vector=true
   )
  ]
  particleUpdateScript [
   &ParticleState(Kill Particles When Lifetime Has Elapsed=false, Loop Particles Lifetime=true)
   &ShapeLocation(
    MaskSpawnGroup=false
    Non Uniform Scale=<1 0.7 1>
    Offset=&AddVectorToPosition(
     Position=&Lerp_Vector(
      A=<0 0 -110>
      B=<0 0 0>
      Alpha=&ClampFloat(Float=&Multiply_Float(A=$Emitter.Age, B=1.25))
     )
     Vector=&RandomRangeVector(Minimum=<0 -15 0>, Maximum=<0 15 0>)
    )
    Offset Coordinate Space=::Local
    Offset Mode=::NewEnumerator0
    Override Seed=false
    Ring / Disc Distribution Mode:ENiagaraSphereDistributionMode=::NewEnumerator1
    Ring Radius=50
    Rotation Angle=-30
    Rotation Angle Type:ENiagara_AngleInput=::NewEnumerator0
    Rotation Axis=<0 1 0>
    Rotation Coordinate Space=::Simulation
    Rotation Mode=::NewEnumerator1
    Shape Primitive=::NewEnumerator4
    U Position=&Add_Float(
     A=&Multiply_Float(A=$Emitter.Age, B=-0.3)
     B=&ReturnNormalizedExecIndex(Normalized Index Includes One=false)
    )
    UseOwnerScale=true
   ) {module=ShapeLocation001}
  ]
  rendererClasses [&NiagaraMeshRendererProperties]
  renderers [
   {rendererClass=&NiagaraMeshRendererProperties, rendererIndex=0}
   {rendererClass=&NiagaraMeshRendererProperties, rendererIndex=1}
  ]
 }
 InfiniteParticleLifetime001 {
  emitterUpdateScript [
   &EmitterState(
    Inactive Response=::NewEnumerator1
    Life Cycle Mode=::NewEnumerator1
    Loop Behavior=::NewEnumerator1
    Loop Duration Mode=::NewEnumerator1
    UseLoopDelay=false
   )
   &SpawnBurst_Instantaneous(Spawn Count=2, Spawn Probability=0.5, Spawn Time=1.5, Use Spawn Probability=false, UseLoopCountLimit=false)
  ]
  particleSpawnScript [
   &InitializeParticle(
    Color Mode=::NewEnumerator0
    Mesh Scale Mode=::NewEnumerator3
    Mesh Uniform Scale=0.025
    Position Mode=::NewEnumerator0
    Write Material Random=true
   )
   &InitialMeshOrientation(
    Mesh Orientation Mode=::NewEnumerator1
    Rotation=&RandomRangeVector(Minimum=<0 0 0>, Maximum=<0 0 1>)
    Rotation Coordinate Space=::NewEnumerator2
    Use Rotation Vector=true
   )
   &MeshRotationForce(Rotation=&RandomRangeVector(Minimum=<0 0 -2.5>, Maximum=<0 0 2.5>))
   &ApplyInitialForces()
  ]
  particleUpdateScript [
   &ParticleState(Kill Particles When Lifetime Has Elapsed=false, Loop Particles Lifetime=true)
   &SolveForcesAndVelocity(Clamp Velocity=false, Limit Acceleration=false)
   &ShapeLocation(
    MaskSpawnGroup=false
    Non Uniform Scale=<1 0.5 1>
    Offset=&AddVectorToPosition(
     Position=&SelectVectorFromArray(
      Array Sampling Mode=::NewEnumerator0
      Direct Array Index=?
      Vector Selection Array=DataInterface() {propertyValues {FloatData [<0 -33 18>, <1 30 18>]}}
      Random Seed=0
     )
     Vector=&RandomRangeVector(Minimum=<0 -5 0>, Maximum=<1 5 0>)
    )
    Offset Coordinate Space=::Local
    Offset Mode=::NewEnumerator0
    Override Seed=false
    Ring / Disc Distribution Mode:ENiagaraSphereDistributionMode=::NewEnumerator1
    Ring Radius=105
    Rotation Angle=-30
    Rotation Angle Type:ENiagara_AngleInput=::NewEnumerator0
    Rotation Axis=<0 1 0>
    Rotation Coordinate Space=::Simulation
    Rotation Mode=::NewEnumerator1
    Shape Primitive=::NewEnumerator4
    U Position=&Add_Float(
     A=&Multiply_Float(A=$Emitter.Age, B=-0.3)
     B=&ReturnNormalizedExecIndex(Normalized Index Includes One=false)
    )
    UseOwnerScale=true
   ) {module=ShapeLocation001}
  ]
  rendererClasses [&NiagaraMeshRendererProperties]
  renderers [{rendererClass=&NiagaraMeshRendererProperties, rendererIndex=0}]
 }
 InfiniteParticleLifetime002 {
  bEnabled=false
  emitterUpdateScript [
   &EmitterState(
    Inactive Response=::NewEnumerator1
    Life Cycle Mode=::NewEnumerator1
    Loop Behavior=::NewEnumerator1
    Loop Duration Mode=::NewEnumerator1
    UseLoopDelay=false
   )
   &SpawnBurst_Instantaneous(Spawn Count=16, Use Spawn Probability=false, UseLoopCountLimit=false)
  ]
  particleSpawnScript [
   &InitializeParticle(
    Color Mode=::NewEnumerator0
    Mesh Scale Mode=::NewEnumerator1
    Mesh Uniform Scale Min=1.5
    Position Mode=::NewEnumerator0
    Write Material Random=true
   )
   &InitialMeshOrientation(
    Mesh Orientation Mode=::NewEnumerator1
    Rotation=&RandomRangeVector(Minimum=<0 0 0>, Maximum=<0 0 1>)
    Rotation Coordinate Space=::NewEnumerator2
    Use Rotation Vector=true
   )
  ]
  particleUpdateScript [
   &ParticleState(Kill Particles When Lifetime Has Elapsed=false, Loop Particles Lifetime=true)
   &ShapeLocation(
    MaskSpawnGroup=false
    Non Uniform Scale=<1 0.5 1>
    Offset=&AddVectorToPosition(
     Position=&Lerp_Vector(
      A=<0 0 -110>
      B=<0 0 0>
      Alpha=&ClampFloat(Float=&Multiply_Float(A=$Emitter.Age, B=1.25))
     )
     Vector=&RandomRangeVector(Minimum=<0 -10 0>, Maximum=<0 10 0>)
    )
    Offset Coordinate Space=::Local
    Offset Mode=::NewEnumerator0
    Override Seed=false
    Ring / Disc Distribution Mode:ENiagaraSphereDistributionMode=::NewEnumerator1
    Rotation Angle=-30
    Rotation Angle Type:ENiagara_AngleInput=::NewEnumerator0
    Rotation Axis=<0 1 0>
    Rotation Coordinate Space=::Simulation
    Rotation Mode=::NewEnumerator1
    Shape Primitive=::NewEnumerator4
    U Position=&Add_Float(
     A=&Multiply_Float(A=$Emitter.Age, B=-0.3)
     B=&ReturnNormalizedExecIndex(Normalized Index Includes One=false)
    )
    UseOwnerScale=true
   ) {module=ShapeLocation001}
  ]
  rendererClasses [&NiagaraMeshRendererProperties]
  renderers [
   {rendererClass=&NiagaraMeshRendererProperties, rendererIndex=0}
   {rendererClass=&NiagaraMeshRendererProperties, rendererIndex=1}
  ]
 }
 InfiniteParticleLifetime004 {
  emitterUpdateScript [
   &EmitterState(
    Inactive Response=::NewEnumerator1
    Life Cycle Mode=::NewEnumerator1
    Loop Behavior=::NewEnumerator1
    Loop Duration Mode=::NewEnumerator1
    UseLoopDelay=false
   )
   &SpawnBurst_Instantaneous(Spawn Count=14, Use Spawn Probability=false, UseLoopCountLimit=false)
  ]
  particleSpawnScript [
   &InitializeParticle(
    Color Mode=::NewEnumerator0
    Mesh Scale Mode=::NewEnumerator1
    Mesh Uniform Scale Max=1.75
    Mesh Uniform Scale Min=1.25
    Position Mode=::NewEnumerator0
    Write Material Random=true
   )
   &InitialMeshOrientation(
    Mesh Orientation Mode=::NewEnumerator1
    Rotation=&RandomRangeVector(Minimum=<0 0 0>, Maximum=<0 0 1>)
    Rotation Coordinate Space=::NewEnumerator2
    Use Rotation Vector=true
   )
  ]
  particleUpdateScript [
   &ParticleState(Kill Particles When Lifetime Has Elapsed=false, Loop Particles Lifetime=true)
   &ShapeLocation(
    MaskSpawnGroup=false
    Non Uniform Scale=<1 0.5 1>
    Offset=&AddVectorToPosition(
     Position=&Lerp_Vector(
      A=<0 0 -110>
      B=<0 0 0>
      Alpha=&ClampFloat(Float=&Multiply_Float(A=$Emitter.Age, B=1.25))
     )
     Vector=&RandomRangeVector(Minimum=<0 -15 0>, Maximum=<0 15 0>)
    )
    Offset Coordinate Space=::Local
    Offset Mode=::NewEnumerator0
    Override Seed=false
    Ring / Disc Distribution Mode:ENiagaraSphereDistributionMode=::NewEnumerator1
    Ring Radius=105
    Rotation Angle=-30
    Rotation Angle Type:ENiagara_AngleInput=::NewEnumerator0
    Rotation Axis=<0 1 0>
    Rotation Coordinate Space=::Simulation
    Rotation Mode=::NewEnumerator1
    Shape Primitive=::NewEnumerator4
    U Position=&Add_Float(
     A=&Multiply_Float(A=$Emitter.Age, B=-0.3)
     B=&ReturnNormalizedExecIndex(Normalized Index Includes One=false)
    )
    UseOwnerScale=true
   ) {module=ShapeLocation001}
  ]
  rendererClasses [&NiagaraMeshRendererProperties]
  renderers [
   {rendererClass=&NiagaraMeshRendererProperties, rendererIndex=0}
   {rendererClass=&NiagaraMeshRendererProperties, rendererIndex=1}
  ]
 }
}
summary {
 systemName=NS_Emote_FunkyFeather_GroundSwim
 emitters [
  {emitterName=InfiniteParticleLifetime004, rendererClasses [&NiagaraMeshRendererProperties]}
  {emitterName=InfiniteParticleLifetime001, rendererClasses [&NiagaraMeshRendererProperties]}
  {emitterName=InfiniteParticleLifetime, rendererClasses [&NiagaraMeshRendererProperties]}
  {bEnabled=false, emitterName=InfiniteParticleLifetime002, rendererClasses [&NiagaraMeshRendererProperties]}
 ]
 userVariables []
}
