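# Niagara system for the swim FunkyFeather emote that sheds short-lived, tumbling mesh shards from bones or sockets on the character's skeletal mesh.
# Two hand emitters place shards at spawn, with rates lerped by the nsRate_handL and nsRate_handR user parameters.
# Rocks003 spawns at a fixed rate and samples randomly between indices 2 and 3 during particle update; partitioned Rocks001 is disabled.
# All emitters use cone velocity, gravity, drag and a grow-then-shrink scale curve.
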NiagaraSystem /BRCosmetics/Animation/Game/MainPlayer/Emotes/FunkyFeather/VFX/NS_Emote_FunkyFeather_Swim_BodyAttach_Shards
dependencies {
 usedDataInterfaces [
  &NiagaraDataInterfaceMeshRendererInfo
  &NiagaraDataInterfaceCurve
  &NiagaraDataInterfaceSkeletalMesh
 ]
 usedDynamicInputs []
 usedModules [
  &self:SystemSpawnScript.SystemScriptSource.SystemScriptGraph.NiagaraNodeAssignment_0.SetVariables_#1
  &SystemState
 ]
}
emitters {
 Rocks {
  emitterUpdateScript [
   &EmitterState()
   &SpawnRate(Spawn Probability=0.75, SpawnRate=&Lerp_Float(A=0, B=100, Alpha=$User.nsRate_handR), Use Spawn Probability=true)
  ]
  particleSpawnScript [
   &InitializeParticle(
    Color=<rgba 0.1845 0.2259 0.7115 1>
    Lifetime=0.75
    Mesh Renderer Array Visibility Mode=::NewEnumerator3
    Mesh Renderer Info=DataInterface() {propertyValues {MeshRenderer=&self:Rocks_5.NiagaraMeshRendererProperties_0}}
    Mesh Scale Mode=::NewEnumerator1
    Mesh Uniform Scale Max=0.6
    Mesh Uniform Scale Min=0.3
    Position Mode=::NewEnumerator0
    Write Material Random=false
   ) {module=InitializeParticle001}
   &InitialMeshOrientation(Use Rotation Vector=false)
   &SkeletalMeshLocation(
    Bone / Socket Index=1
    Bone Sampling Mode=::NewEnumerator4
    MaskSpawnGroup=false
    Position Sampling=::NewEnumerator1
    Sampled Position Offset=&RandomRangeVector(Minimum=<0 0 -20>, Maximum=<0 0 0>)
    Skeletal Mesh=$System.Skeletal Mesh Reader
   ) {module=SkeletalMeshLocation001}
   &AddVelocity(
    Cone Axis=<0 0.5 1>
    Module.UseConeFalloff=false
    Module.UseVelDistribution=false
    Override Seed=false
    Velocity Mode=::NewEnumerator2
    Velocity Speed=&RandomRangeFloat(Minimum=75, Maximum=125)
   )
  ]
  particleUpdateScript [
   &ParticleState()
   &GravityForce(Gravity=&RandomRangeVector(Minimum=<0 0 -500>, Maximum=<0 0 -700>))
   &MeshRotationForce(
    Coordinate Space:ENiagaraMeshTransforms=::NewEnumerator3
    Rotation=&RandomRangeVector(Minimum=<-45 -45 -45>, Maximum=<45 45 45>)
   )
   &Drag(Drag=0.25, Rotational Drag=0, Use Linear Drag=true, Use Rotational Drag=false)
   &ScaleMeshSize(Scale Factor=&VectorFromFloat(Value=&FloatFromCurve(FloatCurve=DataInterface() {propertyValues {Curve {keys [
          {arriveTangent=5.6831, interpMode=RCIM_Cubic, leaveTangent=5.6831, tangentMode=RCTM_User, time=0, value=0}
          {arriveTangent=-0.0035, interpMode=RCIM_Cubic, leaveTangent=-0.0035, tangentMode=RCTM_User, time=0.25, value=1}
          {arriveTangent=0.0043, interpMode=RCIM_Cubic, leaveTangent=0.0043, tangentMode=RCTM_User, time=0.75, value=1}
          {arriveTangent=-9.8656, tangentMode=RCTM_User, time=1, value=0}
         ]}}})))
   &SolveForcesAndVelocity(Clamp Velocity=false, Limit Acceleration=false, Manually Enable Rotational Solver=true)
  ]
  rendererClasses [&NiagaraMeshRendererProperties]
  renderers [{rendererClass=&NiagaraMeshRendererProperties, rendererIndex=0}]
 }
 Rocks001 {
  bEnabled=false
  emitterUpdateScript [
   &EmitterState()
   &SpawnRate(Spawn Probability=0.75, SpawnRate=80, Use Spawn Probability=true)
  ]
  particleSpawnScript [
   &InitializeParticle(
    Color=<rgba 0.1845 0.2259 0.7115 1>
    Lifetime=0.75
    Mesh Renderer Array Visibility Mode=::NewEnumerator3
    Mesh Renderer Info=DataInterface() {propertyValues {MeshRenderer=&self:Rocks001_0.NiagaraMeshRendererProperties_0}}
    Mesh Scale Mode=::NewEnumerator1
    Mesh Uniform Scale Max=0.7
    Mesh Uniform Scale Min=0.35
    Position Offset=&RandomRangeVector(Minimum=<0 0 0>, Maximum=<0 0 0>)
    UsePositionOffset=false
    Write Material Random=false
   ) {module=InitializeParticle001}
   &InitialMeshOrientation(Use Rotation Vector=false)
   &PartitionParticles(Number Of Partitions=4)
   &AddVelocity(
    Cone Angle=60
    Cone Axis=<0 0 1>
    Module.UseConeFalloff=false
    Module.UseVelDistribution=false
    Override Seed=false
    Velocity Mode=::NewEnumerator2
    Velocity Speed=&RandomRangeFloat(Minimum=75, Maximum=125)
   )
  ]
  particleUpdateScript [
   &ParticleState()
   &SkeletalMeshLocation(
    Bone / Socket Index=$Particles.PartitionParticles.Partition
    Bone Sampling Mode=::NewEnumerator4
    MaskSpawnGroup=false
    Position Sampling=::NewEnumerator1
    Skeletal Mesh=$System.Skeletal Mesh Reader
   )
   &Drag(Drag=0.25, Use Linear Drag=true, Use Rotational Drag=false)
   &GravityForce(Gravity=&RandomRangeVector(Minimum=<0 0 -400>, Maximum=<0 0 -600>))
   &MeshRotationForce(Rotation=&RandomRangeVector(Minimum=<-45 -45 -45>, Maximum=<45 45 45>))
   &SolveForcesAndVelocity(Clamp Velocity=false, Limit Acceleration=false)
   &ScaleMeshSize(Scale Factor=&VectorFromFloat(Value=&FloatFromCurve(FloatCurve=DataInterface() {propertyValues {Curve {keys [
          {arriveTangent=5.6831, interpMode=RCIM_Cubic, leaveTangent=5.6831, tangentMode=RCTM_User, time=0, value=0}
          {arriveTangent=-0.0035, interpMode=RCIM_Cubic, leaveTangent=-0.0035, tangentMode=RCTM_User, time=0.25, value=1}
          {arriveTangent=0.0043, interpMode=RCIM_Cubic, leaveTangent=0.0043, tangentMode=RCTM_User, time=0.75, value=0.9932}
          {arriveTangent=-9.8656, tangentMode=RCTM_User, time=1, value=0}
         ]}}})))
  ]
  rendererClasses [&NiagaraMeshRendererProperties]
  renderers [{rendererClass=&NiagaraMeshRendererProperties, rendererIndex=0}]
 }
 Rocks002 {
  emitterUpdateScript [
   &EmitterState()
   &SpawnRate(Spawn Probability=0.75, SpawnRate=&Lerp_Float(A=0, B=100, Alpha=$User.nsRate_handL), Use Spawn Probability=true)
  ]
  particleSpawnScript [
   &InitializeParticle(
    Color=<rgba 0.1845 0.2259 0.7115 1>
    Lifetime=0.75
    Mesh Renderer Array Visibility Mode=::NewEnumerator3
    Mesh Renderer Info=DataInterface() {propertyValues {MeshRenderer=&self:Rocks002_1.NiagaraMeshRendererProperties_0}}
    Mesh Scale Mode=::NewEnumerator1
    Mesh Uniform Scale Max=0.6
    Mesh Uniform Scale Min=0.3
    Position Mode=::NewEnumerator0
    Write Material Random=false
   ) {module=InitializeParticle001}
   &InitialMeshOrientation(Use Rotation Vector=false)
   &SkeletalMeshLocation(
    Bone Sampling Mode=::NewEnumerator4
    MaskSpawnGroup=false
    Position Sampling=::NewEnumerator1
    Sampled Position Offset=&RandomRangeVector(Minimum=<0 0 0>, Maximum=<1 20 1>)
    Sampled Position Offset Coordinate Space:ENiagaraSkelMeshTransforms=::NewEnumerator2
    Skeletal Mesh=$System.Skeletal Mesh Reader
   )
   &AddVelocity(
    Cone Axis=<0 0.5 1>
    Module.UseConeFalloff=false
    Module.UseVelDistribution=false
    Override Seed=false
    Velocity Mode=::NewEnumerator2
    Velocity Speed=&RandomRangeFloat(Minimum=75, Maximum=125)
   )
  ]
  particleUpdateScript [
   &ParticleState()
   &Drag(Drag=0.25, Use Linear Drag=true, Use Rotational Drag=false)
   &GravityForce(Gravity=&RandomRangeVector(Minimum=<0 0 -500>, Maximum=<0 0 -700>))
   &MeshRotationForce(
    Coordinate Space:ENiagaraMeshTransforms=::NewEnumerator3
    Rotation=&RandomRangeVector(Minimum=<-45 -45 -45>, Maximum=<45 45 45>)
   )
   &ScaleMeshSize(Scale Factor=&VectorFromFloat(Value=&FloatFromCurve(FloatCurve=DataInterface() {propertyValues {Curve {keys [
          {arriveTangent=5.6831, interpMode=RCIM_Cubic, leaveTangent=5.6831, tangentMode=RCTM_User, time=0, value=0}
          {arriveTangent=-0.0035, interpMode=RCIM_Cubic, leaveTangent=-0.0035, tangentMode=RCTM_User, time=0.25, value=1}
          {arriveTangent=0.0043, interpMode=RCIM_Cubic, leaveTangent=0.0043, tangentMode=RCTM_User, time=0.75, value=1}
          {arriveTangent=-9.8656, tangentMode=RCTM_User, time=1, value=0}
         ]}}})))
   &SolveForcesAndVelocity(Clamp Velocity=false, Limit Acceleration=false)
  ]
  rendererClasses [&NiagaraMeshRendererProperties]
  renderers [{rendererClass=&NiagaraMeshRendererProperties, rendererIndex=0}]
 }
 Rocks003 {
  emitterUpdateScript [
   &EmitterState()
   &SpawnRate(Spawn Probability=0.75, SpawnRate=40, Use Spawn Probability=true)
  ]
  particleSpawnScript [
   &InitializeParticle(
    Color=<rgba 0.1845 0.2259 0.7115 1>
    Lifetime=0.75
    Mesh Renderer Array Visibility Mode=::NewEnumerator3
    Mesh Renderer Info=DataInterface() {propertyValues {MeshRenderer=&self:Rocks003_2.NiagaraMeshRendererProperties_0}}
    Mesh Scale Mode=::NewEnumerator1
    Mesh Uniform Scale Max=0.6
    Mesh Uniform Scale Min=0.3
    Position Offset=&RandomRangeVector(Minimum=<0 0 0>, Maximum=<0 0 0>)
    UsePositionOffset=false
    Write Material Random=false
   ) {module=InitializeParticle001}
   &InitialMeshOrientation(Use Rotation Vector=false)
   &AddVelocity(
    Cone Angle=60
    Cone Axis=<0 0 1>
    Module.UseConeFalloff=false
    Module.UseVelDistribution=false
    Override Seed=false
    Velocity Mode=::NewEnumerator2
    Velocity Speed=&RandomRangeFloat(Minimum=75, Maximum=125)
   )
  ]
  particleUpdateScript [
   &ParticleState()
   &SkeletalMeshLocation(
    Bone / Socket Index=&RandomRangeInt(Minimum=2, Maximum=3)
    Bone Sampling Mode=::NewEnumerator4
    MaskSpawnGroup=false
    Position Sampling=::NewEnumerator1
    Skeletal Mesh=$System.Skeletal Mesh Reader
   )
   &Drag(Drag=0.25, Use Linear Drag=true, Use Rotational Drag=false)
   &GravityForce(Gravity=&RandomRangeVector(Minimum=<0 0 -500>, Maximum=<0 0 -700>))
   &MeshRotationForce(Rotation=&RandomRangeVector(Minimum=<-45 -45 -45>, Maximum=<45 45 45>))
   &SolveForcesAndVelocity(Clamp Velocity=false, Limit Acceleration=false)
   &ScaleMeshSize(Scale Factor=&VectorFromFloat(Value=&FloatFromCurve(FloatCurve=DataInterface() {propertyValues {Curve {keys [
          {arriveTangent=5.6831, interpMode=RCIM_Cubic, leaveTangent=5.6831, tangentMode=RCTM_User, time=0, value=0}
          {arriveTangent=-0.0035, interpMode=RCIM_Cubic, leaveTangent=-0.0035, tangentMode=RCTM_User, time=0.25, value=1}
          {arriveTangent=0.0043, interpMode=RCIM_Cubic, leaveTangent=0.0043, tangentMode=RCTM_User, time=0.75, value=0.9932}
          {arriveTangent=-9.8656, tangentMode=RCTM_User, time=1, value=0}
         ]}}})))
  ]
  rendererClasses [&NiagaraMeshRendererProperties]
  renderers [{rendererClass=&NiagaraMeshRendererProperties, rendererIndex=0}]
 }
}
summary {
 systemName=NS_Emote_FunkyFeather_Swim_BodyAttach_Shards
 emitters [
  {emitterName=Rocks, rendererClasses [&NiagaraMeshRendererProperties]}
  {bEnabled=false, emitterName=Rocks001, rendererClasses [&NiagaraMeshRendererProperties]}
  {emitterName=Rocks002, rendererClasses [&NiagaraMeshRendererProperties]}
  {emitterName=Rocks003, rendererClasses [&NiagaraMeshRendererProperties]}
 ]
 userVariables [
  User.nsRate_handL {type=NiagaraFloat, defaultValue=1, description=""}
  User.nsRate_handR {type=NiagaraFloat, defaultValue=1, description=""}
 ]
}
