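# Niagara system for the FunkyFeather swim emote, emitting tinted, tumbling mesh shards and sprite flickers from hands and, per the asset name, feet.
# User floats nsRate_handL/R scale spawning for each hand, which has paired GPU and CPU emitters.
# Height relative to the simulation origin blends hand shards from heavy drag and curl noise low down to free movement higher up.
# A constant-rate CPU emitter resamples two other bones during update, adding jitter.
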NiagaraSystem /BRCosmetics/Animation/Game/MainPlayer/Emotes/FunkyFeather/VFX/NS_Emote_FunkyFeather_Swim_HandFeet_Shards
dependencies {
 usedDataInterfaces [
  &NiagaraDataInterfaceArrayColor
  &NiagaraDataInterfaceMeshRendererInfo
  &NiagaraDataInterfaceCurve
  &NiagaraDataInterfaceSkeletalMesh
 ]
 usedDynamicInputs []
 usedModules [
  &self:SystemSpawnScript.SystemScriptSource.SystemScriptGraph.NiagaraNodeAssignment_0.SetVariables_#1
  &SystemState
  &BlendSpaceHelper
 ]
}
emitters {
 Rocks {
  simTarget=GPUComputeSim
  emitterUpdateScript [
   &EmitterState()
   &SpawnRate(Spawn Probability=0.75, SpawnRate=&Lerp_Float(A=0, B=250, Alpha=$User.nsRate_handR), Use Spawn Probability=true)
  ]
  particleSpawnScript [
   &InitializeParticle(
    Color=&SelectLinearColorFromArray(
     Array Sampling Mode=::NewEnumerator1
     Direct Array Index=&RandomRangeInt(Minimum=0, Maximum=2)
     Color Selection Array=DataInterface() {propertyValues {ColorData [<rgba 0.7187 0 0.8406 1>, <rgba 0.6286 0.1769 1 1>, <rgba 0.3717 0.1052 1 1>]}}
     Random Seed=?
    )
    Lifetime=&RandomRangeFloat(Minimum=0.75, Maximum=1)
    Mesh Renderer Array Visibility Mode=::NewEnumerator3
    Mesh Renderer Info=DataInterface() {propertyValues {MeshRenderer=&self:Rocks_5.NiagaraMeshRendererProperties_2}}
    Mesh Scale Mode=::NewEnumerator1
    Mesh Uniform Scale Max=0.6
    Mesh Uniform Scale Min=0.3
    Position Mode=::NewEnumerator0
    Sprite Size Mode=::NewEnumerator1
    Uniform Sprite Size Max=60
    Uniform Sprite Size Min=30
    Write Material Random=false
   ) {module=InitializeParticle001}
   &InitialMeshOrientation(Use Rotation Vector=false)
   &SkeletalMeshLocation(
    Bone / Socket Index=1
    Bone Sampling Mode=::NewEnumerator4
    MaskSpawnGroup=false
    Position Sampling=::NewEnumerator1
    Sampled Position Offset=&RandomRangeVector(Minimum=<-4 -4 -10>, Maximum=<4 4 10>)
    Skeletal Mesh=$System.Skeletal Mesh Reader
   ) {module=SkeletalMeshLocation001}
   &AddVelocity(
    Cone Axis=<0 0.5 1>
    Module.UseConeFalloff=false
    Module.UseVelDistribution=false
    Override Seed=false
    Velocity Mode=::NewEnumerator2
    Velocity Speed=&RandomRangeFloat(Minimum=75, Maximum=125)
   )
   &InheritVelocity(Inherited Velocity Amount Scale=<0.5 0.5 0.5>, Inherited Velocity Speed Limit=200, Speed Limit=true, UseSpeedThreshold=false)
  ]
  particleUpdateScript [
   &ParticleState()
   &GravityForce(Gravity=&RandomRangeVector(Minimum=<0 0 -500>, Maximum=<0 0 -700>))
   &ScaleSpriteSize(
    Uniform Curve Index=&Add_Float(A=$Particles.NormalizedAge, B=&RandomRangeFloat(Minimum=0, Maximum=1.25))
    Uniform Curve Sprite Scale=DataInterface() {propertyValues {Curve {
       defaultValue=0
       keys [<at 0 0 RCIM_Constant>,<at 0.1 0.25 RCIM_Constant>,<at 0.2 0.5 RCIM_Constant>,<at 0.3 1 RCIM_Constant>,<at 0.4 0.75 RCIM_Constant>,<at 0.5 0.25 RCIM_Constant>,<at 0.6 0 RCIM_Constant>,<at 1 0 RCIM_Constant>]
      }}}
   ) {module=ScaleSpriteSize001}
   &MeshRotationForce(
    Coordinate Space:ENiagaraMeshTransforms=::NewEnumerator3
    Rotation=&RandomRangeVector(Minimum=<-45 -45 -45>, Maximum=<45 45 45>)
   )
   &Drag(
    Drag=&Lerp_Float(
     A=40
     B=0.25
     Alpha=&RemapRange(
      Input Value To Remap=&MakeFloatFromVector(
       Vector=&SubtractVectorFromPosition(Position=$Particles.Position, Vector=/Niagara/DynamicInputs/Helpers/SimulationPosition())
       Channel=::NewEnumerator2
      )
      Input Min=50
      Input Max=85
      Output Min=0
      Output Max=1
     )
    )
    Rotational Drag=0.2
    Use Linear Drag=true
    Use Rotational Drag=true
   ) {module=Drag002}
   &ScaleMeshSize(Scale Factor=&VectorFromFloat(Value=&FloatFromCurve(FloatCurve=DataInterface() {propertyValues {Curve {keys [
          {arriveTangent=5.6831, interpMode=RCIM_Cubic, leaveTangent=5.6831, tangentMode=RCTM_User, time=0, value=0}
          {arriveTangent=-0.0035, interpMode=RCIM_Cubic, leaveTangent=-0.0035, tangentMode=RCTM_User, time=0.25, value=1}
          {arriveTangent=0.0043, interpMode=RCIM_Cubic, leaveTangent=0.0043, tangentMode=RCTM_User, time=0.75, value=1}
          {arriveTangent=-9.8656, tangentMode=RCTM_User, time=1, value=0}
         ]}}})))
   &CurlNoiseForce(
    Bias Noise=false
    Noise Frequency=10
    Noise Strength=&Lerp_Float(
     A=2000
     B=0
     Alpha=&RemapRange(
      Input Value To Remap=&MakeFloatFromVector(
       Vector=&SubtractVectorFromPosition(Position=$Particles.Position, Vector=&SimulationPosition())
       Channel=::NewEnumerator2
      )
      Input Min=50
      Input Max=85
      Output Min=0
      Output Max=1
     )
    )
    Pan Noise=false
   )
   &InheritSourceMovement(Applied Position Scale=<0.5 0.5 0.5>, ApplyForce=false, ApplyPosition=true, ApplyPositionFalloff=false)
   &SolveForcesAndVelocity(Clamp Velocity=false, Limit Acceleration=false, Manually Enable Rotational Solver=true)
   &DynamicMaterialParameters(
    Index 0 Linear Color:LinearColor=&SelectLinearColorFromArray(
     Array Sampling Mode=::NewEnumerator1
     Direct Array Index=&RandomRangeInt(Minimum=0, Maximum=2)
     Color Selection Array=DataInterface() {propertyValues {ColorData [<rgba 0.147 0.305 0.4072 1>, <rgba 0.1488 0.147 0.4072 1>, <rgba 0.0671 0.3445 0.724 1>]}}
     Random Seed=?
    )
    Write Parameter Index 0=::NewEnumerator5
   )
   &KillParticlesInVolume(Box Size:Vector3f=<300 300 3000>, Invert Volume=true, Kill Shape=::NewEnumerator1, Origin Offset:Vector3f=<0 0 100>) {module=KillParticlesInVolume001}
   &ScaleColor(Scale RGB=<2 2 2>, ScaleA=false, ScaleRGB=true)
  ]
  rendererClasses [&NiagaraMeshRendererProperties, &NiagaraSpriteRendererProperties]
  renderers [
   {rendererClass=&NiagaraMeshRendererProperties, rendererIndex=0}
   {rendererClass=&NiagaraSpriteRendererProperties, rendererIndex=1}
  ]
 }
 Rocks001 {
  emitterUpdateScript [
   &EmitterState()
   &SpawnRate(Spawn Probability=0.75, SpawnRate=&Lerp_Float(A=0, B=250, Alpha=$User.nsRate_handR), Use Spawn Probability=true)
  ]
  particleSpawnScript [
   &InitializeParticle(
    Color=&SelectLinearColorFromArray(
     Array Sampling Mode=::NewEnumerator1
     Direct Array Index=&RandomRangeInt(Minimum=0, Maximum=2)
     Color Selection Array=DataInterface() {propertyValues {ColorData [<rgba 0.7187 0 0.8406 1>, <rgba 0.6286 0.1769 1 1>, <rgba 0.3717 0.1052 1 1>]}}
     Random Seed=?
    )
    Lifetime=&RandomRangeFloat(Minimum=0.75, Maximum=1)
    Mesh Renderer Array Visibility Mode=::NewEnumerator3
    Mesh Renderer Info=DataInterface() {propertyValues {MeshRenderer=&self:Rocks001_0.NiagaraMeshRendererProperties_2}}
    Mesh Scale Mode=::NewEnumerator1
    Mesh Uniform Scale Max=0.6
    Mesh Uniform Scale Min=0.3
    Position Mode=::NewEnumerator0
    Sprite Size Mode=::NewEnumerator1
    Uniform Sprite Size Max=60
    Uniform Sprite Size Min=30
    Write Material Random=false
   ) {module=InitializeParticle001}
   &InitialMeshOrientation(Use Rotation Vector=false)
   &SkeletalMeshLocation(
    Bone / Socket Index=1
    Bone Sampling Mode=::NewEnumerator4
    MaskSpawnGroup=false
    Position Sampling=::NewEnumerator1
    Sampled Position Offset=&RandomRangeVector(Minimum=<-4 -4 -10>, Maximum=<4 4 10>)
    Skeletal Mesh=$System.Skeletal Mesh Reader
   ) {module=SkeletalMeshLocation001}
   &AddVelocity(
    Cone Axis=<0 0.5 1>
    Module.UseConeFalloff=false
    Module.UseVelDistribution=false
    Override Seed=false
    Velocity Mode=::NewEnumerator2
    Velocity Speed=&RandomRangeFloat(Minimum=75, Maximum=125)
   )
   &InheritVelocity(Inherited Velocity Amount Scale=<0.5 0.5 0.5>, Inherited Velocity Speed Limit=200, Speed Limit=true, UseSpeedThreshold=false)
  ]
  particleUpdateScript [
   &ParticleState()
   &GravityForce(Gravity=&RandomRangeVector(Minimum=<0 0 -500>, Maximum=<0 0 -700>))
   &ScaleSpriteSize(
    Uniform Curve Index=&Add_Float(A=$Particles.NormalizedAge, B=&RandomRangeFloat(Minimum=0, Maximum=1.25))
    Uniform Curve Sprite Scale=DataInterface() {propertyValues {Curve {
       defaultValue=0
       keys [<at 0 0 RCIM_Constant>,<at 0.1 0.25 RCIM_Constant>,<at 0.2 0.5 RCIM_Constant>,<at 0.3 1 RCIM_Constant>,<at 0.4 0.75 RCIM_Constant>,<at 0.5 0.25 RCIM_Constant>,<at 0.6 0 RCIM_Constant>,<at 1 0 RCIM_Constant>]
      }}}
   ) {module=ScaleSpriteSize001}
   &MeshRotationForce(
    Coordinate Space:ENiagaraMeshTransforms=::NewEnumerator3
    Rotation=&RandomRangeVector(Minimum=<-45 -45 -45>, Maximum=<45 45 45>)
   )
   &Drag(
    Drag=&Lerp_Float(
     A=40
     B=0.25
     Alpha=&RemapRange(
      Input Value To Remap=&MakeFloatFromVector(
       Vector=&SubtractVectorFromPosition(Position=$Particles.Position, Vector=&SimulationPosition())
       Channel=::NewEnumerator2
      )
      Input Min=50
      Input Max=85
      Output Min=0
      Output Max=1
     )
    )
    Rotational Drag=0.2
    Use Linear Drag=true
    Use Rotational Drag=true
   ) {module=Drag002}
   &ScaleMeshSize(Scale Factor=&VectorFromFloat(Value=&FloatFromCurve(FloatCurve=DataInterface() {propertyValues {Curve {keys [
          {arriveTangent=5.6831, interpMode=RCIM_Cubic, leaveTangent=5.6831, tangentMode=RCTM_User, time=0, value=0}
          {arriveTangent=-0.0035, interpMode=RCIM_Cubic, leaveTangent=-0.0035, tangentMode=RCTM_User, time=0.25, value=1}
          {arriveTangent=0.0043, interpMode=RCIM_Cubic, leaveTangent=0.0043, tangentMode=RCTM_User, time=0.75, value=1}
          {arriveTangent=-9.8656, tangentMode=RCTM_User, time=1, value=0}
         ]}}})))
   &CurlNoiseForce(
    Bias Noise=false
    Noise Frequency=10
    Noise Strength=&Lerp_Float(
     A=2000
     B=0
     Alpha=&RemapRange(
      Input Value To Remap=&MakeFloatFromVector(
       Vector=&SubtractVectorFromPosition(Position=$Particles.Position, Vector=&SimulationPosition())
       Channel=::NewEnumerator2
      )
      Input Min=50
      Input Max=85
      Output Min=0
      Output Max=1
     )
    )
    Pan Noise=false
   )
   &InheritSourceMovement(Applied Position Scale=<0.5 0.5 0.5>, ApplyForce=false, ApplyPosition=true, ApplyPositionFalloff=false)
   &SolveForcesAndVelocity(Clamp Velocity=false, Limit Acceleration=false, Manually Enable Rotational Solver=true)
   &DynamicMaterialParameters(
    Index 0 Linear Color:LinearColor=&SelectLinearColorFromArray(
     Array Sampling Mode=::NewEnumerator1
     Direct Array Index=&RandomRangeInt(Minimum=0, Maximum=2)
     Color Selection Array=DataInterface() {propertyValues {ColorData [<rgba 0.147 0.305 0.4072 1>, <rgba 0.1488 0.147 0.4072 1>, <rgba 0.0671 0.3445 0.724 1>]}}
     Random Seed=?
    )
    Write Parameter Index 0=::NewEnumerator5
   )
   &KillParticlesInVolume(Box Size:Vector3f=<300 300 3000>, Invert Volume=true, Kill Shape=::NewEnumerator1, Origin Offset:Vector3f=<0 0 100>) {module=KillParticlesInVolume001}
   &ScaleColor(Scale RGB=<2 2 2>, ScaleA=false, ScaleRGB=true)
  ]
  rendererClasses [&NiagaraMeshRendererProperties, &NiagaraSpriteRendererProperties]
  renderers [
   {rendererClass=&NiagaraMeshRendererProperties, rendererIndex=0}
   {rendererClass=&NiagaraSpriteRendererProperties, rendererIndex=1}
  ]
 }
 Rocks002 {
  simTarget=GPUComputeSim
  emitterUpdateScript [
   &EmitterState()
   &SpawnRate(Spawn Probability=0.75, SpawnRate=&Lerp_Float(A=0, B=250, Alpha=$User.nsRate_handL), Use Spawn Probability=true)
  ]
  particleSpawnScript [
   &InitializeParticle(
    Color=&SelectLinearColorFromArray(
     Array Sampling Mode=::NewEnumerator1
     Direct Array Index=&RandomRangeInt(Minimum=0, Maximum=2)
     Color Selection Array=DataInterface() {propertyValues {ColorData [<rgba 0.7187 0 0.8406 1>, <rgba 0.6286 0.1769 1 1>, <rgba 0.3717 0.1052 1 1>]}}
     Random Seed=?
    )
    Lifetime=&RandomRangeFloat(Minimum=0.75, Maximum=1)
    Mesh Renderer Array Visibility Mode=::NewEnumerator3
    Mesh Renderer Info=DataInterface() {propertyValues {MeshRenderer=&self:Rocks002_1.NiagaraMeshRendererProperties_2}}
    Mesh Scale Mode=::NewEnumerator1
    Mesh Uniform Scale Max=0.6
    Mesh Uniform Scale Min=0.3
    Position Mode=::NewEnumerator0
    Sprite Size Mode=::NewEnumerator1
    Uniform Sprite Size Max=60
    Uniform Sprite Size Min=30
    Write Material Random=false
   ) {module=InitializeParticle001}
   &InitialMeshOrientation(Use Rotation Vector=false)
   &SkeletalMeshLocation(
    Bone Sampling Mode=::NewEnumerator4
    MaskSpawnGroup=false
    Position Sampling=::NewEnumerator1
    Sampled Position Offset=&RandomRangeVector(Minimum=<-4 -4 -10>, Maximum=<4 4 10>)
    Skeletal Mesh=$System.Skeletal Mesh Reader
   )
   &AddVelocity(
    Cone Axis=<0 0.5 1>
    Module.UseConeFalloff=false
    Module.UseVelDistribution=false
    Override Seed=false
    Velocity Mode=::NewEnumerator2
    Velocity Speed=&RandomRangeFloat(Minimum=75, Maximum=125)
   )
   &InheritVelocity(Inherited Velocity Amount Scale=<0.5 0.5 0.5>, Inherited Velocity Speed Limit=200, Speed Limit=true, UseSpeedThreshold=false)
  ]
  particleUpdateScript [
   &ParticleState()
   &Drag(
    Drag=&Lerp_Float(
     A=40
     B=0.25
     Alpha=&RemapRange(
      Input Value To Remap=&MakeFloatFromVector(
       Vector=&SubtractVectorFromPosition(Position=$Particles.Position, Vector=&SimulationPosition())
       Channel=::NewEnumerator2
      )
      Input Min=50
      Input Max=85
      Output Min=0
      Output Max=1
     )
    )
    Rotational Drag=0.2
    Use Linear Drag=true
    Use Rotational Drag=true
   )
   &ScaleSpriteSize(
    Uniform Curve Index=&Add_Float(A=$Particles.NormalizedAge, B=&RandomRangeFloat(Minimum=0, Maximum=1.25))
    Uniform Curve Sprite Scale=DataInterface() {propertyValues {Curve {
       defaultValue=0
       keys [<at 0 0 RCIM_Constant>,<at 0.1 0.25 RCIM_Constant>,<at 0.2 0.5 RCIM_Constant>,<at 0.3 1 RCIM_Constant>,<at 0.4 0.75 RCIM_Constant>,<at 0.5 0.25 RCIM_Constant>,<at 0.6 0 RCIM_Constant>]
      }}}
   ) {module=ScaleSpriteSize001}
   &GravityForce(Gravity=&RandomRangeVector(Minimum=<0 0 -500>, Maximum=<0 0 -700>))
   &MeshRotationForce(
    Coordinate Space:ENiagaraMeshTransforms=::NewEnumerator3
    Rotation=&RandomRangeVector(Minimum=<-45 -45 -45>, Maximum=<45 45 45>)
   )
   &ScaleMeshSize(Scale Factor=&VectorFromFloat(Value=&FloatFromCurve(FloatCurve=DataInterface() {propertyValues {Curve {keys [
          {arriveTangent=5.6831, interpMode=RCIM_Cubic, leaveTangent=5.6831, tangentMode=RCTM_User, time=0, value=0}
          {arriveTangent=-0.0035, interpMode=RCIM_Cubic, leaveTangent=-0.0035, tangentMode=RCTM_User, time=0.25, value=1}
          {arriveTangent=0.0043, interpMode=RCIM_Cubic, leaveTangent=0.0043, tangentMode=RCTM_User, time=0.75, value=1}
          {arriveTangent=-9.8656, tangentMode=RCTM_User, time=1, value=0}
         ]}}})))
   &CurlNoiseForce(
    Bias Noise=false
    Noise Frequency=10
    Noise Strength=&Lerp_Float(
     A=2000
     B=0
     Alpha=&RemapRange(
      Input Value To Remap=&MakeFloatFromVector(
       Vector=&SubtractVectorFromPosition(Position=$Particles.Position, Vector=&SimulationPosition())
       Channel=::NewEnumerator2
      )
      Input Min=50
      Input Max=85
      Output Min=0
      Output Max=1
     )
    )
    Pan Noise=false
   )
   &InheritSourceMovement(Applied Position Scale=<0.5 0.5 0.5>, ApplyForce=false, ApplyPosition=true, ApplyPositionFalloff=false)
   &SolveForcesAndVelocity(Clamp Velocity=false, Limit Acceleration=false)
   &DynamicMaterialParameters(
    Index 0 Linear Color:LinearColor=&SelectLinearColorFromArray(
     Array Sampling Mode=::NewEnumerator1
     Direct Array Index=&RandomRangeInt(Minimum=0, Maximum=2)
     Color Selection Array=DataInterface() {propertyValues {ColorData [<rgba 0.147 0.305 0.4072 1>, <rgba 0.1488 0.147 0.4072 1>, <rgba 0.0671 0.3445 0.724 1>]}}
     Random Seed=?
    )
    Write Parameter Index 0=::NewEnumerator5
   )
   &KillParticlesInVolume(Box Size:Vector3f=<300 300 3000>, Invert Volume=true, Kill Shape=::NewEnumerator1, Origin Offset:Vector3f=<0 0 100>)
  ]
  rendererClasses [&NiagaraMeshRendererProperties, &NiagaraSpriteRendererProperties]
  renderers [
   {rendererClass=&NiagaraMeshRendererProperties, rendererIndex=0}
   {rendererClass=&NiagaraSpriteRendererProperties, rendererIndex=1}
  ]
 }
 Rocks003 {
  emitterUpdateScript [
   &EmitterState()
   &SpawnRate(Spawn Probability=0.75, SpawnRate=100, Use Spawn Probability=true)
  ]
  particleSpawnScript [
   &InitializeParticle(
    Color=&SelectLinearColorFromArray(
     Array Sampling Mode=::NewEnumerator1
     Direct Array Index=&RandomRangeInt(Minimum=0, Maximum=2)
     Color Selection Array=DataInterface() {propertyValues {ColorData [<rgba 0.7187 0 0.8406 1>, <rgba 0.6286 0.1769 1 1>, <rgba 0.3717 0.1052 1 1>]}}
     Random Seed=?
    )
    Lifetime=0.6
    Mesh Renderer Array Visibility Mode=::NewEnumerator3
    Mesh Renderer Info=DataInterface() {propertyValues {MeshRenderer=&self:Rocks003_2.NiagaraMeshRendererProperties_2}}
    Mesh Scale Mode=::NewEnumerator1
    Mesh Uniform Scale Max=0.6
    Mesh Uniform Scale Min=0.3
    Position Offset=&RandomRangeVector(Minimum=<-25 -25 -25>, Maximum=<0.25 0.25 0.25>)
    Sprite Size Mode=::NewEnumerator1
    Uniform Sprite Size Max=60
    Uniform Sprite Size Min=30
    UsePositionOffset=true
    Write Material Random=false
   ) {module=InitializeParticle001}
   &InitialMeshOrientation(Use Rotation Vector=false)
   &AddVelocity(
    Cone Angle=60
    Cone Axis=<0 0 1>
    Module.UseConeFalloff=false
    Module.UseVelDistribution=false
    Override Seed=false
    Velocity Mode=::NewEnumerator2
    Velocity Speed=&RandomRangeFloat(Minimum=75, Maximum=125)
   )
   &InheritVelocity(Inherited Velocity Amount Scale=<-0.5 -0.5 -0.5>, Inherited Velocity Speed Limit=200, Speed Limit=true, UseSpeedThreshold=false)
  ]
  particleUpdateScript [
   &ParticleState()
   &SkeletalMeshLocation(
    Bone / Socket Index=&RandomRangeInt(Minimum=2, Maximum=3)
    Bone Sampling Mode=::NewEnumerator4
    MaskSpawnGroup=false
    Position Sampling=::NewEnumerator1
    Skeletal Mesh=$System.Skeletal Mesh Reader
   )
   &GravityForce(Gravity=&RandomRangeVector(Minimum=<0 0 -500>, Maximum=<0 0 -700>))
   &ScaleSpriteSize(
    Uniform Curve Index=&Add_Float(A=$Particles.NormalizedAge, B=&RandomRangeFloat(Minimum=0, Maximum=10))
    Uniform Curve Sprite Scale=DataInterface() {propertyValues {Curve {defaultValue=0, keys [<at 0.5 0>,<at 0.55 1>,<at 0.6 1>,<at 0.65 0>]}}}
   ) {module=ScaleSpriteSize001}
   &MeshRotationForce(Rotation=&RandomRangeVector(Minimum=<-45 -45 -45>, Maximum=<45 45 45>))
   &Drag(
    Drag=&Lerp_Float(
     A=40
     B=0
     Alpha=&RemapRange(
      Input Value To Remap=&MakeFloatFromVector(
       Vector=&SubtractVectorFromPosition(Position=$Particles.Position, Vector=&SimulationPosition())
       Channel=::NewEnumerator2
      )
      Input Min=50
      Input Max=85
      Output Min=0
      Output Max=1
     )
    )
    Rotational Drag=0.2
    Use Linear Drag=true
    Use Rotational Drag=true
   ) {module=Drag002}
   &SolveForcesAndVelocity(Clamp Velocity=false, Limit Acceleration=false)
   &ScaleMeshSize(Scale Factor=&VectorFromFloat(Value=&FloatFromCurve(FloatCurve=DataInterface() {propertyValues {Curve {keys [
          {arriveTangent=5.6831, time=0, value=0}
          {arriveTangent=-0.0035, leaveTangent=-0.0035, time=0.25, value=1}
          {arriveTangent=0.0043, leaveTangent=0.0043, time=0.6, value=1}
          {arriveTangent=-9.8656, time=1, value=0}
         ]}}})))
   &DynamicMaterialParameters(
    Index 0 Linear Color:LinearColor=&SelectLinearColorFromArray(
     Array Sampling Mode=::NewEnumerator1
     Direct Array Index=&RandomRangeInt(Minimum=0, Maximum=2)
     Color Selection Array=DataInterface() {propertyValues {ColorData [<rgba 0.147 0.305 0.4072 1>, <rgba 0.1488 0.147 0.4072 1>, <rgba 0.0671 0.3445 0.724 1>]}}
     Random Seed=?
    )
    Write Parameter Index 0=::NewEnumerator5
   )
   &KillParticlesInVolume(Kill Shape=::NewEnumerator3, Origin Offset:Vector3f=<0 0 -40>)
   /SetVariables(Particles.Position=&Lerp_Vector(
     A=Linked() {linkedVariable {name=Particles.Position, type=NiagaraPosition}}
     B=&AddVectorToPosition(Position=$Particles.Position, Vector=&RandomRangeVector(Minimum=<-3 -3 -3>, Maximum=<3 3 0>))
     Alpha=$Particles.NormalizedAge
    )) {isSetParameters=true}
   &BlendSpace("Position Blend (World <-> Local)":NiagaraFloat=&Lerp_Float(A=0, B=0.05, Alpha=$Particles.NormalizedAge, Translation Contribution:NiagaraFloat=1, Rotation Contribution:NiagaraFloat=1))
   &ScaleColor(Scale RGB=<2 2 2>, ScaleA=false, ScaleRGB=true)
  ]
  rendererClasses [&NiagaraMeshRendererProperties, &NiagaraSpriteRendererProperties]
  renderers [
   {rendererClass=&NiagaraMeshRendererProperties, rendererIndex=0}
   {rendererClass=&NiagaraSpriteRendererProperties, rendererIndex=1}
  ]
 }
 Rocks004 {
  emitterUpdateScript [
   &EmitterState()
   &SpawnRate(Spawn Probability=0.75, SpawnRate=&Lerp_Float(A=0, B=250, Alpha=$User.nsRate_handL), Use Spawn Probability=true)
  ]
  particleSpawnScript [
   &InitializeParticle(
    Color=&SelectLinearColorFromArray(
     Array Sampling Mode=::NewEnumerator1
     Direct Array Index=&RandomRangeInt(Minimum=0, Maximum=2)
     Color Selection Array=DataInterface() {propertyValues {ColorData [<rgba 0.7187 0 0.8406 1>, <rgba 0.6286 0.1769 1 1>, <rgba 0.3717 0.1052 1 1>]}}
     Random Seed=?
    )
    Lifetime=&RandomRangeFloat(Minimum=0.75, Maximum=1)
    Mesh Renderer Array Visibility Mode=::NewEnumerator3
    Mesh Renderer Info=DataInterface() {propertyValues {MeshRenderer=&self:Rocks004_0.NiagaraMeshRendererProperties_2}}
    Mesh Scale Mode=::NewEnumerator1
    Mesh Uniform Scale Max=0.6
    Mesh Uniform Scale Min=0.3
    Position Mode=::NewEnumerator0
    Sprite Size Mode=::NewEnumerator1
    Uniform Sprite Size Max=60
    Uniform Sprite Size Min=30
    Write Material Random=false
   ) {module=InitializeParticle001}
   &InitialMeshOrientation(Use Rotation Vector=false)
   &SkeletalMeshLocation(
    Bone Sampling Mode=::NewEnumerator4
    MaskSpawnGroup=false
    Position Sampling=::NewEnumerator1
    Sampled Position Offset=&RandomRangeVector(Minimum=<-4 -4 -10>, Maximum=<4 4 10>)
    Skeletal Mesh=$System.Skeletal Mesh Reader
   )
   &AddVelocity(
    Cone Axis=<0 0.5 1>
    Module.UseConeFalloff=false
    Module.UseVelDistribution=false
    Override Seed=false
    Velocity Mode=::NewEnumerator2
    Velocity Speed=&RandomRangeFloat(Minimum=75, Maximum=125)
   )
   &InheritVelocity(Inherited Velocity Amount Scale=<0.5 0.5 0.5>, Inherited Velocity Speed Limit=200, Speed Limit=true, UseSpeedThreshold=false)
  ]
  particleUpdateScript [
   &ParticleState()
   &Drag(
    Drag=&Lerp_Float(
     A=40
     B=0.25
     Alpha=&RemapRange(
      Input Value To Remap=&MakeFloatFromVector(
       Vector=&SubtractVectorFromPosition(Position=$Particles.Position, Vector=&SimulationPosition())
       Channel=::NewEnumerator2
      )
      Input Min=50
      Input Max=85
      Output Min=0
      Output Max=1
     )
    )
    Rotational Drag=0.2
    Use Linear Drag=true
    Use Rotational Drag=true
   ) {module=Drag002}
   &ScaleSpriteSize(
    Uniform Curve Index=&Add_Float(A=$Particles.NormalizedAge, B=&RandomRangeFloat(Minimum=0, Maximum=1.25))
    Uniform Curve Sprite Scale=DataInterface() {propertyValues {Curve {
       defaultValue=0
       keys [<at 0 0 RCIM_Constant>,<at 0.1 0.25 RCIM_Constant>,<at 0.2 0.5 RCIM_Constant>,<at 0.3 1 RCIM_Constant>,<at 0.4 0.75 RCIM_Constant>,<at 0.5 0.25 RCIM_Constant>,<at 0.6 0 RCIM_Constant>]
      }}}
   ) {module=ScaleSpriteSize001}
   &GravityForce(Gravity=&RandomRangeVector(Minimum=<0 0 -500>, Maximum=<0 0 -700>))
   &MeshRotationForce(
    Coordinate Space:ENiagaraMeshTransforms=::NewEnumerator3
    Rotation=&RandomRangeVector(Minimum=<-45 -45 -45>, Maximum=<45 45 45>)
   )
   &ScaleMeshSize(Scale Factor=&VectorFromFloat(Value=&FloatFromCurve(FloatCurve=DataInterface() {propertyValues {Curve {keys [
          {arriveTangent=5.6831, interpMode=RCIM_Cubic, leaveTangent=5.6831, tangentMode=RCTM_User, time=0, value=0}
          {arriveTangent=-0.0035, interpMode=RCIM_Cubic, leaveTangent=-0.0035, tangentMode=RCTM_User, time=0.25, value=1}
          {arriveTangent=0.0043, interpMode=RCIM_Cubic, leaveTangent=0.0043, tangentMode=RCTM_User, time=0.75, value=1}
          {arriveTangent=-9.8656, tangentMode=RCTM_User, time=1, value=0}
         ]}}})))
   &CurlNoiseForce(
    Bias Noise=false
    Noise Frequency=10
    Noise Strength=&Lerp_Float(
     A=2000
     B=0
     Alpha=&RemapRange(
      Input Value To Remap=&MakeFloatFromVector(
       Vector=&SubtractVectorFromPosition(Position=$Particles.Position, Vector=&SimulationPosition())
       Channel=::NewEnumerator2
      )
      Input Min=50
      Input Max=85
      Output Min=0
      Output Max=1
     )
    )
    Pan Noise=false
   )
   &InheritSourceMovement(Applied Position Scale=<0.5 0.5 0.5>, ApplyForce=false, ApplyPosition=true, ApplyPositionFalloff=false)
   &SolveForcesAndVelocity(Clamp Velocity=false, Limit Acceleration=false)
   &DynamicMaterialParameters(
    Index 0 Linear Color:LinearColor=&SelectLinearColorFromArray(
     Array Sampling Mode=::NewEnumerator1
     Direct Array Index=&RandomRangeInt(Minimum=0, Maximum=2)
     Color Selection Array=DataInterface() {propertyValues {ColorData [<rgba 0.147 0.305 0.4072 1>, <rgba 0.1488 0.147 0.4072 1>, <rgba 0.0671 0.3445 0.724 1>]}}
     Random Seed=?
    )
    Write Parameter Index 0=::NewEnumerator5
   )
   &KillParticlesInVolume(Box Size:Vector3f=<300 300 3000>, Invert Volume=true, Kill Shape=::NewEnumerator1, Origin Offset:Vector3f=<0 0 100>)
   &ScaleColor(Scale RGB=<2 2 2>, ScaleA=false, ScaleRGB=true)
  ]
  rendererClasses [&NiagaraMeshRendererProperties, &NiagaraSpriteRendererProperties]
  renderers [
   {rendererClass=&NiagaraMeshRendererProperties, rendererIndex=0}
   {rendererClass=&NiagaraSpriteRendererProperties, rendererIndex=1}
  ]
 }
}
summary {
 systemName=NS_Emote_FunkyFeather_Swim_HandFeet_Shards
 emitters [
  {
   emitterName=Rocks
   simTarget=GPUComputeSim
   rendererClasses [&NiagaraMeshRendererProperties, &NiagaraSpriteRendererProperties]
  }
  {
   emitterName=Rocks002
   simTarget=GPUComputeSim
   rendererClasses [&NiagaraMeshRendererProperties, &NiagaraSpriteRendererProperties]
  }
  {emitterName=Rocks003, rendererClasses [&NiagaraMeshRendererProperties, &NiagaraSpriteRendererProperties]}
  {emitterName=Rocks001, rendererClasses [&NiagaraMeshRendererProperties, &NiagaraSpriteRendererProperties]}
  {emitterName=Rocks004, rendererClasses [&NiagaraMeshRendererProperties, &NiagaraSpriteRendererProperties]}
 ]
 userVariables [
  User.nsRate_handL {type=NiagaraFloat, defaultValue=1, description=""}
  User.nsRate_handR {type=NiagaraFloat, defaultValue=1, description=""}
 ]
}
