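# Niagara system for the FunkyFeather emote's energy-water VFX, combining mesh, sprite, decal and light renderers with audio player modules.
# Emitters include a burst-spawned caustics decal with animated material parameters, plus puddle, trail, rock and light elements.
# Its modules and inputs include wind and vortex forces, per-unit spawning, volume kills, blend-space helpers, camera-relative math and time-of-day scalars.
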NiagaraSystem /BRCosmetics/Animation/Game/MainPlayer/Emotes/FunkyFeather/VFX/NS_Emote_FunkyFeather_energyWater
dependencies {
 usedDataInterfaces [
  {refPath=&NiagaraDataInterfaceAudioPlayer}
  {refPath=&NiagaraDataInterfaceArrayPosition}
  {refPath=&NiagaraDataInterfaceMeshRendererInfo}
  {refPath=&NiagaraDataInterfaceArrayFloat3}
  {refPath=&NiagaraDataInterfaceCurve}
  {refPath=&NiagaraDataInterfaceArrayColor}
  {refPath=&NiagaraDataInterfaceCamera}
  {refPath=&NiagaraDataInterfaceVectorCurve}
  {refPath=&NiagaraDataInterfaceArrayFloat}
 ]
 usedDynamicInputs [
  {refPath=/Niagara/DynamicInputs/Arrays/SelectPositionFromArray}
  {refPath=&SpawnGroup}
  {refPath=&RandomRangeVector}
  {refPath=&SelectVectorFromArray}
  {refPath=&RandomRangeFloat}
  {refPath=&Multiply_VectorByFloat}
  {refPath=&FloatFromCurve}
  {refPath=&RemapRange}
  {refPath=&DistanceBetweenPositions}
  {refPath=&AddVectorToPosition}
  {refPath=&Lerp_Vector}
  {refPath=&Sine}
  {refPath=&Add_Float}
  {refPath=&Multiply_LinearColorByFloat}
  {refPath=&MakeCustomFloatFromBool}
  {refPath=&SetBoolByFloatComparison}
  {refPath=&VectorLength}
  {refPath=&SelectLinearColorFromArray}
  {refPath=&Multiply_Float}
  {refPath=&FN_Select_Float_from_Time_Of_Day}
  {refPath=&ClampFloat}
  {refPath=&Abs_Float}
  {refPath=&DotProduct}
  {refPath=&ReturnMeshOrientationAxis}
  {refPath=&ReturnCameraVector}
  {refPath=&Lerp_Float}
  {refPath=&VectorFromCurve}
  {refPath=&MakeLinearColorFromVectorAndFloat}
  {refPath=&MakeVectorFromLinearColorRGB}
  {refPath=&LerpLinearColors}
  {refPath=&SelectFloatFromArray}
  {refPath=&RandomRangeInt}
  {refPath=&MaskVectorBySpawnGroup}
  {refPath=&VectorFromFloat}
  {refPath=/Niagara/DynamicInputs/Helpers/SimulationPosition}
  {refPath=&FNTimeOfDayScalar}
  {refPath=&RandomRangeLinearColor}
 ]
 usedModules [
  {refPath=&InitializeAudioPlayer}
  {refPath=&EmitterState}
  {refPath=&SpawnRate}
  {refPath=&InitializeParticle}
  {refPath=&InitialMeshOrientation}
  {refPath=&MeshRotationForce}
  {refPath=&ShapeLocation}
  {refPath=&AddVelocity}
  {refPath=&ApplyInitialForces}
  {refPath=&ParticleState}
  {refPath=&ScaleMeshSize}
  {refPath=&ScaleSpriteSize}
  {refPath=&WindForce}
  {refPath=&VortexForce}
  {refPath=&Drag}
  {refPath=&OffsetPosition}
  {refPath=&self:NiagaraEmitter_0.NiagaraScriptSource_0.NiagaraGraph_0.NiagaraNodeAssignment_0.SetVariables_#1}
  {refPath=&SolveForcesAndVelocity}
  {refPath=/Niagara/Modules/Audio/PlayPersistentAudio}
  {refPath=&BlendSpace}
  {refPath=&KillParticlesInVolume}
  {refPath=&ScaleColor}
  {refPath=&SpawnPerUnit}
  {refPath=&CameraOffset}
  {refPath=&self:PuddleRadial_Trail_0.NiagaraScriptSource_0.NiagaraGraph_0.NiagaraNodeAssignment_0.SetVariables_#2}
  {refPath=&self:PuddleRadial_Trail_0.NiagaraScriptSource_0.NiagaraGraph_0.NiagaraNodeAssignment_2.SetVariables_#3}
  {refPath=&self:Rocks_0.NiagaraScriptSource_0.NiagaraGraph_0.NiagaraNodeAssignment_0.SetVariables_#4}
  {refPath=&DynamicMaterialParameters}
  {refPath=&self:Rocks001_0.NiagaraScriptSource_0.NiagaraGraph_0.NiagaraNodeAssignment_0.SetVariables_#4}
  {refPath=&SpawnBurst_Instantaneous}
  {refPath=&self:Decal_Caustics_0.NiagaraScriptSource_0.NiagaraGraph_0.NiagaraNodeAssignment_1.SetVariables_#5}
  {refPath=&Color}
  {refPath=&self:Light_0.NiagaraScriptSource_0.NiagaraGraph_0.NiagaraNodeAssignment_1.SetVariables_#6}
  {refPath=&self:PuddleRadial001_1.NiagaraScriptSource_0.NiagaraGraph_0.NiagaraNodeAssignment_0.SetVariables_#7}
  {refPath=&self:PuddleRadial001_1.NiagaraScriptSource_0.NiagaraGraph_0.NiagaraNodeAssignment_1.SetVariables_#8}
  {refPath=&SystemState}
  {refPath=&BlendSpaceHelper}
 ]
 usedRenderers [
  {refPath=&NiagaraMeshRendererProperties}
  {refPath=&NiagaraSpriteRendererProperties}
  {refPath=&NiagaraDecalRendererProperties}
  {refPath=&NiagaraLightRendererProperties}
 ]
}
emitters {
 Decal_Caustics {
  emitterSpawnScript {scriptName=EmitterSpawnScript, modules []}
  emitterUpdateScript {
   scriptName=EmitterUpdateScript
   modules [
    {
     bIsSetParametersModule=false
     enabled=true
     moduleName=EmitterState
     inputs [
      Life Cycle Mode {
       bIsDynamic=false
       bIsEditable=true
       bIsStaticSwitch=true
       bIsVisible=true
       type {classStructOrEnum {refPath=/Niagara/Enums/ENiagaraEmitterLifeCycleMode}}
      }
      Inactive Response {
       bIsDynamic=false
       bIsEditable=true
       bIsStaticSwitch=true
       bIsVisible=true
       type {classStructOrEnum {refPath=/Niagara/Enums/ENiagaraInactiveMode}}
      }
      Loop Behavior {
       bIsDynamic=false
       bIsEditable=true
       bIsStaticSwitch=true
       bIsVisible=true
       type {classStructOrEnum {refPath=/Niagara/Enums/ENiagara_EmitterStateOptions}}
      }
      Loop Duration Mode {
       bIsDynamic=false
       bIsEditable=true
       bIsStaticSwitch=true
       bIsVisible=true
       type {classStructOrEnum {refPath=/Niagara/Enums/ENiagara_InfiniteLoopDuration}}
      }
      Loop Count {
       bIsDynamic=false
       bIsEditable=false
       bIsStaticSwitch=false
       bIsVisible=false
       type {classStructOrEnum {refPath=/Script/Niagara.NiagaraInt32}}
      }
      Loop Duration {
       bIsDynamic=false
       bIsEditable=false
       bIsStaticSwitch=false
       bIsVisible=false
       type {classStructOrEnum {refPath=/Script/Niagara.NiagaraFloat}}
      }
      Recalculate Duration Each Loop {bIsDynamic=false, bIsEditable=false, bIsStaticSwitch=false, bIsVisible=false, type {classStructOrEnum {refPath=/Script/Niagara.NiagaraBool}}}
      UseLoopDelay {bIsDynamic=false, bIsEditable=true, bIsStaticSwitch=false, bIsVisible=true, type {classStructOrEnum {refPath=&NiagaraBool}}}
      Loop Delay {bIsDynamic=false, bIsEditable=false, bIsStaticSwitch=false, bIsVisible=true, type {classStructOrEnum {refPath=&NiagaraFloat}}}
      Delay First Loop Only {bIsDynamic=false, bIsEditable=false, bIsStaticSwitch=false, bIsVisible=false, type {classStructOrEnum {refPath=&NiagaraBool}}}
      Scalability Mode {
       bIsDynamic=false
       bIsEditable=true
       bIsStaticSwitch=true
       bIsVisible=true
       type {classStructOrEnum {refPath=/Niagara/Enums/ENiagaraEmitterScalabilityMode_Limited}}
      }
      Enable Distance Culling {bIsDynamic=false, bIsEditable=false, bIsStaticSwitch=true, bIsVisible=false, type {classStructOrEnum {refPath=&NiagaraBool}}}
      MinDistance {
       bIsDynamic=false
       bIsEditable=false
       bIsStaticSwitch=false
       bIsVisible=false
       type {classStructOrEnum {refPath=&NiagaraFloat}}
      }
      Min Distance Response {
       bIsDynamic=false
       bIsEditable=false
       bIsStaticSwitch=false
       bIsVisible=false
       type {classStructOrEnum {refPath=/Script/Niagara.ENiagaraExecutionStateManagement}}
      }
      MaxDistance {
       bIsDynamic=false
       bIsEditable=false
       bIsStaticSwitch=false
       bIsVisible=false
       type {classStructOrEnum {refPath=&NiagaraFloat}}
      }
      Max Distance Response {
       bIsDynamic=false
       bIsEditable=false
       bIsStaticSwitch=false
       bIsVisible=false
       type {classStructOrEnum {refPath=&ENiagaraExecutionStateManagement}}
      }
      Scale Spawn Count {bIsDynamic=false, bIsEditable=false, bIsStaticSwitch=true, bIsVisible=false, type {classStructOrEnum {refPath=&NiagaraBool}}}
      Spawn Count Scale {
       bIsDynamic=false
       bIsEditable=false
       bIsStaticSwitch=false
       bIsVisible=false
       type {classStructOrEnum {refPath=&NiagaraFloat}}
      }
      Spawn Count Scale By Distance {
       bIsDynamic=false
       bIsEditable=false
       bIsStaticSwitch=false
       bIsVisible=false
       type {classStructOrEnum {refPath=&NiagaraDataInterfaceCurve}}
      }
      Enable Visibility Culling {bIsDynamic=false, bIsEditable=false, bIsStaticSwitch=true, bIsVisible=false, type {classStructOrEnum {refPath=&NiagaraBool}}}
      Visibility Cull Response {
       bIsDynamic=false
       bIsEditable=false
       bIsStaticSwitch=false
       bIsVisible=false
       type {classStructOrEnum {refPath=&ENiagaraExecutionStateManagement}}
      }
      Visibility Cull Delay {
       bIsDynamic=false
       bIsEditable=false
       bIsStaticSwitch=false
       bIsVisible=false
       type {classStructOrEnum {refPath=&NiagaraFloat}}
      }
      Reset Age On Awaken {bIsDynamic=false, bIsEditable=false, bIsStaticSwitch=false, bIsVisible=false, type {classStructOrEnum {refPath=&NiagaraBool}}}
     ]
     moduleScript {refPath=&EmitterState}
    }
    {
     bIsSetParametersModule=false
     enabled=true
     moduleName=SpawnBurst_Instantaneous001
     inputs [
      Spawn Count {bIsDynamic=false, bIsEditable=true, bIsStaticSwitch=false, bIsVisible=true, type {classStructOrEnum {refPath=&NiagaraInt32}}}
      Spawn Time {bIsDynamic=false, bIsEditable=true, bIsStaticSwitch=false, bIsVisible=true, type {classStructOrEnum {refPath=&NiagaraFloat}}}
      Use Spawn Probability {bIsDynamic=false, bIsEditable=true, bIsStaticSwitch=false, bIsVisible=true, type {classStructOrEnum {refPath=&NiagaraBool}}}
      Spawn Probability {bIsDynamic=false, bIsEditable=false, bIsStaticSwitch=false, bIsVisible=true, type {classStructOrEnum {refPath=&NiagaraFloat}}}
      UseLoopCountLimit {bIsDynamic=false, bIsEditable=true, bIsStaticSwitch=false, bIsVisible=true, type {classStructOrEnum {refPath=&NiagaraBool}}}
      Loop Count Limit {bIsDynamic=false, bIsEditable=false, bIsStaticSwitch=false, bIsVisible=true, type {classStructOrEnum {refPath=&NiagaraInt32}}}
      Spawn Group {bIsDynamic=false, bIsEditable=true, bIsStaticSwitch=false, bIsVisible=true, type {classStructOrEnum {refPath=&NiagaraInt32}}}
      Age {bIsDynamic=false, bIsEditable=true, bIsStaticSwitch=false, bIsVisible=true, type {classStructOrEnum {refPath=&NiagaraFloat}}}
     ]
     moduleScript {refPath=&SpawnBurst_Instantaneous}
    }
   ]
  }
  particleSpawnScript {
   scriptName=ParticleSpawnScript
   modules [{
     bIsSetParametersModule=false
     enabled=true
     moduleName=InitializeParticle001
     inputs [
      Lifetime Mode {
       bIsDynamic=false
       bIsEditable=true
       bIsStaticSwitch=true
       bIsVisible=true
       type {classStructOrEnum {refPath=/Niagara/Enums/ENiagara_LifetimeMode}}
      }
      Lifetime Min {
       bIsDynamic=false
       bIsEditable=false
       bIsStaticSwitch=false
       bIsVisible=false
       type {classStructOrEnum {refPath=&NiagaraFloat}}
      }
      Lifetime {bIsDynamic=false, bIsEditable=true, bIsStaticSwitch=false, bIsVisible=true, type {classStructOrEnum {refPath=&NiagaraFloat}}}
      Lifetime Max {
       bIsDynamic=false
       bIsEditable=false
       bIsStaticSwitch=false
       bIsVisible=false
       type {classStructOrEnum {refPath=&NiagaraFloat}}
      }
      Lifetime Randomness Mode {
       bIsDynamic=false
       bIsEditable=false
       bIsStaticSwitch=true
       bIsVisible=false
       type {classStructOrEnum {refPath=/Niagara/Enums/ENiagaraRandomnessMode}}
      }
      Color Mode {
       bIsDynamic=false
       bIsEditable=true
       bIsStaticSwitch=true
       bIsVisible=true
       type {classStructOrEnum {refPath=/Niagara/Enums/ENiagara_ColorInitializationMode}}
      }
      Color Minimum {
       bIsDynamic=false
       bIsEditable=false
       bIsStaticSwitch=false
       bIsVisible=false
       type {classStructOrEnum {refPath=/Script/CoreUObject.LinearColor}}
      }
      Color {
       bIsDynamic=false
       bIsEditable=true
       bIsStaticSwitch=false
       bIsVisible=true
       type {classStructOrEnum {refPath=&LinearColor}}
      }
      AdjustHue {bIsDynamic=false, bIsEditable=false, bIsStaticSwitch=false, bIsVisible=false, type {classStructOrEnum {refPath=&NiagaraBool}}}
      Hue Shift Range {
       bIsDynamic=false
       bIsEditable=false
       bIsStaticSwitch=false
       bIsVisible=false
       type {classStructOrEnum {refPath=/Script/CoreUObject.Vector2f}}
      }
      Color Maximum {
       bIsDynamic=false
       bIsEditable=false
       bIsStaticSwitch=false
       bIsVisible=false
       type {classStructOrEnum {refPath=&LinearColor}}
      }
      AdjustSaturation {bIsDynamic=false, bIsEditable=false, bIsStaticSwitch=false, bIsVisible=false, type {classStructOrEnum {refPath=&NiagaraBool}}}
      Saturation Range {
       bIsDynamic=false
       bIsEditable=false
       bIsStaticSwitch=false
       bIsVisible=false
       type {classStructOrEnum {refPath=&Vector2f}}
      }
      AdjustValue {bIsDynamic=false, bIsEditable=false, bIsStaticSwitch=false, bIsVisible=false, type {classStructOrEnum {refPath=&NiagaraBool}}}
      Value Range {
       bIsDynamic=false
       bIsEditable=false
       bIsStaticSwitch=false
       bIsVisible=false
       type {classStructOrEnum {refPath=&Vector2f}}
      }
      AdjustAlpha {bIsDynamic=false, bIsEditable=false, bIsStaticSwitch=false, bIsVisible=false, type {classStructOrEnum {refPath=&NiagaraBool}}}
      Alpha Scale Range {
       bIsDynamic=false
       bIsEditable=false
       bIsStaticSwitch=false
       bIsVisible=false
       type {classStructOrEnum {refPath=&Vector2f}}
      }
      Color Channel Mode {
       bIsDynamic=false
       bIsEditable=false
       bIsStaticSwitch=true
       bIsVisible=false
       type {classStructOrEnum {refPath=/Niagara/Enums/ENiagaraChannelCorrelation}}
      }
      Color Randomness Mode {
       bIsDynamic=false
       bIsEditable=false
       bIsStaticSwitch=true
       bIsVisible=false
       type {classStructOrEnum {refPath=&ENiagaraRandomnessMode}}
      }
      Position Mode {
       bIsDynamic=false
       bIsEditable=true
       bIsStaticSwitch=true
       bIsVisible=true
       type {classStructOrEnum {refPath=/Niagara/Enums/ENiagara_PositionInitializationMode}}
      }
      Position {
       bIsDynamic=false
       bIsEditable=false
       bIsStaticSwitch=false
       bIsVisible=false
       type {classStructOrEnum {refPath=/Script/Niagara.NiagaraPosition}}
      }
      UsePositionOffset {bIsDynamic=false, bIsEditable=true, bIsStaticSwitch=false, bIsVisible=true, type {classStructOrEnum {refPath=&NiagaraBool}}}
      Position Offset {bIsDynamic=false, bIsEditable=false, bIsStaticSwitch=false, bIsVisible=true, type {classStructOrEnum {refPath=/Script/CoreUObject.Vector3f}}}
      Position Offset Coordinate Space {
       bIsDynamic=false
       bIsEditable=false
       bIsStaticSwitch=false
       bIsVisible=false
       type {classStructOrEnum {refPath=/Script/Niagara.ENiagaraCoordinateSpace}}
      }
      Mass Mode {
       bIsDynamic=false
       bIsEditable=true
       bIsStaticSwitch=true
       bIsVisible=true
       type {classStructOrEnum {refPath=/Niagara/Enums/ENiagara_MassInitializationMode}}
      }
      Mass Min {
       bIsDynamic=false
       bIsEditable=false
       bIsStaticSwitch=false
       bIsVisible=false
       type {classStructOrEnum {refPath=&NiagaraFloat}}
      }
      Mass {
       bIsDynamic=false
       bIsEditable=false
       bIsStaticSwitch=false
       bIsVisible=false
       type {classStructOrEnum {refPath=&NiagaraFloat}}
      }
      Mass Max {
       bIsDynamic=false
       bIsEditable=false
       bIsStaticSwitch=false
       bIsVisible=false
       type {classStructOrEnum {refPath=&NiagaraFloat}}
      }
      Mass Randomness Mode {
       bIsDynamic=false
       bIsEditable=false
       bIsStaticSwitch=true
       bIsVisible=false
       type {classStructOrEnum {refPath=&ENiagaraRandomnessMode}}
      }
      Write Material Random {bIsDynamic=false, bIsEditable=true, bIsStaticSwitch=false, bIsVisible=true, type {classStructOrEnum {refPath=&NiagaraBool}}}
      Material Random {bIsDynamic=false, bIsEditable=true, bIsStaticSwitch=false, bIsVisible=true, type {classStructOrEnum {refPath=&NiagaraFloat}}}
      Sprite Size Mode {
       bIsDynamic=false
       bIsEditable=true
       bIsStaticSwitch=true
       bIsVisible=true
       type {classStructOrEnum {refPath=/Niagara/Enums/ENiagara_SizeScaleMode}}
      }
      Sprite Size {
       bIsDynamic=false
       bIsEditable=false
       bIsStaticSwitch=false
       bIsVisible=false
       type {classStructOrEnum {refPath=&Vector2f}}
      }
      Uniform Sprite Size {
       bIsDynamic=false
       bIsEditable=false
       bIsStaticSwitch=false
       bIsVisible=false
       type {classStructOrEnum {refPath=&NiagaraFloat}}
      }
      Uniform Sprite Size Min {
       bIsDynamic=false
       bIsEditable=false
       bIsStaticSwitch=false
       bIsVisible=false
       type {classStructOrEnum {refPath=&NiagaraFloat}}
      }
      Sprite Size Min {
       bIsDynamic=false
       bIsEditable=false
       bIsStaticSwitch=false
       bIsVisible=false
       type {classStructOrEnum {refPath=&Vector2f}}
      }
      Sprite Size Max {
       bIsDynamic=false
       bIsEditable=false
       bIsStaticSwitch=false
       bIsVisible=false
       type {classStructOrEnum {refPath=&Vector2f}}
      }
      Uniform Sprite Size Max {
       bIsDynamic=false
       bIsEditable=false
       bIsStaticSwitch=false
       bIsVisible=false
       type {classStructOrEnum {refPath=&NiagaraFloat}}
      }
      Sprite Size Randomness Mode {
       bIsDynamic=false
       bIsEditable=false
       bIsStaticSwitch=true
       bIsVisible=false
       type {classStructOrEnum {refPath=&ENiagaraRandomnessMode}}
      }
      Sprite Rotation Mode {
       bIsDynamic=false
       bIsEditable=true
       bIsStaticSwitch=true
       bIsVisible=true
       type {classStructOrEnum {refPath=/Niagara/Enums/ENiagara_SpriteRotationMode}}
      }
      Sprite Rotation Angle Min {
       bIsDynamic=false
       bIsEditable=false
       bIsStaticSwitch=false
       bIsVisible=false
       type {classStructOrEnum {refPath=&NiagaraFloat}}
      }
      Sprite Rotation Angle {
       bIsDynamic=false
       bIsEditable=false
       bIsStaticSwitch=false
       bIsVisible=false
       type {classStructOrEnum {refPath=&NiagaraFloat}}
      }
      Sprite Rotation Angle Max {
       bIsDynamic=false
       bIsEditable=false
       bIsStaticSwitch=false
       bIsVisible=false
       type {classStructOrEnum {refPath=&NiagaraFloat}}
      }
      Sprite Rotation Randomness Mode {
       bIsDynamic=false
       bIsEditable=false
       bIsStaticSwitch=true
       bIsVisible=false
       type {classStructOrEnum {refPath=&ENiagaraRandomnessMode}}
      }
      Sprite UV Mode {
       bIsDynamic=false
       bIsEditable=true
       bIsStaticSwitch=true
       bIsVisible=true
       type {classStructOrEnum {refPath=/Niagara/Enums/ENiagara_UVFlippingMode}}
      }
      Sprite UV Randomness Mode {
       bIsDynamic=false
       bIsEditable=false
       bIsStaticSwitch=true
       bIsVisible=false
       type {classStructOrEnum {refPath=&ENiagaraRandomnessMode}}
      }
      Sprite UV Scale {
       bIsDynamic=false
       bIsEditable=false
       bIsStaticSwitch=false
       bIsVisible=false
       type {classStructOrEnum {refPath=&Vector2f}}
      }
      Mesh Scale Mode {
       bIsDynamic=false
       bIsEditable=true
       bIsStaticSwitch=true
       bIsVisible=true
       type {classStructOrEnum {refPath=&ENiagara_SizeScaleMode}}
      }
      Mesh Uniform Scale Min {
       bIsDynamic=false
       bIsEditable=false
       bIsStaticSwitch=false
       bIsVisible=false
       type {classStructOrEnum {refPath=&NiagaraFloat}}
      }
      Mesh Uniform Scale {
       bIsDynamic=false
       bIsEditable=false
       bIsStaticSwitch=false
       bIsVisible=false
       type {classStructOrEnum {refPath=&NiagaraFloat}}
      }
      Mesh Scale Min {
       bIsDynamic=false
       bIsEditable=false
       bIsStaticSwitch=false
       bIsVisible=false
       type {classStructOrEnum {refPath=&Vector3f}}
      }
      Mesh Scale {
       bIsDynamic=false
       bIsEditable=false
       bIsStaticSwitch=false
       bIsVisible=false
       type {classStructOrEnum {refPath=&Vector3f}}
      }
      Mesh Scale Max {
       bIsDynamic=false
       bIsEditable=false
       bIsStaticSwitch=false
       bIsVisible=false
       type {classStructOrEnum {refPath=&Vector3f}}
      }
      Mesh Uniform Scale Max {
       bIsDynamic=false
       bIsEditable=false
       bIsStaticSwitch=false
       bIsVisible=false
       type {classStructOrEnum {refPath=&NiagaraFloat}}
      }
      Mesh Scale Randomness Mode {
       bIsDynamic=false
       bIsEditable=false
       bIsStaticSwitch=true
       bIsVisible=false
       type {classStructOrEnum {refPath=&ENiagaraRandomnessMode}}
      }
      Mesh Renderer Array Visibility Mode {
       bIsDynamic=false
       bIsEditable=true
       bIsStaticSwitch=true
       bIsVisible=true
       type {classStructOrEnum {refPath=/Niagara/Enums/Ribbons/ENiagara_UnsetDirectSetRandom}}
      }
      Mesh Renderer Mesh Index {
       bIsDynamic=false
       bIsEditable=false
       bIsStaticSwitch=false
       bIsVisible=false
       type {classStructOrEnum {refPath=&NiagaraInt32}}
      }
      Wrap Index By Mesh Count {bIsDynamic=false, bIsEditable=false, bIsStaticSwitch=true, bIsVisible=false, type {classStructOrEnum {refPath=&NiagaraBool}}}
      Mesh Renderer Info {
       bIsDynamic=false
       bIsEditable=false
       bIsStaticSwitch=false
       bIsVisible=false
       type {classStructOrEnum {refPath=&NiagaraDataInterfaceMeshRendererInfo}}
      }
      Mesh Renderer Random Seed {
       bIsDynamic=false
       bIsEditable=false
       bIsStaticSwitch=false
       bIsVisible=false
       type {classStructOrEnum {refPath=&NiagaraInt32}}
      }
      Ribbon Width Mode {
       bIsDynamic=false
       bIsEditable=true
       bIsStaticSwitch=true
       bIsVisible=true
       type {classStructOrEnum {refPath=/Niagara/Enums/Ribbons/ENiagara_UnsetDirectSet}}
      }
      Ribbon Width {
       bIsDynamic=false
       bIsEditable=false
       bIsStaticSwitch=false
       bIsVisible=false
       type {classStructOrEnum {refPath=&NiagaraFloat}}
      }
      Ribbon Facing Mode {
       bIsDynamic=false
       bIsEditable=true
       bIsStaticSwitch=true
       bIsVisible=true
       type {classStructOrEnum {refPath=&ENiagara_UnsetDirectSet}}
      }
      Ribbon Facing Vector {
       bIsDynamic=false
       bIsEditable=false
       bIsStaticSwitch=false
       bIsVisible=false
       type {classStructOrEnum {refPath=&Vector3f}}
      }
      Ribbon Twist Mode {
       bIsDynamic=false
       bIsEditable=true
       bIsStaticSwitch=true
       bIsVisible=true
       type {classStructOrEnum {refPath=&ENiagara_UnsetDirectSet}}
      }
      "Ribbon Twist " {
       bIsDynamic=false
       bIsEditable=false
       bIsStaticSwitch=false
       bIsVisible=false
       type {classStructOrEnum {refPath=&NiagaraFloat}}
      }
     ]
     moduleScript {refPath=&InitializeParticle}
    }]
  }
  particleUpdateScript {
   scriptName=ParticleUpdateScript
   modules [
    {
     bIsSetParametersModule=false
     enabled=true
     moduleName=ParticleState
     inputs [
      Kill Particles When Lifetime Has Elapsed {bIsDynamic=false, bIsEditable=true, bIsStaticSwitch=true, bIsVisible=true, type {classStructOrEnum {refPath=&NiagaraBool}}}
      Loop Particles Lifetime {bIsDynamic=false, bIsEditable=true, bIsStaticSwitch=false, bIsVisible=true, type {classStructOrEnum {refPath=&NiagaraBool}}}
      Let Infinitely Lived Particles Die When Emitter Deactivates {bIsDynamic=false, bIsEditable=true, bIsStaticSwitch=false, bIsVisible=true, type {classStructOrEnum {refPath=&NiagaraBool}}}
      Lifetime {bIsDynamic=false, bIsEditable=true, bIsStaticSwitch=false, bIsVisible=true, type {classStructOrEnum {refPath=&NiagaraFloat}}}
      DeltaTime {bIsDynamic=false, bIsEditable=true, bIsStaticSwitch=false, bIsVisible=true, type {classStructOrEnum {refPath=&NiagaraFloat}}}
     ]
     moduleScript {refPath=&ParticleState}
    }
    {
     bIsSetParametersModule=false
     enabled=true
     moduleName=ScaleSpriteSize001
     inputs [
      Scale Sprite Size Mode {
       bIsDynamic=false
       bIsEditable=true
       bIsStaticSwitch=true
       bIsVisible=true
       type {classStructOrEnum {refPath=/Niagara/Enums/SpriteRenderer/ENiagara_ScaleSpriteSize}}
      }
      Uniform Curve Sprite Scale {
       bIsDynamic=false
       bIsEditable=true
       bIsStaticSwitch=false
       bIsVisible=true
       type {classStructOrEnum {refPath=&NiagaraDataInterfaceCurve}}
      }
      Non-Uniform Curve Sprite Scale {
       bIsDynamic=false
       bIsEditable=false
       bIsStaticSwitch=false
       bIsVisible=false
       type {classStructOrEnum {refPath=&NiagaraDataInterfaceVector2DCurve}}
      }
      Uniform Curve Index {bIsDynamic=false, bIsEditable=true, bIsStaticSwitch=false, bIsVisible=true, type {classStructOrEnum {refPath=&NiagaraFloat}}}
      Non-Uniform Curve Index {
       bIsDynamic=false
       bIsEditable=false
       bIsStaticSwitch=false
       bIsVisible=false
       type {classStructOrEnum {refPath=&NiagaraFloat}}
      }
      Non-Uniform Curve Scale {
       bIsDynamic=false
       bIsEditable=false
       bIsStaticSwitch=false
       bIsVisible=false
       type {classStructOrEnum {refPath=&Vector2f}}
      }
      Uniform Curve Scale {bIsDynamic=false, bIsEditable=true, bIsStaticSwitch=false, bIsVisible=true, type {classStructOrEnum {refPath=&NiagaraFloat}}}
      Scale Factor {
       bIsDynamic=false
       bIsEditable=false
       bIsStaticSwitch=false
       bIsVisible=false
       type {classStructOrEnum {refPath=&Vector2f}}
      }
      Uniform Scale Factor {
       bIsDynamic=false
       bIsEditable=false
       bIsStaticSwitch=false
       bIsVisible=false
       type {classStructOrEnum {refPath=&NiagaraFloat}}
      }
      Initial Sprite Size {bIsDynamic=false, bIsEditable=true, bIsStaticSwitch=false, bIsVisible=true, type {classStructOrEnum {refPath=&Vector2f}}}
     ]
     moduleScript {refPath=&ScaleSpriteSize}
    }
    {
     bIsSetParametersModule=true
     enabled=true
     moduleName=SetVariables_#5
     inputs [
      Particles.DecalSize {bIsDynamic=false, bIsEditable=true, bIsStaticSwitch=false, bIsVisible=true, type {classStructOrEnum {refPath=&Vector3f}}}
      Particles.Color_Decal {
       bIsDynamic=true
       bIsEditable=true
       bIsStaticSwitch=false
       bIsVisible=true
       type {classStructOrEnum {refPath=&LinearColor}}
      }
     ]
     moduleScript {refPath=&self:Decal_Caustics_0.NiagaraScriptSource_0.NiagaraGraph_0.NiagaraNodeAssignment_1.SetVariables_#5}
    }
    {
     bIsSetParametersModule=false
     enabled=true
     moduleName=DynamicMaterialParameters
     inputs [
      Write Parameter Index 0 {
       bIsDynamic=false
       bIsEditable=true
       bIsStaticSwitch=true
       bIsVisible=true
       type {classStructOrEnum {refPath=/Niagara/Enums/Utility/ENiagara_DynamicMaterialParametersInputMode}}
      }
      Index 0 Vector 1 {
       bIsDynamic=false
       bIsEditable=false
       bIsStaticSwitch=false
       bIsVisible=false
       type {classStructOrEnum {refPath=&Vector3f}}
      }
      Module.Index0Float0Write {bIsDynamic=false, bIsEditable=true, bIsStaticSwitch=true, bIsVisible=true, type {classStructOrEnum {refPath=&NiagaraBool}}}
      Index 0 Param 1 {bIsDynamic=false, bIsEditable=true, bIsStaticSwitch=false, bIsVisible=true, type {classStructOrEnum {refPath=&NiagaraFloat}}}
      Module.Index0Float1Write {bIsDynamic=false, bIsEditable=true, bIsStaticSwitch=true, bIsVisible=true, type {classStructOrEnum {refPath=&NiagaraBool}}}
      Index 0 Param 2 {bIsDynamic=false, bIsEditable=false, bIsStaticSwitch=false, bIsVisible=true, type {classStructOrEnum {refPath=&NiagaraFloat}}}
      Module.Index0Float2Write {bIsDynamic=false, bIsEditable=true, bIsStaticSwitch=true, bIsVisible=true, type {classStructOrEnum {refPath=&NiagaraBool}}}
      Index 0 Param 3 {bIsDynamic=false, bIsEditable=false, bIsStaticSwitch=false, bIsVisible=true, type {classStructOrEnum {refPath=&NiagaraFloat}}}
      Module.Index0Float3Write {bIsDynamic=false, bIsEditable=true, bIsStaticSwitch=true, bIsVisible=true, type {classStructOrEnum {refPath=&NiagaraBool}}}
      Index 0 Param 4 {bIsDynamic=false, bIsEditable=true, bIsStaticSwitch=false, bIsVisible=true, type {classStructOrEnum {refPath=&NiagaraFloat}}}
      Index 0 Vector2D 1 {
       bIsDynamic=false
       bIsEditable=false
       bIsStaticSwitch=false
       bIsVisible=false
       type {classStructOrEnum {refPath=&Vector2f}}
      }
      Index 0 Vector2D 2 {
       bIsDynamic=false
       bIsEditable=false
       bIsStaticSwitch=false
       bIsVisible=false
       type {classStructOrEnum {refPath=&Vector2f}}
      }
      Index 0 Vector 4 {
       bIsDynamic=false
       bIsEditable=false
       bIsStaticSwitch=false
       bIsVisible=false
       type {classStructOrEnum {refPath=/Script/CoreUObject.Vector4f}}
      }
      Index 0 Linear Color {
       bIsDynamic=false
       bIsEditable=false
       bIsStaticSwitch=false
       bIsVisible=false
       type {classStructOrEnum {refPath=&LinearColor}}
      }
      Write Parameter Index 1 {
       bIsDynamic=false
       bIsEditable=true
       bIsStaticSwitch=true
       bIsVisible=true
       type {classStructOrEnum {refPath=&ENiagara_DynamicMaterialParametersInputMode}}
      }
      Index 1 Vector 1 {
       bIsDynamic=false
       bIsEditable=false
       bIsStaticSwitch=false
       bIsVisible=false
       type {classStructOrEnum {refPath=&Vector3f}}
      }
      Index 1 Param 1 {
       bIsDynamic=false
       bIsEditable=false
       bIsStaticSwitch=false
       bIsVisible=false
       type {classStructOrEnum {refPath=&NiagaraFloat}}
      }
      Index 1 Param 2 {
       bIsDynamic=false
       bIsEditable=false
       bIsStaticSwitch=false
       bIsVisible=false
       type {classStructOrEnum {refPath=&NiagaraFloat}}
      }
      Index 1 Param 3 {
       bIsDynamic=false
       bIsEditable=false
       bIsStaticSwitch=false
       bIsVisible=false
       type {classStructOrEnum {refPath=&NiagaraFloat}}
      }
      Index 1 Param 4 {
       bIsDynamic=false
       bIsEditable=false
       bIsStaticSwitch=false
       bIsVisible=false
       type {classStructOrEnum {refPath=&NiagaraFloat}}
      }
      Index 1 Vector2D 1 {
       bIsDynamic=false
       bIsEditable=false
       bIsStaticSwitch=false
       bIsVisible=false
       type {classStructOrEnum {refPath=&Vector2f}}
      }
      Index 1 Vector2D 2 {
       bIsDynamic=false
       bIsEditable=false
       bIsStaticSwitch=false
       bIsVisible=false
       type {classStructOrEnum {refPath=&Vector2f}}
      }
      Index 1 Vector 4 {
       bIsDynamic=false
       bIsEditable=false
       bIsStaticSwitch=false
       bIsVisible=false
       type {classStructOrEnum {refPath=&Vector4f}}
      }
      Index 1 Linear Color {
       bIsDynamic=false
       bIsEditable=false
       bIsStaticSwitch=false
       bIsVisible=false
       type {classStructOrEnum {refPath=&LinearColor}}
      }
      Write Parameter Index 2 {
       bIsDynamic=false
       bIsEditable=true
       bIsStaticSwitch=true
       bIsVisible=true
       type {classStructOrEnum {refPath=&ENiagara_DynamicMaterialParametersInputMode}}
      }
      Index 2 Vector 1 {
       bIsDynamic=false
       bIsEditable=false
       bIsStaticSwitch=false
       bIsVisible=false
       type {classStructOrEnum {refPath=&Vector3f}}
      }
      Index 2 Param 1 {
       bIsDynamic=false
       bIsEditable=false
       bIsStaticSwitch=false
       bIsVisible=false
       type {classStructOrEnum {refPath=&NiagaraFloat}}
      }
      Index 2 Param 2 {
       bIsDynamic=false
       bIsEditable=false
       bIsStaticSwitch=false
       bIsVisible=false
       type {classStructOrEnum {refPath=&NiagaraFloat}}
      }
      Index 2 Param 3 {
       bIsDynamic=false
       bIsEditable=false
       bIsStaticSwitch=false
       bIsVisible=false
       type {classStructOrEnum {refPath=&NiagaraFloat}}
      }
      Index 2 Param 4 {
       bIsDynamic=false
       bIsEditable=false
       bIsStaticSwitch=false
       bIsVisible=false
       type {classStructOrEnum {refPath=&NiagaraFloat}}
      }
      Index 2 Vector2D 1 {
       bIsDynamic=false
       bIsEditable=false
       bIsStaticSwitch=false
       bIsVisible=false
       type {classStructOrEnum {refPath=&Vector2f}}
      }
      Index 2 Vector2D 2 {
       bIsDynamic=false
       bIsEditable=false
       bIsStaticSwitch=false
       bIsVisible=false
       type {classStructOrEnum {refPath=&Vector2f}}
      }
      Index 2 Vector 4 {
       bIsDynamic=false
       bIsEditable=false
       bIsStaticSwitch=false
       bIsVisible=false
       type {classStructOrEnum {refPath=&Vector4f}}
      }
      Index 2 Linear Color {
       bIsDynamic=false
       bIsEditable=false
       bIsStaticSwitch=false
       bIsVisible=false
       type {classStructOrEnum {refPath=&LinearColor}}
      }
      Write Parameter Index 3 {
       bIsDynamic=false
       bIsEditable=true
       bIsStaticSwitch=true
       bIsVisible=true
       type {classStructOrEnum {refPath=&ENiagara_DynamicMaterialParametersInputMode}}
      }
      Index 3 Vector 1 {
       bIsDynamic=false
       bIsEditable=false
       bIsStaticSwitch=false
       bIsVisible=false
       type {classStructOrEnum {refPath=&Vector3f}}
      }
      Index 3 Param 1 {
       bIsDynamic=false
       bIsEditable=false
       bIsStaticSwitch=false
       bIsVisible=false
       type {classStructOrEnum {refPath=&NiagaraFloat}}
      }
      Index 3 Param 2 {
       bIsDynamic=false
       bIsEditable=false
       bIsStaticSwitch=false
       bIsVisible=false
       type {classStructOrEnum {refPath=&NiagaraFloat}}
      }
      Index 3 Param 3 {
       bIsDynamic=false
       bIsEditable=false
       bIsStaticSwitch=false
       bIsVisible=false
       type {classStructOrEnum {refPath=&NiagaraFloat}}
      }
      Index 3 Param 4 {
       bIsDynamic=false
       bIsEditable=false
       bIsStaticSwitch=false
       bIsVisible=false
       type {classStructOrEnum {refPath=&NiagaraFloat}}
      }
      Index 3 Vector2D 1 {
       bIsDynamic=false
       bIsEditable=false
       bIsStaticSwitch=false
       bIsVisible=false
       type {classStructOrEnum {refPath=&Vector2f}}
      }
      Index 3 Vector2D 2 {
       bIsDynamic=false
       bIsEditable=false
       bIsStaticSwitch=false
       bIsVisible=false
       type {classStructOrEnum {refPath=&Vector2f}}
      }
      Index 3 Vector 4 {
       bIsDynamic=false
       bIsEditable=false
       bIsStaticSwitch=false
       bIsVisible=false
       type {classStructOrEnum {refPath=&Vector4f}}
      }
      Index 3 Linear Color {
       bIsDynamic=false
       bIsEditable=false
       bIsStaticSwitch=false
       bIsVisible=false
       type {classStructOrEnum {refPath=&LinearColor}}
      }
      Dynamic Parameter Output {
       bIsDynamic=false
       bIsEditable=true
       bIsStaticSwitch=true
       bIsVisible=true
       type {classStructOrEnum {refPath=/Niagara/Enums/ENiagara_AttributeSamplingApplyOutput}}
      }
     ]
     moduleScript {refPath=&DynamicMaterialParameters}
    }
   ]
  }
  rendererClasses [{refPath=&NiagaraDecalRendererProperties}]
  renderers [{rendererIndex=0, rendererClass {refPath=&NiagaraDecalRendererProperties}}]
 }
 InfiniteParticleLifetime002 {
  emitterSpawnScript {
   scriptName=EmitterSpawnScript
   modules [{
     bIsSetParametersModule=false
     enabled=true
     moduleName=InitializeAudioPlayer
     inputs [Audio Player {
       bIsDynamic=false
       bIsEditable=true
       bIsStaticSwitch=false
       bIsVisible=true
       type {classStructOrEnum {refPath=&NiagaraDataInterfaceAudioPlayer}}
      }]
     moduleScript {refPath=&InitializeAudioPlayer}
    }]
  }
  emitterUpdateScript {
   scriptName=EmitterUpdateScript
   modules [
    {
     bIsSetParametersModule=false
     enabled=true
     moduleName=EmitterState
     inputs [
      Life Cycle Mode {
       bIsDynamic=false
       bIsEditable=true
       bIsStaticSwitch=true
       bIsVisible=true
       type {classStructOrEnum {refPath=&ENiagaraEmitterLifeCycleMode}}
      }
      Inactive Response {
       bIsDynamic=false
       bIsEditable=true
       bIsStaticSwitch=true
       bIsVisible=true
       type {classStructOrEnum {refPath=&ENiagaraInactiveMode}}
      }
      Loop Behavior {
       bIsDynamic=false
       bIsEditable=true
       bIsStaticSwitch=true
       bIsVisible=true
       type {classStructOrEnum {refPath=&ENiagara_EmitterStateOptions}}
      }
      Loop Duration Mode {
       bIsDynamic=false
       bIsEditable=false
       bIsStaticSwitch=true
       bIsVisible=false
       type {classStructOrEnum {refPath=&ENiagara_InfiniteLoopDuration}}
      }
      Loop Count {
       bIsDynamic=false
       bIsEditable=false
       bIsStaticSwitch=false
       bIsVisible=false
       type {classStructOrEnum {refPath=&NiagaraInt32}}
      }
      Loop Duration {bIsDynamic=false, bIsEditable=true, bIsStaticSwitch=false, bIsVisible=true, type {classStructOrEnum {refPath=&NiagaraFloat}}}
      Recalculate Duration Each Loop {bIsDynamic=false, bIsEditable=true, bIsStaticSwitch=false, bIsVisible=true, type {classStructOrEnum {refPath=&NiagaraBool}}}
      UseLoopDelay {bIsDynamic=false, bIsEditable=true, bIsStaticSwitch=false, bIsVisible=true, type {classStructOrEnum {refPath=&NiagaraBool}}}
      Loop Delay {bIsDynamic=false, bIsEditable=false, bIsStaticSwitch=false, bIsVisible=true, type {classStructOrEnum {refPath=&NiagaraFloat}}}
      Delay First Loop Only {bIsDynamic=false, bIsEditable=false, bIsStaticSwitch=false, bIsVisible=false, type {classStructOrEnum {refPath=&NiagaraBool}}}
      Scalability Mode {
       bIsDynamic=false
       bIsEditable=true
       bIsStaticSwitch=true
       bIsVisible=true
       type {classStructOrEnum {refPath=&ENiagaraEmitterScalabilityMode_Limited}}
      }
      Enable Distance Culling {bIsDynamic=false, bIsEditable=false, bIsStaticSwitch=true, bIsVisible=false, type {classStructOrEnum {refPath=&NiagaraBool}}}
      MinDistance {
       bIsDynamic=false
       bIsEditable=false
       bIsStaticSwitch=false
       bIsVisible=false
       type {classStructOrEnum {refPath=&NiagaraFloat}}
      }
      Min Distance Response {
       bIsDynamic=false
       bIsEditable=false
       bIsStaticSwitch=false
       bIsVisible=false
       type {classStructOrEnum {refPath=&ENiagaraExecutionStateManagement}}
      }
      MaxDistance {
       bIsDynamic=false
       bIsEditable=false
       bIsStaticSwitch=false
       bIsVisible=false
       type {classStructOrEnum {refPath=&NiagaraFloat}}
      }
      Max Distance Response {
       bIsDynamic=false
       bIsEditable=false
       bIsStaticSwitch=false
       bIsVisible=false
       type {classStructOrEnum {refPath=&ENiagaraExecutionStateManagement}}
      }
      Scale Spawn Count {bIsDynamic=false, bIsEditable=false, bIsStaticSwitch=true, bIsVisible=false, type {classStructOrEnum {refPath=&NiagaraBool}}}
      Spawn Count Scale {
       bIsDynamic=false
       bIsEditable=false
       bIsStaticSwitch=false
       bIsVisible=false
       type {classStructOrEnum {refPath=&NiagaraFloat}}
      }
      Spawn Count Scale By Distance {
       bIsDynamic=false
       bIsEditable=false
       bIsStaticSwitch=false
       bIsVisible=false
       type {classStructOrEnum {refPath=&NiagaraDataInterfaceCurve}}
      }
      Enable Visibility Culling {bIsDynamic=false, bIsEditable=false, bIsStaticSwitch=true, bIsVisible=false, type {classStructOrEnum {refPath=&NiagaraBool}}}
      Visibility Cull Response {
       bIsDynamic=false
       bIsEditable=false
       bIsStaticSwitch=false
       bIsVisible=false
       type {classStructOrEnum {refPath=&ENiagaraExecutionStateManagement}}
      }
      Visibility Cull Delay {
       bIsDynamic=false
       bIsEditable=false
       bIsStaticSwitch=false
       bIsVisible=false
       type {classStructOrEnum {refPath=&NiagaraFloat}}
      }
      Reset Age On Awaken {bIsDynamic=false, bIsEditable=false, bIsStaticSwitch=false, bIsVisible=false, type {classStructOrEnum {refPath=&NiagaraBool}}}
     ]
     moduleScript {refPath=&EmitterState}
    }
    {
     bIsSetParametersModule=false
     enabled=true
     moduleName=SpawnRate
     inputs [
      SpawnRate {bIsDynamic=false, bIsEditable=true, bIsStaticSwitch=false, bIsVisible=true, type {classStructOrEnum {refPath=&NiagaraFloat}}}
      Use Spawn Probability {bIsDynamic=false, bIsEditable=true, bIsStaticSwitch=false, bIsVisible=true, type {classStructOrEnum {refPath=&NiagaraBool}}}
      Spawn Probability {bIsDynamic=false, bIsEditable=true, bIsStaticSwitch=false, bIsVisible=true, type {classStructOrEnum {refPath=&NiagaraFloat}}}
      SpawnGroup {bIsDynamic=false, bIsEditable=true, bIsStaticSwitch=false, bIsVisible=true, type {classStructOrEnum {refPath=&NiagaraInt32}}}
     ]
     moduleScript {refPath=&SpawnRate}
    }
    {
     bIsSetParametersModule=false
     enabled=true
     moduleName=SpawnRate001
     inputs [
      SpawnRate {bIsDynamic=false, bIsEditable=true, bIsStaticSwitch=false, bIsVisible=true, type {classStructOrEnum {refPath=&NiagaraFloat}}}
      Use Spawn Probability {bIsDynamic=false, bIsEditable=true, bIsStaticSwitch=false, bIsVisible=true, type {classStructOrEnum {refPath=&NiagaraBool}}}
      Spawn Probability {bIsDynamic=false, bIsEditable=true, bIsStaticSwitch=false, bIsVisible=true, type {classStructOrEnum {refPath=&NiagaraFloat}}}
      SpawnGroup {bIsDynamic=false, bIsEditable=true, bIsStaticSwitch=false, bIsVisible=true, type {classStructOrEnum {refPath=&NiagaraInt32}}}
     ]
     moduleScript {refPath=&SpawnRate}
    }
   ]
  }
  particleSpawnScript {
   scriptName=ParticleSpawnScript
   modules [
    {
     bIsSetParametersModule=false
     enabled=true
     moduleName=InitializeParticle
     inputs [
      Lifetime Mode {
       bIsDynamic=false
       bIsEditable=true
       bIsStaticSwitch=true
       bIsVisible=true
       type {classStructOrEnum {refPath=&ENiagara_LifetimeMode}}
      }
      Lifetime Min {
       bIsDynamic=false
       bIsEditable=false
       bIsStaticSwitch=false
       bIsVisible=false
       type {classStructOrEnum {refPath=&NiagaraFloat}}
      }
      Lifetime {bIsDynamic=false, bIsEditable=true, bIsStaticSwitch=false, bIsVisible=true, type {classStructOrEnum {refPath=&NiagaraFloat}}}
      Lifetime Max {
       bIsDynamic=false
       bIsEditable=false
       bIsStaticSwitch=false
       bIsVisible=false
       type {classStructOrEnum {refPath=&NiagaraFloat}}
      }
      Lifetime Randomness Mode {
       bIsDynamic=false
       bIsEditable=false
       bIsStaticSwitch=true
       bIsVisible=false
       type {classStructOrEnum {refPath=&ENiagaraRandomnessMode}}
      }
      Color Mode {
       bIsDynamic=false
       bIsEditable=true
       bIsStaticSwitch=true
       bIsVisible=true
       type {classStructOrEnum {refPath=&ENiagara_ColorInitializationMode}}
      }
      Color Minimum {
       bIsDynamic=false
       bIsEditable=false
       bIsStaticSwitch=false
       bIsVisible=false
       type {classStructOrEnum {refPath=&LinearColor}}
      }
      Color {
       bIsDynamic=false
       bIsEditable=true
       bIsStaticSwitch=false
       bIsVisible=true
       type {classStructOrEnum {refPath=&LinearColor}}
      }
      AdjustHue {bIsDynamic=false, bIsEditable=false, bIsStaticSwitch=false, bIsVisible=false, type {classStructOrEnum {refPath=&NiagaraBool}}}
      Hue Shift Range {
       bIsDynamic=false
       bIsEditable=false
       bIsStaticSwitch=false
       bIsVisible=false
       type {classStructOrEnum {refPath=&Vector2f}}
      }
      Color Maximum {
       bIsDynamic=false
       bIsEditable=false
       bIsStaticSwitch=false
       bIsVisible=false
       type {classStructOrEnum {refPath=&LinearColor}}
      }
      AdjustSaturation {bIsDynamic=false, bIsEditable=false, bIsStaticSwitch=false, bIsVisible=false, type {classStructOrEnum {refPath=&NiagaraBool}}}
      Saturation Range {
       bIsDynamic=false
       bIsEditable=false
       bIsStaticSwitch=false
       bIsVisible=false
       type {classStructOrEnum {refPath=&Vector2f}}
      }
      AdjustValue {bIsDynamic=false, bIsEditable=false, bIsStaticSwitch=false, bIsVisible=false, type {classStructOrEnum {refPath=&NiagaraBool}}}
      Value Range {
       bIsDynamic=false
       bIsEditable=false
       bIsStaticSwitch=false
       bIsVisible=false
       type {classStructOrEnum {refPath=&Vector2f}}
      }
      AdjustAlpha {bIsDynamic=false, bIsEditable=false, bIsStaticSwitch=false, bIsVisible=false, type {classStructOrEnum {refPath=&NiagaraBool}}}
      Alpha Scale Range {
       bIsDynamic=false
       bIsEditable=false
       bIsStaticSwitch=false
       bIsVisible=false
       type {classStructOrEnum {refPath=&Vector2f}}
      }
      Color Channel Mode {
       bIsDynamic=false
       bIsEditable=false
       bIsStaticSwitch=true
       bIsVisible=false
       type {classStructOrEnum {refPath=&ENiagaraChannelCorrelation}}
      }
      Color Randomness Mode {
       bIsDynamic=false
       bIsEditable=false
       bIsStaticSwitch=true
       bIsVisible=false
       type {classStructOrEnum {refPath=&ENiagaraRandomnessMode}}
      }
      Position Mode {
       bIsDynamic=false
       bIsEditable=true
       bIsStaticSwitch=true
       bIsVisible=true
       type {classStructOrEnum {refPath=&ENiagara_PositionInitializationMode}}
      }
      Position {
       bIsDynamic=false
       bIsEditable=false
       bIsStaticSwitch=false
       bIsVisible=false
       type {classStructOrEnum {refPath=&NiagaraPosition}}
      }
      UsePositionOffset {bIsDynamic=false, bIsEditable=true, bIsStaticSwitch=false, bIsVisible=true, type {classStructOrEnum {refPath=&NiagaraBool}}}
      Position Offset {bIsDynamic=true, bIsEditable=false, bIsStaticSwitch=false, bIsVisible=true, type {classStructOrEnum {refPath=&Vector3f}}}
      Position Offset Coordinate Space {
       bIsDynamic=false
       bIsEditable=false
       bIsStaticSwitch=false
       bIsVisible=false
       type {classStructOrEnum {refPath=&ENiagaraCoordinateSpace}}
      }
      Mass Mode {
       bIsDynamic=false
       bIsEditable=true
       bIsStaticSwitch=true
       bIsVisible=true
       type {classStructOrEnum {refPath=&ENiagara_MassInitializationMode}}
      }
      Mass Min {
       bIsDynamic=false
       bIsEditable=false
       bIsStaticSwitch=false
       bIsVisible=false
       type {classStructOrEnum {refPath=&NiagaraFloat}}
      }
      Mass {
       bIsDynamic=false
       bIsEditable=false
       bIsStaticSwitch=false
       bIsVisible=false
       type {classStructOrEnum {refPath=&NiagaraFloat}}
      }
      Mass Max {
       bIsDynamic=false
       bIsEditable=false
       bIsStaticSwitch=false
       bIsVisible=false
       type {classStructOrEnum {refPath=&NiagaraFloat}}
      }
      Mass Randomness Mode {
       bIsDynamic=false
       bIsEditable=false
       bIsStaticSwitch=true
       bIsVisible=false
       type {classStructOrEnum {refPath=&ENiagaraRandomnessMode}}
      }
      Write Material Random {bIsDynamic=false, bIsEditable=true, bIsStaticSwitch=false, bIsVisible=true, type {classStructOrEnum {refPath=&NiagaraBool}}}
      Material Random {bIsDynamic=false, bIsEditable=true, bIsStaticSwitch=false, bIsVisible=true, type {classStructOrEnum {refPath=&NiagaraFloat}}}
      Sprite Size Mode {
       bIsDynamic=false
       bIsEditable=true
       bIsStaticSwitch=true
       bIsVisible=true
       type {classStructOrEnum {refPath=&ENiagara_SizeScaleMode}}
      }
      Sprite Size {
       bIsDynamic=false
       bIsEditable=false
       bIsStaticSwitch=false
       bIsVisible=false
       type {classStructOrEnum {refPath=&Vector2f}}
      }
      Uniform Sprite Size {
       bIsDynamic=false
       bIsEditable=false
       bIsStaticSwitch=false
       bIsVisible=false
       type {classStructOrEnum {refPath=&NiagaraFloat}}
      }
      Uniform Sprite Size Min {bIsDynamic=false, bIsEditable=true, bIsStaticSwitch=false, bIsVisible=true, type {classStructOrEnum {refPath=&NiagaraFloat}}}
      Sprite Size Min {
       bIsDynamic=false
       bIsEditable=false
       bIsStaticSwitch=false
       bIsVisible=false
       type {classStructOrEnum {refPath=&Vector2f}}
      }
      Sprite Size Max {
       bIsDynamic=false
       bIsEditable=false
       bIsStaticSwitch=false
       bIsVisible=false
       type {classStructOrEnum {refPath=&Vector2f}}
      }
      Uniform Sprite Size Max {bIsDynamic=false, bIsEditable=true, bIsStaticSwitch=false, bIsVisible=true, type {classStructOrEnum {refPath=&NiagaraFloat}}}
      Sprite Size Randomness Mode {
       bIsDynamic=false
       bIsEditable=true
       bIsStaticSwitch=true
       bIsVisible=true
       type {classStructOrEnum {refPath=&ENiagaraRandomnessMode}}
      }
      Sprite Rotation Mode {
       bIsDynamic=false
       bIsEditable=true
       bIsStaticSwitch=true
       bIsVisible=true
       type {classStructOrEnum {refPath=&ENiagara_SpriteRotationMode}}
      }
      Sprite Rotation Angle Min {
       bIsDynamic=false
       bIsEditable=false
       bIsStaticSwitch=false
       bIsVisible=false
       type {classStructOrEnum {refPath=&NiagaraFloat}}
      }
      Sprite Rotation Angle {
       bIsDynamic=false
       bIsEditable=false
       bIsStaticSwitch=false
       bIsVisible=false
       type {classStructOrEnum {refPath=&NiagaraFloat}}
      }
      Sprite Rotation Angle Max {
       bIsDynamic=false
       bIsEditable=false
       bIsStaticSwitch=false
       bIsVisible=false
       type {classStructOrEnum {refPath=&NiagaraFloat}}
      }
      Sprite Rotation Randomness Mode {
       bIsDynamic=false
       bIsEditable=false
       bIsStaticSwitch=true
       bIsVisible=false
       type {classStructOrEnum {refPath=&ENiagaraRandomnessMode}}
      }
      Sprite UV Mode {
       bIsDynamic=false
       bIsEditable=true
       bIsStaticSwitch=true
       bIsVisible=true
       type {classStructOrEnum {refPath=&ENiagara_UVFlippingMode}}
      }
      Sprite UV Randomness Mode {
       bIsDynamic=false
       bIsEditable=false
       bIsStaticSwitch=true
       bIsVisible=false
       type {classStructOrEnum {refPath=&ENiagaraRandomnessMode}}
      }
      Sprite UV Scale {
       bIsDynamic=false
       bIsEditable=false
       bIsStaticSwitch=false
       bIsVisible=false
       type {classStructOrEnum {refPath=&Vector2f}}
      }
      Mesh Scale Mode {
       bIsDynamic=false
       bIsEditable=true
       bIsStaticSwitch=true
       bIsVisible=true
       type {classStructOrEnum {refPath=&ENiagara_SizeScaleMode}}
      }
      Mesh Uniform Scale Min {
       bIsDynamic=false
       bIsEditable=false
       bIsStaticSwitch=false
       bIsVisible=false
       type {classStructOrEnum {refPath=&NiagaraFloat}}
      }
      Mesh Uniform Scale {bIsDynamic=false, bIsEditable=true, bIsStaticSwitch=false, bIsVisible=true, type {classStructOrEnum {refPath=&NiagaraFloat}}}
      Mesh Scale Min {
       bIsDynamic=false
       bIsEditable=false
       bIsStaticSwitch=false
       bIsVisible=false
       type {classStructOrEnum {refPath=&Vector3f}}
      }
      Mesh Scale {
       bIsDynamic=false
       bIsEditable=false
       bIsStaticSwitch=false
       bIsVisible=false
       type {classStructOrEnum {refPath=&Vector3f}}
      }
      Mesh Scale Max {
       bIsDynamic=false
       bIsEditable=false
       bIsStaticSwitch=false
       bIsVisible=false
       type {classStructOrEnum {refPath=&Vector3f}}
      }
      Mesh Uniform Scale Max {
       bIsDynamic=false
       bIsEditable=false
       bIsStaticSwitch=false
       bIsVisible=false
       type {classStructOrEnum {refPath=&NiagaraFloat}}
      }
      Mesh Scale Randomness Mode {
       bIsDynamic=false
       bIsEditable=false
       bIsStaticSwitch=true
       bIsVisible=false
       type {classStructOrEnum {refPath=&ENiagaraRandomnessMode}}
      }
      Mesh Renderer Array Visibility Mode {
       bIsDynamic=false
       bIsEditable=true
       bIsStaticSwitch=true
       bIsVisible=true
       type {classStructOrEnum {refPath=&ENiagara_UnsetDirectSetRandom}}
      }
      Mesh Renderer Mesh Index {
       bIsDynamic=false
       bIsEditable=false
       bIsStaticSwitch=false
       bIsVisible=false
       type {classStructOrEnum {refPath=&NiagaraInt32}}
      }
      Wrap Index By Mesh Count {bIsDynamic=false, bIsEditable=false, bIsStaticSwitch=true, bIsVisible=false, type {classStructOrEnum {refPath=&NiagaraBool}}}
      Mesh Renderer Info {
       bIsDynamic=false
       bIsEditable=false
       bIsStaticSwitch=false
       bIsVisible=false
       type {classStructOrEnum {refPath=&NiagaraDataInterfaceMeshRendererInfo}}
      }
      Mesh Renderer Random Seed {
       bIsDynamic=false
       bIsEditable=false
       bIsStaticSwitch=false
       bIsVisible=false
       type {classStructOrEnum {refPath=&NiagaraInt32}}
      }
      Ribbon Width Mode {
       bIsDynamic=false
       bIsEditable=true
       bIsStaticSwitch=true
       bIsVisible=true
       type {classStructOrEnum {refPath=&ENiagara_UnsetDirectSet}}
      }
      Ribbon Width {
       bIsDynamic=false
       bIsEditable=false
       bIsStaticSwitch=false
       bIsVisible=false
       type {classStructOrEnum {refPath=&NiagaraFloat}}
      }
      Ribbon Facing Mode {
       bIsDynamic=false
       bIsEditable=true
       bIsStaticSwitch=true
       bIsVisible=true
       type {classStructOrEnum {refPath=&ENiagara_UnsetDirectSet}}
      }
      Ribbon Facing Vector {
       bIsDynamic=false
       bIsEditable=false
       bIsStaticSwitch=false
       bIsVisible=false
       type {classStructOrEnum {refPath=&Vector3f}}
      }
      Ribbon Twist Mode {
       bIsDynamic=false
       bIsEditable=true
       bIsStaticSwitch=true
       bIsVisible=true
       type {classStructOrEnum {refPath=&ENiagara_UnsetDirectSet}}
      }
      "Ribbon Twist " {
       bIsDynamic=false
       bIsEditable=false
       bIsStaticSwitch=false
       bIsVisible=false
       type {classStructOrEnum {refPath=&NiagaraFloat}}
      }
     ]
     moduleScript {refPath=&InitializeParticle}
    }
    {
     bIsSetParametersModule=false
     enabled=true
     moduleName=InitialMeshOrientation
     inputs [
      Mesh Orientation Mode {
       bIsDynamic=false
       bIsEditable=true
       bIsStaticSwitch=true
       bIsVisible=true
       type {classStructOrEnum {refPath=/Niagara/Enums/Transforms/ENiagara_MeshOrientationOptions}}
      }
      Orientation Quaternion {bIsDynamic=false, bIsEditable=false, bIsStaticSwitch=false, bIsVisible=false, type {classStructOrEnum {refPath=/Script/CoreUObject.Quat4f}}}
      Orientation Matrix {
       bIsDynamic=false
       bIsEditable=false
       bIsStaticSwitch=false
       bIsVisible=false
       type {classStructOrEnum {refPath=/Script/Niagara.NiagaraMatrix}}
      }
      Orientation Vector {
       bIsDynamic=false
       bIsEditable=false
       bIsStaticSwitch=false
       bIsVisible=false
       type {classStructOrEnum {refPath=&Vector3f}}
      }
      Orientation Coordinate Space {
       bIsDynamic=false
       bIsEditable=false
       bIsStaticSwitch=true
       bIsVisible=false
       type {classStructOrEnum {refPath=&ENiagaraCoordinateSpace}}
      }
      Mesh Axis To Orient {
       bIsDynamic=false
       bIsEditable=false
       bIsStaticSwitch=false
       bIsVisible=false
       type {classStructOrEnum {refPath=&Vector3f}}
      }
      Random Seed {
       bIsDynamic=false
       bIsEditable=false
       bIsStaticSwitch=false
       bIsVisible=false
       type {classStructOrEnum {refPath=&NiagaraInt32}}
      }
      Use Rotation Vector {bIsDynamic=false, bIsEditable=true, bIsStaticSwitch=false, bIsVisible=true, type {classStructOrEnum {refPath=&NiagaraBool}}}
      Rotation {bIsDynamic=true, bIsEditable=true, bIsStaticSwitch=false, bIsVisible=true, type {classStructOrEnum {refPath=&Vector3f}}}
      Rotation Coordinate Space {
       bIsDynamic=false
       bIsEditable=true
       bIsStaticSwitch=true
       bIsVisible=true
       type {classStructOrEnum {refPath=/Niagara/Enums/ENiagaraMeshTransforms}}
      }
     ]
     moduleScript {refPath=&InitialMeshOrientation}
    }
    {
     bIsSetParametersModule=false
     enabled=true
     moduleName=MeshRotationForce
     inputs [
      Rotation {bIsDynamic=true, bIsEditable=true, bIsStaticSwitch=false, bIsVisible=true, type {classStructOrEnum {refPath=&Vector3f}}}
      Coordinate Space {
       bIsDynamic=false
       bIsEditable=true
       bIsStaticSwitch=true
       bIsVisible=true
       type {classStructOrEnum {refPath=&ENiagaraMeshTransforms}}
      }
      "Lever Radius (cm)" {bIsDynamic=false, bIsEditable=true, bIsStaticSwitch=false, bIsVisible=true, type {classStructOrEnum {refPath=&NiagaraFloat}}}
     ]
     moduleScript {refPath=&MeshRotationForce}
    }
    {
     bIsSetParametersModule=false
     enabled=true
     moduleName=ShapeLocation
     inputs [
      Shape Primitive {
       bIsDynamic=false
       bIsEditable=true
       bIsStaticSwitch=true
       bIsVisible=true
       type {classStructOrEnum {refPath=/Niagara/Enums/Location/ENiagara_LocationShapes}}
      }
      Box / Plane Mode {
       bIsDynamic=false
       bIsEditable=true
       bIsStaticSwitch=true
       bIsVisible=true
       type {classStructOrEnum {refPath=/Niagara/Enums/Location/ENiagara_BoxPlaneMode}}
      }
      Cone Mode {
       bIsDynamic=false
       bIsEditable=false
       bIsStaticSwitch=true
       bIsVisible=false
       type {classStructOrEnum {refPath=/Niagara/Enums/Location/ENiagara_ConeMode}}
      }
      Cylinder Height {
       bIsDynamic=false
       bIsEditable=false
       bIsStaticSwitch=false
       bIsVisible=false
       type {classStructOrEnum {refPath=&NiagaraFloat}}
      }
      Sphere Radius {
       bIsDynamic=false
       bIsEditable=false
       bIsStaticSwitch=false
       bIsVisible=false
       type {classStructOrEnum {refPath=&NiagaraFloat}}
      }
      Ring / Disc Mode {
       bIsDynamic=false
       bIsEditable=false
       bIsStaticSwitch=true
       bIsVisible=false
       type {classStructOrEnum {refPath=/Niagara/Enums/Location/ENiagaraRingDiscMode}}
      }
      Torus Mode {
       bIsDynamic=false
       bIsEditable=false
       bIsStaticSwitch=true
       bIsVisible=false
       type {classStructOrEnum {refPath=/Niagara/Enums/ENiagaraShapeTorusMode}}
      }
      Knot Ring Loops {
       bIsDynamic=false
       bIsEditable=false
       bIsStaticSwitch=false
       bIsVisible=false
       type {classStructOrEnum {refPath=&NiagaraInt32}}
      }
      Knot Helix Loops {
       bIsDynamic=false
       bIsEditable=false
       bIsStaticSwitch=false
       bIsVisible=false
       type {classStructOrEnum {refPath=&NiagaraInt32}}
      }
      Enable Torus Knot Radius {bIsDynamic=false, bIsEditable=false, bIsStaticSwitch=false, bIsVisible=false, type {classStructOrEnum {refPath=&NiagaraBool}}}
      Torus Knot Tube Radius {
       bIsDynamic=false
       bIsEditable=false
       bIsStaticSwitch=false
       bIsVisible=false
       type {classStructOrEnum {refPath=&NiagaraFloat}}
      }
      Cylinder Radius {
       bIsDynamic=false
       bIsEditable=false
       bIsStaticSwitch=false
       bIsVisible=false
       type {classStructOrEnum {refPath=&NiagaraFloat}}
      }
      Cylinder Height Midpoint {
       bIsDynamic=false
       bIsEditable=false
       bIsStaticSwitch=false
       bIsVisible=false
       type {classStructOrEnum {refPath=&NiagaraFloat}}
      }
      Ring Radius {
       bIsDynamic=false
       bIsEditable=false
       bIsStaticSwitch=false
       bIsVisible=false
       type {classStructOrEnum {refPath=&NiagaraFloat}}
      }
      Box Size {bIsDynamic=false, bIsEditable=true, bIsStaticSwitch=false, bIsVisible=true, type {classStructOrEnum {refPath=&Vector3f}}}
      Box Midpoint {bIsDynamic=true, bIsEditable=true, bIsStaticSwitch=false, bIsVisible=true, type {classStructOrEnum {refPath=&Vector3f}}}
      Cone Length {
       bIsDynamic=false
       bIsEditable=false
       bIsStaticSwitch=false
       bIsVisible=false
       type {classStructOrEnum {refPath=&NiagaraFloat}}
      }
      Plane Size {
       bIsDynamic=false
       bIsEditable=false
       bIsStaticSwitch=false
       bIsVisible=false
       type {classStructOrEnum {refPath=&Vector2f}}
      }
      Wedge Horizontal Angle {
       bIsDynamic=false
       bIsEditable=false
       bIsStaticSwitch=false
       bIsVisible=false
       type {classStructOrEnum {refPath=&NiagaraFloat}}
      }
      Large Radius {
       bIsDynamic=false
       bIsEditable=false
       bIsStaticSwitch=false
       bIsVisible=false
       type {classStructOrEnum {refPath=&NiagaraFloat}}
      }
      Plane Midpoint {
       bIsDynamic=false
       bIsEditable=false
       bIsStaticSwitch=false
       bIsVisible=false
       type {classStructOrEnum {refPath=&Vector2f}}
      }
      Wedge Vertical Angle {
       bIsDynamic=false
       bIsEditable=false
       bIsStaticSwitch=false
       bIsVisible=false
       type {classStructOrEnum {refPath=&NiagaraFloat}}
      }
      Cone Angle {
       bIsDynamic=false
       bIsEditable=false
       bIsStaticSwitch=false
       bIsVisible=false
       type {classStructOrEnum {refPath=&NiagaraFloat}}
      }
      Handle Radius {
       bIsDynamic=false
       bIsEditable=false
       bIsStaticSwitch=false
       bIsVisible=false
       type {classStructOrEnum {refPath=&NiagaraFloat}}
      }
      Cone Inner Angle {
       bIsDynamic=false
       bIsEditable=false
       bIsStaticSwitch=false
       bIsVisible=false
       type {classStructOrEnum {refPath=&NiagaraFloat}}
      }
      Cone Radial Angle {
       bIsDynamic=false
       bIsEditable=false
       bIsStaticSwitch=false
       bIsVisible=false
       type {classStructOrEnum {refPath=&NiagaraFloat}}
      }
      Cone Angle Type {
       bIsDynamic=false
       bIsEditable=false
       bIsStaticSwitch=true
       bIsVisible=false
       type {classStructOrEnum {refPath=/Niagara/Enums/Angles/ENiagara_AngleInput}}
      }
      Flatten Endcaps {bIsDynamic=false, bIsEditable=false, bIsStaticSwitch=true, bIsVisible=false, type {classStructOrEnum {refPath=&NiagaraBool}}}
      Apply To Particle Position {
       bIsDynamic=false
       bIsEditable=true
       bIsStaticSwitch=true
       bIsVisible=true
       type {classStructOrEnum {refPath=&ENiagara_AttributeSamplingApplyOutput}}
      }
      Cone Surface Distribution {
       bIsDynamic=false
       bIsEditable=false
       bIsStaticSwitch=false
       bIsVisible=false
       type {classStructOrEnum {refPath=&NiagaraFloat}}
      }
      Ring / Disc Distribution Mode {
       bIsDynamic=false
       bIsEditable=false
       bIsStaticSwitch=true
       bIsVisible=false
       type {classStructOrEnum {refPath=/Niagara/Enums/ENiagaraSphereDistributionMode}}
      }
      Sphere Distribution {
       bIsDynamic=false
       bIsEditable=false
       bIsStaticSwitch=true
       bIsVisible=false
       type {classStructOrEnum {refPath=&ENiagaraSphereDistributionMode}}
      }
      Torus Distribution Mode {
       bIsDynamic=false
       bIsEditable=false
       bIsStaticSwitch=true
       bIsVisible=false
       type {classStructOrEnum {refPath=/Niagara/Enums/ENiagaraTorusDistributionMode}}
      }
      Sphere Surface Distribution {
       bIsDynamic=false
       bIsEditable=false
       bIsStaticSwitch=false
       bIsVisible=false
       type {classStructOrEnum {refPath=&NiagaraFloat}}
      }
      Hemisphere Distribution {
       bIsDynamic=false
       bIsEditable=false
       bIsStaticSwitch=false
       bIsVisible=false
       type {classStructOrEnum {refPath=&Vector2f}}
      }
      Cylinder Distribution {
       bIsDynamic=false
       bIsEditable=false
       bIsStaticSwitch=true
       bIsVisible=false
       type {classStructOrEnum {refPath=/Niagara/Enums/Location/ENiagara_CylinderMode}}
      }
      Surface Only {bIsDynamic=false, bIsEditable=true, bIsStaticSwitch=false, bIsVisible=true, type {classStructOrEnum {refPath=&NiagaraBool}}}
      Surface Only Band Thickness {bIsDynamic=false, bIsEditable=false, bIsStaticSwitch=false, bIsVisible=true, type {classStructOrEnum {refPath=&NiagaraFloat}}}
      Use Endcaps In Surface Only Mode {bIsDynamic=false, bIsEditable=false, bIsStaticSwitch=true, bIsVisible=false, type {classStructOrEnum {refPath=&NiagaraBool}}}
      Hemicircle X {bIsDynamic=false, bIsEditable=false, bIsStaticSwitch=true, bIsVisible=false, type {classStructOrEnum {refPath=&NiagaraBool}}}
      Hemicircle Y {bIsDynamic=false, bIsEditable=false, bIsStaticSwitch=true, bIsVisible=false, type {classStructOrEnum {refPath=&NiagaraBool}}}
      Disc Coverage {
       bIsDynamic=false
       bIsEditable=false
       bIsStaticSwitch=false
       bIsVisible=false
       type {classStructOrEnum {refPath=&NiagaraFloat}}
      }
      Hemicircle Internal Cap {bIsDynamic=false, bIsEditable=false, bIsStaticSwitch=true, bIsVisible=false, type {classStructOrEnum {refPath=&NiagaraBool}}}
      U Position {
       bIsDynamic=false
       bIsEditable=false
       bIsStaticSwitch=false
       bIsVisible=false
       type {classStructOrEnum {refPath=&NiagaraFloat}}
      }
      Enable Lathe Profile {bIsDynamic=false, bIsEditable=false, bIsStaticSwitch=true, bIsVisible=false, type {classStructOrEnum {refPath=&NiagaraBool}}}
      Lathe Profile {
       bIsDynamic=false
       bIsEditable=false
       bIsStaticSwitch=false
       bIsVisible=false
       type {classStructOrEnum {refPath=&NiagaraDataInterfaceCurve}}
      }
      Surface Distribution {
       bIsDynamic=false
       bIsEditable=false
       bIsStaticSwitch=false
       bIsVisible=false
       type {classStructOrEnum {refPath=&NiagaraFloat}}
      }
      Surface Expansion Mode {
       bIsDynamic=false
       bIsEditable=false
       bIsStaticSwitch=false
       bIsVisible=false
       type {classStructOrEnum {refPath=/Niagara/Enums/ENiagaraExpansionMode}}
      }
      U Distribution {
       bIsDynamic=false
       bIsEditable=false
       bIsStaticSwitch=false
       bIsVisible=false
       type {classStructOrEnum {refPath=&NiagaraFloat}}
      }
      V Position {
       bIsDynamic=false
       bIsEditable=false
       bIsStaticSwitch=false
       bIsVisible=false
       type {classStructOrEnum {refPath=&NiagaraFloat}}
      }
      Radius Position {
       bIsDynamic=false
       bIsEditable=false
       bIsStaticSwitch=false
       bIsVisible=false
       type {classStructOrEnum {refPath=&NiagaraFloat}}
      }
      Sphere Radius Position {
       bIsDynamic=false
       bIsEditable=false
       bIsStaticSwitch=false
       bIsVisible=false
       type {classStructOrEnum {refPath=&NiagaraFloat}}
      }
      V Distribution {
       bIsDynamic=false
       bIsEditable=false
       bIsStaticSwitch=false
       bIsVisible=false
       type {classStructOrEnum {refPath=&NiagaraFloat}}
      }
      Evaluation Type {
       bIsDynamic=false
       bIsEditable=false
       bIsStaticSwitch=true
       bIsVisible=false
       type {classStructOrEnum {refPath=/Niagara/Enums/ENiagaraRandomnessEvaluation}}
      }
      Uniform Distribution {
       bIsDynamic=false
       bIsEditable=false
       bIsStaticSwitch=false
       bIsVisible=false
       type {classStructOrEnum {refPath=&NiagaraFloat}}
      }
      Uniform Spiral Amount {
       bIsDynamic=false
       bIsEditable=false
       bIsStaticSwitch=false
       bIsVisible=false
       type {classStructOrEnum {refPath=&NiagaraFloat}}
      }
      Uniform Spiral Falloff {
       bIsDynamic=false
       bIsEditable=false
       bIsStaticSwitch=false
       bIsVisible=false
       type {classStructOrEnum {refPath=&NiagaraFloat}}
      }
      Uniform Count {
       bIsDynamic=false
       bIsEditable=false
       bIsStaticSwitch=false
       bIsVisible=false
       type {classStructOrEnum {refPath=&NiagaraInt32}}
      }
      Uniform Particle Index {
       bIsDynamic=false
       bIsEditable=false
       bIsStaticSwitch=false
       bIsVisible=false
       type {classStructOrEnum {refPath=&NiagaraInt32}}
      }
      Hemisphere Angle Type {
       bIsDynamic=false
       bIsEditable=false
       bIsStaticSwitch=true
       bIsVisible=false
       type {classStructOrEnum {refPath=&ENiagara_AngleInput}}
      }
      Uniform Endpoint Offset {
       bIsDynamic=false
       bIsEditable=false
       bIsStaticSwitch=false
       bIsVisible=false
       type {classStructOrEnum {refPath=&NiagaraFloat}}
      }
      MaskSpawnGroup {bIsDynamic=false, bIsEditable=true, bIsStaticSwitch=false, bIsVisible=true, type {classStructOrEnum {refPath=&NiagaraBool}}}
      Spawn Group Mask {bIsDynamic=false, bIsEditable=false, bIsStaticSwitch=false, bIsVisible=true, type {classStructOrEnum {refPath=&NiagaraInt32}}}
      Spawn Group {
       bIsDynamic=false
       bIsEditable=false
       bIsStaticSwitch=false
       bIsVisible=false
       type {classStructOrEnum {refPath=&NiagaraInt32}}
      }
      Randomness Mode {
       bIsDynamic=false
       bIsEditable=true
       bIsStaticSwitch=true
       bIsVisible=true
       type {classStructOrEnum {refPath=&ENiagaraRandomnessMode}}
      }
      Override Seed {bIsDynamic=false, bIsEditable=true, bIsStaticSwitch=false, bIsVisible=true, type {classStructOrEnum {refPath=&NiagaraBool}}}
      Random Seed {bIsDynamic=false, bIsEditable=false, bIsStaticSwitch=false, bIsVisible=true, type {classStructOrEnum {refPath=&NiagaraInt32}}}
      Fixed Random Seed {bIsDynamic=false, bIsEditable=false, bIsStaticSwitch=true, bIsVisible=false, type {classStructOrEnum {refPath=&NiagaraBool}}}
      Transform Order {
       bIsDynamic=false
       bIsEditable=true
       bIsStaticSwitch=true
       bIsVisible=true
       type {classStructOrEnum {refPath=/Niagara/Enums/Transforms/ENiagara_TransformOrder}}
      }
      Scale Mode {
       bIsDynamic=false
       bIsEditable=true
       bIsStaticSwitch=true
       bIsVisible=true
       type {classStructOrEnum {refPath=/Niagara/Enums/Transforms/ENiagara_ScaleMode}}
      }
      Non Uniform Scale {bIsDynamic=false, bIsEditable=true, bIsStaticSwitch=false, bIsVisible=true, type {classStructOrEnum {refPath=&Vector3f}}}
      UseOwnerScale {bIsDynamic=false, bIsEditable=true, bIsStaticSwitch=false, bIsVisible=true, type {classStructOrEnum {refPath=&NiagaraBool}}}
      Apply Owner Scale {bIsDynamic=false, bIsEditable=true, bIsStaticSwitch=false, bIsVisible=true, type {classStructOrEnum {refPath=&Vector3f}}}
      Rotation Mode {
       bIsDynamic=false
       bIsEditable=true
       bIsStaticSwitch=true
       bIsVisible=true
       type {classStructOrEnum {refPath=/Niagara/Enums/Transforms/ENiagara_RotationMode}}
      }
      Rotation Axis {bIsDynamic=false, bIsEditable=true, bIsStaticSwitch=false, bIsVisible=true, type {classStructOrEnum {refPath=&Vector3f}}}
      Rotation Matrix {
       bIsDynamic=false
       bIsEditable=false
       bIsStaticSwitch=false
       bIsVisible=false
       type {classStructOrEnum {refPath=&NiagaraMatrix}}
      }
      Rotation Quaternion {bIsDynamic=false, bIsEditable=false, bIsStaticSwitch=false, bIsVisible=false, type {classStructOrEnum {refPath=&Quat4f}}}
      Yaw / Pitch / Roll {
       bIsDynamic=false
       bIsEditable=false
       bIsStaticSwitch=false
       bIsVisible=false
       type {classStructOrEnum {refPath=&Vector3f}}
      }
      Rotation Angle {bIsDynamic=false, bIsEditable=true, bIsStaticSwitch=false, bIsVisible=true, type {classStructOrEnum {refPath=&NiagaraFloat}}}
      Invert Rotation Quaternion {bIsDynamic=false, bIsEditable=false, bIsStaticSwitch=false, bIsVisible=false, type {classStructOrEnum {refPath=&NiagaraBool}}}
      Additional Quaternion Rotation {bIsDynamic=false, bIsEditable=false, bIsStaticSwitch=false, bIsVisible=false, type {classStructOrEnum {refPath=&NiagaraBool}}}
      Additional Yaw / Pitch / Roll {
       bIsDynamic=false
       bIsEditable=false
       bIsStaticSwitch=false
       bIsVisible=false
       type {classStructOrEnum {refPath=&Vector3f}}
      }
      Rotation Angle Type {
       bIsDynamic=false
       bIsEditable=true
       bIsStaticSwitch=true
       bIsVisible=true
       type {classStructOrEnum {refPath=&ENiagara_AngleInput}}
      }
      Rotation Coordinate Space {
       bIsDynamic=false
       bIsEditable=true
       bIsStaticSwitch=false
       bIsVisible=true
       type {classStructOrEnum {refPath=&ENiagaraCoordinateSpace}}
      }
      Offset Mode {
       bIsDynamic=false
       bIsEditable=true
       bIsStaticSwitch=true
       bIsVisible=true
       type {classStructOrEnum {refPath=/Niagara/Enums/Transforms/ENiagara_OffsetMode}}
      }
      Shape Origin {
       bIsDynamic=false
       bIsEditable=true
       bIsStaticSwitch=false
       bIsVisible=true
       type {classStructOrEnum {refPath=&NiagaraPosition}}
      }
      Offset {bIsDynamic=true, bIsEditable=true, bIsStaticSwitch=false, bIsVisible=true, type {classStructOrEnum {refPath=&Vector3f}}}
      Offset Coordinate Space {
       bIsDynamic=false
       bIsEditable=true
       bIsStaticSwitch=false
       bIsVisible=true
       type {classStructOrEnum {refPath=&ENiagaraCoordinateSpace}}
      }
      Custom Transform Matrix {
       bIsDynamic=false
       bIsEditable=false
       bIsStaticSwitch=false
       bIsVisible=false
       type {classStructOrEnum {refPath=&NiagaraMatrix}}
      }
      Transform Method {
       bIsDynamic=false
       bIsEditable=true
       bIsStaticSwitch=true
       bIsVisible=true
       type {classStructOrEnum {refPath=/Niagara/Enums/Transforms/ENiagara_TransformType}}
      }
     ]
     moduleScript {refPath=&ShapeLocation}
    }
    {
     bIsSetParametersModule=false
     enabled=true
     moduleName=AddVelocity
     inputs [
      Velocity Mode {
       bIsDynamic=false
       bIsEditable=true
       bIsStaticSwitch=true
       bIsVisible=true
       type {classStructOrEnum {refPath=/Niagara/Enums/Utility/ENiagara_VelocityMode}}
      }
      Velocity {bIsDynamic=false, bIsEditable=true, bIsStaticSwitch=false, bIsVisible=true, type {classStructOrEnum {refPath=&Vector3f}}}
      Velocity Speed {
       bIsDynamic=false
       bIsEditable=false
       bIsStaticSwitch=false
       bIsVisible=false
       type {classStructOrEnum {refPath=&NiagaraFloat}}
      }
      Velocity Speed Scale {bIsDynamic=true, bIsEditable=true, bIsStaticSwitch=false, bIsVisible=true, type {classStructOrEnum {refPath=&NiagaraFloat}}}
      Constrain To Radius {
       bIsDynamic=false
       bIsEditable=false
       bIsStaticSwitch=false
       bIsVisible=false
       type {classStructOrEnum {refPath=&NiagaraFloat}}
      }
      Distribution Along Cone Axis {
       bIsDynamic=false
       bIsEditable=false
       bIsStaticSwitch=false
       bIsVisible=false
       type {classStructOrEnum {refPath=&NiagaraFloat}}
      }
      Radius Falloff Near / Far {
       bIsDynamic=false
       bIsEditable=false
       bIsStaticSwitch=false
       bIsVisible=false
       type {classStructOrEnum {refPath=&Vector2f}}
      }
      Radius Falloff Exponent {
       bIsDynamic=false
       bIsEditable=false
       bIsStaticSwitch=false
       bIsVisible=false
       type {classStructOrEnum {refPath=&NiagaraFloat}}
      }
      Speed Falloff From Cone Axis {
       bIsDynamic=false
       bIsEditable=false
       bIsStaticSwitch=false
       bIsVisible=false
       type {classStructOrEnum {refPath=&NiagaraFloat}}
      }
      Invert Velocity Falloff {bIsDynamic=false, bIsEditable=false, bIsStaticSwitch=false, bIsVisible=false, type {classStructOrEnum {refPath=&NiagaraBool}}}
      Override Seed {bIsDynamic=false, bIsEditable=false, bIsStaticSwitch=false, bIsVisible=false, type {classStructOrEnum {refPath=&NiagaraBool}}}
      Random Seed {
       bIsDynamic=false
       bIsEditable=false
       bIsStaticSwitch=false
       bIsVisible=false
       type {classStructOrEnum {refPath=&NiagaraInt32}}
      }
      Fixed Random Seed {bIsDynamic=false, bIsEditable=false, bIsStaticSwitch=true, bIsVisible=false, type {classStructOrEnum {refPath=&NiagaraBool}}}
      Randomness Mode {
       bIsDynamic=false
       bIsEditable=false
       bIsStaticSwitch=true
       bIsVisible=false
       type {classStructOrEnum {refPath=&ENiagaraRandomnessMode}}
      }
      Cone Axis {
       bIsDynamic=false
       bIsEditable=false
       bIsStaticSwitch=false
       bIsVisible=false
       type {classStructOrEnum {refPath=&Vector3f}}
      }
      Cone Angle {
       bIsDynamic=false
       bIsEditable=false
       bIsStaticSwitch=false
       bIsVisible=false
       type {classStructOrEnum {refPath=&NiagaraFloat}}
      }
      Inner Cone Angle {
       bIsDynamic=false
       bIsEditable=false
       bIsStaticSwitch=false
       bIsVisible=false
       type {classStructOrEnum {refPath=&NiagaraFloat}}
      }
      Cone Angle Mode {
       bIsDynamic=false
       bIsEditable=false
       bIsStaticSwitch=true
       bIsVisible=false
       type {classStructOrEnum {refPath=&ENiagara_AngleInput}}
      }
      Rotation Mode {
       bIsDynamic=false
       bIsEditable=true
       bIsStaticSwitch=true
       bIsVisible=true
       type {classStructOrEnum {refPath=&ENiagara_RotationMode}}
      }
      Rotation Axis {
       bIsDynamic=false
       bIsEditable=false
       bIsStaticSwitch=false
       bIsVisible=false
       type {classStructOrEnum {refPath=&Vector3f}}
      }
      Yaw / Pitch / Roll {
       bIsDynamic=false
       bIsEditable=false
       bIsStaticSwitch=false
       bIsVisible=false
       type {classStructOrEnum {refPath=&Vector3f}}
      }
      Rotation Quaternion {bIsDynamic=false, bIsEditable=false, bIsStaticSwitch=false, bIsVisible=false, type {classStructOrEnum {refPath=&Quat4f}}}
      Rotation Matrix {
       bIsDynamic=false
       bIsEditable=false
       bIsStaticSwitch=false
       bIsVisible=false
       type {classStructOrEnum {refPath=&NiagaraMatrix}}
      }
      Rotation Angle {
       bIsDynamic=false
       bIsEditable=false
       bIsStaticSwitch=false
       bIsVisible=false
       type {classStructOrEnum {refPath=&NiagaraFloat}}
      }
      Invert Rotation Quaternion {bIsDynamic=false, bIsEditable=false, bIsStaticSwitch=false, bIsVisible=false, type {classStructOrEnum {refPath=&NiagaraBool}}}
      Additional Quaternion Rotation {bIsDynamic=false, bIsEditable=false, bIsStaticSwitch=false, bIsVisible=false, type {classStructOrEnum {refPath=&NiagaraBool}}}
      Additional Yaw / Pitch / Roll {
       bIsDynamic=false
       bIsEditable=false
       bIsStaticSwitch=false
       bIsVisible=false
       type {classStructOrEnum {refPath=&Vector3f}}
      }
      Rotation Angle Type {
       bIsDynamic=false
       bIsEditable=false
       bIsStaticSwitch=true
       bIsVisible=false
       type {classStructOrEnum {refPath=&ENiagara_AngleInput}}
      }
      Rotation Coordinate Space {
       bIsDynamic=false
       bIsEditable=true
       bIsStaticSwitch=false
       bIsVisible=true
       type {classStructOrEnum {refPath=&ENiagaraCoordinateSpace}}
      }
      Velocity Origin {
       bIsDynamic=false
       bIsEditable=false
       bIsStaticSwitch=false
       bIsVisible=false
       type {classStructOrEnum {refPath=&NiagaraPosition}}
      }
      Origin Offset {
       bIsDynamic=false
       bIsEditable=false
       bIsStaticSwitch=false
       bIsVisible=false
       type {classStructOrEnum {refPath=&Vector3f}}
      }
      Origin Offset Coordinate Space {
       bIsDynamic=false
       bIsEditable=false
       bIsStaticSwitch=false
       bIsVisible=false
       type {classStructOrEnum {refPath=&ENiagaraCoordinateSpace}}
      }
      Default Position {
       bIsDynamic=false
       bIsEditable=false
       bIsStaticSwitch=false
       bIsVisible=false
       type {classStructOrEnum {refPath=&NiagaraPosition}}
      }
     ]
     moduleScript {refPath=&AddVelocity}
    }
    {
     bIsSetParametersModule=false
     enabled=true
     moduleName=ApplyInitialForces
     inputs [
      Apply Force to Velocity {bIsDynamic=false, bIsEditable=true, bIsStaticSwitch=false, bIsVisible=true, type {classStructOrEnum {refPath=&NiagaraBool}}}
      Apply Rotational Force to Rotational Velocity {bIsDynamic=false, bIsEditable=true, bIsStaticSwitch=false, bIsVisible=true, type {classStructOrEnum {refPath=&NiagaraBool}}}
      Apply Force to Position {bIsDynamic=false, bIsEditable=true, bIsStaticSwitch=false, bIsVisible=true, type {classStructOrEnum {refPath=&NiagaraBool}}}
      Positional Force Warmup Time {bIsDynamic=false, bIsEditable=true, bIsStaticSwitch=false, bIsVisible=true, type {classStructOrEnum {refPath=&NiagaraFloat}}}
     ]
     moduleScript {refPath=&ApplyInitialForces}
    }
   ]
  }
  particleUpdateScript {
   scriptName=ParticleUpdateScript
   modules [
    {
     bIsSetParametersModule=false
     enabled=true
     moduleName=ParticleState
     inputs [
      Kill Particles When Lifetime Has Elapsed {bIsDynamic=false, bIsEditable=true, bIsStaticSwitch=true, bIsVisible=true, type {classStructOrEnum {refPath=&NiagaraBool}}}
      Loop Particles Lifetime {bIsDynamic=false, bIsEditable=false, bIsStaticSwitch=false, bIsVisible=false, type {classStructOrEnum {refPath=&NiagaraBool}}}
      Let Infinitely Lived Particles Die When Emitter Deactivates {bIsDynamic=false, bIsEditable=false, bIsStaticSwitch=false, bIsVisible=false, type {classStructOrEnum {refPath=&NiagaraBool}}}
      Lifetime {bIsDynamic=false, bIsEditable=true, bIsStaticSwitch=false, bIsVisible=true, type {classStructOrEnum {refPath=&NiagaraFloat}}}
      DeltaTime {bIsDynamic=false, bIsEditable=true, bIsStaticSwitch=false, bIsVisible=true, type {classStructOrEnum {refPath=&NiagaraFloat}}}
     ]
     moduleScript {refPath=&ParticleState}
    }
    {
     bIsSetParametersModule=false
     enabled=true
     moduleName=ScaleMeshSize
     inputs [
      Scale Factor {bIsDynamic=true, bIsEditable=true, bIsStaticSwitch=false, bIsVisible=true, type {classStructOrEnum {refPath=&Vector3f}}}
      Initial Mesh Scale {bIsDynamic=false, bIsEditable=true, bIsStaticSwitch=false, bIsVisible=true, type {classStructOrEnum {refPath=&Vector3f}}}
     ]
     moduleScript {refPath=&ScaleMeshSize}
    }
    {
     bIsSetParametersModule=false
     enabled=true
     moduleName=ScaleSpriteSize
     inputs [
      Scale Sprite Size Mode {
       bIsDynamic=false
       bIsEditable=true
       bIsStaticSwitch=true
       bIsVisible=true
       type {classStructOrEnum {refPath=&ENiagara_ScaleSpriteSize}}
      }
      Uniform Curve Sprite Scale {
       bIsDynamic=false
       bIsEditable=true
       bIsStaticSwitch=false
       bIsVisible=true
       type {classStructOrEnum {refPath=&NiagaraDataInterfaceCurve}}
      }
      Non-Uniform Curve Sprite Scale {
       bIsDynamic=false
       bIsEditable=false
       bIsStaticSwitch=false
       bIsVisible=false
       type {classStructOrEnum {refPath=&NiagaraDataInterfaceVector2DCurve}}
      }
      Uniform Curve Index {bIsDynamic=false, bIsEditable=true, bIsStaticSwitch=false, bIsVisible=true, type {classStructOrEnum {refPath=&NiagaraFloat}}}
      Non-Uniform Curve Index {
       bIsDynamic=false
       bIsEditable=false
       bIsStaticSwitch=false
       bIsVisible=false
       type {classStructOrEnum {refPath=&NiagaraFloat}}
      }
      Non-Uniform Curve Scale {
       bIsDynamic=false
       bIsEditable=false
       bIsStaticSwitch=false
       bIsVisible=false
       type {classStructOrEnum {refPath=&Vector2f}}
      }
      Uniform Curve Scale {bIsDynamic=true, bIsEditable=true, bIsStaticSwitch=false, bIsVisible=true, type {classStructOrEnum {refPath=&NiagaraFloat}}}
      Scale Factor {
       bIsDynamic=false
       bIsEditable=false
       bIsStaticSwitch=false
       bIsVisible=false
       type {classStructOrEnum {refPath=&Vector2f}}
      }
      Uniform Scale Factor {
       bIsDynamic=false
       bIsEditable=false
       bIsStaticSwitch=false
       bIsVisible=false
       type {classStructOrEnum {refPath=&NiagaraFloat}}
      }
      Initial Sprite Size {bIsDynamic=false, bIsEditable=true, bIsStaticSwitch=false, bIsVisible=true, type {classStructOrEnum {refPath=&Vector2f}}}
     ]
     moduleScript {refPath=&ScaleSpriteSize}
    }
    {
     bIsSetParametersModule=false
     enabled=true
     moduleName=WindForce
     inputs [
      Debug Scale {
       bIsDynamic=false
       bIsEditable=false
       bIsStaticSwitch=false
       bIsVisible=false
       type {classStructOrEnum {refPath=&NiagaraFloat}}
      }
      Debug {
       bIsDynamic=false
       bIsEditable=false
       bIsStaticSwitch=false
       bIsVisible=false
       type {classStructOrEnum {refPath=&NiagaraDataInterfaceDebugDraw}}
      }
      Wind Speed {bIsDynamic=true, bIsEditable=true, bIsStaticSwitch=false, bIsVisible=true, type {classStructOrEnum {refPath=&Vector3f}}}
      Wind Speed Scale {bIsDynamic=false, bIsEditable=true, bIsStaticSwitch=false, bIsVisible=true, type {classStructOrEnum {refPath=&NiagaraFloat}}}
      Turbulence Mode {
       bIsDynamic=false
       bIsEditable=true
       bIsStaticSwitch=true
       bIsVisible=true
       type {classStructOrEnum {refPath=/Niagara/Enums/Wind/ENiagaraWindTurbulenceMode}}
      }
      Turbulence Scale Mode {
       bIsDynamic=false
       bIsEditable=false
       bIsStaticSwitch=true
       bIsVisible=false
       type {classStructOrEnum {refPath=/Niagara/Enums/Wind/ENiagaraWindTurbulenceContributionMode}}
      }
      Scale {
       bIsDynamic=false
       bIsEditable=false
       bIsStaticSwitch=false
       bIsVisible=false
       type {classStructOrEnum {refPath=&NiagaraFloat}}
      }
      Scale Min / Max {
       bIsDynamic=false
       bIsEditable=false
       bIsStaticSwitch=false
       bIsVisible=false
       type {classStructOrEnum {refPath=&Vector2f}}
      }
      Speed Min / Max {
       bIsDynamic=false
       bIsEditable=false
       bIsStaticSwitch=false
       bIsVisible=false
       type {classStructOrEnum {refPath=&Vector2f}}
      }
      Scale Curve {
       bIsDynamic=false
       bIsEditable=false
       bIsStaticSwitch=false
       bIsVisible=false
       type {classStructOrEnum {refPath=&NiagaraDataInterfaceCurve}}
      }
      Turbulence Frequency Mode {
       bIsDynamic=false
       bIsEditable=false
       bIsStaticSwitch=true
       bIsVisible=false
       type {classStructOrEnum {refPath=/Niagara/Enums/Wind/ENiagaraWindTurbulenceFrequencyMode}}
      }
      Frequency Min / Max {
       bIsDynamic=false
       bIsEditable=false
       bIsStaticSwitch=false
       bIsVisible=false
       type {classStructOrEnum {refPath=&Vector2f}}
      }
      Frequency Blend {
       bIsDynamic=false
       bIsEditable=false
       bIsStaticSwitch=false
       bIsVisible=false
       type {classStructOrEnum {refPath=&NiagaraFloat}}
      }
      Frequency {
       bIsDynamic=false
       bIsEditable=false
       bIsStaticSwitch=false
       bIsVisible=false
       type {classStructOrEnum {refPath=&NiagaraFloat}}
      }
      Turbulence Offset Mode {
       bIsDynamic=false
       bIsEditable=false
       bIsStaticSwitch=true
       bIsVisible=false
       type {classStructOrEnum {refPath=/Niagara/Enums/Wind/ENiagaraWindOffsetMode}}
      }
      Offset By Time {
       bIsDynamic=false
       bIsEditable=false
       bIsStaticSwitch=false
       bIsVisible=false
       type {classStructOrEnum {refPath=&Vector3f}}
      }
      Offset {
       bIsDynamic=false
       bIsEditable=false
       bIsStaticSwitch=false
       bIsVisible=false
       type {classStructOrEnum {refPath=&Vector3f}}
      }
      Turbulence Flow Loop Duration {
       bIsDynamic=false
       bIsEditable=false
       bIsStaticSwitch=false
       bIsVisible=false
       type {classStructOrEnum {refPath=&NiagaraFloat}}
      }
      Noise Quality / Cost {
       bIsDynamic=false
       bIsEditable=false
       bIsStaticSwitch=true
       bIsVisible=false
       type {classStructOrEnum {refPath=/Niagara/Enums/ENiagaraCurlNoiseQuality}}
      }
      Curl Noise {
       bIsDynamic=false
       bIsEditable=false
       bIsStaticSwitch=false
       bIsVisible=false
       type {classStructOrEnum {refPath=&NiagaraDataInterfaceCurlNoise}}
      }
      Time Variable {
       bIsDynamic=false
       bIsEditable=false
       bIsStaticSwitch=false
       bIsVisible=false
       type {classStructOrEnum {refPath=&NiagaraFloat}}
      }
      Surface Alignment Mode {
       bIsDynamic=false
       bIsEditable=false
       bIsStaticSwitch=true
       bIsVisible=false
       type {classStructOrEnum {refPath=/Niagara/Enums/Wind/ENiagaraWindCombingMode}}
      }
      Alignment Falloff Start / End {
       bIsDynamic=false
       bIsEditable=false
       bIsStaticSwitch=false
       bIsVisible=false
       type {classStructOrEnum {refPath=&Vector2f}}
      }
      Friction Mode {
       bIsDynamic=false
       bIsEditable=false
       bIsStaticSwitch=true
       bIsVisible=false
       type {classStructOrEnum {refPath=/Niagara/Enums/Wind/ENiagaraWindFrictionMode}}
      }
      Friction Distance Mode {
       bIsDynamic=false
       bIsEditable=false
       bIsStaticSwitch=true
       bIsVisible=false
       type {classStructOrEnum {refPath=/Niagara/Enums/Wind/ENiagaraWindFrictionDistanceMode}}
      }
      Friction Distance Limit {
       bIsDynamic=false
       bIsEditable=false
       bIsStaticSwitch=false
       bIsVisible=false
       type {classStructOrEnum {refPath=&NiagaraFloat}}
      }
      Friction {
       bIsDynamic=false
       bIsEditable=false
       bIsStaticSwitch=false
       bIsVisible=false
       type {classStructOrEnum {refPath=&NiagaraFloat}}
      }
      Friction Falloff Start / End {
       bIsDynamic=false
       bIsEditable=false
       bIsStaticSwitch=false
       bIsVisible=false
       type {classStructOrEnum {refPath=&Vector2f}}
      }
      Normal Friction Mask {
       bIsDynamic=false
       bIsEditable=false
       bIsStaticSwitch=false
       bIsVisible=false
       type {classStructOrEnum {refPath=&NiagaraFloat}}
      }
      Use Depth Buffer {bIsDynamic=false, bIsEditable=false, bIsStaticSwitch=true, bIsVisible=false, type {classStructOrEnum {refPath=&NiagaraBool}}}
      Use Landscape {bIsDynamic=false, bIsEditable=false, bIsStaticSwitch=true, bIsVisible=false, type {classStructOrEnum {refPath=&NiagaraBool}}}
      Use Ground Mask {bIsDynamic=false, bIsEditable=true, bIsStaticSwitch=true, bIsVisible=true, type {classStructOrEnum {refPath=&NiagaraBool}}}
      Ground Mask Elevation Mode {
       bIsDynamic=false
       bIsEditable=false
       bIsStaticSwitch=true
       bIsVisible=false
       type {classStructOrEnum {refPath=/Niagara/Enums/Wind/ENiagaraWindGroundMaskMode}}
      }
      Minimum Ground Elevation {
       bIsDynamic=false
       bIsEditable=false
       bIsStaticSwitch=false
       bIsVisible=false
       type {classStructOrEnum {refPath=&NiagaraFloat}}
      }
      Ground Elevation {
       bIsDynamic=false
       bIsEditable=false
       bIsStaticSwitch=false
       bIsVisible=false
       type {classStructOrEnum {refPath=&NiagaraFloat}}
      }
      Ground Distance Min / Max {
       bIsDynamic=false
       bIsEditable=false
       bIsStaticSwitch=false
       bIsVisible=false
       type {classStructOrEnum {refPath=&Vector2f}}
      }
      Ground Mask Min / Max {
       bIsDynamic=false
       bIsEditable=false
       bIsStaticSwitch=false
       bIsVisible=false
       type {classStructOrEnum {refPath=&Vector2f}}
      }
      Use Physics Field {bIsDynamic=false, bIsEditable=true, bIsStaticSwitch=true, bIsVisible=true, type {classStructOrEnum {refPath=&NiagaraBool}}}
      Physics Field Scale {
       bIsDynamic=false
       bIsEditable=false
       bIsStaticSwitch=false
       bIsVisible=false
       type {classStructOrEnum {refPath=&NiagaraFloat}}
      }
      Position {
       bIsDynamic=false
       bIsEditable=true
       bIsStaticSwitch=false
       bIsVisible=true
       type {classStructOrEnum {refPath=&NiagaraPosition}}
      }
      CollisionQuery {
       bIsDynamic=false
       bIsEditable=false
       bIsStaticSwitch=false
       bIsVisible=false
       type {classStructOrEnum {refPath=&NiagaraDataInterfaceCollisionQuery}}
      }
      Physics Field {
       bIsDynamic=false
       bIsEditable=false
       bIsStaticSwitch=false
       bIsVisible=false
       type {classStructOrEnum {refPath=/Script/ChaosNiagara.NiagaraDataInterfacePhysicsField}}
      }
      Landscape Sample {
       bIsDynamic=false
       bIsEditable=false
       bIsStaticSwitch=false
       bIsVisible=false
       type {classStructOrEnum {refPath=&NiagaraDataInterfaceLandscape}}
      }
     ]
     moduleScript {refPath=&WindForce}
    }
    {
     bIsSetParametersModule=false
     enabled=true
     moduleName=VortexForce
     inputs [
      Vortex Force Amount {bIsDynamic=true, bIsEditable=true, bIsStaticSwitch=false, bIsVisible=true, type {classStructOrEnum {refPath=&NiagaraFloat}}}
      Vortex Axis {bIsDynamic=false, bIsEditable=true, bIsStaticSwitch=false, bIsVisible=true, type {classStructOrEnum {refPath=&Vector3f}}}
      Vortex Axis Coordinate Space {
       bIsDynamic=false
       bIsEditable=true
       bIsStaticSwitch=false
       bIsVisible=true
       type {classStructOrEnum {refPath=&ENiagaraCoordinateSpace}}
      }
      Vortex Influence Position {
       bIsDynamic=false
       bIsEditable=true
       bIsStaticSwitch=false
       bIsVisible=true
       type {classStructOrEnum {refPath=&NiagaraPosition}}
      }
      Write to Intrinsic Parameters {bIsDynamic=false, bIsEditable=true, bIsStaticSwitch=true, bIsVisible=true, type {classStructOrEnum {refPath=&NiagaraBool}}}
      Vortex Origin {
       bIsDynamic=false
       bIsEditable=true
       bIsStaticSwitch=false
       bIsVisible=true
       type {classStructOrEnum {refPath=&NiagaraPosition}}
      }
      Origin Pull Amount {bIsDynamic=false, bIsEditable=true, bIsStaticSwitch=false, bIsVisible=true, type {classStructOrEnum {refPath=&NiagaraFloat}}}
      Origin Pull Only On Vortex Axis {bIsDynamic=false, bIsEditable=true, bIsStaticSwitch=true, bIsVisible=true, type {classStructOrEnum {refPath=&NiagaraBool}}}
      UseOriginOffset {bIsDynamic=false, bIsEditable=true, bIsStaticSwitch=false, bIsVisible=true, type {classStructOrEnum {refPath=&NiagaraBool}}}
      Vortex Origin Offset {bIsDynamic=false, bIsEditable=false, bIsStaticSwitch=false, bIsVisible=true, type {classStructOrEnum {refPath=&Vector3f}}}
      Vortex Origin Offset Coordinate Space {
       bIsDynamic=false
       bIsEditable=false
       bIsStaticSwitch=false
       bIsVisible=false
       type {classStructOrEnum {refPath=&ENiagaraCoordinateSpace}}
      }
      Use Influence Falloff {bIsDynamic=false, bIsEditable=true, bIsStaticSwitch=true, bIsVisible=true, type {classStructOrEnum {refPath=&NiagaraBool}}}
      Influence Falloff Radius {bIsDynamic=false, bIsEditable=true, bIsStaticSwitch=false, bIsVisible=true, type {classStructOrEnum {refPath=&NiagaraFloat}}}
      Influence Falloff Exponent {bIsDynamic=false, bIsEditable=true, bIsStaticSwitch=false, bIsVisible=true, type {classStructOrEnum {refPath=&NiagaraFloat}}}
      Invert Influence Falloff {bIsDynamic=false, bIsEditable=true, bIsStaticSwitch=false, bIsVisible=true, type {classStructOrEnum {refPath=&NiagaraBool}}}
      Debug Every N Particles {
       bIsDynamic=false
       bIsEditable=false
       bIsStaticSwitch=false
       bIsVisible=false
       type {classStructOrEnum {refPath=&NiagaraInt32}}
      }
      Debug Force Line Color {
       bIsDynamic=false
       bIsEditable=false
       bIsStaticSwitch=false
       bIsVisible=false
       type {classStructOrEnum {refPath=&LinearColor}}
      }
      Debug Force Line Length {
       bIsDynamic=false
       bIsEditable=false
       bIsStaticSwitch=false
       bIsVisible=false
       type {classStructOrEnum {refPath=&NiagaraFloat}}
      }
      Debug Influence Falloff Color {
       bIsDynamic=false
       bIsEditable=false
       bIsStaticSwitch=false
       bIsVisible=false
       type {classStructOrEnum {refPath=&LinearColor}}
      }
     ]
     moduleScript {refPath=&VortexForce}
    }
    {
     bIsSetParametersModule=false
     enabled=true
     moduleName=Drag
     inputs [
      Use Linear Drag {bIsDynamic=false, bIsEditable=true, bIsStaticSwitch=false, bIsVisible=true, type {classStructOrEnum {refPath=&NiagaraBool}}}
      Drag {bIsDynamic=true, bIsEditable=true, bIsStaticSwitch=false, bIsVisible=true, type {classStructOrEnum {refPath=&NiagaraFloat}}}
      Use Rotational Drag {bIsDynamic=false, bIsEditable=true, bIsStaticSwitch=false, bIsVisible=true, type {classStructOrEnum {refPath=&NiagaraBool}}}
      Rotational Drag {bIsDynamic=false, bIsEditable=false, bIsStaticSwitch=false, bIsVisible=true, type {classStructOrEnum {refPath=&NiagaraFloat}}}
      Ignore Mass {bIsDynamic=false, bIsEditable=true, bIsStaticSwitch=false, bIsVisible=true, type {classStructOrEnum {refPath=&NiagaraBool}}}
     ]
     moduleScript {refPath=&Drag}
    }
    {
     bIsSetParametersModule=false
     enabled=true
     moduleName=OffsetPosition
     inputs [
      Position to Offset {
       bIsDynamic=false
       bIsEditable=true
       bIsStaticSwitch=false
       bIsVisible=true
       type {classStructOrEnum {refPath=&NiagaraPosition}}
      }
      Position Output {bIsDynamic=false, bIsEditable=true, bIsStaticSwitch=true, bIsVisible=true, type {classStructOrEnum {refPath=&NiagaraInt32}}}
      Offset Mode {bIsDynamic=false, bIsEditable=true, bIsStaticSwitch=true, bIsVisible=true, type {classStructOrEnum {refPath=&NiagaraInt32}}}
      Jitter Mode {bIsDynamic=false, bIsEditable=false, bIsStaticSwitch=true, bIsVisible=false, type {classStructOrEnum {refPath=&NiagaraInt32}}}
      Jitter Offset {
       bIsDynamic=false
       bIsEditable=false
       bIsStaticSwitch=false
       bIsVisible=false
       type {classStructOrEnum {refPath=&Vector3f}}
      }
      Jitter Offset Coordinate Space {
       bIsDynamic=false
       bIsEditable=false
       bIsStaticSwitch=false
       bIsVisible=false
       type {classStructOrEnum {refPath=&ENiagaraCoordinateSpace}}
      }
      Jitter Scale {
       bIsDynamic=false
       bIsEditable=false
       bIsStaticSwitch=false
       bIsVisible=false
       type {classStructOrEnum {refPath=&Vector3f}}
      }
      Jitter Amount {
       bIsDynamic=false
       bIsEditable=false
       bIsStaticSwitch=false
       bIsVisible=false
       type {classStructOrEnum {refPath=&NiagaraFloat}}
      }
      Jitter Delay {
       bIsDynamic=false
       bIsEditable=false
       bIsStaticSwitch=false
       bIsVisible=false
       type {classStructOrEnum {refPath=&NiagaraFloat}}
      }
      DeltaTime {
       bIsDynamic=false
       bIsEditable=false
       bIsStaticSwitch=false
       bIsVisible=false
       type {classStructOrEnum {refPath=&NiagaraFloat}}
      }
      Jitter Time Offset {
       bIsDynamic=false
       bIsEditable=false
       bIsStaticSwitch=false
       bIsVisible=false
       type {classStructOrEnum {refPath=&NiagaraFloat}}
      }
      Jitter Interpolation {bIsDynamic=false, bIsEditable=false, bIsStaticSwitch=true, bIsVisible=false, type {classStructOrEnum {refPath=&NiagaraInt32}}}
      Interpolation Curve {
       bIsDynamic=false
       bIsEditable=false
       bIsStaticSwitch=false
       bIsVisible=false
       type {classStructOrEnum {refPath=&NiagaraDataInterfaceCurve}}
      }
      Interpolation Scale {
       bIsDynamic=false
       bIsEditable=false
       bIsStaticSwitch=false
       bIsVisible=false
       type {classStructOrEnum {refPath=&NiagaraFloat}}
      }
      Position Offset {bIsDynamic=true, bIsEditable=true, bIsStaticSwitch=false, bIsVisible=true, type {classStructOrEnum {refPath=&Vector3f}}}
      Position Offset Coordinate Space {
       bIsDynamic=false
       bIsEditable=true
       bIsStaticSwitch=false
       bIsVisible=true
       type {classStructOrEnum {refPath=&ENiagaraCoordinateSpace}}
      }
      Debug Draw {
       bIsDynamic=false
       bIsEditable=false
       bIsStaticSwitch=false
       bIsVisible=false
       type {classStructOrEnum {refPath=&NiagaraDataInterfaceDebugDraw}}
      }
      SkipID {
       bIsDynamic=false
       bIsEditable=false
       bIsStaticSwitch=false
       bIsVisible=false
       type {classStructOrEnum {refPath=&NiagaraInt32}}
      }
      Skip {
       bIsDynamic=false
       bIsEditable=false
       bIsStaticSwitch=false
       bIsVisible=false
       type {classStructOrEnum {refPath=&NiagaraInt32}}
      }
      Gizmo Scale {
       bIsDynamic=false
       bIsEditable=false
       bIsStaticSwitch=false
       bIsVisible=false
       type {classStructOrEnum {refPath=&NiagaraFloat}}
      }
     ]
     moduleScript {refPath=&OffsetPosition}
    }
    {
     bIsSetParametersModule=true
     enabled=true
     moduleName=SetVariables_#1
     inputs [
      Particles.Color_Halo {
       bIsDynamic=true
       bIsEditable=true
       bIsStaticSwitch=false
       bIsVisible=true
       type {classStructOrEnum {refPath=&LinearColor}}
      }
      Particles.CameraOffset_Halo {bIsDynamic=false, bIsEditable=true, bIsStaticSwitch=false, bIsVisible=true, type {classStructOrEnum {refPath=&NiagaraFloat}}}
     ]
     moduleScript {refPath=&self:NiagaraEmitter_0.NiagaraScriptSource_0.NiagaraGraph_0.NiagaraNodeAssignment_0.SetVariables_#1}
    }
    {
     bIsSetParametersModule=false
     enabled=true
     moduleName=SolveForcesAndVelocity
     inputs [
      Clamp Velocity {bIsDynamic=false, bIsEditable=true, bIsStaticSwitch=false, bIsVisible=true, type {classStructOrEnum {refPath=&NiagaraBool}}}
      Speed Limit {bIsDynamic=false, bIsEditable=false, bIsStaticSwitch=false, bIsVisible=true, type {classStructOrEnum {refPath=&NiagaraFloat}}}
      Limit Acceleration {bIsDynamic=false, bIsEditable=true, bIsStaticSwitch=false, bIsVisible=true, type {classStructOrEnum {refPath=&NiagaraBool}}}
      Acceleration Limit {bIsDynamic=false, bIsEditable=false, bIsStaticSwitch=false, bIsVisible=true, type {classStructOrEnum {refPath=&NiagaraFloat}}}
      Manually Enable Rotational Solver {bIsDynamic=false, bIsEditable=true, bIsStaticSwitch=false, bIsVisible=true, type {classStructOrEnum {refPath=&NiagaraBool}}}
      Write to Intrinsic Properties {bIsDynamic=false, bIsEditable=true, bIsStaticSwitch=false, bIsVisible=true, type {classStructOrEnum {refPath=&NiagaraBool}}}
      Write to Presolve Properties {bIsDynamic=false, bIsEditable=true, bIsStaticSwitch=true, bIsVisible=true, type {classStructOrEnum {refPath=&NiagaraBool}}}
      DeltaTime {bIsDynamic=false, bIsEditable=true, bIsStaticSwitch=false, bIsVisible=true, type {classStructOrEnum {refPath=&NiagaraFloat}}}
      Mass {bIsDynamic=false, bIsEditable=true, bIsStaticSwitch=false, bIsVisible=true, type {classStructOrEnum {refPath=&NiagaraFloat}}}
      Position {
       bIsDynamic=false
       bIsEditable=true
       bIsStaticSwitch=false
       bIsVisible=true
       type {classStructOrEnum {refPath=&NiagaraPosition}}
      }
      Velocity {bIsDynamic=false, bIsEditable=true, bIsStaticSwitch=false, bIsVisible=true, type {classStructOrEnum {refPath=&Vector3f}}}
      PreviousVelocity {bIsDynamic=false, bIsEditable=true, bIsStaticSwitch=false, bIsVisible=true, type {classStructOrEnum {refPath=&Vector3f}}}
      Force {bIsDynamic=false, bIsEditable=true, bIsStaticSwitch=false, bIsVisible=true, type {classStructOrEnum {refPath=&Vector3f}}}
      Mesh Orientation {bIsDynamic=false, bIsEditable=true, bIsStaticSwitch=false, bIsVisible=true, type {classStructOrEnum {refPath=&Quat4f}}}
      Rotational Velocity {bIsDynamic=false, bIsEditable=true, bIsStaticSwitch=false, bIsVisible=true, type {classStructOrEnum {refPath=&Vector3f}}}
      Rotational Force {bIsDynamic=false, bIsEditable=true, bIsStaticSwitch=false, bIsVisible=true, type {classStructOrEnum {refPath=&Vector3f}}}
      Rotational Inertia {bIsDynamic=false, bIsEditable=true, bIsStaticSwitch=false, bIsVisible=true, type {classStructOrEnum {refPath=&NiagaraFloat}}}
     ]
     moduleScript {refPath=&SolveForcesAndVelocity}
    }
    {
     bIsSetParametersModule=false
     enabled=true
     moduleName=PlayPersistentAudio
     inputs [
      Play Audio {bIsDynamic=false, bIsEditable=true, bIsStaticSwitch=false, bIsVisible=true, type {classStructOrEnum {refPath=&NiagaraBool}}}
      Audio Playback Position {
       bIsDynamic=false
       bIsEditable=true
       bIsStaticSwitch=false
       bIsVisible=true
       type {classStructOrEnum {refPath=&NiagaraPosition}}
      }
      Audio Position Coordinate Space {
       bIsDynamic=false
       bIsEditable=true
       bIsStaticSwitch=false
       bIsVisible=true
       type {classStructOrEnum {refPath=&ENiagaraCoordinateSpace}}
      }
      Pitch {bIsDynamic=false, bIsEditable=true, bIsStaticSwitch=false, bIsVisible=true, type {classStructOrEnum {refPath=&NiagaraFloat}}}
      Volume {bIsDynamic=false, bIsEditable=true, bIsStaticSwitch=false, bIsVisible=true, type {classStructOrEnum {refPath=&NiagaraFloat}}}
      Filter Audio Playback By Distance {bIsDynamic=false, bIsEditable=true, bIsStaticSwitch=true, bIsVisible=true, type {classStructOrEnum {refPath=&NiagaraBool}}}
      Number Of Nearest Particles To Play Audio {
       bIsDynamic=false
       bIsEditable=false
       bIsStaticSwitch=false
       bIsVisible=false
       type {classStructOrEnum {refPath=&NiagaraInt32}}
      }
      Enable Fade In / Out {bIsDynamic=false, bIsEditable=true, bIsStaticSwitch=true, bIsVisible=true, type {classStructOrEnum {refPath=&NiagaraBool}}}
      Fade In Time {
       bIsDynamic=false
       bIsEditable=false
       bIsStaticSwitch=false
       bIsVisible=false
       type {classStructOrEnum {refPath=&NiagaraFloat}}
      }
      Fade Out Time {
       bIsDynamic=false
       bIsEditable=false
       bIsStaticSwitch=false
       bIsVisible=false
       type {classStructOrEnum {refPath=&NiagaraFloat}}
      }
      Audio Player {
       bIsDynamic=false
       bIsEditable=true
       bIsStaticSwitch=false
       bIsVisible=true
       type {classStructOrEnum {refPath=&NiagaraDataInterfaceAudioPlayer}}
      }
      Rotation {bIsDynamic=false, bIsEditable=true, bIsStaticSwitch=false, bIsVisible=true, type {classStructOrEnum {refPath=&Vector3f}}}
      Start Time {bIsDynamic=false, bIsEditable=true, bIsStaticSwitch=false, bIsVisible=true, type {classStructOrEnum {refPath=&NiagaraFloat}}}
      Camera Query {
       bIsDynamic=false
       bIsEditable=false
       bIsStaticSwitch=false
       bIsVisible=false
       type {classStructOrEnum {refPath=&NiagaraDataInterfaceCamera}}
      }
     ]
     moduleScript {refPath=&PlayPersistentAudio}
    }
    {
     bIsSetParametersModule=false
     enabled=true
     moduleName=BlendSpace
     inputs [
      Position Mode {bIsDynamic=false, bIsEditable=true, bIsStaticSwitch=true, bIsVisible=true, type {classStructOrEnum {refPath=&NiagaraInt32}}}
      Position Coordinate Space {bIsDynamic=false, bIsEditable=true, bIsStaticSwitch=true, bIsVisible=true, type {classStructOrEnum {refPath=&NiagaraInt32}}}
      "Position Blend (World <-> Local)" {bIsDynamic=true, bIsEditable=true, bIsStaticSwitch=false, bIsVisible=true, type {classStructOrEnum {refPath=&NiagaraFloat}}}
      Mesh Orientation Mode {bIsDynamic=false, bIsEditable=true, bIsStaticSwitch=true, bIsVisible=true, type {classStructOrEnum {refPath=&NiagaraInt32}}}
      Mesh Orientation Coordinate Space {bIsDynamic=false, bIsEditable=false, bIsStaticSwitch=true, bIsVisible=false, type {classStructOrEnum {refPath=&NiagaraInt32}}}
      "Mesh Orientation Blend (World <-> Local)" {
       bIsDynamic=false
       bIsEditable=false
       bIsStaticSwitch=false
       bIsVisible=false
       type {classStructOrEnum {refPath=&NiagaraFloat}}
      }
      Velocity Mode {bIsDynamic=false, bIsEditable=true, bIsStaticSwitch=true, bIsVisible=true, type {classStructOrEnum {refPath=&NiagaraInt32}}}
      Velocity Coordinate Space {bIsDynamic=false, bIsEditable=false, bIsStaticSwitch=true, bIsVisible=false, type {classStructOrEnum {refPath=&NiagaraInt32}}}
      "Velocity Blend (World <-> Local)" {
       bIsDynamic=false
       bIsEditable=false
       bIsStaticSwitch=false
       bIsVisible=false
       type {classStructOrEnum {refPath=&NiagaraFloat}}
      }
     ]
     moduleScript {refPath=&BlendSpace}
    }
    {
     bIsSetParametersModule=false
     enabled=true
     moduleName=KillParticlesInVolume001
     inputs [
      Kill Shape {
       bIsDynamic=false
       bIsEditable=true
       bIsStaticSwitch=true
       bIsVisible=true
       type {classStructOrEnum {refPath=/Niagara/Enums/ENiagaraKillVolumeOptions}}
      }
      Box Size {bIsDynamic=false, bIsEditable=true, bIsStaticSwitch=false, bIsVisible=true, type {classStructOrEnum {refPath=&Vector3f}}}
      Plane Normal {
       bIsDynamic=false
       bIsEditable=false
       bIsStaticSwitch=false
       bIsVisible=false
       type {classStructOrEnum {refPath=&Vector3f}}
      }
      Slab Width {
       bIsDynamic=false
       bIsEditable=false
       bIsStaticSwitch=false
       bIsVisible=false
       type {classStructOrEnum {refPath=&NiagaraFloat}}
      }
      Sphere Radius {
       bIsDynamic=false
       bIsEditable=false
       bIsStaticSwitch=false
       bIsVisible=false
       type {classStructOrEnum {refPath=&NiagaraFloat}}
      }
      Slab Axis {
       bIsDynamic=false
       bIsEditable=false
       bIsStaticSwitch=false
       bIsVisible=false
       type {classStructOrEnum {refPath=&Vector3f}}
      }
      Cone Angle {
       bIsDynamic=false
       bIsEditable=false
       bIsStaticSwitch=false
       bIsVisible=false
       type {classStructOrEnum {refPath=&NiagaraFloat}}
      }
      Cone Axis {
       bIsDynamic=false
       bIsEditable=false
       bIsStaticSwitch=false
       bIsVisible=false
       type {classStructOrEnum {refPath=&Vector3f}}
      }
      Cone Length {
       bIsDynamic=false
       bIsEditable=false
       bIsStaticSwitch=false
       bIsVisible=false
       type {classStructOrEnum {refPath=&NiagaraFloat}}
      }
      Kill Volume Enabled {bIsDynamic=false, bIsEditable=true, bIsStaticSwitch=false, bIsVisible=true, type {classStructOrEnum {refPath=&NiagaraBool}}}
      Invert Volume {bIsDynamic=false, bIsEditable=true, bIsStaticSwitch=false, bIsVisible=true, type {classStructOrEnum {refPath=&NiagaraBool}}}
      Volume Origin {bIsDynamic=false, bIsEditable=true, bIsStaticSwitch=false, bIsVisible=true, type {classStructOrEnum {refPath=&Vector3f}}}
      Origin Offset {bIsDynamic=false, bIsEditable=true, bIsStaticSwitch=false, bIsVisible=true, type {classStructOrEnum {refPath=&Vector3f}}}
      Position {
       bIsDynamic=false
       bIsEditable=true
       bIsStaticSwitch=false
       bIsVisible=true
       type {classStructOrEnum {refPath=&NiagaraPosition}}
      }
     ]
     moduleScript {refPath=&KillParticlesInVolume}
    }
   ]
  }
  rendererClasses [{refPath=&NiagaraMeshRendererProperties}, {refPath=&NiagaraSpriteRendererProperties}]
  renderers [
   {rendererIndex=0, rendererClass {refPath=&NiagaraMeshRendererProperties}}
   {rendererIndex=1, rendererClass {refPath=&NiagaraSpriteRendererProperties}}
  ]
 }
 LW_FogHalo {
  emitterSpawnScript {scriptName=None, modules []}
  emitterUpdateScript {scriptName=None, modules []}
  particleSpawnScript {scriptName=None, modules []}
  particleUpdateScript {scriptName=None, modules []}
  rendererClasses []
  renderers []
 }
 LW_Fog_Axis {
  emitterSpawnScript {scriptName=None, modules []}
  emitterUpdateScript {scriptName=None, modules []}
  particleSpawnScript {scriptName=None, modules []}
  particleUpdateScript {scriptName=None, modules []}
  rendererClasses []
  renderers []
 }
 Light {
  emitterSpawnScript {scriptName=EmitterSpawnScript, modules []}
  emitterUpdateScript {
   scriptName=EmitterUpdateScript
   modules [
    {
     bIsSetParametersModule=false
     enabled=true
     moduleName=EmitterState
     inputs [
      Life Cycle Mode {
       bIsDynamic=false
       bIsEditable=true
       bIsStaticSwitch=true
       bIsVisible=true
       type {classStructOrEnum {refPath=&ENiagaraEmitterLifeCycleMode}}
      }
      Inactive Response {
       bIsDynamic=false
       bIsEditable=true
       bIsStaticSwitch=true
       bIsVisible=true
       type {classStructOrEnum {refPath=&ENiagaraInactiveMode}}
      }
      Loop Behavior {
       bIsDynamic=false
       bIsEditable=true
       bIsStaticSwitch=true
       bIsVisible=true
       type {classStructOrEnum {refPath=&ENiagara_EmitterStateOptions}}
      }
      Loop Duration Mode {
       bIsDynamic=false
       bIsEditable=true
       bIsStaticSwitch=true
       bIsVisible=true
       type {classStructOrEnum {refPath=&ENiagara_InfiniteLoopDuration}}
      }
      Loop Count {
       bIsDynamic=false
       bIsEditable=false
       bIsStaticSwitch=false
       bIsVisible=false
       type {classStructOrEnum {refPath=&NiagaraInt32}}
      }
      Loop Duration {
       bIsDynamic=false
       bIsEditable=false
       bIsStaticSwitch=false
       bIsVisible=false
       type {classStructOrEnum {refPath=&NiagaraFloat}}
      }
      Recalculate Duration Each Loop {bIsDynamic=false, bIsEditable=false, bIsStaticSwitch=false, bIsVisible=false, type {classStructOrEnum {refPath=&NiagaraBool}}}
      UseLoopDelay {bIsDynamic=false, bIsEditable=true, bIsStaticSwitch=false, bIsVisible=true, type {classStructOrEnum {refPath=&NiagaraBool}}}
      Loop Delay {bIsDynamic=false, bIsEditable=false, bIsStaticSwitch=false, bIsVisible=true, type {classStructOrEnum {refPath=&NiagaraFloat}}}
      Delay First Loop Only {bIsDynamic=false, bIsEditable=false, bIsStaticSwitch=false, bIsVisible=false, type {classStructOrEnum {refPath=&NiagaraBool}}}
      Scalability Mode {
       bIsDynamic=false
       bIsEditable=true
       bIsStaticSwitch=true
       bIsVisible=true
       type {classStructOrEnum {refPath=&ENiagaraEmitterScalabilityMode_Limited}}
      }
      Enable Distance Culling {bIsDynamic=false, bIsEditable=false, bIsStaticSwitch=true, bIsVisible=false, type {classStructOrEnum {refPath=&NiagaraBool}}}
      MinDistance {
       bIsDynamic=false
       bIsEditable=false
       bIsStaticSwitch=false
       bIsVisible=false
       type {classStructOrEnum {refPath=&NiagaraFloat}}
      }
      Min Distance Response {
       bIsDynamic=false
       bIsEditable=false
       bIsStaticSwitch=false
       bIsVisible=false
       type {classStructOrEnum {refPath=&ENiagaraExecutionStateManagement}}
      }
      MaxDistance {
       bIsDynamic=false
       bIsEditable=false
       bIsStaticSwitch=false
       bIsVisible=false
       type {classStructOrEnum {refPath=&NiagaraFloat}}
      }
      Max Distance Response {
       bIsDynamic=false
       bIsEditable=false
       bIsStaticSwitch=false
       bIsVisible=false
       type {classStructOrEnum {refPath=&ENiagaraExecutionStateManagement}}
      }
      Scale Spawn Count {bIsDynamic=false, bIsEditable=false, bIsStaticSwitch=true, bIsVisible=false, type {classStructOrEnum {refPath=&NiagaraBool}}}
      Spawn Count Scale {
       bIsDynamic=false
       bIsEditable=false
       bIsStaticSwitch=false
       bIsVisible=false
       type {classStructOrEnum {refPath=&NiagaraFloat}}
      }
      Spawn Count Scale By Distance {
       bIsDynamic=false
       bIsEditable=false
       bIsStaticSwitch=false
       bIsVisible=false
       type {classStructOrEnum {refPath=&NiagaraDataInterfaceCurve}}
      }
      Enable Visibility Culling {bIsDynamic=false, bIsEditable=false, bIsStaticSwitch=true, bIsVisible=false, type {classStructOrEnum {refPath=&NiagaraBool}}}
      Visibility Cull Response {
       bIsDynamic=false
       bIsEditable=false
       bIsStaticSwitch=false
       bIsVisible=false
       type {classStructOrEnum {refPath=&ENiagaraExecutionStateManagement}}
      }
      Visibility Cull Delay {
       bIsDynamic=false
       bIsEditable=false
       bIsStaticSwitch=false
       bIsVisible=false
       type {classStructOrEnum {refPath=&NiagaraFloat}}
      }
      Reset Age On Awaken {bIsDynamic=false, bIsEditable=false, bIsStaticSwitch=false, bIsVisible=false, type {classStructOrEnum {refPath=&NiagaraBool}}}
     ]
     moduleScript {refPath=&EmitterState}
    }
    {
     bIsSetParametersModule=false
     enabled=true
     moduleName=SpawnBurst_Instantaneous001
     inputs [
      Spawn Count {bIsDynamic=false, bIsEditable=true, bIsStaticSwitch=false, bIsVisible=true, type {classStructOrEnum {refPath=&NiagaraInt32}}}
      Spawn Time {bIsDynamic=false, bIsEditable=true, bIsStaticSwitch=false, bIsVisible=true, type {classStructOrEnum {refPath=&NiagaraFloat}}}
      Use Spawn Probability {bIsDynamic=false, bIsEditable=true, bIsStaticSwitch=false, bIsVisible=true, type {classStructOrEnum {refPath=&NiagaraBool}}}
      Spawn Probability {bIsDynamic=false, bIsEditable=false, bIsStaticSwitch=false, bIsVisible=true, type {classStructOrEnum {refPath=&NiagaraFloat}}}
      UseLoopCountLimit {bIsDynamic=false, bIsEditable=true, bIsStaticSwitch=false, bIsVisible=true, type {classStructOrEnum {refPath=&NiagaraBool}}}
      Loop Count Limit {bIsDynamic=false, bIsEditable=false, bIsStaticSwitch=false, bIsVisible=true, type {classStructOrEnum {refPath=&NiagaraInt32}}}
      Spawn Group {bIsDynamic=false, bIsEditable=true, bIsStaticSwitch=false, bIsVisible=true, type {classStructOrEnum {refPath=&NiagaraInt32}}}
      Age {bIsDynamic=false, bIsEditable=true, bIsStaticSwitch=false, bIsVisible=true, type {classStructOrEnum {refPath=&NiagaraFloat}}}
     ]
     moduleScript {refPath=&SpawnBurst_Instantaneous}
    }
   ]
  }
  particleSpawnScript {
   scriptName=ParticleSpawnScript
   modules [{
     bIsSetParametersModule=false
     enabled=true
     moduleName=InitializeParticle001
     inputs [
      Lifetime Mode {
       bIsDynamic=false
       bIsEditable=true
       bIsStaticSwitch=true
       bIsVisible=true
       type {classStructOrEnum {refPath=&ENiagara_LifetimeMode}}
      }
      Lifetime Min {
       bIsDynamic=false
       bIsEditable=false
       bIsStaticSwitch=false
       bIsVisible=false
       type {classStructOrEnum {refPath=&NiagaraFloat}}
      }
      Lifetime {bIsDynamic=false, bIsEditable=true, bIsStaticSwitch=false, bIsVisible=true, type {classStructOrEnum {refPath=&NiagaraFloat}}}
      Lifetime Max {
       bIsDynamic=false
       bIsEditable=false
       bIsStaticSwitch=false
       bIsVisible=false
       type {classStructOrEnum {refPath=&NiagaraFloat}}
      }
      Lifetime Randomness Mode {
       bIsDynamic=false
       bIsEditable=false
       bIsStaticSwitch=true
       bIsVisible=false
       type {classStructOrEnum {refPath=&ENiagaraRandomnessMode}}
      }
      Color Mode {
       bIsDynamic=false
       bIsEditable=true
       bIsStaticSwitch=true
       bIsVisible=true
       type {classStructOrEnum {refPath=&ENiagara_ColorInitializationMode}}
      }
      Color Minimum {
       bIsDynamic=false
       bIsEditable=false
       bIsStaticSwitch=false
       bIsVisible=false
       type {classStructOrEnum {refPath=&LinearColor}}
      }
      Color {
       bIsDynamic=false
       bIsEditable=true
       bIsStaticSwitch=false
       bIsVisible=true
       type {classStructOrEnum {refPath=&LinearColor}}
      }
      AdjustHue {bIsDynamic=false, bIsEditable=false, bIsStaticSwitch=false, bIsVisible=false, type {classStructOrEnum {refPath=&NiagaraBool}}}
      Hue Shift Range {
       bIsDynamic=false
       bIsEditable=false
       bIsStaticSwitch=false
       bIsVisible=false
       type {classStructOrEnum {refPath=&Vector2f}}
      }
      Color Maximum {
       bIsDynamic=false
       bIsEditable=false
       bIsStaticSwitch=false
       bIsVisible=false
       type {classStructOrEnum {refPath=&LinearColor}}
      }
      AdjustSaturation {bIsDynamic=false, bIsEditable=false, bIsStaticSwitch=false, bIsVisible=false, type {classStructOrEnum {refPath=&NiagaraBool}}}
      Saturation Range {
       bIsDynamic=false
       bIsEditable=false
       bIsStaticSwitch=false
       bIsVisible=false
       type {classStructOrEnum {refPath=&Vector2f}}
      }
      AdjustValue {bIsDynamic=false, bIsEditable=false, bIsStaticSwitch=false, bIsVisible=false, type {classStructOrEnum {refPath=&NiagaraBool}}}
      Value Range {
       bIsDynamic=false
       bIsEditable=false
       bIsStaticSwitch=false
       bIsVisible=false
       type {classStructOrEnum {refPath=&Vector2f}}
      }
      AdjustAlpha {bIsDynamic=false, bIsEditable=false, bIsStaticSwitch=false, bIsVisible=false, type {classStructOrEnum {refPath=&NiagaraBool}}}
      Alpha Scale Range {
       bIsDynamic=false
       bIsEditable=false
       bIsStaticSwitch=false
       bIsVisible=false
       type {classStructOrEnum {refPath=&Vector2f}}
      }
      Color Channel Mode {
       bIsDynamic=false
       bIsEditable=false
       bIsStaticSwitch=true
       bIsVisible=false
       type {classStructOrEnum {refPath=&ENiagaraChannelCorrelation}}
      }
      Color Randomness Mode {
       bIsDynamic=false
       bIsEditable=false
       bIsStaticSwitch=true
       bIsVisible=false
       type {classStructOrEnum {refPath=&ENiagaraRandomnessMode}}
      }
      Position Mode {
       bIsDynamic=false
       bIsEditable=true
       bIsStaticSwitch=true
       bIsVisible=true
       type {classStructOrEnum {refPath=&ENiagara_PositionInitializationMode}}
      }
      Position {
       bIsDynamic=false
       bIsEditable=false
       bIsStaticSwitch=false
       bIsVisible=false
       type {classStructOrEnum {refPath=&NiagaraPosition}}
      }
      UsePositionOffset {bIsDynamic=false, bIsEditable=true, bIsStaticSwitch=false, bIsVisible=true, type {classStructOrEnum {refPath=&NiagaraBool}}}
      Position Offset {bIsDynamic=false, bIsEditable=false, bIsStaticSwitch=false, bIsVisible=true, type {classStructOrEnum {refPath=&Vector3f}}}
      Position Offset Coordinate Space {
       bIsDynamic=false
       bIsEditable=false
       bIsStaticSwitch=false
       bIsVisible=false
       type {classStructOrEnum {refPath=&ENiagaraCoordinateSpace}}
      }
      Mass Mode {
       bIsDynamic=false
       bIsEditable=true
       bIsStaticSwitch=true
       bIsVisible=true
       type {classStructOrEnum {refPath=&ENiagara_MassInitializationMode}}
      }
      Mass Min {
       bIsDynamic=false
       bIsEditable=false
       bIsStaticSwitch=false
       bIsVisible=false
       type {classStructOrEnum {refPath=&NiagaraFloat}}
      }
      Mass {
       bIsDynamic=false
       bIsEditable=false
       bIsStaticSwitch=false
       bIsVisible=false
       type {classStructOrEnum {refPath=&NiagaraFloat}}
      }
      Mass Max {
       bIsDynamic=false
       bIsEditable=false
       bIsStaticSwitch=false
       bIsVisible=false
       type {classStructOrEnum {refPath=&NiagaraFloat}}
      }
      Mass Randomness Mode {
       bIsDynamic=false
       bIsEditable=false
       bIsStaticSwitch=true
       bIsVisible=false
       type {classStructOrEnum {refPath=&ENiagaraRandomnessMode}}
      }
      Write Material Random {bIsDynamic=false, bIsEditable=true, bIsStaticSwitch=false, bIsVisible=true, type {classStructOrEnum {refPath=&NiagaraBool}}}
      Material Random {bIsDynamic=false, bIsEditable=true, bIsStaticSwitch=false, bIsVisible=true, type {classStructOrEnum {refPath=&NiagaraFloat}}}
      Sprite Size Mode {
       bIsDynamic=false
       bIsEditable=true
       bIsStaticSwitch=true
       bIsVisible=true
       type {classStructOrEnum {refPath=&ENiagara_SizeScaleMode}}
      }
      Sprite Size {
       bIsDynamic=false
       bIsEditable=false
       bIsStaticSwitch=false
       bIsVisible=false
       type {classStructOrEnum {refPath=&Vector2f}}
      }
      Uniform Sprite Size {
       bIsDynamic=false
       bIsEditable=false
       bIsStaticSwitch=false
       bIsVisible=false
       type {classStructOrEnum {refPath=&NiagaraFloat}}
      }
      Uniform Sprite Size Min {
       bIsDynamic=false
       bIsEditable=false
       bIsStaticSwitch=false
       bIsVisible=false
       type {classStructOrEnum {refPath=&NiagaraFloat}}
      }
      Sprite Size Min {
       bIsDynamic=false
       bIsEditable=false
       bIsStaticSwitch=false
       bIsVisible=false
       type {classStructOrEnum {refPath=&Vector2f}}
      }
      Sprite Size Max {
       bIsDynamic=false
       bIsEditable=false
       bIsStaticSwitch=false
       bIsVisible=false
       type {classStructOrEnum {refPath=&Vector2f}}
      }
      Uniform Sprite Size Max {
       bIsDynamic=false
       bIsEditable=false
       bIsStaticSwitch=false
       bIsVisible=false
       type {classStructOrEnum {refPath=&NiagaraFloat}}
      }
      Sprite Size Randomness Mode {
       bIsDynamic=false
       bIsEditable=false
       bIsStaticSwitch=true
       bIsVisible=false
       type {classStructOrEnum {refPath=&ENiagaraRandomnessMode}}
      }
      Sprite Rotation Mode {
       bIsDynamic=false
       bIsEditable=true
       bIsStaticSwitch=true
       bIsVisible=true
       type {classStructOrEnum {refPath=&ENiagara_SpriteRotationMode}}
      }
      Sprite Rotation Angle Min {
       bIsDynamic=false
       bIsEditable=false
       bIsStaticSwitch=false
       bIsVisible=false
       type {classStructOrEnum {refPath=&NiagaraFloat}}
      }
      Sprite Rotation Angle {
       bIsDynamic=false
       bIsEditable=false
       bIsStaticSwitch=false
       bIsVisible=false
       type {classStructOrEnum {refPath=&NiagaraFloat}}
      }
      Sprite Rotation Angle Max {
       bIsDynamic=false
       bIsEditable=false
       bIsStaticSwitch=false
       bIsVisible=false
       type {classStructOrEnum {refPath=&NiagaraFloat}}
      }
      Sprite Rotation Randomness Mode {
       bIsDynamic=false
       bIsEditable=false
       bIsStaticSwitch=true
       bIsVisible=false
       type {classStructOrEnum {refPath=&ENiagaraRandomnessMode}}
      }
      Sprite UV Mode {
       bIsDynamic=false
       bIsEditable=true
       bIsStaticSwitch=true
       bIsVisible=true
       type {classStructOrEnum {refPath=&ENiagara_UVFlippingMode}}
      }
      Sprite UV Randomness Mode {
       bIsDynamic=false
       bIsEditable=false
       bIsStaticSwitch=true
       bIsVisible=false
       type {classStructOrEnum {refPath=&ENiagaraRandomnessMode}}
      }
      Sprite UV Scale {
       bIsDynamic=false
       bIsEditable=false
       bIsStaticSwitch=false
       bIsVisible=false
       type {classStructOrEnum {refPath=&Vector2f}}
      }
      Mesh Scale Mode {
       bIsDynamic=false
       bIsEditable=true
       bIsStaticSwitch=true
       bIsVisible=true
       type {classStructOrEnum {refPath=&ENiagara_SizeScaleMode}}
      }
      Mesh Uniform Scale Min {
       bIsDynamic=false
       bIsEditable=false
       bIsStaticSwitch=false
       bIsVisible=false
       type {classStructOrEnum {refPath=&NiagaraFloat}}
      }
      Mesh Uniform Scale {
       bIsDynamic=false
       bIsEditable=false
       bIsStaticSwitch=false
       bIsVisible=false
       type {classStructOrEnum {refPath=&NiagaraFloat}}
      }
      Mesh Scale Min {
       bIsDynamic=false
       bIsEditable=false
       bIsStaticSwitch=false
       bIsVisible=false
       type {classStructOrEnum {refPath=&Vector3f}}
      }
      Mesh Scale {
       bIsDynamic=false
       bIsEditable=false
       bIsStaticSwitch=false
       bIsVisible=false
       type {classStructOrEnum {refPath=&Vector3f}}
      }
      Mesh Scale Max {
       bIsDynamic=false
       bIsEditable=false
       bIsStaticSwitch=false
       bIsVisible=false
       type {classStructOrEnum {refPath=&Vector3f}}
      }
      Mesh Uniform Scale Max {
       bIsDynamic=false
       bIsEditable=false
       bIsStaticSwitch=false
       bIsVisible=false
       type {classStructOrEnum {refPath=&NiagaraFloat}}
      }
      Mesh Scale Randomness Mode {
       bIsDynamic=false
       bIsEditable=false
       bIsStaticSwitch=true
       bIsVisible=false
       type {classStructOrEnum {refPath=&ENiagaraRandomnessMode}}
      }
      Mesh Renderer Array Visibility Mode {
       bIsDynamic=false
       bIsEditable=true
       bIsStaticSwitch=true
       bIsVisible=true
       type {classStructOrEnum {refPath=&ENiagara_UnsetDirectSetRandom}}
      }
      Mesh Renderer Mesh Index {
       bIsDynamic=false
       bIsEditable=false
       bIsStaticSwitch=false
       bIsVisible=false
       type {classStructOrEnum {refPath=&NiagaraInt32}}
      }
      Wrap Index By Mesh Count {bIsDynamic=false, bIsEditable=false, bIsStaticSwitch=true, bIsVisible=false, type {classStructOrEnum {refPath=&NiagaraBool}}}
      Mesh Renderer Info {
       bIsDynamic=false
       bIsEditable=false
       bIsStaticSwitch=false
       bIsVisible=false
       type {classStructOrEnum {refPath=&NiagaraDataInterfaceMeshRendererInfo}}
      }
      Mesh Renderer Random Seed {
       bIsDynamic=false
       bIsEditable=false
       bIsStaticSwitch=false
       bIsVisible=false
       type {classStructOrEnum {refPath=&NiagaraInt32}}
      }
      Ribbon Width Mode {
       bIsDynamic=false
       bIsEditable=true
       bIsStaticSwitch=true
       bIsVisible=true
       type {classStructOrEnum {refPath=&ENiagara_UnsetDirectSet}}
      }
      Ribbon Width {
       bIsDynamic=false
       bIsEditable=false
       bIsStaticSwitch=false
       bIsVisible=false
       type {classStructOrEnum {refPath=&NiagaraFloat}}
      }
      Ribbon Facing Mode {
       bIsDynamic=false
       bIsEditable=true
       bIsStaticSwitch=true
       bIsVisible=true
       type {classStructOrEnum {refPath=&ENiagara_UnsetDirectSet}}
      }
      Ribbon Facing Vector {
       bIsDynamic=false
       bIsEditable=false
       bIsStaticSwitch=false
       bIsVisible=false
       type {classStructOrEnum {refPath=&Vector3f}}
      }
      Ribbon Twist Mode {
       bIsDynamic=false
       bIsEditable=true
       bIsStaticSwitch=true
       bIsVisible=true
       type {classStructOrEnum {refPath=&ENiagara_UnsetDirectSet}}
      }
      "Ribbon Twist " {
       bIsDynamic=false
       bIsEditable=false
       bIsStaticSwitch=false
       bIsVisible=false
       type {classStructOrEnum {refPath=&NiagaraFloat}}
      }
     ]
     moduleScript {refPath=&InitializeParticle}
    }]
  }
  particleUpdateScript {
   scriptName=ParticleUpdateScript
   modules [
    {
     bIsSetParametersModule=false
     enabled=true
     moduleName=ParticleState
     inputs [
      Kill Particles When Lifetime Has Elapsed {bIsDynamic=false, bIsEditable=true, bIsStaticSwitch=true, bIsVisible=true, type {classStructOrEnum {refPath=&NiagaraBool}}}
      Loop Particles Lifetime {bIsDynamic=false, bIsEditable=true, bIsStaticSwitch=false, bIsVisible=true, type {classStructOrEnum {refPath=&NiagaraBool}}}
      Let Infinitely Lived Particles Die When Emitter Deactivates {bIsDynamic=false, bIsEditable=true, bIsStaticSwitch=false, bIsVisible=true, type {classStructOrEnum {refPath=&NiagaraBool}}}
      Lifetime {bIsDynamic=false, bIsEditable=true, bIsStaticSwitch=false, bIsVisible=true, type {classStructOrEnum {refPath=&NiagaraFloat}}}
      DeltaTime {bIsDynamic=false, bIsEditable=true, bIsStaticSwitch=false, bIsVisible=true, type {classStructOrEnum {refPath=&NiagaraFloat}}}
     ]
     moduleScript {refPath=&ParticleState}
    }
    {
     bIsSetParametersModule=false
     enabled=true
     moduleName=ScaleSpriteSize001
     inputs [
      Scale Sprite Size Mode {
       bIsDynamic=false
       bIsEditable=true
       bIsStaticSwitch=true
       bIsVisible=true
       type {classStructOrEnum {refPath=&ENiagara_ScaleSpriteSize}}
      }
      Uniform Curve Sprite Scale {
       bIsDynamic=false
       bIsEditable=true
       bIsStaticSwitch=false
       bIsVisible=true
       type {classStructOrEnum {refPath=&NiagaraDataInterfaceCurve}}
      }
      Non-Uniform Curve Sprite Scale {
       bIsDynamic=false
       bIsEditable=false
       bIsStaticSwitch=false
       bIsVisible=false
       type {classStructOrEnum {refPath=&NiagaraDataInterfaceVector2DCurve}}
      }
      Uniform Curve Index {bIsDynamic=false, bIsEditable=true, bIsStaticSwitch=false, bIsVisible=true, type {classStructOrEnum {refPath=&NiagaraFloat}}}
      Non-Uniform Curve Index {
       bIsDynamic=false
       bIsEditable=false
       bIsStaticSwitch=false
       bIsVisible=false
       type {classStructOrEnum {refPath=&NiagaraFloat}}
      }
      Non-Uniform Curve Scale {
       bIsDynamic=false
       bIsEditable=false
       bIsStaticSwitch=false
       bIsVisible=false
       type {classStructOrEnum {refPath=&Vector2f}}
      }
      Uniform Curve Scale {bIsDynamic=false, bIsEditable=true, bIsStaticSwitch=false, bIsVisible=true, type {classStructOrEnum {refPath=&NiagaraFloat}}}
      Scale Factor {
       bIsDynamic=false
       bIsEditable=false
       bIsStaticSwitch=false
       bIsVisible=false
       type {classStructOrEnum {refPath=&Vector2f}}
      }
      Uniform Scale Factor {
       bIsDynamic=false
       bIsEditable=false
       bIsStaticSwitch=false
       bIsVisible=false
       type {classStructOrEnum {refPath=&NiagaraFloat}}
      }
      Initial Sprite Size {bIsDynamic=false, bIsEditable=true, bIsStaticSwitch=false, bIsVisible=true, type {classStructOrEnum {refPath=&Vector2f}}}
     ]
     moduleScript {refPath=&ScaleSpriteSize}
    }
    {
     bIsSetParametersModule=false
     enabled=true
     moduleName=Color
     inputs [
      Color {
       bIsDynamic=false
       bIsEditable=true
       bIsStaticSwitch=false
       bIsVisible=true
       type {classStructOrEnum {refPath=&LinearColor}}
      }
      Scale Color {bIsDynamic=false, bIsEditable=true, bIsStaticSwitch=false, bIsVisible=true, type {classStructOrEnum {refPath=&Vector3f}}}
      Scale Alpha {bIsDynamic=true, bIsEditable=true, bIsStaticSwitch=false, bIsVisible=true, type {classStructOrEnum {refPath=&NiagaraFloat}}}
     ]
     moduleScript {refPath=&Color}
    }
    {
     bIsSetParametersModule=true
     enabled=true
     moduleName=SetVariables_#6
     inputs [
      Particles.Color_Light {
       bIsDynamic=true
       bIsEditable=true
       bIsStaticSwitch=false
       bIsVisible=true
       type {classStructOrEnum {refPath=&LinearColor}}
      }
      Particles.Position_Light {
       bIsDynamic=true
       bIsEditable=true
       bIsStaticSwitch=false
       bIsVisible=true
       type {classStructOrEnum {refPath=&NiagaraPosition}}
      }
     ]
     moduleScript {refPath=&self:Light_0.NiagaraScriptSource_0.NiagaraGraph_0.NiagaraNodeAssignment_1.SetVariables_#6}
    }
   ]
  }
  rendererClasses [{refPath=&NiagaraLightRendererProperties}]
  renderers [{rendererIndex=0, rendererClass {refPath=&NiagaraLightRendererProperties}}]
 }
 PuddleRadial001 {
  emitterSpawnScript {scriptName=EmitterSpawnScript, modules []}
  emitterUpdateScript {
   scriptName=EmitterUpdateScript
   modules [
    {
     bIsSetParametersModule=false
     enabled=true
     moduleName=EmitterState
     inputs [
      Life Cycle Mode {
       bIsDynamic=false
       bIsEditable=true
       bIsStaticSwitch=true
       bIsVisible=true
       type {classStructOrEnum {refPath=&ENiagaraEmitterLifeCycleMode}}
      }
      Inactive Response {
       bIsDynamic=false
       bIsEditable=true
       bIsStaticSwitch=true
       bIsVisible=true
       type {classStructOrEnum {refPath=&ENiagaraInactiveMode}}
      }
      Loop Behavior {
       bIsDynamic=false
       bIsEditable=true
       bIsStaticSwitch=true
       bIsVisible=true
       type {classStructOrEnum {refPath=&ENiagara_EmitterStateOptions}}
      }
      Loop Duration Mode {
       bIsDynamic=false
       bIsEditable=false
       bIsStaticSwitch=true
       bIsVisible=false
       type {classStructOrEnum {refPath=&ENiagara_InfiniteLoopDuration}}
      }
      Loop Count {
       bIsDynamic=false
       bIsEditable=false
       bIsStaticSwitch=false
       bIsVisible=false
       type {classStructOrEnum {refPath=&NiagaraInt32}}
      }
      Loop Duration {bIsDynamic=false, bIsEditable=true, bIsStaticSwitch=false, bIsVisible=true, type {classStructOrEnum {refPath=&NiagaraFloat}}}
      Recalculate Duration Each Loop {bIsDynamic=false, bIsEditable=true, bIsStaticSwitch=false, bIsVisible=true, type {classStructOrEnum {refPath=&NiagaraBool}}}
      UseLoopDelay {bIsDynamic=false, bIsEditable=true, bIsStaticSwitch=false, bIsVisible=true, type {classStructOrEnum {refPath=&NiagaraBool}}}
      Loop Delay {bIsDynamic=false, bIsEditable=false, bIsStaticSwitch=false, bIsVisible=true, type {classStructOrEnum {refPath=&NiagaraFloat}}}
      Delay First Loop Only {bIsDynamic=false, bIsEditable=false, bIsStaticSwitch=false, bIsVisible=false, type {classStructOrEnum {refPath=&NiagaraBool}}}
      Scalability Mode {
       bIsDynamic=false
       bIsEditable=true
       bIsStaticSwitch=true
       bIsVisible=true
       type {classStructOrEnum {refPath=&ENiagaraEmitterScalabilityMode_Limited}}
      }
      Enable Distance Culling {bIsDynamic=false, bIsEditable=false, bIsStaticSwitch=true, bIsVisible=false, type {classStructOrEnum {refPath=&NiagaraBool}}}
      MinDistance {
       bIsDynamic=false
       bIsEditable=false
       bIsStaticSwitch=false
       bIsVisible=false
       type {classStructOrEnum {refPath=&NiagaraFloat}}
      }
      Min Distance Response {
       bIsDynamic=false
       bIsEditable=false
       bIsStaticSwitch=false
       bIsVisible=false
       type {classStructOrEnum {refPath=&ENiagaraExecutionStateManagement}}
      }
      MaxDistance {
       bIsDynamic=false
       bIsEditable=false
       bIsStaticSwitch=false
       bIsVisible=false
       type {classStructOrEnum {refPath=&NiagaraFloat}}
      }
      Max Distance Response {
       bIsDynamic=false
       bIsEditable=false
       bIsStaticSwitch=false
       bIsVisible=false
       type {classStructOrEnum {refPath=&ENiagaraExecutionStateManagement}}
      }
      Scale Spawn Count {bIsDynamic=false, bIsEditable=false, bIsStaticSwitch=true, bIsVisible=false, type {classStructOrEnum {refPath=&NiagaraBool}}}
      Spawn Count Scale {
       bIsDynamic=false
       bIsEditable=false
       bIsStaticSwitch=false
       bIsVisible=false
       type {classStructOrEnum {refPath=&NiagaraFloat}}
      }
      Spawn Count Scale By Distance {
       bIsDynamic=false
       bIsEditable=false
       bIsStaticSwitch=false
       bIsVisible=false
       type {classStructOrEnum {refPath=&NiagaraDataInterfaceCurve}}
      }
      Enable Visibility Culling {bIsDynamic=false, bIsEditable=false, bIsStaticSwitch=true, bIsVisible=false, type {classStructOrEnum {refPath=&NiagaraBool}}}
      Visibility Cull Response {
       bIsDynamic=false
       bIsEditable=false
       bIsStaticSwitch=false
       bIsVisible=false
       type {classStructOrEnum {refPath=&ENiagaraExecutionStateManagement}}
      }
      Visibility Cull Delay {
       bIsDynamic=false
       bIsEditable=false
       bIsStaticSwitch=false
       bIsVisible=false
       type {classStructOrEnum {refPath=&NiagaraFloat}}
      }
      Reset Age On Awaken {bIsDynamic=false, bIsEditable=false, bIsStaticSwitch=false, bIsVisible=false, type {classStructOrEnum {refPath=&NiagaraBool}}}
     ]
     moduleScript {refPath=&EmitterState}
    }
    {
     bIsSetParametersModule=false
     enabled=true
     moduleName=SpawnRate
     inputs [
      SpawnRate {bIsDynamic=false, bIsEditable=true, bIsStaticSwitch=false, bIsVisible=true, type {classStructOrEnum {refPath=&NiagaraFloat}}}
      Use Spawn Probability {bIsDynamic=false, bIsEditable=true, bIsStaticSwitch=false, bIsVisible=true, type {classStructOrEnum {refPath=&NiagaraBool}}}
      Spawn Probability {bIsDynamic=false, bIsEditable=false, bIsStaticSwitch=false, bIsVisible=true, type {classStructOrEnum {refPath=&NiagaraFloat}}}
      SpawnGroup {bIsDynamic=false, bIsEditable=true, bIsStaticSwitch=false, bIsVisible=true, type {classStructOrEnum {refPath=&NiagaraInt32}}}
     ]
     moduleScript {refPath=&SpawnRate}
    }
   ]
  }
  particleSpawnScript {
   scriptName=ParticleSpawnScript
   modules [{
     bIsSetParametersModule=false
     enabled=true
     moduleName=InitializeParticle
     inputs [
      Lifetime Mode {
       bIsDynamic=false
       bIsEditable=true
       bIsStaticSwitch=true
       bIsVisible=true
       type {classStructOrEnum {refPath=&ENiagara_LifetimeMode}}
      }
      Lifetime Min {
       bIsDynamic=false
       bIsEditable=false
       bIsStaticSwitch=false
       bIsVisible=false
       type {classStructOrEnum {refPath=&NiagaraFloat}}
      }
      Lifetime {bIsDynamic=false, bIsEditable=true, bIsStaticSwitch=false, bIsVisible=true, type {classStructOrEnum {refPath=&NiagaraFloat}}}
      Lifetime Max {
       bIsDynamic=false
       bIsEditable=false
       bIsStaticSwitch=false
       bIsVisible=false
       type {classStructOrEnum {refPath=&NiagaraFloat}}
      }
      Lifetime Randomness Mode {
       bIsDynamic=false
       bIsEditable=false
       bIsStaticSwitch=true
       bIsVisible=false
       type {classStructOrEnum {refPath=&ENiagaraRandomnessMode}}
      }
      Color Mode {
       bIsDynamic=false
       bIsEditable=true
       bIsStaticSwitch=true
       bIsVisible=true
       type {classStructOrEnum {refPath=&ENiagara_ColorInitializationMode}}
      }
      Color Minimum {
       bIsDynamic=false
       bIsEditable=false
       bIsStaticSwitch=false
       bIsVisible=false
       type {classStructOrEnum {refPath=&LinearColor}}
      }
      Color {
       bIsDynamic=false
       bIsEditable=true
       bIsStaticSwitch=false
       bIsVisible=true
       type {classStructOrEnum {refPath=&LinearColor}}
      }
      AdjustHue {bIsDynamic=false, bIsEditable=false, bIsStaticSwitch=false, bIsVisible=false, type {classStructOrEnum {refPath=&NiagaraBool}}}
      Hue Shift Range {
       bIsDynamic=false
       bIsEditable=false
       bIsStaticSwitch=false
       bIsVisible=false
       type {classStructOrEnum {refPath=&Vector2f}}
      }
      Color Maximum {
       bIsDynamic=false
       bIsEditable=false
       bIsStaticSwitch=false
       bIsVisible=false
       type {classStructOrEnum {refPath=&LinearColor}}
      }
      AdjustSaturation {bIsDynamic=false, bIsEditable=false, bIsStaticSwitch=false, bIsVisible=false, type {classStructOrEnum {refPath=&NiagaraBool}}}
      Saturation Range {
       bIsDynamic=false
       bIsEditable=false
       bIsStaticSwitch=false
       bIsVisible=false
       type {classStructOrEnum {refPath=&Vector2f}}
      }
      AdjustValue {bIsDynamic=false, bIsEditable=false, bIsStaticSwitch=false, bIsVisible=false, type {classStructOrEnum {refPath=&NiagaraBool}}}
      Value Range {
       bIsDynamic=false
       bIsEditable=false
       bIsStaticSwitch=false
       bIsVisible=false
       type {classStructOrEnum {refPath=&Vector2f}}
      }
      AdjustAlpha {bIsDynamic=false, bIsEditable=false, bIsStaticSwitch=false, bIsVisible=false, type {classStructOrEnum {refPath=&NiagaraBool}}}
      Alpha Scale Range {
       bIsDynamic=false
       bIsEditable=false
       bIsStaticSwitch=false
       bIsVisible=false
       type {classStructOrEnum {refPath=&Vector2f}}
      }
      Color Channel Mode {
       bIsDynamic=false
       bIsEditable=false
       bIsStaticSwitch=true
       bIsVisible=false
       type {classStructOrEnum {refPath=&ENiagaraChannelCorrelation}}
      }
      Color Randomness Mode {
       bIsDynamic=false
       bIsEditable=false
       bIsStaticSwitch=true
       bIsVisible=false
       type {classStructOrEnum {refPath=&ENiagaraRandomnessMode}}
      }
      Position Mode {
       bIsDynamic=false
       bIsEditable=true
       bIsStaticSwitch=true
       bIsVisible=true
       type {classStructOrEnum {refPath=&ENiagara_PositionInitializationMode}}
      }
      Position {
       bIsDynamic=false
       bIsEditable=false
       bIsStaticSwitch=false
       bIsVisible=false
       type {classStructOrEnum {refPath=&NiagaraPosition}}
      }
      UsePositionOffset {bIsDynamic=false, bIsEditable=true, bIsStaticSwitch=false, bIsVisible=true, type {classStructOrEnum {refPath=&NiagaraBool}}}
      Position Offset {bIsDynamic=false, bIsEditable=true, bIsStaticSwitch=false, bIsVisible=true, type {classStructOrEnum {refPath=&Vector3f}}}
      Position Offset Coordinate Space {
       bIsDynamic=false
       bIsEditable=true
       bIsStaticSwitch=false
       bIsVisible=true
       type {classStructOrEnum {refPath=&ENiagaraCoordinateSpace}}
      }
      Mass Mode {
       bIsDynamic=false
       bIsEditable=true
       bIsStaticSwitch=true
       bIsVisible=true
       type {classStructOrEnum {refPath=&ENiagara_MassInitializationMode}}
      }
      Mass Min {
       bIsDynamic=false
       bIsEditable=false
       bIsStaticSwitch=false
       bIsVisible=false
       type {classStructOrEnum {refPath=&NiagaraFloat}}
      }
      Mass {
       bIsDynamic=false
       bIsEditable=false
       bIsStaticSwitch=false
       bIsVisible=false
       type {classStructOrEnum {refPath=&NiagaraFloat}}
      }
      Mass Max {
       bIsDynamic=false
       bIsEditable=false
       bIsStaticSwitch=false
       bIsVisible=false
       type {classStructOrEnum {refPath=&NiagaraFloat}}
      }
      Mass Randomness Mode {
       bIsDynamic=false
       bIsEditable=false
       bIsStaticSwitch=true
       bIsVisible=false
       type {classStructOrEnum {refPath=&ENiagaraRandomnessMode}}
      }
      Write Material Random {bIsDynamic=false, bIsEditable=true, bIsStaticSwitch=false, bIsVisible=true, type {classStructOrEnum {refPath=&NiagaraBool}}}
      Material Random {bIsDynamic=false, bIsEditable=true, bIsStaticSwitch=false, bIsVisible=true, type {classStructOrEnum {refPath=&NiagaraFloat}}}
      Sprite Size Mode {
       bIsDynamic=false
       bIsEditable=true
       bIsStaticSwitch=true
       bIsVisible=true
       type {classStructOrEnum {refPath=&ENiagara_SizeScaleMode}}
      }
      Sprite Size {
       bIsDynamic=false
       bIsEditable=false
       bIsStaticSwitch=false
       bIsVisible=false
       type {classStructOrEnum {refPath=&Vector2f}}
      }
      Uniform Sprite Size {
       bIsDynamic=false
       bIsEditable=false
       bIsStaticSwitch=false
       bIsVisible=false
       type {classStructOrEnum {refPath=&NiagaraFloat}}
      }
      Uniform Sprite Size Min {
       bIsDynamic=false
       bIsEditable=false
       bIsStaticSwitch=false
       bIsVisible=false
       type {classStructOrEnum {refPath=&NiagaraFloat}}
      }
      Sprite Size Min {
       bIsDynamic=false
       bIsEditable=false
       bIsStaticSwitch=false
       bIsVisible=false
       type {classStructOrEnum {refPath=&Vector2f}}
      }
      Sprite Size Max {
       bIsDynamic=false
       bIsEditable=false
       bIsStaticSwitch=false
       bIsVisible=false
       type {classStructOrEnum {refPath=&Vector2f}}
      }
      Uniform Sprite Size Max {
       bIsDynamic=false
       bIsEditable=false
       bIsStaticSwitch=false
       bIsVisible=false
       type {classStructOrEnum {refPath=&NiagaraFloat}}
      }
      Sprite Size Randomness Mode {
       bIsDynamic=false
       bIsEditable=false
       bIsStaticSwitch=true
       bIsVisible=false
       type {classStructOrEnum {refPath=&ENiagaraRandomnessMode}}
      }
      Sprite Rotation Mode {
       bIsDynamic=false
       bIsEditable=true
       bIsStaticSwitch=true
       bIsVisible=true
       type {classStructOrEnum {refPath=&ENiagara_SpriteRotationMode}}
      }
      Sprite Rotation Angle Min {
       bIsDynamic=false
       bIsEditable=false
       bIsStaticSwitch=false
       bIsVisible=false
       type {classStructOrEnum {refPath=&NiagaraFloat}}
      }
      Sprite Rotation Angle {
       bIsDynamic=false
       bIsEditable=false
       bIsStaticSwitch=false
       bIsVisible=false
       type {classStructOrEnum {refPath=&NiagaraFloat}}
      }
      Sprite Rotation Angle Max {
       bIsDynamic=false
       bIsEditable=false
       bIsStaticSwitch=false
       bIsVisible=false
       type {classStructOrEnum {refPath=&NiagaraFloat}}
      }
      Sprite Rotation Randomness Mode {
       bIsDynamic=false
       bIsEditable=false
       bIsStaticSwitch=true
       bIsVisible=false
       type {classStructOrEnum {refPath=&ENiagaraRandomnessMode}}
      }
      Sprite UV Mode {
       bIsDynamic=false
       bIsEditable=true
       bIsStaticSwitch=true
       bIsVisible=true
       type {classStructOrEnum {refPath=&ENiagara_UVFlippingMode}}
      }
      Sprite UV Randomness Mode {
       bIsDynamic=false
       bIsEditable=false
       bIsStaticSwitch=true
       bIsVisible=false
       type {classStructOrEnum {refPath=&ENiagaraRandomnessMode}}
      }
      Sprite UV Scale {
       bIsDynamic=false
       bIsEditable=false
       bIsStaticSwitch=false
       bIsVisible=false
       type {classStructOrEnum {refPath=&Vector2f}}
      }
      Mesh Scale Mode {
       bIsDynamic=false
       bIsEditable=true
       bIsStaticSwitch=true
       bIsVisible=true
       type {classStructOrEnum {refPath=&ENiagara_SizeScaleMode}}
      }
      Mesh Uniform Scale Min {
       bIsDynamic=false
       bIsEditable=false
       bIsStaticSwitch=false
       bIsVisible=false
       type {classStructOrEnum {refPath=&NiagaraFloat}}
      }
      Mesh Uniform Scale {bIsDynamic=true, bIsEditable=false, bIsStaticSwitch=false, bIsVisible=false, type {classStructOrEnum {refPath=&NiagaraFloat}}}
      Mesh Scale Min {
       bIsDynamic=false
       bIsEditable=false
       bIsStaticSwitch=false
       bIsVisible=false
       type {classStructOrEnum {refPath=&Vector3f}}
      }
      Mesh Scale {bIsDynamic=false, bIsEditable=true, bIsStaticSwitch=false, bIsVisible=true, type {classStructOrEnum {refPath=&Vector3f}}}
      Mesh Scale Max {
       bIsDynamic=false
       bIsEditable=false
       bIsStaticSwitch=false
       bIsVisible=false
       type {classStructOrEnum {refPath=&Vector3f}}
      }
      Mesh Uniform Scale Max {
       bIsDynamic=false
       bIsEditable=false
       bIsStaticSwitch=false
       bIsVisible=false
       type {classStructOrEnum {refPath=&NiagaraFloat}}
      }
      Mesh Scale Randomness Mode {
       bIsDynamic=false
       bIsEditable=false
       bIsStaticSwitch=true
       bIsVisible=false
       type {classStructOrEnum {refPath=&ENiagaraRandomnessMode}}
      }
      Mesh Renderer Array Visibility Mode {
       bIsDynamic=false
       bIsEditable=true
       bIsStaticSwitch=true
       bIsVisible=true
       type {classStructOrEnum {refPath=&ENiagara_UnsetDirectSetRandom}}
      }
      Mesh Renderer Mesh Index {
       bIsDynamic=false
       bIsEditable=false
       bIsStaticSwitch=false
       bIsVisible=false
       type {classStructOrEnum {refPath=&NiagaraInt32}}
      }
      Wrap Index By Mesh Count {bIsDynamic=false, bIsEditable=false, bIsStaticSwitch=true, bIsVisible=false, type {classStructOrEnum {refPath=&NiagaraBool}}}
      Mesh Renderer Info {
       bIsDynamic=false
       bIsEditable=false
       bIsStaticSwitch=false
       bIsVisible=false
       type {classStructOrEnum {refPath=&NiagaraDataInterfaceMeshRendererInfo}}
      }
      Mesh Renderer Random Seed {
       bIsDynamic=false
       bIsEditable=false
       bIsStaticSwitch=false
       bIsVisible=false
       type {classStructOrEnum {refPath=&NiagaraInt32}}
      }
      Ribbon Width Mode {
       bIsDynamic=false
       bIsEditable=true
       bIsStaticSwitch=true
       bIsVisible=true
       type {classStructOrEnum {refPath=&ENiagara_UnsetDirectSet}}
      }
      Ribbon Width {
       bIsDynamic=false
       bIsEditable=false
       bIsStaticSwitch=false
       bIsVisible=false
       type {classStructOrEnum {refPath=&NiagaraFloat}}
      }
      Ribbon Facing Mode {
       bIsDynamic=false
       bIsEditable=true
       bIsStaticSwitch=true
       bIsVisible=true
       type {classStructOrEnum {refPath=&ENiagara_UnsetDirectSet}}
      }
      Ribbon Facing Vector {
       bIsDynamic=false
       bIsEditable=false
       bIsStaticSwitch=false
       bIsVisible=false
       type {classStructOrEnum {refPath=&Vector3f}}
      }
      Ribbon Twist Mode {
       bIsDynamic=false
       bIsEditable=true
       bIsStaticSwitch=true
       bIsVisible=true
       type {classStructOrEnum {refPath=&ENiagara_UnsetDirectSet}}
      }
      "Ribbon Twist " {
       bIsDynamic=false
       bIsEditable=false
       bIsStaticSwitch=false
       bIsVisible=false
       type {classStructOrEnum {refPath=&NiagaraFloat}}
      }
     ]
     moduleScript {refPath=&InitializeParticle}
    }]
  }
  particleUpdateScript {
   scriptName=ParticleUpdateScript
   modules [
    {
     bIsSetParametersModule=false
     enabled=true
     moduleName=ParticleState
     inputs [
      Kill Particles When Lifetime Has Elapsed {bIsDynamic=false, bIsEditable=true, bIsStaticSwitch=true, bIsVisible=true, type {classStructOrEnum {refPath=&NiagaraBool}}}
      Loop Particles Lifetime {bIsDynamic=false, bIsEditable=false, bIsStaticSwitch=false, bIsVisible=false, type {classStructOrEnum {refPath=&NiagaraBool}}}
      Let Infinitely Lived Particles Die When Emitter Deactivates {bIsDynamic=false, bIsEditable=false, bIsStaticSwitch=false, bIsVisible=false, type {classStructOrEnum {refPath=&NiagaraBool}}}
      Lifetime {bIsDynamic=false, bIsEditable=true, bIsStaticSwitch=false, bIsVisible=true, type {classStructOrEnum {refPath=&NiagaraFloat}}}
      DeltaTime {bIsDynamic=false, bIsEditable=true, bIsStaticSwitch=false, bIsVisible=true, type {classStructOrEnum {refPath=&NiagaraFloat}}}
     ]
     moduleScript {refPath=&ParticleState}
    }
    {
     bIsSetParametersModule=false
     enabled=true
     moduleName=ScaleMeshSize
     inputs [
      Scale Factor {bIsDynamic=true, bIsEditable=true, bIsStaticSwitch=false, bIsVisible=true, type {classStructOrEnum {refPath=&Vector3f}}}
      Initial Mesh Scale {bIsDynamic=false, bIsEditable=true, bIsStaticSwitch=false, bIsVisible=true, type {classStructOrEnum {refPath=&Vector3f}}}
     ]
     moduleScript {refPath=&ScaleMeshSize}
    }
    {
     bIsSetParametersModule=false
     enabled=true
     moduleName=ScaleColor002
     inputs [
      Scale Mode {
       bIsDynamic=false
       bIsEditable=true
       bIsStaticSwitch=true
       bIsVisible=true
       type {classStructOrEnum {refPath=/Niagara/Enums/ENiagaraScaleColorMode}}
      }
      Linear Color Curve {
       bIsDynamic=false
       bIsEditable=false
       bIsStaticSwitch=false
       bIsVisible=false
       type {classStructOrEnum {refPath=&NiagaraDataInterfaceColorCurve}}
      }
      ScaleRGB {bIsDynamic=false, bIsEditable=true, bIsStaticSwitch=false, bIsVisible=true, type {classStructOrEnum {refPath=&NiagaraBool}}}
      Scale RGB {bIsDynamic=true, bIsEditable=true, bIsStaticSwitch=false, bIsVisible=true, type {classStructOrEnum {refPath=&Vector3f}}}
      ScaleRGBA {bIsDynamic=false, bIsEditable=false, bIsStaticSwitch=false, bIsVisible=false, type {classStructOrEnum {refPath=&NiagaraBool}}}
      Scale RGBA {
       bIsDynamic=false
       bIsEditable=false
       bIsStaticSwitch=false
       bIsVisible=false
       type {classStructOrEnum {refPath=&Vector4f}}
      }
      Curve Index {
       bIsDynamic=false
       bIsEditable=false
       bIsStaticSwitch=false
       bIsVisible=false
       type {classStructOrEnum {refPath=&NiagaraFloat}}
      }
      ScaleA {bIsDynamic=false, bIsEditable=true, bIsStaticSwitch=false, bIsVisible=true, type {classStructOrEnum {refPath=&NiagaraBool}}}
      Scale Alpha {bIsDynamic=true, bIsEditable=true, bIsStaticSwitch=false, bIsVisible=true, type {classStructOrEnum {refPath=&NiagaraFloat}}}
      Color Value To Scale {
       bIsDynamic=false
       bIsEditable=true
       bIsStaticSwitch=false
       bIsVisible=true
       type {classStructOrEnum {refPath=&LinearColor}}
      }
     ]
     moduleScript {refPath=&ScaleColor}
    }
    {
     bIsSetParametersModule=true
     enabled=true
     moduleName=SetVariables_#7
     inputs [
      Particles.Scale_BlueRipple {bIsDynamic=true, bIsEditable=true, bIsStaticSwitch=false, bIsVisible=true, type {classStructOrEnum {refPath=&Vector3f}}}
      Particles.Linear Color_BlueRipple {
       bIsDynamic=true
       bIsEditable=true
       bIsStaticSwitch=false
       bIsVisible=true
       type {classStructOrEnum {refPath=&LinearColor}}
      }
     ]
     moduleScript {refPath=&self:PuddleRadial001_1.NiagaraScriptSource_0.NiagaraGraph_0.NiagaraNodeAssignment_0.SetVariables_#7}
    }
    {
     bIsSetParametersModule=true
     enabled=true
     moduleName=SetVariables_#8
     inputs [Particles.Color_GlowRing {
       bIsDynamic=true
       bIsEditable=true
       bIsStaticSwitch=false
       bIsVisible=true
       type {classStructOrEnum {refPath=&LinearColor}}
      }]
     moduleScript {refPath=&self:PuddleRadial001_1.NiagaraScriptSource_0.NiagaraGraph_0.NiagaraNodeAssignment_1.SetVariables_#8}
    }
   ]
  }
  rendererClasses [{refPath=&NiagaraMeshRendererProperties}]
  renderers [
   {rendererIndex=0, rendererClass {refPath=&NiagaraMeshRendererProperties}}
   {rendererIndex=1, rendererClass {refPath=&NiagaraMeshRendererProperties}}
   {rendererIndex=2, rendererClass {refPath=&NiagaraMeshRendererProperties}}
  ]
 }
 PuddleRadial_Trail {
  emitterSpawnScript {scriptName=EmitterSpawnScript, modules []}
  emitterUpdateScript {
   scriptName=EmitterUpdateScript
   modules [
    {
     bIsSetParametersModule=false
     enabled=true
     moduleName=EmitterState
     inputs [
      Life Cycle Mode {
       bIsDynamic=false
       bIsEditable=true
       bIsStaticSwitch=true
       bIsVisible=true
       type {classStructOrEnum {refPath=&ENiagaraEmitterLifeCycleMode}}
      }
      Inactive Response {
       bIsDynamic=false
       bIsEditable=true
       bIsStaticSwitch=true
       bIsVisible=true
       type {classStructOrEnum {refPath=&ENiagaraInactiveMode}}
      }
      Loop Behavior {
       bIsDynamic=false
       bIsEditable=true
       bIsStaticSwitch=true
       bIsVisible=true
       type {classStructOrEnum {refPath=&ENiagara_EmitterStateOptions}}
      }
      Loop Duration Mode {
       bIsDynamic=false
       bIsEditable=false
       bIsStaticSwitch=true
       bIsVisible=false
       type {classStructOrEnum {refPath=&ENiagara_InfiniteLoopDuration}}
      }
      Loop Count {
       bIsDynamic=false
       bIsEditable=false
       bIsStaticSwitch=false
       bIsVisible=false
       type {classStructOrEnum {refPath=&NiagaraInt32}}
      }
      Loop Duration {bIsDynamic=false, bIsEditable=true, bIsStaticSwitch=false, bIsVisible=true, type {classStructOrEnum {refPath=&NiagaraFloat}}}
      Recalculate Duration Each Loop {bIsDynamic=false, bIsEditable=true, bIsStaticSwitch=false, bIsVisible=true, type {classStructOrEnum {refPath=&NiagaraBool}}}
      UseLoopDelay {bIsDynamic=false, bIsEditable=true, bIsStaticSwitch=false, bIsVisible=true, type {classStructOrEnum {refPath=&NiagaraBool}}}
      Loop Delay {bIsDynamic=false, bIsEditable=false, bIsStaticSwitch=false, bIsVisible=true, type {classStructOrEnum {refPath=&NiagaraFloat}}}
      Delay First Loop Only {bIsDynamic=false, bIsEditable=false, bIsStaticSwitch=false, bIsVisible=false, type {classStructOrEnum {refPath=&NiagaraBool}}}
      Scalability Mode {
       bIsDynamic=false
       bIsEditable=true
       bIsStaticSwitch=true
       bIsVisible=true
       type {classStructOrEnum {refPath=&ENiagaraEmitterScalabilityMode_Limited}}
      }
      Enable Distance Culling {bIsDynamic=false, bIsEditable=false, bIsStaticSwitch=true, bIsVisible=false, type {classStructOrEnum {refPath=&NiagaraBool}}}
      MinDistance {
       bIsDynamic=false
       bIsEditable=false
       bIsStaticSwitch=false
       bIsVisible=false
       type {classStructOrEnum {refPath=&NiagaraFloat}}
      }
      Min Distance Response {
       bIsDynamic=false
       bIsEditable=false
       bIsStaticSwitch=false
       bIsVisible=false
       type {classStructOrEnum {refPath=&ENiagaraExecutionStateManagement}}
      }
      MaxDistance {
       bIsDynamic=false
       bIsEditable=false
       bIsStaticSwitch=false
       bIsVisible=false
       type {classStructOrEnum {refPath=&NiagaraFloat}}
      }
      Max Distance Response {
       bIsDynamic=false
       bIsEditable=false
       bIsStaticSwitch=false
       bIsVisible=false
       type {classStructOrEnum {refPath=&ENiagaraExecutionStateManagement}}
      }
      Scale Spawn Count {bIsDynamic=false, bIsEditable=false, bIsStaticSwitch=true, bIsVisible=false, type {classStructOrEnum {refPath=&NiagaraBool}}}
      Spawn Count Scale {
       bIsDynamic=false
       bIsEditable=false
       bIsStaticSwitch=false
       bIsVisible=false
       type {classStructOrEnum {refPath=&NiagaraFloat}}
      }
      Spawn Count Scale By Distance {
       bIsDynamic=false
       bIsEditable=false
       bIsStaticSwitch=false
       bIsVisible=false
       type {classStructOrEnum {refPath=&NiagaraDataInterfaceCurve}}
      }
      Enable Visibility Culling {bIsDynamic=false, bIsEditable=false, bIsStaticSwitch=true, bIsVisible=false, type {classStructOrEnum {refPath=&NiagaraBool}}}
      Visibility Cull Response {
       bIsDynamic=false
       bIsEditable=false
       bIsStaticSwitch=false
       bIsVisible=false
       type {classStructOrEnum {refPath=&ENiagaraExecutionStateManagement}}
      }
      Visibility Cull Delay {
       bIsDynamic=false
       bIsEditable=false
       bIsStaticSwitch=false
       bIsVisible=false
       type {classStructOrEnum {refPath=&NiagaraFloat}}
      }
      Reset Age On Awaken {bIsDynamic=false, bIsEditable=false, bIsStaticSwitch=false, bIsVisible=false, type {classStructOrEnum {refPath=&NiagaraBool}}}
     ]
     moduleScript {refPath=&EmitterState}
    }
    {
     bIsSetParametersModule=false
     enabled=true
     moduleName=SpawnPerUnit
     inputs [
      Spawn Spacing {bIsDynamic=true, bIsEditable=true, bIsStaticSwitch=false, bIsVisible=true, type {classStructOrEnum {refPath=&NiagaraFloat}}}
      Module.Use Max Movement Threshold {bIsDynamic=false, bIsEditable=true, bIsStaticSwitch=false, bIsVisible=true, type {classStructOrEnum {refPath=&NiagaraBool}}}
      Max Movement Threshold {bIsDynamic=false, bIsEditable=false, bIsStaticSwitch=false, bIsVisible=true, type {classStructOrEnum {refPath=&NiagaraFloat}}}
      Module.Use Movement Tolerance {bIsDynamic=false, bIsEditable=true, bIsStaticSwitch=false, bIsVisible=true, type {classStructOrEnum {refPath=&NiagaraBool}}}
      Movement Tolerance {bIsDynamic=false, bIsEditable=false, bIsStaticSwitch=false, bIsVisible=true, type {classStructOrEnum {refPath=&NiagaraFloat}}}
      Use Spawn Probability {bIsDynamic=false, bIsEditable=true, bIsStaticSwitch=false, bIsVisible=true, type {classStructOrEnum {refPath=&NiagaraBool}}}
      Spawn Probability {bIsDynamic=false, bIsEditable=false, bIsStaticSwitch=false, bIsVisible=true, type {classStructOrEnum {refPath=&NiagaraFloat}}}
      Velocity Vector {bIsDynamic=false, bIsEditable=true, bIsStaticSwitch=false, bIsVisible=true, type {classStructOrEnum {refPath=&Vector3f}}}
      Delta Time {bIsDynamic=false, bIsEditable=true, bIsStaticSwitch=false, bIsVisible=true, type {classStructOrEnum {refPath=&NiagaraFloat}}}
      Spawn Group {bIsDynamic=false, bIsEditable=true, bIsStaticSwitch=false, bIsVisible=true, type {classStructOrEnum {refPath=&NiagaraInt32}}}
     ]
     moduleScript {refPath=&SpawnPerUnit}
    }
    {
     bIsSetParametersModule=false
     enabled=false
     moduleName=SpawnRate
     inputs [
      SpawnRate {bIsDynamic=false, bIsEditable=true, bIsStaticSwitch=false, bIsVisible=true, type {classStructOrEnum {refPath=&NiagaraFloat}}}
      Use Spawn Probability {bIsDynamic=false, bIsEditable=true, bIsStaticSwitch=false, bIsVisible=true, type {classStructOrEnum {refPath=&NiagaraBool}}}
      Spawn Probability {bIsDynamic=false, bIsEditable=false, bIsStaticSwitch=false, bIsVisible=true, type {classStructOrEnum {refPath=&NiagaraFloat}}}
      SpawnGroup {bIsDynamic=false, bIsEditable=true, bIsStaticSwitch=false, bIsVisible=true, type {classStructOrEnum {refPath=&NiagaraInt32}}}
     ]
     moduleScript {refPath=&SpawnRate}
    }
   ]
  }
  particleSpawnScript {
   scriptName=ParticleSpawnScript
   modules [
    {
     bIsSetParametersModule=false
     enabled=true
     moduleName=InitializeParticle
     inputs [
      Lifetime Mode {
       bIsDynamic=false
       bIsEditable=true
       bIsStaticSwitch=true
       bIsVisible=true
       type {classStructOrEnum {refPath=&ENiagara_LifetimeMode}}
      }
      Lifetime Min {
       bIsDynamic=false
       bIsEditable=false
       bIsStaticSwitch=false
       bIsVisible=false
       type {classStructOrEnum {refPath=&NiagaraFloat}}
      }
      Lifetime {bIsDynamic=false, bIsEditable=true, bIsStaticSwitch=false, bIsVisible=true, type {classStructOrEnum {refPath=&NiagaraFloat}}}
      Lifetime Max {
       bIsDynamic=false
       bIsEditable=false
       bIsStaticSwitch=false
       bIsVisible=false
       type {classStructOrEnum {refPath=&NiagaraFloat}}
      }
      Lifetime Randomness Mode {
       bIsDynamic=false
       bIsEditable=false
       bIsStaticSwitch=true
       bIsVisible=false
       type {classStructOrEnum {refPath=&ENiagaraRandomnessMode}}
      }
      Color Mode {
       bIsDynamic=false
       bIsEditable=true
       bIsStaticSwitch=true
       bIsVisible=true
       type {classStructOrEnum {refPath=&ENiagara_ColorInitializationMode}}
      }
      Color Minimum {
       bIsDynamic=false
       bIsEditable=false
       bIsStaticSwitch=false
       bIsVisible=false
       type {classStructOrEnum {refPath=&LinearColor}}
      }
      Color {
       bIsDynamic=false
       bIsEditable=true
       bIsStaticSwitch=false
       bIsVisible=true
       type {classStructOrEnum {refPath=&LinearColor}}
      }
      AdjustHue {bIsDynamic=false, bIsEditable=false, bIsStaticSwitch=false, bIsVisible=false, type {classStructOrEnum {refPath=&NiagaraBool}}}
      Hue Shift Range {
       bIsDynamic=false
       bIsEditable=false
       bIsStaticSwitch=false
       bIsVisible=false
       type {classStructOrEnum {refPath=&Vector2f}}
      }
      Color Maximum {
       bIsDynamic=false
       bIsEditable=false
       bIsStaticSwitch=false
       bIsVisible=false
       type {classStructOrEnum {refPath=&LinearColor}}
      }
      AdjustSaturation {bIsDynamic=false, bIsEditable=false, bIsStaticSwitch=false, bIsVisible=false, type {classStructOrEnum {refPath=&NiagaraBool}}}
      Saturation Range {
       bIsDynamic=false
       bIsEditable=false
       bIsStaticSwitch=false
       bIsVisible=false
       type {classStructOrEnum {refPath=&Vector2f}}
      }
      AdjustValue {bIsDynamic=false, bIsEditable=false, bIsStaticSwitch=false, bIsVisible=false, type {classStructOrEnum {refPath=&NiagaraBool}}}
      Value Range {
       bIsDynamic=false
       bIsEditable=false
       bIsStaticSwitch=false
       bIsVisible=false
       type {classStructOrEnum {refPath=&Vector2f}}
      }
      AdjustAlpha {bIsDynamic=false, bIsEditable=false, bIsStaticSwitch=false, bIsVisible=false, type {classStructOrEnum {refPath=&NiagaraBool}}}
      Alpha Scale Range {
       bIsDynamic=false
       bIsEditable=false
       bIsStaticSwitch=false
       bIsVisible=false
       type {classStructOrEnum {refPath=&Vector2f}}
      }
      Color Channel Mode {
       bIsDynamic=false
       bIsEditable=false
       bIsStaticSwitch=true
       bIsVisible=false
       type {classStructOrEnum {refPath=&ENiagaraChannelCorrelation}}
      }
      Color Randomness Mode {
       bIsDynamic=false
       bIsEditable=false
       bIsStaticSwitch=true
       bIsVisible=false
       type {classStructOrEnum {refPath=&ENiagaraRandomnessMode}}
      }
      Position Mode {
       bIsDynamic=false
       bIsEditable=true
       bIsStaticSwitch=true
       bIsVisible=true
       type {classStructOrEnum {refPath=&ENiagara_PositionInitializationMode}}
      }
      Position {
       bIsDynamic=false
       bIsEditable=false
       bIsStaticSwitch=false
       bIsVisible=false
       type {classStructOrEnum {refPath=&NiagaraPosition}}
      }
      UsePositionOffset {bIsDynamic=false, bIsEditable=true, bIsStaticSwitch=false, bIsVisible=true, type {classStructOrEnum {refPath=&NiagaraBool}}}
      Position Offset {bIsDynamic=false, bIsEditable=true, bIsStaticSwitch=false, bIsVisible=true, type {classStructOrEnum {refPath=&Vector3f}}}
      Position Offset Coordinate Space {
       bIsDynamic=false
       bIsEditable=true
       bIsStaticSwitch=false
       bIsVisible=true
       type {classStructOrEnum {refPath=&ENiagaraCoordinateSpace}}
      }
      Mass Mode {
       bIsDynamic=false
       bIsEditable=true
       bIsStaticSwitch=true
       bIsVisible=true
       type {classStructOrEnum {refPath=&ENiagara_MassInitializationMode}}
      }
      Mass Min {
       bIsDynamic=false
       bIsEditable=false
       bIsStaticSwitch=false
       bIsVisible=false
       type {classStructOrEnum {refPath=&NiagaraFloat}}
      }
      Mass {
       bIsDynamic=false
       bIsEditable=false
       bIsStaticSwitch=false
       bIsVisible=false
       type {classStructOrEnum {refPath=&NiagaraFloat}}
      }
      Mass Max {
       bIsDynamic=false
       bIsEditable=false
       bIsStaticSwitch=false
       bIsVisible=false
       type {classStructOrEnum {refPath=&NiagaraFloat}}
      }
      Mass Randomness Mode {
       bIsDynamic=false
       bIsEditable=false
       bIsStaticSwitch=true
       bIsVisible=false
       type {classStructOrEnum {refPath=&ENiagaraRandomnessMode}}
      }
      Write Material Random {bIsDynamic=false, bIsEditable=true, bIsStaticSwitch=false, bIsVisible=true, type {classStructOrEnum {refPath=&NiagaraBool}}}
      Material Random {bIsDynamic=false, bIsEditable=true, bIsStaticSwitch=false, bIsVisible=true, type {classStructOrEnum {refPath=&NiagaraFloat}}}
      Sprite Size Mode {
       bIsDynamic=false
       bIsEditable=true
       bIsStaticSwitch=true
       bIsVisible=true
       type {classStructOrEnum {refPath=&ENiagara_SizeScaleMode}}
      }
      Sprite Size {
       bIsDynamic=false
       bIsEditable=false
       bIsStaticSwitch=false
       bIsVisible=false
       type {classStructOrEnum {refPath=&Vector2f}}
      }
      Uniform Sprite Size {
       bIsDynamic=false
       bIsEditable=false
       bIsStaticSwitch=false
       bIsVisible=false
       type {classStructOrEnum {refPath=&NiagaraFloat}}
      }
      Uniform Sprite Size Min {
       bIsDynamic=false
       bIsEditable=false
       bIsStaticSwitch=false
       bIsVisible=false
       type {classStructOrEnum {refPath=&NiagaraFloat}}
      }
      Sprite Size Min {
       bIsDynamic=false
       bIsEditable=false
       bIsStaticSwitch=false
       bIsVisible=false
       type {classStructOrEnum {refPath=&Vector2f}}
      }
      Sprite Size Max {
       bIsDynamic=false
       bIsEditable=false
       bIsStaticSwitch=false
       bIsVisible=false
       type {classStructOrEnum {refPath=&Vector2f}}
      }
      Uniform Sprite Size Max {
       bIsDynamic=false
       bIsEditable=false
       bIsStaticSwitch=false
       bIsVisible=false
       type {classStructOrEnum {refPath=&NiagaraFloat}}
      }
      Sprite Size Randomness Mode {
       bIsDynamic=false
       bIsEditable=false
       bIsStaticSwitch=true
       bIsVisible=false
       type {classStructOrEnum {refPath=&ENiagaraRandomnessMode}}
      }
      Sprite Rotation Mode {
       bIsDynamic=false
       bIsEditable=true
       bIsStaticSwitch=true
       bIsVisible=true
       type {classStructOrEnum {refPath=&ENiagara_SpriteRotationMode}}
      }
      Sprite Rotation Angle Min {
       bIsDynamic=false
       bIsEditable=false
       bIsStaticSwitch=false
       bIsVisible=false
       type {classStructOrEnum {refPath=&NiagaraFloat}}
      }
      Sprite Rotation Angle {
       bIsDynamic=false
       bIsEditable=false
       bIsStaticSwitch=false
       bIsVisible=false
       type {classStructOrEnum {refPath=&NiagaraFloat}}
      }
      Sprite Rotation Angle Max {
       bIsDynamic=false
       bIsEditable=false
       bIsStaticSwitch=false
       bIsVisible=false
       type {classStructOrEnum {refPath=&NiagaraFloat}}
      }
      Sprite Rotation Randomness Mode {
       bIsDynamic=false
       bIsEditable=false
       bIsStaticSwitch=true
       bIsVisible=false
       type {classStructOrEnum {refPath=&ENiagaraRandomnessMode}}
      }
      Sprite UV Mode {
       bIsDynamic=false
       bIsEditable=true
       bIsStaticSwitch=true
       bIsVisible=true
       type {classStructOrEnum {refPath=&ENiagara_UVFlippingMode}}
      }
      Sprite UV Randomness Mode {
       bIsDynamic=false
       bIsEditable=false
       bIsStaticSwitch=true
       bIsVisible=false
       type {classStructOrEnum {refPath=&ENiagaraRandomnessMode}}
      }
      Sprite UV Scale {
       bIsDynamic=false
       bIsEditable=false
       bIsStaticSwitch=false
       bIsVisible=false
       type {classStructOrEnum {refPath=&Vector2f}}
      }
      Mesh Scale Mode {
       bIsDynamic=false
       bIsEditable=true
       bIsStaticSwitch=true
       bIsVisible=true
       type {classStructOrEnum {refPath=&ENiagara_SizeScaleMode}}
      }
      Mesh Uniform Scale Min {
       bIsDynamic=false
       bIsEditable=false
       bIsStaticSwitch=false
       bIsVisible=false
       type {classStructOrEnum {refPath=&NiagaraFloat}}
      }
      Mesh Uniform Scale {bIsDynamic=false, bIsEditable=true, bIsStaticSwitch=false, bIsVisible=true, type {classStructOrEnum {refPath=&NiagaraFloat}}}
      Mesh Scale Min {
       bIsDynamic=false
       bIsEditable=false
       bIsStaticSwitch=false
       bIsVisible=false
       type {classStructOrEnum {refPath=&Vector3f}}
      }
      Mesh Scale {
       bIsDynamic=false
       bIsEditable=false
       bIsStaticSwitch=false
       bIsVisible=false
       type {classStructOrEnum {refPath=&Vector3f}}
      }
      Mesh Scale Max {
       bIsDynamic=false
       bIsEditable=false
       bIsStaticSwitch=false
       bIsVisible=false
       type {classStructOrEnum {refPath=&Vector3f}}
      }
      Mesh Uniform Scale Max {
       bIsDynamic=false
       bIsEditable=false
       bIsStaticSwitch=false
       bIsVisible=false
       type {classStructOrEnum {refPath=&NiagaraFloat}}
      }
      Mesh Scale Randomness Mode {
       bIsDynamic=false
       bIsEditable=false
       bIsStaticSwitch=true
       bIsVisible=false
       type {classStructOrEnum {refPath=&ENiagaraRandomnessMode}}
      }
      Mesh Renderer Array Visibility Mode {
       bIsDynamic=false
       bIsEditable=true
       bIsStaticSwitch=true
       bIsVisible=true
       type {classStructOrEnum {refPath=&ENiagara_UnsetDirectSetRandom}}
      }
      Mesh Renderer Mesh Index {
       bIsDynamic=false
       bIsEditable=false
       bIsStaticSwitch=false
       bIsVisible=false
       type {classStructOrEnum {refPath=&NiagaraInt32}}
      }
      Wrap Index By Mesh Count {bIsDynamic=false, bIsEditable=false, bIsStaticSwitch=true, bIsVisible=false, type {classStructOrEnum {refPath=&NiagaraBool}}}
      Mesh Renderer Info {
       bIsDynamic=false
       bIsEditable=false
       bIsStaticSwitch=false
       bIsVisible=false
       type {classStructOrEnum {refPath=&NiagaraDataInterfaceMeshRendererInfo}}
      }
      Mesh Renderer Random Seed {
       bIsDynamic=false
       bIsEditable=false
       bIsStaticSwitch=false
       bIsVisible=false
       type {classStructOrEnum {refPath=&NiagaraInt32}}
      }
      Ribbon Width Mode {
       bIsDynamic=false
       bIsEditable=true
       bIsStaticSwitch=true
       bIsVisible=true
       type {classStructOrEnum {refPath=&ENiagara_UnsetDirectSet}}
      }
      Ribbon Width {
       bIsDynamic=false
       bIsEditable=false
       bIsStaticSwitch=false
       bIsVisible=false
       type {classStructOrEnum {refPath=&NiagaraFloat}}
      }
      Ribbon Facing Mode {
       bIsDynamic=false
       bIsEditable=true
       bIsStaticSwitch=true
       bIsVisible=true
       type {classStructOrEnum {refPath=&ENiagara_UnsetDirectSet}}
      }
      Ribbon Facing Vector {
       bIsDynamic=false
       bIsEditable=false
       bIsStaticSwitch=false
       bIsVisible=false
       type {classStructOrEnum {refPath=&Vector3f}}
      }
      Ribbon Twist Mode {
       bIsDynamic=false
       bIsEditable=true
       bIsStaticSwitch=true
       bIsVisible=true
       type {classStructOrEnum {refPath=&ENiagara_UnsetDirectSet}}
      }
      "Ribbon Twist " {
       bIsDynamic=false
       bIsEditable=false
       bIsStaticSwitch=false
       bIsVisible=false
       type {classStructOrEnum {refPath=&NiagaraFloat}}
      }
     ]
     moduleScript {refPath=&InitializeParticle}
    }
    {
     bIsSetParametersModule=false
     enabled=true
     moduleName=InitialMeshOrientation
     inputs [
      Mesh Orientation Mode {
       bIsDynamic=false
       bIsEditable=true
       bIsStaticSwitch=true
       bIsVisible=true
       type {classStructOrEnum {refPath=&ENiagara_MeshOrientationOptions}}
      }
      Orientation Quaternion {bIsDynamic=false, bIsEditable=false, bIsStaticSwitch=false, bIsVisible=false, type {classStructOrEnum {refPath=&Quat4f}}}
      Orientation Matrix {
       bIsDynamic=false
       bIsEditable=false
       bIsStaticSwitch=false
       bIsVisible=false
       type {classStructOrEnum {refPath=&NiagaraMatrix}}
      }
      Orientation Vector {
       bIsDynamic=false
       bIsEditable=false
       bIsStaticSwitch=false
       bIsVisible=false
       type {classStructOrEnum {refPath=&Vector3f}}
      }
      Orientation Coordinate Space {
       bIsDynamic=false
       bIsEditable=false
       bIsStaticSwitch=true
       bIsVisible=false
       type {classStructOrEnum {refPath=&ENiagaraCoordinateSpace}}
      }
      Mesh Axis To Orient {
       bIsDynamic=false
       bIsEditable=false
       bIsStaticSwitch=false
       bIsVisible=false
       type {classStructOrEnum {refPath=&Vector3f}}
      }
      Random Seed {
       bIsDynamic=false
       bIsEditable=false
       bIsStaticSwitch=false
       bIsVisible=false
       type {classStructOrEnum {refPath=&NiagaraInt32}}
      }
      Use Rotation Vector {bIsDynamic=false, bIsEditable=true, bIsStaticSwitch=false, bIsVisible=true, type {classStructOrEnum {refPath=&NiagaraBool}}}
      Rotation {bIsDynamic=true, bIsEditable=true, bIsStaticSwitch=false, bIsVisible=true, type {classStructOrEnum {refPath=&Vector3f}}}
      Rotation Coordinate Space {
       bIsDynamic=false
       bIsEditable=true
       bIsStaticSwitch=true
       bIsVisible=true
       type {classStructOrEnum {refPath=&ENiagaraMeshTransforms}}
      }
     ]
     moduleScript {refPath=&InitialMeshOrientation}
    }
   ]
  }
  particleUpdateScript {
   scriptName=ParticleUpdateScript
   modules [
    {
     bIsSetParametersModule=false
     enabled=true
     moduleName=ParticleState001
     inputs [
      Kill Particles When Lifetime Has Elapsed {bIsDynamic=false, bIsEditable=true, bIsStaticSwitch=true, bIsVisible=true, type {classStructOrEnum {refPath=&NiagaraBool}}}
      Loop Particles Lifetime {bIsDynamic=false, bIsEditable=false, bIsStaticSwitch=false, bIsVisible=false, type {classStructOrEnum {refPath=&NiagaraBool}}}
      Let Infinitely Lived Particles Die When Emitter Deactivates {bIsDynamic=false, bIsEditable=false, bIsStaticSwitch=false, bIsVisible=false, type {classStructOrEnum {refPath=&NiagaraBool}}}
      Lifetime {bIsDynamic=false, bIsEditable=true, bIsStaticSwitch=false, bIsVisible=true, type {classStructOrEnum {refPath=&NiagaraFloat}}}
      DeltaTime {bIsDynamic=false, bIsEditable=true, bIsStaticSwitch=false, bIsVisible=true, type {classStructOrEnum {refPath=&NiagaraFloat}}}
     ]
     moduleScript {refPath=&ParticleState}
    }
    {
     bIsSetParametersModule=false
     enabled=true
     moduleName=ScaleMeshSize
     inputs [
      Scale Factor {bIsDynamic=true, bIsEditable=true, bIsStaticSwitch=false, bIsVisible=true, type {classStructOrEnum {refPath=&Vector3f}}}
      Initial Mesh Scale {bIsDynamic=false, bIsEditable=true, bIsStaticSwitch=false, bIsVisible=true, type {classStructOrEnum {refPath=&Vector3f}}}
     ]
     moduleScript {refPath=&ScaleMeshSize}
    }
    {
     bIsSetParametersModule=false
     enabled=true
     moduleName=CameraOffset
     inputs [Camera Offset Amount {bIsDynamic=false, bIsEditable=true, bIsStaticSwitch=false, bIsVisible=true, type {classStructOrEnum {refPath=&NiagaraFloat}}}]
     moduleScript {refPath=&CameraOffset}
    }
    {
     bIsSetParametersModule=false
     enabled=true
     moduleName=ScaleColor002
     inputs [
      Scale Mode {
       bIsDynamic=false
       bIsEditable=true
       bIsStaticSwitch=true
       bIsVisible=true
       type {classStructOrEnum {refPath=&ENiagaraScaleColorMode}}
      }
      Linear Color Curve {
       bIsDynamic=false
       bIsEditable=false
       bIsStaticSwitch=false
       bIsVisible=false
       type {classStructOrEnum {refPath=&NiagaraDataInterfaceColorCurve}}
      }
      ScaleRGB {bIsDynamic=false, bIsEditable=true, bIsStaticSwitch=false, bIsVisible=true, type {classStructOrEnum {refPath=&NiagaraBool}}}
      Scale RGB {bIsDynamic=false, bIsEditable=true, bIsStaticSwitch=false, bIsVisible=true, type {classStructOrEnum {refPath=&Vector3f}}}
      ScaleRGBA {bIsDynamic=false, bIsEditable=false, bIsStaticSwitch=false, bIsVisible=false, type {classStructOrEnum {refPath=&NiagaraBool}}}
      Scale RGBA {
       bIsDynamic=false
       bIsEditable=false
       bIsStaticSwitch=false
       bIsVisible=false
       type {classStructOrEnum {refPath=&Vector4f}}
      }
      Curve Index {
       bIsDynamic=false
       bIsEditable=false
       bIsStaticSwitch=false
       bIsVisible=false
       type {classStructOrEnum {refPath=&NiagaraFloat}}
      }
      ScaleA {bIsDynamic=false, bIsEditable=true, bIsStaticSwitch=false, bIsVisible=true, type {classStructOrEnum {refPath=&NiagaraBool}}}
      Scale Alpha {bIsDynamic=true, bIsEditable=true, bIsStaticSwitch=false, bIsVisible=true, type {classStructOrEnum {refPath=&NiagaraFloat}}}
      Color Value To Scale {
       bIsDynamic=false
       bIsEditable=true
       bIsStaticSwitch=false
       bIsVisible=true
       type {classStructOrEnum {refPath=&LinearColor}}
      }
     ]
     moduleScript {refPath=&ScaleColor}
    }
    {
     bIsSetParametersModule=true
     enabled=true
     moduleName=SetVariables_#2
     inputs [
      Particles.Scale_BlueRipple {bIsDynamic=true, bIsEditable=true, bIsStaticSwitch=false, bIsVisible=true, type {classStructOrEnum {refPath=&Vector3f}}}
      Particles.Linear Color_BlueRipple {
       bIsDynamic=true
       bIsEditable=true
       bIsStaticSwitch=false
       bIsVisible=true
       type {classStructOrEnum {refPath=&LinearColor}}
      }
     ]
     moduleScript {refPath=&self:PuddleRadial_Trail_0.NiagaraScriptSource_0.NiagaraGraph_0.NiagaraNodeAssignment_0.SetVariables_#2}
    }
    {
     bIsSetParametersModule=false
     enabled=true
     moduleName=BlendSpace
     inputs [
      Position Mode {bIsDynamic=false, bIsEditable=true, bIsStaticSwitch=true, bIsVisible=true, type {classStructOrEnum {refPath=&NiagaraInt32}}}
      Position Coordinate Space {bIsDynamic=false, bIsEditable=true, bIsStaticSwitch=true, bIsVisible=true, type {classStructOrEnum {refPath=&NiagaraInt32}}}
      "Position Blend (World <-> Local)" {bIsDynamic=true, bIsEditable=true, bIsStaticSwitch=false, bIsVisible=true, type {classStructOrEnum {refPath=&NiagaraFloat}}}
      Mesh Orientation Mode {bIsDynamic=false, bIsEditable=true, bIsStaticSwitch=true, bIsVisible=true, type {classStructOrEnum {refPath=&NiagaraInt32}}}
      Mesh Orientation Coordinate Space {bIsDynamic=false, bIsEditable=false, bIsStaticSwitch=true, bIsVisible=false, type {classStructOrEnum {refPath=&NiagaraInt32}}}
      "Mesh Orientation Blend (World <-> Local)" {
       bIsDynamic=false
       bIsEditable=false
       bIsStaticSwitch=false
       bIsVisible=false
       type {classStructOrEnum {refPath=&NiagaraFloat}}
      }
      Velocity Mode {bIsDynamic=false, bIsEditable=true, bIsStaticSwitch=true, bIsVisible=true, type {classStructOrEnum {refPath=&NiagaraInt32}}}
      Velocity Coordinate Space {bIsDynamic=false, bIsEditable=false, bIsStaticSwitch=true, bIsVisible=false, type {classStructOrEnum {refPath=&NiagaraInt32}}}
      "Velocity Blend (World <-> Local)" {
       bIsDynamic=false
       bIsEditable=false
       bIsStaticSwitch=false
       bIsVisible=false
       type {classStructOrEnum {refPath=&NiagaraFloat}}
      }
     ]
     moduleScript {refPath=&BlendSpace}
    }
    {
     bIsSetParametersModule=false
     enabled=true
     moduleName=SolveForcesAndVelocity
     inputs [
      Clamp Velocity {bIsDynamic=false, bIsEditable=true, bIsStaticSwitch=false, bIsVisible=true, type {classStructOrEnum {refPath=&NiagaraBool}}}
      Speed Limit {bIsDynamic=false, bIsEditable=false, bIsStaticSwitch=false, bIsVisible=true, type {classStructOrEnum {refPath=&NiagaraFloat}}}
      Limit Acceleration {bIsDynamic=false, bIsEditable=true, bIsStaticSwitch=false, bIsVisible=true, type {classStructOrEnum {refPath=&NiagaraBool}}}
      Acceleration Limit {bIsDynamic=false, bIsEditable=false, bIsStaticSwitch=false, bIsVisible=true, type {classStructOrEnum {refPath=&NiagaraFloat}}}
      Manually Enable Rotational Solver {bIsDynamic=false, bIsEditable=true, bIsStaticSwitch=false, bIsVisible=true, type {classStructOrEnum {refPath=&NiagaraBool}}}
      Write to Intrinsic Properties {bIsDynamic=false, bIsEditable=true, bIsStaticSwitch=false, bIsVisible=true, type {classStructOrEnum {refPath=&NiagaraBool}}}
      Write to Presolve Properties {bIsDynamic=false, bIsEditable=true, bIsStaticSwitch=true, bIsVisible=true, type {classStructOrEnum {refPath=&NiagaraBool}}}
      DeltaTime {bIsDynamic=false, bIsEditable=true, bIsStaticSwitch=false, bIsVisible=true, type {classStructOrEnum {refPath=&NiagaraFloat}}}
      Mass {bIsDynamic=false, bIsEditable=true, bIsStaticSwitch=false, bIsVisible=true, type {classStructOrEnum {refPath=&NiagaraFloat}}}
      Position {
       bIsDynamic=false
       bIsEditable=true
       bIsStaticSwitch=false
       bIsVisible=true
       type {classStructOrEnum {refPath=&NiagaraPosition}}
      }
      Velocity {bIsDynamic=false, bIsEditable=true, bIsStaticSwitch=false, bIsVisible=true, type {classStructOrEnum {refPath=&Vector3f}}}
      PreviousVelocity {bIsDynamic=false, bIsEditable=true, bIsStaticSwitch=false, bIsVisible=true, type {classStructOrEnum {refPath=&Vector3f}}}
      Force {bIsDynamic=false, bIsEditable=true, bIsStaticSwitch=false, bIsVisible=true, type {classStructOrEnum {refPath=&Vector3f}}}
      Mesh Orientation {bIsDynamic=false, bIsEditable=false, bIsStaticSwitch=false, bIsVisible=false, type {classStructOrEnum {refPath=&Quat4f}}}
      Rotational Velocity {
       bIsDynamic=false
       bIsEditable=false
       bIsStaticSwitch=false
       bIsVisible=false
       type {classStructOrEnum {refPath=&Vector3f}}
      }
      Rotational Force {
       bIsDynamic=false
       bIsEditable=false
       bIsStaticSwitch=false
       bIsVisible=false
       type {classStructOrEnum {refPath=&Vector3f}}
      }
      Rotational Inertia {
       bIsDynamic=false
       bIsEditable=false
       bIsStaticSwitch=false
       bIsVisible=false
       type {classStructOrEnum {refPath=&NiagaraFloat}}
      }
     ]
     moduleScript {refPath=&SolveForcesAndVelocity}
    }
    {
     bIsSetParametersModule=true
     enabled=true
     moduleName=SetVariables_#3
     inputs [Particles.Color_GlowRing {
       bIsDynamic=true
       bIsEditable=true
       bIsStaticSwitch=false
       bIsVisible=true
       type {classStructOrEnum {refPath=&LinearColor}}
      }]
     moduleScript {refPath=&self:PuddleRadial_Trail_0.NiagaraScriptSource_0.NiagaraGraph_0.NiagaraNodeAssignment_2.SetVariables_#3}
    }
   ]
  }
  rendererClasses [{refPath=&NiagaraMeshRendererProperties}]
  renderers [
   {rendererIndex=0, rendererClass {refPath=&NiagaraMeshRendererProperties}}
   {rendererIndex=1, rendererClass {refPath=&NiagaraMeshRendererProperties}}
   {rendererIndex=2, rendererClass {refPath=&NiagaraMeshRendererProperties}}
  ]
 }
 Rocks {
  simTarget=GPUComputeSim
  emitterSpawnScript {scriptName=EmitterSpawnScript, modules []}
  emitterUpdateScript {
   scriptName=EmitterUpdateScript
   modules [
    {
     bIsSetParametersModule=false
     enabled=true
     moduleName=EmitterState
     inputs [
      Life Cycle Mode {
       bIsDynamic=false
       bIsEditable=true
       bIsStaticSwitch=true
       bIsVisible=true
       type {classStructOrEnum {refPath=&ENiagaraEmitterLifeCycleMode}}
      }
      Inactive Response {
       bIsDynamic=false
       bIsEditable=true
       bIsStaticSwitch=true
       bIsVisible=true
       type {classStructOrEnum {refPath=&ENiagaraInactiveMode}}
      }
      Loop Behavior {
       bIsDynamic=false
       bIsEditable=true
       bIsStaticSwitch=true
       bIsVisible=true
       type {classStructOrEnum {refPath=&ENiagara_EmitterStateOptions}}
      }
      Loop Duration Mode {
       bIsDynamic=false
       bIsEditable=false
       bIsStaticSwitch=true
       bIsVisible=false
       type {classStructOrEnum {refPath=&ENiagara_InfiniteLoopDuration}}
      }
      Loop Count {
       bIsDynamic=false
       bIsEditable=false
       bIsStaticSwitch=false
       bIsVisible=false
       type {classStructOrEnum {refPath=&NiagaraInt32}}
      }
      Loop Duration {bIsDynamic=false, bIsEditable=true, bIsStaticSwitch=false, bIsVisible=true, type {classStructOrEnum {refPath=&NiagaraFloat}}}
      Recalculate Duration Each Loop {bIsDynamic=false, bIsEditable=true, bIsStaticSwitch=false, bIsVisible=true, type {classStructOrEnum {refPath=&NiagaraBool}}}
      UseLoopDelay {bIsDynamic=false, bIsEditable=true, bIsStaticSwitch=false, bIsVisible=true, type {classStructOrEnum {refPath=&NiagaraBool}}}
      Loop Delay {bIsDynamic=false, bIsEditable=false, bIsStaticSwitch=false, bIsVisible=true, type {classStructOrEnum {refPath=&NiagaraFloat}}}
      Delay First Loop Only {bIsDynamic=false, bIsEditable=false, bIsStaticSwitch=false, bIsVisible=false, type {classStructOrEnum {refPath=&NiagaraBool}}}
      Scalability Mode {
       bIsDynamic=false
       bIsEditable=true
       bIsStaticSwitch=true
       bIsVisible=true
       type {classStructOrEnum {refPath=&ENiagaraEmitterScalabilityMode_Limited}}
      }
      Enable Distance Culling {bIsDynamic=false, bIsEditable=false, bIsStaticSwitch=true, bIsVisible=false, type {classStructOrEnum {refPath=&NiagaraBool}}}
      MinDistance {
       bIsDynamic=false
       bIsEditable=false
       bIsStaticSwitch=false
       bIsVisible=false
       type {classStructOrEnum {refPath=&NiagaraFloat}}
      }
      Min Distance Response {
       bIsDynamic=false
       bIsEditable=false
       bIsStaticSwitch=false
       bIsVisible=false
       type {classStructOrEnum {refPath=&ENiagaraExecutionStateManagement}}
      }
      MaxDistance {
       bIsDynamic=false
       bIsEditable=false
       bIsStaticSwitch=false
       bIsVisible=false
       type {classStructOrEnum {refPath=&NiagaraFloat}}
      }
      Max Distance Response {
       bIsDynamic=false
       bIsEditable=false
       bIsStaticSwitch=false
       bIsVisible=false
       type {classStructOrEnum {refPath=&ENiagaraExecutionStateManagement}}
      }
      Scale Spawn Count {bIsDynamic=false, bIsEditable=false, bIsStaticSwitch=true, bIsVisible=false, type {classStructOrEnum {refPath=&NiagaraBool}}}
      Spawn Count Scale {
       bIsDynamic=false
       bIsEditable=false
       bIsStaticSwitch=false
       bIsVisible=false
       type {classStructOrEnum {refPath=&NiagaraFloat}}
      }
      Spawn Count Scale By Distance {
       bIsDynamic=false
       bIsEditable=false
       bIsStaticSwitch=false
       bIsVisible=false
       type {classStructOrEnum {refPath=&NiagaraDataInterfaceCurve}}
      }
      Enable Visibility Culling {bIsDynamic=false, bIsEditable=false, bIsStaticSwitch=true, bIsVisible=false, type {classStructOrEnum {refPath=&NiagaraBool}}}
      Visibility Cull Response {
       bIsDynamic=false
       bIsEditable=false
       bIsStaticSwitch=false
       bIsVisible=false
       type {classStructOrEnum {refPath=&ENiagaraExecutionStateManagement}}
      }
      Visibility Cull Delay {
       bIsDynamic=false
       bIsEditable=false
       bIsStaticSwitch=false
       bIsVisible=false
       type {classStructOrEnum {refPath=&NiagaraFloat}}
      }
      Reset Age On Awaken {bIsDynamic=false, bIsEditable=false, bIsStaticSwitch=false, bIsVisible=false, type {classStructOrEnum {refPath=&NiagaraBool}}}
     ]
     moduleScript {refPath=&EmitterState}
    }
    {
     bIsSetParametersModule=false
     enabled=true
     moduleName=SpawnRate
     inputs [
      SpawnRate {bIsDynamic=false, bIsEditable=true, bIsStaticSwitch=false, bIsVisible=true, type {classStructOrEnum {refPath=&NiagaraFloat}}}
      Use Spawn Probability {bIsDynamic=false, bIsEditable=true, bIsStaticSwitch=false, bIsVisible=true, type {classStructOrEnum {refPath=&NiagaraBool}}}
      Spawn Probability {bIsDynamic=false, bIsEditable=true, bIsStaticSwitch=false, bIsVisible=true, type {classStructOrEnum {refPath=&NiagaraFloat}}}
      SpawnGroup {bIsDynamic=false, bIsEditable=true, bIsStaticSwitch=false, bIsVisible=true, type {classStructOrEnum {refPath=&NiagaraInt32}}}
     ]
     moduleScript {refPath=&SpawnRate}
    }
    {
     bIsSetParametersModule=false
     enabled=true
     moduleName=SpawnRate001
     inputs [
      SpawnRate {bIsDynamic=false, bIsEditable=true, bIsStaticSwitch=false, bIsVisible=true, type {classStructOrEnum {refPath=&NiagaraFloat}}}
      Use Spawn Probability {bIsDynamic=false, bIsEditable=true, bIsStaticSwitch=false, bIsVisible=true, type {classStructOrEnum {refPath=&NiagaraBool}}}
      Spawn Probability {bIsDynamic=false, bIsEditable=true, bIsStaticSwitch=false, bIsVisible=true, type {classStructOrEnum {refPath=&NiagaraFloat}}}
      SpawnGroup {bIsDynamic=false, bIsEditable=true, bIsStaticSwitch=false, bIsVisible=true, type {classStructOrEnum {refPath=&NiagaraInt32}}}
     ]
     moduleScript {refPath=&SpawnRate}
    }
   ]
  }
  particleSpawnScript {
   scriptName=ParticleSpawnScript
   modules [
    {
     bIsSetParametersModule=false
     enabled=true
     moduleName=InitializeParticle001
     inputs [
      Lifetime Mode {
       bIsDynamic=false
       bIsEditable=true
       bIsStaticSwitch=true
       bIsVisible=true
       type {classStructOrEnum {refPath=&ENiagara_LifetimeMode}}
      }
      Lifetime Min {bIsDynamic=true, bIsEditable=true, bIsStaticSwitch=false, bIsVisible=true, type {classStructOrEnum {refPath=&NiagaraFloat}}}
      Lifetime {bIsDynamic=true, bIsEditable=false, bIsStaticSwitch=false, bIsVisible=false, type {classStructOrEnum {refPath=&NiagaraFloat}}}
      Lifetime Max {bIsDynamic=true, bIsEditable=true, bIsStaticSwitch=false, bIsVisible=true, type {classStructOrEnum {refPath=&NiagaraFloat}}}
      Lifetime Randomness Mode {
       bIsDynamic=false
       bIsEditable=true
       bIsStaticSwitch=true
       bIsVisible=true
       type {classStructOrEnum {refPath=&ENiagaraRandomnessMode}}
      }
      Color Mode {
       bIsDynamic=false
       bIsEditable=true
       bIsStaticSwitch=true
       bIsVisible=true
       type {classStructOrEnum {refPath=&ENiagara_ColorInitializationMode}}
      }
      Color Minimum {
       bIsDynamic=false
       bIsEditable=false
       bIsStaticSwitch=false
       bIsVisible=false
       type {classStructOrEnum {refPath=&LinearColor}}
      }
      Color {
       bIsDynamic=true
       bIsEditable=true
       bIsStaticSwitch=false
       bIsVisible=true
       type {classStructOrEnum {refPath=&LinearColor}}
      }
      AdjustHue {bIsDynamic=false, bIsEditable=false, bIsStaticSwitch=false, bIsVisible=false, type {classStructOrEnum {refPath=&NiagaraBool}}}
      Hue Shift Range {
       bIsDynamic=false
       bIsEditable=false
       bIsStaticSwitch=false
       bIsVisible=false
       type {classStructOrEnum {refPath=&Vector2f}}
      }
      Color Maximum {
       bIsDynamic=false
       bIsEditable=false
       bIsStaticSwitch=false
       bIsVisible=false
       type {classStructOrEnum {refPath=&LinearColor}}
      }
      AdjustSaturation {bIsDynamic=false, bIsEditable=false, bIsStaticSwitch=false, bIsVisible=false, type {classStructOrEnum {refPath=&NiagaraBool}}}
      Saturation Range {
       bIsDynamic=false
       bIsEditable=false
       bIsStaticSwitch=false
       bIsVisible=false
       type {classStructOrEnum {refPath=&Vector2f}}
      }
      AdjustValue {bIsDynamic=false, bIsEditable=false, bIsStaticSwitch=false, bIsVisible=false, type {classStructOrEnum {refPath=&NiagaraBool}}}
      Value Range {
       bIsDynamic=false
       bIsEditable=false
       bIsStaticSwitch=false
       bIsVisible=false
       type {classStructOrEnum {refPath=&Vector2f}}
      }
      AdjustAlpha {bIsDynamic=false, bIsEditable=false, bIsStaticSwitch=false, bIsVisible=false, type {classStructOrEnum {refPath=&NiagaraBool}}}
      Alpha Scale Range {
       bIsDynamic=false
       bIsEditable=false
       bIsStaticSwitch=false
       bIsVisible=false
       type {classStructOrEnum {refPath=&Vector2f}}
      }
      Color Channel Mode {
       bIsDynamic=false
       bIsEditable=false
       bIsStaticSwitch=true
       bIsVisible=false
       type {classStructOrEnum {refPath=&ENiagaraChannelCorrelation}}
      }
      Color Randomness Mode {
       bIsDynamic=false
       bIsEditable=false
       bIsStaticSwitch=true
       bIsVisible=false
       type {classStructOrEnum {refPath=&ENiagaraRandomnessMode}}
      }
      Position Mode {
       bIsDynamic=false
       bIsEditable=true
       bIsStaticSwitch=true
       bIsVisible=true
       type {classStructOrEnum {refPath=&ENiagara_PositionInitializationMode}}
      }
      Position {
       bIsDynamic=true
       bIsEditable=false
       bIsStaticSwitch=false
       bIsVisible=false
       type {classStructOrEnum {refPath=&NiagaraPosition}}
      }
      UsePositionOffset {bIsDynamic=false, bIsEditable=true, bIsStaticSwitch=false, bIsVisible=true, type {classStructOrEnum {refPath=&NiagaraBool}}}
      Position Offset {bIsDynamic=true, bIsEditable=true, bIsStaticSwitch=false, bIsVisible=true, type {classStructOrEnum {refPath=&Vector3f}}}
      Position Offset Coordinate Space {
       bIsDynamic=false
       bIsEditable=true
       bIsStaticSwitch=false
       bIsVisible=true
       type {classStructOrEnum {refPath=&ENiagaraCoordinateSpace}}
      }
      Mass Mode {
       bIsDynamic=false
       bIsEditable=true
       bIsStaticSwitch=true
       bIsVisible=true
       type {classStructOrEnum {refPath=&ENiagara_MassInitializationMode}}
      }
      Mass Min {
       bIsDynamic=false
       bIsEditable=false
       bIsStaticSwitch=false
       bIsVisible=false
       type {classStructOrEnum {refPath=&NiagaraFloat}}
      }
      Mass {
       bIsDynamic=false
       bIsEditable=false
       bIsStaticSwitch=false
       bIsVisible=false
       type {classStructOrEnum {refPath=&NiagaraFloat}}
      }
      Mass Max {
       bIsDynamic=false
       bIsEditable=false
       bIsStaticSwitch=false
       bIsVisible=false
       type {classStructOrEnum {refPath=&NiagaraFloat}}
      }
      Mass Randomness Mode {
       bIsDynamic=false
       bIsEditable=false
       bIsStaticSwitch=true
       bIsVisible=false
       type {classStructOrEnum {refPath=&ENiagaraRandomnessMode}}
      }
      Write Material Random {bIsDynamic=false, bIsEditable=true, bIsStaticSwitch=false, bIsVisible=true, type {classStructOrEnum {refPath=&NiagaraBool}}}
      Material Random {bIsDynamic=false, bIsEditable=false, bIsStaticSwitch=false, bIsVisible=true, type {classStructOrEnum {refPath=&NiagaraFloat}}}
      Sprite Size Mode {
       bIsDynamic=false
       bIsEditable=true
       bIsStaticSwitch=true
       bIsVisible=true
       type {classStructOrEnum {refPath=&ENiagara_SizeScaleMode}}
      }
      Sprite Size {
       bIsDynamic=false
       bIsEditable=false
       bIsStaticSwitch=false
       bIsVisible=false
       type {classStructOrEnum {refPath=&Vector2f}}
      }
      Uniform Sprite Size {bIsDynamic=true, bIsEditable=true, bIsStaticSwitch=false, bIsVisible=true, type {classStructOrEnum {refPath=&NiagaraFloat}}}
      Uniform Sprite Size Min {
       bIsDynamic=false
       bIsEditable=false
       bIsStaticSwitch=false
       bIsVisible=false
       type {classStructOrEnum {refPath=&NiagaraFloat}}
      }
      Sprite Size Min {
       bIsDynamic=false
       bIsEditable=false
       bIsStaticSwitch=false
       bIsVisible=false
       type {classStructOrEnum {refPath=&Vector2f}}
      }
      Sprite Size Max {
       bIsDynamic=false
       bIsEditable=false
       bIsStaticSwitch=false
       bIsVisible=false
       type {classStructOrEnum {refPath=&Vector2f}}
      }
      Uniform Sprite Size Max {
       bIsDynamic=false
       bIsEditable=false
       bIsStaticSwitch=false
       bIsVisible=false
       type {classStructOrEnum {refPath=&NiagaraFloat}}
      }
      Sprite Size Randomness Mode {
       bIsDynamic=false
       bIsEditable=false
       bIsStaticSwitch=true
       bIsVisible=false
       type {classStructOrEnum {refPath=&ENiagaraRandomnessMode}}
      }
      Sprite Rotation Mode {
       bIsDynamic=false
       bIsEditable=true
       bIsStaticSwitch=true
       bIsVisible=true
       type {classStructOrEnum {refPath=&ENiagara_SpriteRotationMode}}
      }
      Sprite Rotation Angle Min {
       bIsDynamic=false
       bIsEditable=false
       bIsStaticSwitch=false
       bIsVisible=false
       type {classStructOrEnum {refPath=&NiagaraFloat}}
      }
      Sprite Rotation Angle {
       bIsDynamic=false
       bIsEditable=false
       bIsStaticSwitch=false
       bIsVisible=false
       type {classStructOrEnum {refPath=&NiagaraFloat}}
      }
      Sprite Rotation Angle Max {
       bIsDynamic=false
       bIsEditable=false
       bIsStaticSwitch=false
       bIsVisible=false
       type {classStructOrEnum {refPath=&NiagaraFloat}}
      }
      Sprite Rotation Randomness Mode {
       bIsDynamic=false
       bIsEditable=false
       bIsStaticSwitch=true
       bIsVisible=false
       type {classStructOrEnum {refPath=&ENiagaraRandomnessMode}}
      }
      Sprite UV Mode {
       bIsDynamic=false
       bIsEditable=true
       bIsStaticSwitch=true
       bIsVisible=true
       type {classStructOrEnum {refPath=&ENiagara_UVFlippingMode}}
      }
      Sprite UV Randomness Mode {
       bIsDynamic=false
       bIsEditable=false
       bIsStaticSwitch=true
       bIsVisible=false
       type {classStructOrEnum {refPath=&ENiagaraRandomnessMode}}
      }
      Sprite UV Scale {
       bIsDynamic=false
       bIsEditable=false
       bIsStaticSwitch=false
       bIsVisible=false
       type {classStructOrEnum {refPath=&Vector2f}}
      }
      Mesh Scale Mode {
       bIsDynamic=false
       bIsEditable=true
       bIsStaticSwitch=true
       bIsVisible=true
       type {classStructOrEnum {refPath=&ENiagara_SizeScaleMode}}
      }
      Mesh Uniform Scale Min {bIsDynamic=false, bIsEditable=true, bIsStaticSwitch=false, bIsVisible=true, type {classStructOrEnum {refPath=&NiagaraFloat}}}
      Mesh Uniform Scale {
       bIsDynamic=false
       bIsEditable=false
       bIsStaticSwitch=false
       bIsVisible=false
       type {classStructOrEnum {refPath=&NiagaraFloat}}
      }
      Mesh Scale Min {
       bIsDynamic=false
       bIsEditable=false
       bIsStaticSwitch=false
       bIsVisible=false
       type {classStructOrEnum {refPath=&Vector3f}}
      }
      Mesh Scale {bIsDynamic=true, bIsEditable=false, bIsStaticSwitch=false, bIsVisible=false, type {classStructOrEnum {refPath=&Vector3f}}}
      Mesh Scale Max {
       bIsDynamic=false
       bIsEditable=false
       bIsStaticSwitch=false
       bIsVisible=false
       type {classStructOrEnum {refPath=&Vector3f}}
      }
      Mesh Uniform Scale Max {bIsDynamic=false, bIsEditable=true, bIsStaticSwitch=false, bIsVisible=true, type {classStructOrEnum {refPath=&NiagaraFloat}}}
      Mesh Scale Randomness Mode {
       bIsDynamic=false
       bIsEditable=true
       bIsStaticSwitch=true
       bIsVisible=true
       type {classStructOrEnum {refPath=&ENiagaraRandomnessMode}}
      }
      Mesh Renderer Array Visibility Mode {
       bIsDynamic=false
       bIsEditable=true
       bIsStaticSwitch=true
       bIsVisible=true
       type {classStructOrEnum {refPath=&ENiagara_UnsetDirectSetRandom}}
      }
      Mesh Renderer Mesh Index {
       bIsDynamic=false
       bIsEditable=false
       bIsStaticSwitch=false
       bIsVisible=false
       type {classStructOrEnum {refPath=&NiagaraInt32}}
      }
      Wrap Index By Mesh Count {bIsDynamic=false, bIsEditable=false, bIsStaticSwitch=true, bIsVisible=false, type {classStructOrEnum {refPath=&NiagaraBool}}}
      Mesh Renderer Info {
       bIsDynamic=false
       bIsEditable=true
       bIsStaticSwitch=false
       bIsVisible=true
       type {classStructOrEnum {refPath=&NiagaraDataInterfaceMeshRendererInfo}}
      }
      Mesh Renderer Random Seed {bIsDynamic=false, bIsEditable=true, bIsStaticSwitch=false, bIsVisible=true, type {classStructOrEnum {refPath=&NiagaraInt32}}}
      Ribbon Width Mode {
       bIsDynamic=false
       bIsEditable=true
       bIsStaticSwitch=true
       bIsVisible=true
       type {classStructOrEnum {refPath=&ENiagara_UnsetDirectSet}}
      }
      Ribbon Width {
       bIsDynamic=false
       bIsEditable=false
       bIsStaticSwitch=false
       bIsVisible=false
       type {classStructOrEnum {refPath=&NiagaraFloat}}
      }
      Ribbon Facing Mode {
       bIsDynamic=false
       bIsEditable=true
       bIsStaticSwitch=true
       bIsVisible=true
       type {classStructOrEnum {refPath=&ENiagara_UnsetDirectSet}}
      }
      Ribbon Facing Vector {
       bIsDynamic=false
       bIsEditable=false
       bIsStaticSwitch=false
       bIsVisible=false
       type {classStructOrEnum {refPath=&Vector3f}}
      }
      Ribbon Twist Mode {
       bIsDynamic=false
       bIsEditable=true
       bIsStaticSwitch=true
       bIsVisible=true
       type {classStructOrEnum {refPath=&ENiagara_UnsetDirectSet}}
      }
      "Ribbon Twist " {
       bIsDynamic=false
       bIsEditable=false
       bIsStaticSwitch=false
       bIsVisible=false
       type {classStructOrEnum {refPath=&NiagaraFloat}}
      }
     ]
     moduleScript {refPath=&InitializeParticle}
    }
    {
     bIsSetParametersModule=false
     enabled=true
     moduleName=InitialMeshOrientation
     inputs [
      Mesh Orientation Mode {
       bIsDynamic=false
       bIsEditable=true
       bIsStaticSwitch=true
       bIsVisible=true
       type {classStructOrEnum {refPath=&ENiagara_MeshOrientationOptions}}
      }
      Orientation Quaternion {bIsDynamic=false, bIsEditable=false, bIsStaticSwitch=false, bIsVisible=false, type {classStructOrEnum {refPath=&Quat4f}}}
      Orientation Matrix {
       bIsDynamic=false
       bIsEditable=false
       bIsStaticSwitch=false
       bIsVisible=false
       type {classStructOrEnum {refPath=&NiagaraMatrix}}
      }
      Orientation Vector {
       bIsDynamic=false
       bIsEditable=false
       bIsStaticSwitch=false
       bIsVisible=false
       type {classStructOrEnum {refPath=&Vector3f}}
      }
      Orientation Coordinate Space {
       bIsDynamic=false
       bIsEditable=false
       bIsStaticSwitch=true
       bIsVisible=false
       type {classStructOrEnum {refPath=&ENiagaraCoordinateSpace}}
      }
      Mesh Axis To Orient {
       bIsDynamic=false
       bIsEditable=false
       bIsStaticSwitch=false
       bIsVisible=false
       type {classStructOrEnum {refPath=&Vector3f}}
      }
      Random Seed {bIsDynamic=false, bIsEditable=true, bIsStaticSwitch=false, bIsVisible=true, type {classStructOrEnum {refPath=&NiagaraInt32}}}
      Use Rotation Vector {bIsDynamic=false, bIsEditable=true, bIsStaticSwitch=false, bIsVisible=true, type {classStructOrEnum {refPath=&NiagaraBool}}}
      Rotation {bIsDynamic=false, bIsEditable=false, bIsStaticSwitch=false, bIsVisible=true, type {classStructOrEnum {refPath=&Vector3f}}}
      Rotation Coordinate Space {
       bIsDynamic=false
       bIsEditable=false
       bIsStaticSwitch=true
       bIsVisible=true
       type {classStructOrEnum {refPath=&ENiagaraMeshTransforms}}
      }
     ]
     moduleScript {refPath=&InitialMeshOrientation}
    }
    {
     bIsSetParametersModule=false
     enabled=true
     moduleName=ShapeLocation
     inputs [
      Shape Primitive {
       bIsDynamic=false
       bIsEditable=true
       bIsStaticSwitch=true
       bIsVisible=true
       type {classStructOrEnum {refPath=&ENiagara_LocationShapes}}
      }
      Box / Plane Mode {
       bIsDynamic=false
       bIsEditable=false
       bIsStaticSwitch=true
       bIsVisible=false
       type {classStructOrEnum {refPath=&ENiagara_BoxPlaneMode}}
      }
      Cone Mode {
       bIsDynamic=false
       bIsEditable=false
       bIsStaticSwitch=true
       bIsVisible=false
       type {classStructOrEnum {refPath=&ENiagara_ConeMode}}
      }
      Cylinder Height {
       bIsDynamic=false
       bIsEditable=false
       bIsStaticSwitch=false
       bIsVisible=false
       type {classStructOrEnum {refPath=&NiagaraFloat}}
      }
      Sphere Radius {
       bIsDynamic=false
       bIsEditable=false
       bIsStaticSwitch=false
       bIsVisible=false
       type {classStructOrEnum {refPath=&NiagaraFloat}}
      }
      Ring / Disc Mode {
       bIsDynamic=false
       bIsEditable=true
       bIsStaticSwitch=true
       bIsVisible=true
       type {classStructOrEnum {refPath=&ENiagaraRingDiscMode}}
      }
      Torus Mode {
       bIsDynamic=false
       bIsEditable=false
       bIsStaticSwitch=true
       bIsVisible=false
       type {classStructOrEnum {refPath=&ENiagaraShapeTorusMode}}
      }
      Knot Ring Loops {
       bIsDynamic=false
       bIsEditable=false
       bIsStaticSwitch=false
       bIsVisible=false
       type {classStructOrEnum {refPath=&NiagaraInt32}}
      }
      Knot Helix Loops {
       bIsDynamic=false
       bIsEditable=false
       bIsStaticSwitch=false
       bIsVisible=false
       type {classStructOrEnum {refPath=&NiagaraInt32}}
      }
      Enable Torus Knot Radius {bIsDynamic=false, bIsEditable=false, bIsStaticSwitch=false, bIsVisible=false, type {classStructOrEnum {refPath=&NiagaraBool}}}
      Torus Knot Tube Radius {
       bIsDynamic=false
       bIsEditable=false
       bIsStaticSwitch=false
       bIsVisible=false
       type {classStructOrEnum {refPath=&NiagaraFloat}}
      }
      Cylinder Radius {
       bIsDynamic=false
       bIsEditable=false
       bIsStaticSwitch=false
       bIsVisible=false
       type {classStructOrEnum {refPath=&NiagaraFloat}}
      }
      Cylinder Height Midpoint {
       bIsDynamic=false
       bIsEditable=false
       bIsStaticSwitch=false
       bIsVisible=false
       type {classStructOrEnum {refPath=&NiagaraFloat}}
      }
      Ring Radius {bIsDynamic=false, bIsEditable=true, bIsStaticSwitch=false, bIsVisible=true, type {classStructOrEnum {refPath=&NiagaraFloat}}}
      Box Size {
       bIsDynamic=false
       bIsEditable=false
       bIsStaticSwitch=false
       bIsVisible=false
       type {classStructOrEnum {refPath=&Vector3f}}
      }
      Box Midpoint {
       bIsDynamic=false
       bIsEditable=false
       bIsStaticSwitch=false
       bIsVisible=false
       type {classStructOrEnum {refPath=&Vector3f}}
      }
      Cone Length {
       bIsDynamic=false
       bIsEditable=false
       bIsStaticSwitch=false
       bIsVisible=false
       type {classStructOrEnum {refPath=&NiagaraFloat}}
      }
      Plane Size {
       bIsDynamic=false
       bIsEditable=false
       bIsStaticSwitch=false
       bIsVisible=false
       type {classStructOrEnum {refPath=&Vector2f}}
      }
      Wedge Horizontal Angle {
       bIsDynamic=false
       bIsEditable=false
       bIsStaticSwitch=false
       bIsVisible=false
       type {classStructOrEnum {refPath=&NiagaraFloat}}
      }
      Large Radius {
       bIsDynamic=false
       bIsEditable=false
       bIsStaticSwitch=false
       bIsVisible=false
       type {classStructOrEnum {refPath=&NiagaraFloat}}
      }
      Plane Midpoint {
       bIsDynamic=false
       bIsEditable=false
       bIsStaticSwitch=false
       bIsVisible=false
       type {classStructOrEnum {refPath=&Vector2f}}
      }
      Wedge Vertical Angle {
       bIsDynamic=false
       bIsEditable=false
       bIsStaticSwitch=false
       bIsVisible=false
       type {classStructOrEnum {refPath=&NiagaraFloat}}
      }
      Cone Angle {
       bIsDynamic=false
       bIsEditable=false
       bIsStaticSwitch=false
       bIsVisible=false
       type {classStructOrEnum {refPath=&NiagaraFloat}}
      }
      Handle Radius {
       bIsDynamic=false
       bIsEditable=false
       bIsStaticSwitch=false
       bIsVisible=false
       type {classStructOrEnum {refPath=&NiagaraFloat}}
      }
      Cone Inner Angle {
       bIsDynamic=false
       bIsEditable=false
       bIsStaticSwitch=false
       bIsVisible=false
       type {classStructOrEnum {refPath=&NiagaraFloat}}
      }
      Cone Radial Angle {
       bIsDynamic=false
       bIsEditable=false
       bIsStaticSwitch=false
       bIsVisible=false
       type {classStructOrEnum {refPath=&NiagaraFloat}}
      }
      Cone Angle Type {
       bIsDynamic=false
       bIsEditable=false
       bIsStaticSwitch=true
       bIsVisible=false
       type {classStructOrEnum {refPath=&ENiagara_AngleInput}}
      }
      Flatten Endcaps {bIsDynamic=false, bIsEditable=false, bIsStaticSwitch=true, bIsVisible=false, type {classStructOrEnum {refPath=&NiagaraBool}}}
      Apply To Particle Position {
       bIsDynamic=false
       bIsEditable=true
       bIsStaticSwitch=true
       bIsVisible=true
       type {classStructOrEnum {refPath=&ENiagara_AttributeSamplingApplyOutput}}
      }
      Cone Surface Distribution {
       bIsDynamic=false
       bIsEditable=false
       bIsStaticSwitch=false
       bIsVisible=false
       type {classStructOrEnum {refPath=&NiagaraFloat}}
      }
      Ring / Disc Distribution Mode {
       bIsDynamic=false
       bIsEditable=true
       bIsStaticSwitch=true
       bIsVisible=true
       type {classStructOrEnum {refPath=&ENiagaraSphereDistributionMode}}
      }
      Sphere Distribution {
       bIsDynamic=false
       bIsEditable=false
       bIsStaticSwitch=true
       bIsVisible=false
       type {classStructOrEnum {refPath=&ENiagaraSphereDistributionMode}}
      }
      Torus Distribution Mode {
       bIsDynamic=false
       bIsEditable=false
       bIsStaticSwitch=true
       bIsVisible=false
       type {classStructOrEnum {refPath=&ENiagaraTorusDistributionMode}}
      }
      Sphere Surface Distribution {
       bIsDynamic=false
       bIsEditable=false
       bIsStaticSwitch=false
       bIsVisible=false
       type {classStructOrEnum {refPath=&NiagaraFloat}}
      }
      Hemisphere Distribution {
       bIsDynamic=false
       bIsEditable=false
       bIsStaticSwitch=false
       bIsVisible=false
       type {classStructOrEnum {refPath=&Vector2f}}
      }
      Cylinder Distribution {
       bIsDynamic=false
       bIsEditable=false
       bIsStaticSwitch=true
       bIsVisible=false
       type {classStructOrEnum {refPath=&ENiagara_CylinderMode}}
      }
      Surface Only {bIsDynamic=false, bIsEditable=false, bIsStaticSwitch=false, bIsVisible=false, type {classStructOrEnum {refPath=&NiagaraBool}}}
      Surface Only Band Thickness {
       bIsDynamic=false
       bIsEditable=false
       bIsStaticSwitch=false
       bIsVisible=false
       type {classStructOrEnum {refPath=&NiagaraFloat}}
      }
      Use Endcaps In Surface Only Mode {bIsDynamic=false, bIsEditable=false, bIsStaticSwitch=true, bIsVisible=false, type {classStructOrEnum {refPath=&NiagaraBool}}}
      Hemicircle X {bIsDynamic=false, bIsEditable=false, bIsStaticSwitch=true, bIsVisible=false, type {classStructOrEnum {refPath=&NiagaraBool}}}
      Hemicircle Y {bIsDynamic=false, bIsEditable=false, bIsStaticSwitch=true, bIsVisible=false, type {classStructOrEnum {refPath=&NiagaraBool}}}
      Disc Coverage {bIsDynamic=false, bIsEditable=true, bIsStaticSwitch=false, bIsVisible=true, type {classStructOrEnum {refPath=&NiagaraFloat}}}
      Hemicircle Internal Cap {bIsDynamic=false, bIsEditable=false, bIsStaticSwitch=true, bIsVisible=false, type {classStructOrEnum {refPath=&NiagaraBool}}}
      U Position {
       bIsDynamic=false
       bIsEditable=false
       bIsStaticSwitch=false
       bIsVisible=false
       type {classStructOrEnum {refPath=&NiagaraFloat}}
      }
      Enable Lathe Profile {bIsDynamic=false, bIsEditable=false, bIsStaticSwitch=true, bIsVisible=false, type {classStructOrEnum {refPath=&NiagaraBool}}}
      Lathe Profile {
       bIsDynamic=false
       bIsEditable=false
       bIsStaticSwitch=false
       bIsVisible=false
       type {classStructOrEnum {refPath=&NiagaraDataInterfaceCurve}}
      }
      Surface Distribution {
       bIsDynamic=false
       bIsEditable=false
       bIsStaticSwitch=false
       bIsVisible=false
       type {classStructOrEnum {refPath=&NiagaraFloat}}
      }
      Surface Expansion Mode {
       bIsDynamic=false
       bIsEditable=false
       bIsStaticSwitch=false
       bIsVisible=false
       type {classStructOrEnum {refPath=&ENiagaraExpansionMode}}
      }
      U Distribution {bIsDynamic=false, bIsEditable=true, bIsStaticSwitch=false, bIsVisible=true, type {classStructOrEnum {refPath=&NiagaraFloat}}}
      V Position {
       bIsDynamic=false
       bIsEditable=false
       bIsStaticSwitch=false
       bIsVisible=false
       type {classStructOrEnum {refPath=&NiagaraFloat}}
      }
      Radius Position {
       bIsDynamic=false
       bIsEditable=false
       bIsStaticSwitch=false
       bIsVisible=false
       type {classStructOrEnum {refPath=&NiagaraFloat}}
      }
      Sphere Radius Position {
       bIsDynamic=false
       bIsEditable=false
       bIsStaticSwitch=false
       bIsVisible=false
       type {classStructOrEnum {refPath=&NiagaraFloat}}
      }
      V Distribution {
       bIsDynamic=false
       bIsEditable=false
       bIsStaticSwitch=false
       bIsVisible=false
       type {classStructOrEnum {refPath=&NiagaraFloat}}
      }
      Evaluation Type {
       bIsDynamic=false
       bIsEditable=false
       bIsStaticSwitch=true
       bIsVisible=false
       type {classStructOrEnum {refPath=&ENiagaraRandomnessEvaluation}}
      }
      Uniform Distribution {
       bIsDynamic=false
       bIsEditable=false
       bIsStaticSwitch=false
       bIsVisible=false
       type {classStructOrEnum {refPath=&NiagaraFloat}}
      }
      Uniform Spiral Amount {
       bIsDynamic=false
       bIsEditable=false
       bIsStaticSwitch=false
       bIsVisible=false
       type {classStructOrEnum {refPath=&NiagaraFloat}}
      }
      Uniform Spiral Falloff {
       bIsDynamic=false
       bIsEditable=false
       bIsStaticSwitch=false
       bIsVisible=false
       type {classStructOrEnum {refPath=&NiagaraFloat}}
      }
      Uniform Count {
       bIsDynamic=false
       bIsEditable=false
       bIsStaticSwitch=false
       bIsVisible=false
       type {classStructOrEnum {refPath=&NiagaraInt32}}
      }
      Uniform Particle Index {
       bIsDynamic=false
       bIsEditable=false
       bIsStaticSwitch=false
       bIsVisible=false
       type {classStructOrEnum {refPath=&NiagaraInt32}}
      }
      Hemisphere Angle Type {
       bIsDynamic=false
       bIsEditable=false
       bIsStaticSwitch=true
       bIsVisible=false
       type {classStructOrEnum {refPath=&ENiagara_AngleInput}}
      }
      Uniform Endpoint Offset {
       bIsDynamic=false
       bIsEditable=false
       bIsStaticSwitch=false
       bIsVisible=false
       type {classStructOrEnum {refPath=&NiagaraFloat}}
      }
      MaskSpawnGroup {bIsDynamic=false, bIsEditable=true, bIsStaticSwitch=false, bIsVisible=true, type {classStructOrEnum {refPath=&NiagaraBool}}}
      Spawn Group Mask {bIsDynamic=false, bIsEditable=false, bIsStaticSwitch=false, bIsVisible=true, type {classStructOrEnum {refPath=&NiagaraInt32}}}
      Spawn Group {
       bIsDynamic=false
       bIsEditable=false
       bIsStaticSwitch=false
       bIsVisible=false
       type {classStructOrEnum {refPath=&NiagaraInt32}}
      }
      Randomness Mode {
       bIsDynamic=false
       bIsEditable=true
       bIsStaticSwitch=true
       bIsVisible=true
       type {classStructOrEnum {refPath=&ENiagaraRandomnessMode}}
      }
      Override Seed {bIsDynamic=false, bIsEditable=true, bIsStaticSwitch=false, bIsVisible=true, type {classStructOrEnum {refPath=&NiagaraBool}}}
      Random Seed {bIsDynamic=false, bIsEditable=false, bIsStaticSwitch=false, bIsVisible=true, type {classStructOrEnum {refPath=&NiagaraInt32}}}
      Fixed Random Seed {bIsDynamic=false, bIsEditable=false, bIsStaticSwitch=true, bIsVisible=false, type {classStructOrEnum {refPath=&NiagaraBool}}}
      Transform Order {
       bIsDynamic=false
       bIsEditable=true
       bIsStaticSwitch=true
       bIsVisible=true
       type {classStructOrEnum {refPath=&ENiagara_TransformOrder}}
      }
      Scale Mode {
       bIsDynamic=false
       bIsEditable=true
       bIsStaticSwitch=true
       bIsVisible=true
       type {classStructOrEnum {refPath=&ENiagara_ScaleMode}}
      }
      Non Uniform Scale {bIsDynamic=false, bIsEditable=true, bIsStaticSwitch=false, bIsVisible=true, type {classStructOrEnum {refPath=&Vector3f}}}
      UseOwnerScale {bIsDynamic=false, bIsEditable=true, bIsStaticSwitch=false, bIsVisible=true, type {classStructOrEnum {refPath=&NiagaraBool}}}
      Apply Owner Scale {bIsDynamic=false, bIsEditable=true, bIsStaticSwitch=false, bIsVisible=true, type {classStructOrEnum {refPath=&Vector3f}}}
      Rotation Mode {
       bIsDynamic=false
       bIsEditable=true
       bIsStaticSwitch=true
       bIsVisible=true
       type {classStructOrEnum {refPath=&ENiagara_RotationMode}}
      }
      Rotation Axis {
       bIsDynamic=false
       bIsEditable=false
       bIsStaticSwitch=false
       bIsVisible=false
       type {classStructOrEnum {refPath=&Vector3f}}
      }
      Rotation Matrix {
       bIsDynamic=false
       bIsEditable=false
       bIsStaticSwitch=false
       bIsVisible=false
       type {classStructOrEnum {refPath=&NiagaraMatrix}}
      }
      Rotation Quaternion {bIsDynamic=false, bIsEditable=false, bIsStaticSwitch=false, bIsVisible=false, type {classStructOrEnum {refPath=&Quat4f}}}
      Yaw / Pitch / Roll {bIsDynamic=false, bIsEditable=true, bIsStaticSwitch=false, bIsVisible=true, type {classStructOrEnum {refPath=&Vector3f}}}
      Rotation Angle {
       bIsDynamic=false
       bIsEditable=false
       bIsStaticSwitch=false
       bIsVisible=false
       type {classStructOrEnum {refPath=&NiagaraFloat}}
      }
      Invert Rotation Quaternion {bIsDynamic=false, bIsEditable=false, bIsStaticSwitch=false, bIsVisible=false, type {classStructOrEnum {refPath=&NiagaraBool}}}
      Additional Quaternion Rotation {bIsDynamic=false, bIsEditable=false, bIsStaticSwitch=false, bIsVisible=false, type {classStructOrEnum {refPath=&NiagaraBool}}}
      Additional Yaw / Pitch / Roll {
       bIsDynamic=false
       bIsEditable=false
       bIsStaticSwitch=false
       bIsVisible=false
       type {classStructOrEnum {refPath=&Vector3f}}
      }
      Rotation Angle Type {
       bIsDynamic=false
       bIsEditable=true
       bIsStaticSwitch=true
       bIsVisible=true
       type {classStructOrEnum {refPath=&ENiagara_AngleInput}}
      }
      Rotation Coordinate Space {
       bIsDynamic=false
       bIsEditable=true
       bIsStaticSwitch=false
       bIsVisible=true
       type {classStructOrEnum {refPath=&ENiagaraCoordinateSpace}}
      }
      Offset Mode {
       bIsDynamic=false
       bIsEditable=true
       bIsStaticSwitch=true
       bIsVisible=true
       type {classStructOrEnum {refPath=&ENiagara_OffsetMode}}
      }
      Shape Origin {
       bIsDynamic=false
       bIsEditable=true
       bIsStaticSwitch=false
       bIsVisible=true
       type {classStructOrEnum {refPath=&NiagaraPosition}}
      }
      Offset {bIsDynamic=true, bIsEditable=true, bIsStaticSwitch=false, bIsVisible=true, type {classStructOrEnum {refPath=&Vector3f}}}
      Offset Coordinate Space {
       bIsDynamic=false
       bIsEditable=true
       bIsStaticSwitch=false
       bIsVisible=true
       type {classStructOrEnum {refPath=&ENiagaraCoordinateSpace}}
      }
      Custom Transform Matrix {
       bIsDynamic=false
       bIsEditable=false
       bIsStaticSwitch=false
       bIsVisible=false
       type {classStructOrEnum {refPath=&NiagaraMatrix}}
      }
      Transform Method {
       bIsDynamic=false
       bIsEditable=true
       bIsStaticSwitch=true
       bIsVisible=true
       type {classStructOrEnum {refPath=&ENiagara_TransformType}}
      }
     ]
     moduleScript {refPath=&ShapeLocation}
    }
    {
     bIsSetParametersModule=false
     enabled=true
     moduleName=AddVelocity
     inputs [
      Velocity Mode {
       bIsDynamic=false
       bIsEditable=true
       bIsStaticSwitch=true
       bIsVisible=true
       type {classStructOrEnum {refPath=&ENiagara_VelocityMode}}
      }
      Velocity {bIsDynamic=false, bIsEditable=true, bIsStaticSwitch=false, bIsVisible=true, type {classStructOrEnum {refPath=&Vector3f}}}
      Velocity Speed {
       bIsDynamic=false
       bIsEditable=false
       bIsStaticSwitch=false
       bIsVisible=false
       type {classStructOrEnum {refPath=&NiagaraFloat}}
      }
      Velocity Speed Scale {bIsDynamic=true, bIsEditable=true, bIsStaticSwitch=false, bIsVisible=true, type {classStructOrEnum {refPath=&NiagaraFloat}}}
      Constrain To Radius {
       bIsDynamic=false
       bIsEditable=false
       bIsStaticSwitch=false
       bIsVisible=false
       type {classStructOrEnum {refPath=&NiagaraFloat}}
      }
      Distribution Along Cone Axis {
       bIsDynamic=false
       bIsEditable=false
       bIsStaticSwitch=false
       bIsVisible=false
       type {classStructOrEnum {refPath=&NiagaraFloat}}
      }
      Radius Falloff Near / Far {
       bIsDynamic=false
       bIsEditable=false
       bIsStaticSwitch=false
       bIsVisible=false
       type {classStructOrEnum {refPath=&Vector2f}}
      }
      Radius Falloff Exponent {
       bIsDynamic=false
       bIsEditable=false
       bIsStaticSwitch=false
       bIsVisible=false
       type {classStructOrEnum {refPath=&NiagaraFloat}}
      }
      Speed Falloff From Cone Axis {
       bIsDynamic=false
       bIsEditable=false
       bIsStaticSwitch=false
       bIsVisible=false
       type {classStructOrEnum {refPath=&NiagaraFloat}}
      }
      Invert Velocity Falloff {bIsDynamic=false, bIsEditable=false, bIsStaticSwitch=false, bIsVisible=false, type {classStructOrEnum {refPath=&NiagaraBool}}}
      Override Seed {bIsDynamic=false, bIsEditable=false, bIsStaticSwitch=false, bIsVisible=false, type {classStructOrEnum {refPath=&NiagaraBool}}}
      Random Seed {
       bIsDynamic=false
       bIsEditable=false
       bIsStaticSwitch=false
       bIsVisible=false
       type {classStructOrEnum {refPath=&NiagaraInt32}}
      }
      Fixed Random Seed {bIsDynamic=false, bIsEditable=false, bIsStaticSwitch=true, bIsVisible=false, type {classStructOrEnum {refPath=&NiagaraBool}}}
      Randomness Mode {
       bIsDynamic=false
       bIsEditable=false
       bIsStaticSwitch=true
       bIsVisible=false
       type {classStructOrEnum {refPath=&ENiagaraRandomnessMode}}
      }
      Cone Axis {
       bIsDynamic=false
       bIsEditable=false
       bIsStaticSwitch=false
       bIsVisible=false
       type {classStructOrEnum {refPath=&Vector3f}}
      }
      Cone Angle {
       bIsDynamic=false
       bIsEditable=false
       bIsStaticSwitch=false
       bIsVisible=false
       type {classStructOrEnum {refPath=&NiagaraFloat}}
      }
      Inner Cone Angle {
       bIsDynamic=false
       bIsEditable=false
       bIsStaticSwitch=false
       bIsVisible=false
       type {classStructOrEnum {refPath=&NiagaraFloat}}
      }
      Cone Angle Mode {
       bIsDynamic=false
       bIsEditable=false
       bIsStaticSwitch=true
       bIsVisible=false
       type {classStructOrEnum {refPath=&ENiagara_AngleInput}}
      }
      Rotation Mode {
       bIsDynamic=false
       bIsEditable=true
       bIsStaticSwitch=true
       bIsVisible=true
       type {classStructOrEnum {refPath=&ENiagara_RotationMode}}
      }
      Rotation Axis {
       bIsDynamic=false
       bIsEditable=false
       bIsStaticSwitch=false
       bIsVisible=false
       type {classStructOrEnum {refPath=&Vector3f}}
      }
      Yaw / Pitch / Roll {
       bIsDynamic=false
       bIsEditable=false
       bIsStaticSwitch=false
       bIsVisible=false
       type {classStructOrEnum {refPath=&Vector3f}}
      }
      Rotation Quaternion {bIsDynamic=false, bIsEditable=false, bIsStaticSwitch=false, bIsVisible=false, type {classStructOrEnum {refPath=&Quat4f}}}
      Rotation Matrix {
       bIsDynamic=false
       bIsEditable=false
       bIsStaticSwitch=false
       bIsVisible=false
       type {classStructOrEnum {refPath=&NiagaraMatrix}}
      }
      Rotation Angle {
       bIsDynamic=false
       bIsEditable=false
       bIsStaticSwitch=false
       bIsVisible=false
       type {classStructOrEnum {refPath=&NiagaraFloat}}
      }
      Invert Rotation Quaternion {bIsDynamic=false, bIsEditable=false, bIsStaticSwitch=false, bIsVisible=false, type {classStructOrEnum {refPath=&NiagaraBool}}}
      Additional Quaternion Rotation {bIsDynamic=false, bIsEditable=false, bIsStaticSwitch=false, bIsVisible=false, type {classStructOrEnum {refPath=&NiagaraBool}}}
      Additional Yaw / Pitch / Roll {
       bIsDynamic=false
       bIsEditable=false
       bIsStaticSwitch=false
       bIsVisible=false
       type {classStructOrEnum {refPath=&Vector3f}}
      }
      Rotation Angle Type {
       bIsDynamic=false
       bIsEditable=false
       bIsStaticSwitch=true
       bIsVisible=false
       type {classStructOrEnum {refPath=&ENiagara_AngleInput}}
      }
      Rotation Coordinate Space {
       bIsDynamic=false
       bIsEditable=true
       bIsStaticSwitch=false
       bIsVisible=true
       type {classStructOrEnum {refPath=&ENiagaraCoordinateSpace}}
      }
      Velocity Origin {
       bIsDynamic=false
       bIsEditable=false
       bIsStaticSwitch=false
       bIsVisible=false
       type {classStructOrEnum {refPath=&NiagaraPosition}}
      }
      Origin Offset {
       bIsDynamic=false
       bIsEditable=false
       bIsStaticSwitch=false
       bIsVisible=false
       type {classStructOrEnum {refPath=&Vector3f}}
      }
      Origin Offset Coordinate Space {
       bIsDynamic=false
       bIsEditable=false
       bIsStaticSwitch=false
       bIsVisible=false
       type {classStructOrEnum {refPath=&ENiagaraCoordinateSpace}}
      }
      Default Position {
       bIsDynamic=false
       bIsEditable=false
       bIsStaticSwitch=false
       bIsVisible=false
       type {classStructOrEnum {refPath=&NiagaraPosition}}
      }
     ]
     moduleScript {refPath=&AddVelocity}
    }
    {
     bIsSetParametersModule=false
     enabled=true
     moduleName=MeshRotationForce001
     inputs [
      Rotation {bIsDynamic=true, bIsEditable=true, bIsStaticSwitch=false, bIsVisible=true, type {classStructOrEnum {refPath=&Vector3f}}}
      Coordinate Space {
       bIsDynamic=false
       bIsEditable=true
       bIsStaticSwitch=true
       bIsVisible=true
       type {classStructOrEnum {refPath=&ENiagaraMeshTransforms}}
      }
      "Lever Radius (cm)" {bIsDynamic=false, bIsEditable=true, bIsStaticSwitch=false, bIsVisible=true, type {classStructOrEnum {refPath=&NiagaraFloat}}}
     ]
     moduleScript {refPath=&MeshRotationForce}
    }
    {
     bIsSetParametersModule=false
     enabled=true
     moduleName=ApplyInitialForces
     inputs [
      Apply Force to Velocity {bIsDynamic=false, bIsEditable=true, bIsStaticSwitch=false, bIsVisible=true, type {classStructOrEnum {refPath=&NiagaraBool}}}
      Apply Rotational Force to Rotational Velocity {bIsDynamic=false, bIsEditable=true, bIsStaticSwitch=false, bIsVisible=true, type {classStructOrEnum {refPath=&NiagaraBool}}}
      Apply Force to Position {bIsDynamic=false, bIsEditable=true, bIsStaticSwitch=false, bIsVisible=true, type {classStructOrEnum {refPath=&NiagaraBool}}}
      Positional Force Warmup Time {bIsDynamic=false, bIsEditable=false, bIsStaticSwitch=false, bIsVisible=true, type {classStructOrEnum {refPath=&NiagaraFloat}}}
     ]
     moduleScript {refPath=&ApplyInitialForces}
    }
   ]
  }
  particleUpdateScript {
   scriptName=ParticleUpdateScript
   modules [
    {
     bIsSetParametersModule=false
     enabled=true
     moduleName=ParticleState
     inputs [
      Kill Particles When Lifetime Has Elapsed {bIsDynamic=false, bIsEditable=true, bIsStaticSwitch=true, bIsVisible=true, type {classStructOrEnum {refPath=&NiagaraBool}}}
      Loop Particles Lifetime {bIsDynamic=false, bIsEditable=false, bIsStaticSwitch=false, bIsVisible=false, type {classStructOrEnum {refPath=&NiagaraBool}}}
      Let Infinitely Lived Particles Die When Emitter Deactivates {bIsDynamic=false, bIsEditable=false, bIsStaticSwitch=false, bIsVisible=false, type {classStructOrEnum {refPath=&NiagaraBool}}}
      Lifetime {bIsDynamic=false, bIsEditable=true, bIsStaticSwitch=false, bIsVisible=true, type {classStructOrEnum {refPath=&NiagaraFloat}}}
      DeltaTime {bIsDynamic=false, bIsEditable=true, bIsStaticSwitch=false, bIsVisible=true, type {classStructOrEnum {refPath=&NiagaraFloat}}}
     ]
     moduleScript {refPath=&ParticleState}
    }
    {
     bIsSetParametersModule=false
     enabled=true
     moduleName=ScaleMeshSize001
     inputs [
      Scale Factor {bIsDynamic=true, bIsEditable=true, bIsStaticSwitch=false, bIsVisible=true, type {classStructOrEnum {refPath=&Vector3f}}}
      Initial Mesh Scale {bIsDynamic=false, bIsEditable=true, bIsStaticSwitch=false, bIsVisible=true, type {classStructOrEnum {refPath=&Vector3f}}}
     ]
     moduleScript {refPath=&ScaleMeshSize}
    }
    {
     bIsSetParametersModule=false
     enabled=true
     moduleName=ScaleMeshSize
     inputs [
      Scale Factor {bIsDynamic=true, bIsEditable=true, bIsStaticSwitch=false, bIsVisible=true, type {classStructOrEnum {refPath=&Vector3f}}}
      Initial Mesh Scale {bIsDynamic=false, bIsEditable=true, bIsStaticSwitch=false, bIsVisible=true, type {classStructOrEnum {refPath=&Vector3f}}}
     ]
     moduleScript {refPath=&ScaleMeshSize}
    }
    {
     bIsSetParametersModule=false
     enabled=true
     moduleName=OffsetPosition
     inputs [
      Position to Offset {
       bIsDynamic=false
       bIsEditable=true
       bIsStaticSwitch=false
       bIsVisible=true
       type {classStructOrEnum {refPath=&NiagaraPosition}}
      }
      Position Output {bIsDynamic=false, bIsEditable=true, bIsStaticSwitch=true, bIsVisible=true, type {classStructOrEnum {refPath=&NiagaraInt32}}}
      Offset Mode {bIsDynamic=false, bIsEditable=true, bIsStaticSwitch=true, bIsVisible=true, type {classStructOrEnum {refPath=&NiagaraInt32}}}
      Jitter Mode {bIsDynamic=false, bIsEditable=false, bIsStaticSwitch=true, bIsVisible=false, type {classStructOrEnum {refPath=&NiagaraInt32}}}
      Jitter Offset {
       bIsDynamic=false
       bIsEditable=false
       bIsStaticSwitch=false
       bIsVisible=false
       type {classStructOrEnum {refPath=&Vector3f}}
      }
      Jitter Offset Coordinate Space {
       bIsDynamic=false
       bIsEditable=false
       bIsStaticSwitch=false
       bIsVisible=false
       type {classStructOrEnum {refPath=&ENiagaraCoordinateSpace}}
      }
      Jitter Scale {
       bIsDynamic=false
       bIsEditable=false
       bIsStaticSwitch=false
       bIsVisible=false
       type {classStructOrEnum {refPath=&Vector3f}}
      }
      Jitter Amount {
       bIsDynamic=false
       bIsEditable=false
       bIsStaticSwitch=false
       bIsVisible=false
       type {classStructOrEnum {refPath=&NiagaraFloat}}
      }
      Jitter Delay {
       bIsDynamic=false
       bIsEditable=false
       bIsStaticSwitch=false
       bIsVisible=false
       type {classStructOrEnum {refPath=&NiagaraFloat}}
      }
      DeltaTime {
       bIsDynamic=false
       bIsEditable=false
       bIsStaticSwitch=false
       bIsVisible=false
       type {classStructOrEnum {refPath=&NiagaraFloat}}
      }
      Jitter Time Offset {
       bIsDynamic=false
       bIsEditable=false
       bIsStaticSwitch=false
       bIsVisible=false
       type {classStructOrEnum {refPath=&NiagaraFloat}}
      }
      Jitter Interpolation {bIsDynamic=false, bIsEditable=false, bIsStaticSwitch=true, bIsVisible=false, type {classStructOrEnum {refPath=&NiagaraInt32}}}
      Interpolation Curve {
       bIsDynamic=false
       bIsEditable=false
       bIsStaticSwitch=false
       bIsVisible=false
       type {classStructOrEnum {refPath=&NiagaraDataInterfaceCurve}}
      }
      Interpolation Scale {
       bIsDynamic=false
       bIsEditable=false
       bIsStaticSwitch=false
       bIsVisible=false
       type {classStructOrEnum {refPath=&NiagaraFloat}}
      }
      Position Offset {bIsDynamic=true, bIsEditable=true, bIsStaticSwitch=false, bIsVisible=true, type {classStructOrEnum {refPath=&Vector3f}}}
      Position Offset Coordinate Space {
       bIsDynamic=false
       bIsEditable=true
       bIsStaticSwitch=false
       bIsVisible=true
       type {classStructOrEnum {refPath=&ENiagaraCoordinateSpace}}
      }
      Debug Draw {
       bIsDynamic=false
       bIsEditable=false
       bIsStaticSwitch=false
       bIsVisible=false
       type {classStructOrEnum {refPath=&NiagaraDataInterfaceDebugDraw}}
      }
      SkipID {
       bIsDynamic=false
       bIsEditable=false
       bIsStaticSwitch=false
       bIsVisible=false
       type {classStructOrEnum {refPath=&NiagaraInt32}}
      }
      Skip {
       bIsDynamic=false
       bIsEditable=false
       bIsStaticSwitch=false
       bIsVisible=false
       type {classStructOrEnum {refPath=&NiagaraInt32}}
      }
      Gizmo Scale {
       bIsDynamic=false
       bIsEditable=false
       bIsStaticSwitch=false
       bIsVisible=false
       type {classStructOrEnum {refPath=&NiagaraFloat}}
      }
     ]
     moduleScript {refPath=&OffsetPosition}
    }
    {
     bIsSetParametersModule=false
     enabled=true
     moduleName=MeshRotationForce
     inputs [
      Rotation {bIsDynamic=true, bIsEditable=true, bIsStaticSwitch=false, bIsVisible=true, type {classStructOrEnum {refPath=&Vector3f}}}
      Coordinate Space {
       bIsDynamic=false
       bIsEditable=true
       bIsStaticSwitch=true
       bIsVisible=true
       type {classStructOrEnum {refPath=&ENiagaraMeshTransforms}}
      }
      "Lever Radius (cm)" {bIsDynamic=false, bIsEditable=true, bIsStaticSwitch=false, bIsVisible=true, type {classStructOrEnum {refPath=&NiagaraFloat}}}
     ]
     moduleScript {refPath=&MeshRotationForce}
    }
    {
     bIsSetParametersModule=false
     enabled=true
     moduleName=ScaleSpriteSize
     inputs [
      Scale Sprite Size Mode {
       bIsDynamic=false
       bIsEditable=true
       bIsStaticSwitch=true
       bIsVisible=true
       type {classStructOrEnum {refPath=&ENiagara_ScaleSpriteSize}}
      }
      Uniform Curve Sprite Scale {
       bIsDynamic=false
       bIsEditable=true
       bIsStaticSwitch=false
       bIsVisible=true
       type {classStructOrEnum {refPath=&NiagaraDataInterfaceCurve}}
      }
      Non-Uniform Curve Sprite Scale {
       bIsDynamic=false
       bIsEditable=false
       bIsStaticSwitch=false
       bIsVisible=false
       type {classStructOrEnum {refPath=&NiagaraDataInterfaceVector2DCurve}}
      }
      Uniform Curve Index {bIsDynamic=false, bIsEditable=true, bIsStaticSwitch=false, bIsVisible=true, type {classStructOrEnum {refPath=&NiagaraFloat}}}
      Non-Uniform Curve Index {
       bIsDynamic=false
       bIsEditable=false
       bIsStaticSwitch=false
       bIsVisible=false
       type {classStructOrEnum {refPath=&NiagaraFloat}}
      }
      Non-Uniform Curve Scale {
       bIsDynamic=false
       bIsEditable=false
       bIsStaticSwitch=false
       bIsVisible=false
       type {classStructOrEnum {refPath=&Vector2f}}
      }
      Uniform Curve Scale {bIsDynamic=false, bIsEditable=true, bIsStaticSwitch=false, bIsVisible=true, type {classStructOrEnum {refPath=&NiagaraFloat}}}
      Scale Factor {
       bIsDynamic=false
       bIsEditable=false
       bIsStaticSwitch=false
       bIsVisible=false
       type {classStructOrEnum {refPath=&Vector2f}}
      }
      Uniform Scale Factor {
       bIsDynamic=false
       bIsEditable=false
       bIsStaticSwitch=false
       bIsVisible=false
       type {classStructOrEnum {refPath=&NiagaraFloat}}
      }
      Initial Sprite Size {bIsDynamic=false, bIsEditable=true, bIsStaticSwitch=false, bIsVisible=true, type {classStructOrEnum {refPath=&Vector2f}}}
     ]
     moduleScript {refPath=&ScaleSpriteSize}
    }
    {
     bIsSetParametersModule=false
     enabled=true
     moduleName=ScaleSpriteSize001
     inputs [
      Scale Sprite Size Mode {
       bIsDynamic=false
       bIsEditable=true
       bIsStaticSwitch=true
       bIsVisible=true
       type {classStructOrEnum {refPath=&ENiagara_ScaleSpriteSize}}
      }
      Uniform Curve Sprite Scale {
       bIsDynamic=false
       bIsEditable=true
       bIsStaticSwitch=false
       bIsVisible=true
       type {classStructOrEnum {refPath=&NiagaraDataInterfaceCurve}}
      }
      Non-Uniform Curve Sprite Scale {
       bIsDynamic=false
       bIsEditable=false
       bIsStaticSwitch=false
       bIsVisible=false
       type {classStructOrEnum {refPath=&NiagaraDataInterfaceVector2DCurve}}
      }
      Uniform Curve Index {bIsDynamic=true, bIsEditable=true, bIsStaticSwitch=false, bIsVisible=true, type {classStructOrEnum {refPath=&NiagaraFloat}}}
      Non-Uniform Curve Index {
       bIsDynamic=false
       bIsEditable=false
       bIsStaticSwitch=false
       bIsVisible=false
       type {classStructOrEnum {refPath=&NiagaraFloat}}
      }
      Non-Uniform Curve Scale {
       bIsDynamic=false
       bIsEditable=false
       bIsStaticSwitch=false
       bIsVisible=false
       type {classStructOrEnum {refPath=&Vector2f}}
      }
      Uniform Curve Scale {bIsDynamic=false, bIsEditable=true, bIsStaticSwitch=false, bIsVisible=true, type {classStructOrEnum {refPath=&NiagaraFloat}}}
      Scale Factor {
       bIsDynamic=false
       bIsEditable=false
       bIsStaticSwitch=false
       bIsVisible=false
       type {classStructOrEnum {refPath=&Vector2f}}
      }
      Uniform Scale Factor {
       bIsDynamic=false
       bIsEditable=false
       bIsStaticSwitch=false
       bIsVisible=false
       type {classStructOrEnum {refPath=&NiagaraFloat}}
      }
      Initial Sprite Size {bIsDynamic=false, bIsEditable=true, bIsStaticSwitch=false, bIsVisible=true, type {classStructOrEnum {refPath=&Vector2f}}}
     ]
     moduleScript {refPath=&ScaleSpriteSize}
    }
    {
     bIsSetParametersModule=false
     enabled=true
     moduleName=WindForce
     inputs [
      Debug Scale {
       bIsDynamic=false
       bIsEditable=false
       bIsStaticSwitch=false
       bIsVisible=false
       type {classStructOrEnum {refPath=&NiagaraFloat}}
      }
      Debug {
       bIsDynamic=false
       bIsEditable=false
       bIsStaticSwitch=false
       bIsVisible=false
       type {classStructOrEnum {refPath=&NiagaraDataInterfaceDebugDraw}}
      }
      Wind Speed {bIsDynamic=true, bIsEditable=true, bIsStaticSwitch=false, bIsVisible=true, type {classStructOrEnum {refPath=&Vector3f}}}
      Wind Speed Scale {bIsDynamic=false, bIsEditable=true, bIsStaticSwitch=false, bIsVisible=true, type {classStructOrEnum {refPath=&NiagaraFloat}}}
      Turbulence Mode {
       bIsDynamic=false
       bIsEditable=true
       bIsStaticSwitch=true
       bIsVisible=true
       type {classStructOrEnum {refPath=&ENiagaraWindTurbulenceMode}}
      }
      Turbulence Scale Mode {
       bIsDynamic=false
       bIsEditable=false
       bIsStaticSwitch=true
       bIsVisible=false
       type {classStructOrEnum {refPath=&ENiagaraWindTurbulenceContributionMode}}
      }
      Scale {
       bIsDynamic=false
       bIsEditable=false
       bIsStaticSwitch=false
       bIsVisible=false
       type {classStructOrEnum {refPath=&NiagaraFloat}}
      }
      Scale Min / Max {
       bIsDynamic=false
       bIsEditable=false
       bIsStaticSwitch=false
       bIsVisible=false
       type {classStructOrEnum {refPath=&Vector2f}}
      }
      Speed Min / Max {
       bIsDynamic=false
       bIsEditable=false
       bIsStaticSwitch=false
       bIsVisible=false
       type {classStructOrEnum {refPath=&Vector2f}}
      }
      Scale Curve {
       bIsDynamic=false
       bIsEditable=false
       bIsStaticSwitch=false
       bIsVisible=false
       type {classStructOrEnum {refPath=&NiagaraDataInterfaceCurve}}
      }
      Turbulence Frequency Mode {
       bIsDynamic=false
       bIsEditable=false
       bIsStaticSwitch=true
       bIsVisible=false
       type {classStructOrEnum {refPath=&ENiagaraWindTurbulenceFrequencyMode}}
      }
      Frequency Min / Max {
       bIsDynamic=false
       bIsEditable=false
       bIsStaticSwitch=false
       bIsVisible=false
       type {classStructOrEnum {refPath=&Vector2f}}
      }
      Frequency Blend {
       bIsDynamic=false
       bIsEditable=false
       bIsStaticSwitch=false
       bIsVisible=false
       type {classStructOrEnum {refPath=&NiagaraFloat}}
      }
      Frequency {
       bIsDynamic=false
       bIsEditable=false
       bIsStaticSwitch=false
       bIsVisible=false
       type {classStructOrEnum {refPath=&NiagaraFloat}}
      }
      Turbulence Offset Mode {
       bIsDynamic=false
       bIsEditable=false
       bIsStaticSwitch=true
       bIsVisible=false
       type {classStructOrEnum {refPath=&ENiagaraWindOffsetMode}}
      }
      Offset By Time {
       bIsDynamic=false
       bIsEditable=false
       bIsStaticSwitch=false
       bIsVisible=false
       type {classStructOrEnum {refPath=&Vector3f}}
      }
      Offset {
       bIsDynamic=false
       bIsEditable=false
       bIsStaticSwitch=false
       bIsVisible=false
       type {classStructOrEnum {refPath=&Vector3f}}
      }
      Turbulence Flow Loop Duration {
       bIsDynamic=false
       bIsEditable=false
       bIsStaticSwitch=false
       bIsVisible=false
       type {classStructOrEnum {refPath=&NiagaraFloat}}
      }
      Noise Quality / Cost {
       bIsDynamic=false
       bIsEditable=false
       bIsStaticSwitch=true
       bIsVisible=false
       type {classStructOrEnum {refPath=&ENiagaraCurlNoiseQuality}}
      }
      Curl Noise {
       bIsDynamic=false
       bIsEditable=false
       bIsStaticSwitch=false
       bIsVisible=false
       type {classStructOrEnum {refPath=&NiagaraDataInterfaceCurlNoise}}
      }
      Time Variable {
       bIsDynamic=false
       bIsEditable=false
       bIsStaticSwitch=false
       bIsVisible=false
       type {classStructOrEnum {refPath=&NiagaraFloat}}
      }
      Surface Alignment Mode {
       bIsDynamic=false
       bIsEditable=true
       bIsStaticSwitch=true
       bIsVisible=true
       type {classStructOrEnum {refPath=&ENiagaraWindCombingMode}}
      }
      Alignment Falloff Start / End {
       bIsDynamic=false
       bIsEditable=false
       bIsStaticSwitch=false
       bIsVisible=false
       type {classStructOrEnum {refPath=&Vector2f}}
      }
      Friction Mode {
       bIsDynamic=false
       bIsEditable=true
       bIsStaticSwitch=true
       bIsVisible=true
       type {classStructOrEnum {refPath=&ENiagaraWindFrictionMode}}
      }
      Friction Distance Mode {
       bIsDynamic=false
       bIsEditable=false
       bIsStaticSwitch=true
       bIsVisible=false
       type {classStructOrEnum {refPath=&ENiagaraWindFrictionDistanceMode}}
      }
      Friction Distance Limit {
       bIsDynamic=false
       bIsEditable=false
       bIsStaticSwitch=false
       bIsVisible=false
       type {classStructOrEnum {refPath=&NiagaraFloat}}
      }
      Friction {
       bIsDynamic=false
       bIsEditable=false
       bIsStaticSwitch=false
       bIsVisible=false
       type {classStructOrEnum {refPath=&NiagaraFloat}}
      }
      Friction Falloff Start / End {
       bIsDynamic=false
       bIsEditable=false
       bIsStaticSwitch=false
       bIsVisible=false
       type {classStructOrEnum {refPath=&Vector2f}}
      }
      Normal Friction Mask {
       bIsDynamic=false
       bIsEditable=false
       bIsStaticSwitch=false
       bIsVisible=false
       type {classStructOrEnum {refPath=&NiagaraFloat}}
      }
      Use Depth Buffer {bIsDynamic=false, bIsEditable=false, bIsStaticSwitch=true, bIsVisible=false, type {classStructOrEnum {refPath=&NiagaraBool}}}
      Use Landscape {bIsDynamic=false, bIsEditable=false, bIsStaticSwitch=true, bIsVisible=false, type {classStructOrEnum {refPath=&NiagaraBool}}}
      Use Ground Mask {bIsDynamic=false, bIsEditable=true, bIsStaticSwitch=true, bIsVisible=true, type {classStructOrEnum {refPath=&NiagaraBool}}}
      Ground Mask Elevation Mode {
       bIsDynamic=false
       bIsEditable=false
       bIsStaticSwitch=true
       bIsVisible=false
       type {classStructOrEnum {refPath=&ENiagaraWindGroundMaskMode}}
      }
      Minimum Ground Elevation {
       bIsDynamic=false
       bIsEditable=false
       bIsStaticSwitch=false
       bIsVisible=false
       type {classStructOrEnum {refPath=&NiagaraFloat}}
      }
      Ground Elevation {
       bIsDynamic=false
       bIsEditable=false
       bIsStaticSwitch=false
       bIsVisible=false
       type {classStructOrEnum {refPath=&NiagaraFloat}}
      }
      Ground Distance Min / Max {
       bIsDynamic=false
       bIsEditable=false
       bIsStaticSwitch=false
       bIsVisible=false
       type {classStructOrEnum {refPath=&Vector2f}}
      }
      Ground Mask Min / Max {
       bIsDynamic=false
       bIsEditable=false
       bIsStaticSwitch=false
       bIsVisible=false
       type {classStructOrEnum {refPath=&Vector2f}}
      }
      Use Physics Field {bIsDynamic=false, bIsEditable=true, bIsStaticSwitch=true, bIsVisible=true, type {classStructOrEnum {refPath=&NiagaraBool}}}
      Physics Field Scale {
       bIsDynamic=false
       bIsEditable=false
       bIsStaticSwitch=false
       bIsVisible=false
       type {classStructOrEnum {refPath=&NiagaraFloat}}
      }
      Position {
       bIsDynamic=false
       bIsEditable=true
       bIsStaticSwitch=false
       bIsVisible=true
       type {classStructOrEnum {refPath=&NiagaraPosition}}
      }
      CollisionQuery {
       bIsDynamic=false
       bIsEditable=false
       bIsStaticSwitch=false
       bIsVisible=false
       type {classStructOrEnum {refPath=&NiagaraDataInterfaceCollisionQuery}}
      }
      Physics Field {
       bIsDynamic=false
       bIsEditable=false
       bIsStaticSwitch=false
       bIsVisible=false
       type {classStructOrEnum {refPath=&NiagaraDataInterfacePhysicsField}}
      }
      Landscape Sample {
       bIsDynamic=false
       bIsEditable=false
       bIsStaticSwitch=false
       bIsVisible=false
       type {classStructOrEnum {refPath=&NiagaraDataInterfaceLandscape}}
      }
     ]
     moduleScript {refPath=&WindForce}
    }
    {
     bIsSetParametersModule=false
     enabled=true
     moduleName=VortexForce
     inputs [
      Vortex Force Amount {bIsDynamic=true, bIsEditable=true, bIsStaticSwitch=false, bIsVisible=true, type {classStructOrEnum {refPath=&NiagaraFloat}}}
      Vortex Axis {bIsDynamic=false, bIsEditable=true, bIsStaticSwitch=false, bIsVisible=true, type {classStructOrEnum {refPath=&Vector3f}}}
      Vortex Axis Coordinate Space {
       bIsDynamic=false
       bIsEditable=true
       bIsStaticSwitch=false
       bIsVisible=true
       type {classStructOrEnum {refPath=&ENiagaraCoordinateSpace}}
      }
      Vortex Influence Position {
       bIsDynamic=false
       bIsEditable=true
       bIsStaticSwitch=false
       bIsVisible=true
       type {classStructOrEnum {refPath=&NiagaraPosition}}
      }
      Write to Intrinsic Parameters {bIsDynamic=false, bIsEditable=true, bIsStaticSwitch=true, bIsVisible=true, type {classStructOrEnum {refPath=&NiagaraBool}}}
      Vortex Origin {
       bIsDynamic=false
       bIsEditable=true
       bIsStaticSwitch=false
       bIsVisible=true
       type {classStructOrEnum {refPath=&NiagaraPosition}}
      }
      Origin Pull Amount {bIsDynamic=false, bIsEditable=true, bIsStaticSwitch=false, bIsVisible=true, type {classStructOrEnum {refPath=&NiagaraFloat}}}
      Origin Pull Only On Vortex Axis {bIsDynamic=false, bIsEditable=true, bIsStaticSwitch=true, bIsVisible=true, type {classStructOrEnum {refPath=&NiagaraBool}}}
      UseOriginOffset {bIsDynamic=false, bIsEditable=true, bIsStaticSwitch=false, bIsVisible=true, type {classStructOrEnum {refPath=&NiagaraBool}}}
      Vortex Origin Offset {bIsDynamic=false, bIsEditable=false, bIsStaticSwitch=false, bIsVisible=true, type {classStructOrEnum {refPath=&Vector3f}}}
      Vortex Origin Offset Coordinate Space {
       bIsDynamic=false
       bIsEditable=false
       bIsStaticSwitch=false
       bIsVisible=false
       type {classStructOrEnum {refPath=&ENiagaraCoordinateSpace}}
      }
      Use Influence Falloff {bIsDynamic=false, bIsEditable=true, bIsStaticSwitch=true, bIsVisible=true, type {classStructOrEnum {refPath=&NiagaraBool}}}
      Influence Falloff Radius {
       bIsDynamic=false
       bIsEditable=false
       bIsStaticSwitch=false
       bIsVisible=false
       type {classStructOrEnum {refPath=&NiagaraFloat}}
      }
      Influence Falloff Exponent {
       bIsDynamic=false
       bIsEditable=false
       bIsStaticSwitch=false
       bIsVisible=false
       type {classStructOrEnum {refPath=&NiagaraFloat}}
      }
      Invert Influence Falloff {bIsDynamic=false, bIsEditable=false, bIsStaticSwitch=false, bIsVisible=false, type {classStructOrEnum {refPath=&NiagaraBool}}}
      Debug Every N Particles {
       bIsDynamic=false
       bIsEditable=false
       bIsStaticSwitch=false
       bIsVisible=false
       type {classStructOrEnum {refPath=&NiagaraInt32}}
      }
      Debug Force Line Color {
       bIsDynamic=false
       bIsEditable=false
       bIsStaticSwitch=false
       bIsVisible=false
       type {classStructOrEnum {refPath=&LinearColor}}
      }
      Debug Force Line Length {
       bIsDynamic=false
       bIsEditable=false
       bIsStaticSwitch=false
       bIsVisible=false
       type {classStructOrEnum {refPath=&NiagaraFloat}}
      }
      Debug Influence Falloff Color {
       bIsDynamic=false
       bIsEditable=false
       bIsStaticSwitch=false
       bIsVisible=false
       type {classStructOrEnum {refPath=&LinearColor}}
      }
     ]
     moduleScript {refPath=&VortexForce}
    }
    {
     bIsSetParametersModule=false
     enabled=true
     moduleName=Drag
     inputs [
      Use Linear Drag {bIsDynamic=false, bIsEditable=true, bIsStaticSwitch=false, bIsVisible=true, type {classStructOrEnum {refPath=&NiagaraBool}}}
      Drag {bIsDynamic=true, bIsEditable=true, bIsStaticSwitch=false, bIsVisible=true, type {classStructOrEnum {refPath=&NiagaraFloat}}}
      Use Rotational Drag {bIsDynamic=false, bIsEditable=true, bIsStaticSwitch=false, bIsVisible=true, type {classStructOrEnum {refPath=&NiagaraBool}}}
      Rotational Drag {bIsDynamic=false, bIsEditable=false, bIsStaticSwitch=false, bIsVisible=true, type {classStructOrEnum {refPath=&NiagaraFloat}}}
      Ignore Mass {bIsDynamic=false, bIsEditable=true, bIsStaticSwitch=false, bIsVisible=true, type {classStructOrEnum {refPath=&NiagaraBool}}}
     ]
     moduleScript {refPath=&Drag}
    }
    {
     bIsSetParametersModule=true
     enabled=true
     moduleName=SetVariables_#4
     inputs [Particles.CameraOffset_LensFlare {bIsDynamic=false, bIsEditable=true, bIsStaticSwitch=false, bIsVisible=true, type {classStructOrEnum {refPath=&NiagaraFloat}}}]
     moduleScript {refPath=&self:Rocks_0.NiagaraScriptSource_0.NiagaraGraph_0.NiagaraNodeAssignment_0.SetVariables_#4}
    }
    {
     bIsSetParametersModule=false
     enabled=true
     moduleName=SolveForcesAndVelocity
     inputs [
      Clamp Velocity {bIsDynamic=false, bIsEditable=true, bIsStaticSwitch=false, bIsVisible=true, type {classStructOrEnum {refPath=&NiagaraBool}}}
      Speed Limit {bIsDynamic=false, bIsEditable=true, bIsStaticSwitch=false, bIsVisible=true, type {classStructOrEnum {refPath=&NiagaraFloat}}}
      Limit Acceleration {bIsDynamic=false, bIsEditable=true, bIsStaticSwitch=false, bIsVisible=true, type {classStructOrEnum {refPath=&NiagaraBool}}}
      Acceleration Limit {bIsDynamic=false, bIsEditable=false, bIsStaticSwitch=false, bIsVisible=true, type {classStructOrEnum {refPath=&NiagaraFloat}}}
      Manually Enable Rotational Solver {bIsDynamic=false, bIsEditable=true, bIsStaticSwitch=false, bIsVisible=true, type {classStructOrEnum {refPath=&NiagaraBool}}}
      Write to Intrinsic Properties {bIsDynamic=false, bIsEditable=true, bIsStaticSwitch=false, bIsVisible=true, type {classStructOrEnum {refPath=&NiagaraBool}}}
      Write to Presolve Properties {bIsDynamic=false, bIsEditable=true, bIsStaticSwitch=true, bIsVisible=true, type {classStructOrEnum {refPath=&NiagaraBool}}}
      DeltaTime {bIsDynamic=false, bIsEditable=true, bIsStaticSwitch=false, bIsVisible=true, type {classStructOrEnum {refPath=&NiagaraFloat}}}
      Mass {bIsDynamic=false, bIsEditable=true, bIsStaticSwitch=false, bIsVisible=true, type {classStructOrEnum {refPath=&NiagaraFloat}}}
      Position {
       bIsDynamic=false
       bIsEditable=true
       bIsStaticSwitch=false
       bIsVisible=true
       type {classStructOrEnum {refPath=&NiagaraPosition}}
      }
      Velocity {bIsDynamic=false, bIsEditable=true, bIsStaticSwitch=false, bIsVisible=true, type {classStructOrEnum {refPath=&Vector3f}}}
      PreviousVelocity {bIsDynamic=false, bIsEditable=true, bIsStaticSwitch=false, bIsVisible=true, type {classStructOrEnum {refPath=&Vector3f}}}
      Force {bIsDynamic=false, bIsEditable=true, bIsStaticSwitch=false, bIsVisible=true, type {classStructOrEnum {refPath=&Vector3f}}}
      Mesh Orientation {bIsDynamic=false, bIsEditable=true, bIsStaticSwitch=false, bIsVisible=true, type {classStructOrEnum {refPath=&Quat4f}}}
      Rotational Velocity {bIsDynamic=false, bIsEditable=true, bIsStaticSwitch=false, bIsVisible=true, type {classStructOrEnum {refPath=&Vector3f}}}
      Rotational Force {bIsDynamic=false, bIsEditable=true, bIsStaticSwitch=false, bIsVisible=true, type {classStructOrEnum {refPath=&Vector3f}}}
      Rotational Inertia {bIsDynamic=false, bIsEditable=true, bIsStaticSwitch=false, bIsVisible=true, type {classStructOrEnum {refPath=&NiagaraFloat}}}
     ]
     moduleScript {refPath=&SolveForcesAndVelocity}
    }
    {
     bIsSetParametersModule=false
     enabled=true
     moduleName=DynamicMaterialParameters
     inputs [
      Write Parameter Index 0 {
       bIsDynamic=false
       bIsEditable=true
       bIsStaticSwitch=true
       bIsVisible=true
       type {classStructOrEnum {refPath=&ENiagara_DynamicMaterialParametersInputMode}}
      }
      Index 0 Vector 1 {
       bIsDynamic=false
       bIsEditable=false
       bIsStaticSwitch=false
       bIsVisible=false
       type {classStructOrEnum {refPath=&Vector3f}}
      }
      Index 0 Param 1 {
       bIsDynamic=false
       bIsEditable=false
       bIsStaticSwitch=false
       bIsVisible=false
       type {classStructOrEnum {refPath=&NiagaraFloat}}
      }
      Index 0 Param 2 {
       bIsDynamic=false
       bIsEditable=false
       bIsStaticSwitch=false
       bIsVisible=false
       type {classStructOrEnum {refPath=&NiagaraFloat}}
      }
      Index 0 Param 3 {
       bIsDynamic=false
       bIsEditable=false
       bIsStaticSwitch=false
       bIsVisible=false
       type {classStructOrEnum {refPath=&NiagaraFloat}}
      }
      Index 0 Param 4 {
       bIsDynamic=false
       bIsEditable=false
       bIsStaticSwitch=false
       bIsVisible=false
       type {classStructOrEnum {refPath=&NiagaraFloat}}
      }
      Index 0 Vector2D 1 {
       bIsDynamic=false
       bIsEditable=false
       bIsStaticSwitch=false
       bIsVisible=false
       type {classStructOrEnum {refPath=&Vector2f}}
      }
      Index 0 Vector2D 2 {
       bIsDynamic=false
       bIsEditable=false
       bIsStaticSwitch=false
       bIsVisible=false
       type {classStructOrEnum {refPath=&Vector2f}}
      }
      Index 0 Vector 4 {
       bIsDynamic=false
       bIsEditable=false
       bIsStaticSwitch=false
       bIsVisible=false
       type {classStructOrEnum {refPath=&Vector4f}}
      }
      Index 0 Linear Color {
       bIsDynamic=true
       bIsEditable=true
       bIsStaticSwitch=false
       bIsVisible=true
       type {classStructOrEnum {refPath=&LinearColor}}
      }
      Write Parameter Index 1 {
       bIsDynamic=false
       bIsEditable=true
       bIsStaticSwitch=true
       bIsVisible=true
       type {classStructOrEnum {refPath=&ENiagara_DynamicMaterialParametersInputMode}}
      }
      Index 1 Vector 1 {
       bIsDynamic=false
       bIsEditable=false
       bIsStaticSwitch=false
       bIsVisible=false
       type {classStructOrEnum {refPath=&Vector3f}}
      }
      Index 1 Param 1 {
       bIsDynamic=false
       bIsEditable=false
       bIsStaticSwitch=false
       bIsVisible=false
       type {classStructOrEnum {refPath=&NiagaraFloat}}
      }
      Index 1 Param 2 {
       bIsDynamic=false
       bIsEditable=false
       bIsStaticSwitch=false
       bIsVisible=false
       type {classStructOrEnum {refPath=&NiagaraFloat}}
      }
      Index 1 Param 3 {
       bIsDynamic=false
       bIsEditable=false
       bIsStaticSwitch=false
       bIsVisible=false
       type {classStructOrEnum {refPath=&NiagaraFloat}}
      }
      Index 1 Param 4 {
       bIsDynamic=false
       bIsEditable=false
       bIsStaticSwitch=false
       bIsVisible=false
       type {classStructOrEnum {refPath=&NiagaraFloat}}
      }
      Index 1 Vector2D 1 {
       bIsDynamic=false
       bIsEditable=false
       bIsStaticSwitch=false
       bIsVisible=false
       type {classStructOrEnum {refPath=&Vector2f}}
      }
      Index 1 Vector2D 2 {
       bIsDynamic=false
       bIsEditable=false
       bIsStaticSwitch=false
       bIsVisible=false
       type {classStructOrEnum {refPath=&Vector2f}}
      }
      Index 1 Vector 4 {
       bIsDynamic=false
       bIsEditable=false
       bIsStaticSwitch=false
       bIsVisible=false
       type {classStructOrEnum {refPath=&Vector4f}}
      }
      Index 1 Linear Color {
       bIsDynamic=false
       bIsEditable=false
       bIsStaticSwitch=false
       bIsVisible=false
       type {classStructOrEnum {refPath=&LinearColor}}
      }
      Write Parameter Index 2 {
       bIsDynamic=false
       bIsEditable=true
       bIsStaticSwitch=true
       bIsVisible=true
       type {classStructOrEnum {refPath=&ENiagara_DynamicMaterialParametersInputMode}}
      }
      Index 2 Vector 1 {
       bIsDynamic=false
       bIsEditable=false
       bIsStaticSwitch=false
       bIsVisible=false
       type {classStructOrEnum {refPath=&Vector3f}}
      }
      Index 2 Param 1 {
       bIsDynamic=false
       bIsEditable=false
       bIsStaticSwitch=false
       bIsVisible=false
       type {classStructOrEnum {refPath=&NiagaraFloat}}
      }
      Index 2 Param 2 {
       bIsDynamic=false
       bIsEditable=false
       bIsStaticSwitch=false
       bIsVisible=false
       type {classStructOrEnum {refPath=&NiagaraFloat}}
      }
      Index 2 Param 3 {
       bIsDynamic=false
       bIsEditable=false
       bIsStaticSwitch=false
       bIsVisible=false
       type {classStructOrEnum {refPath=&NiagaraFloat}}
      }
      Index 2 Param 4 {
       bIsDynamic=false
       bIsEditable=false
       bIsStaticSwitch=false
       bIsVisible=false
       type {classStructOrEnum {refPath=&NiagaraFloat}}
      }
      Index 2 Vector2D 1 {
       bIsDynamic=false
       bIsEditable=false
       bIsStaticSwitch=false
       bIsVisible=false
       type {classStructOrEnum {refPath=&Vector2f}}
      }
      Index 2 Vector2D 2 {
       bIsDynamic=false
       bIsEditable=false
       bIsStaticSwitch=false
       bIsVisible=false
       type {classStructOrEnum {refPath=&Vector2f}}
      }
      Index 2 Vector 4 {
       bIsDynamic=false
       bIsEditable=false
       bIsStaticSwitch=false
       bIsVisible=false
       type {classStructOrEnum {refPath=&Vector4f}}
      }
      Index 2 Linear Color {
       bIsDynamic=false
       bIsEditable=false
       bIsStaticSwitch=false
       bIsVisible=false
       type {classStructOrEnum {refPath=&LinearColor}}
      }
      Write Parameter Index 3 {
       bIsDynamic=false
       bIsEditable=true
       bIsStaticSwitch=true
       bIsVisible=true
       type {classStructOrEnum {refPath=&ENiagara_DynamicMaterialParametersInputMode}}
      }
      Index 3 Vector 1 {
       bIsDynamic=false
       bIsEditable=false
       bIsStaticSwitch=false
       bIsVisible=false
       type {classStructOrEnum {refPath=&Vector3f}}
      }
      Index 3 Param 1 {
       bIsDynamic=false
       bIsEditable=false
       bIsStaticSwitch=false
       bIsVisible=false
       type {classStructOrEnum {refPath=&NiagaraFloat}}
      }
      Index 3 Param 2 {
       bIsDynamic=false
       bIsEditable=false
       bIsStaticSwitch=false
       bIsVisible=false
       type {classStructOrEnum {refPath=&NiagaraFloat}}
      }
      Index 3 Param 3 {
       bIsDynamic=false
       bIsEditable=false
       bIsStaticSwitch=false
       bIsVisible=false
       type {classStructOrEnum {refPath=&NiagaraFloat}}
      }
      Index 3 Param 4 {
       bIsDynamic=false
       bIsEditable=false
       bIsStaticSwitch=false
       bIsVisible=false
       type {classStructOrEnum {refPath=&NiagaraFloat}}
      }
      Index 3 Vector2D 1 {
       bIsDynamic=false
       bIsEditable=false
       bIsStaticSwitch=false
       bIsVisible=false
       type {classStructOrEnum {refPath=&Vector2f}}
      }
      Index 3 Vector2D 2 {
       bIsDynamic=false
       bIsEditable=false
       bIsStaticSwitch=false
       bIsVisible=false
       type {classStructOrEnum {refPath=&Vector2f}}
      }
      Index 3 Vector 4 {
       bIsDynamic=false
       bIsEditable=false
       bIsStaticSwitch=false
       bIsVisible=false
       type {classStructOrEnum {refPath=&Vector4f}}
      }
      Index 3 Linear Color {
       bIsDynamic=false
       bIsEditable=false
       bIsStaticSwitch=false
       bIsVisible=false
       type {classStructOrEnum {refPath=&LinearColor}}
      }
      Dynamic Parameter Output {
       bIsDynamic=false
       bIsEditable=true
       bIsStaticSwitch=true
       bIsVisible=true
       type {classStructOrEnum {refPath=&ENiagara_AttributeSamplingApplyOutput}}
      }
     ]
     moduleScript {refPath=&DynamicMaterialParameters}
    }
    {
     bIsSetParametersModule=false
     enabled=true
     moduleName=BlendSpace
     inputs [
      Position Mode {bIsDynamic=false, bIsEditable=true, bIsStaticSwitch=true, bIsVisible=true, type {classStructOrEnum {refPath=&NiagaraInt32}}}
      Position Coordinate Space {bIsDynamic=false, bIsEditable=true, bIsStaticSwitch=true, bIsVisible=true, type {classStructOrEnum {refPath=&NiagaraInt32}}}
      "Position Blend (World <-> Local)" {bIsDynamic=true, bIsEditable=true, bIsStaticSwitch=false, bIsVisible=true, type {classStructOrEnum {refPath=&NiagaraFloat}}}
      Mesh Orientation Mode {bIsDynamic=false, bIsEditable=true, bIsStaticSwitch=true, bIsVisible=true, type {classStructOrEnum {refPath=&NiagaraInt32}}}
      Mesh Orientation Coordinate Space {bIsDynamic=false, bIsEditable=false, bIsStaticSwitch=true, bIsVisible=false, type {classStructOrEnum {refPath=&NiagaraInt32}}}
      "Mesh Orientation Blend (World <-> Local)" {
       bIsDynamic=false
       bIsEditable=false
       bIsStaticSwitch=false
       bIsVisible=false
       type {classStructOrEnum {refPath=&NiagaraFloat}}
      }
      Velocity Mode {bIsDynamic=false, bIsEditable=true, bIsStaticSwitch=true, bIsVisible=true, type {classStructOrEnum {refPath=&NiagaraInt32}}}
      Velocity Coordinate Space {bIsDynamic=false, bIsEditable=false, bIsStaticSwitch=true, bIsVisible=false, type {classStructOrEnum {refPath=&NiagaraInt32}}}
      "Velocity Blend (World <-> Local)" {
       bIsDynamic=false
       bIsEditable=false
       bIsStaticSwitch=false
       bIsVisible=false
       type {classStructOrEnum {refPath=&NiagaraFloat}}
      }
     ]
     moduleScript {refPath=&BlendSpace}
    }
    {
     bIsSetParametersModule=false
     enabled=true
     moduleName=KillParticlesInVolume001
     inputs [
      Kill Shape {
       bIsDynamic=false
       bIsEditable=true
       bIsStaticSwitch=true
       bIsVisible=true
       type {classStructOrEnum {refPath=&ENiagaraKillVolumeOptions}}
      }
      Box Size {bIsDynamic=false, bIsEditable=true, bIsStaticSwitch=false, bIsVisible=true, type {classStructOrEnum {refPath=&Vector3f}}}
      Plane Normal {
       bIsDynamic=false
       bIsEditable=false
       bIsStaticSwitch=false
       bIsVisible=false
       type {classStructOrEnum {refPath=&Vector3f}}
      }
      Slab Width {
       bIsDynamic=false
       bIsEditable=false
       bIsStaticSwitch=false
       bIsVisible=false
       type {classStructOrEnum {refPath=&NiagaraFloat}}
      }
      Sphere Radius {
       bIsDynamic=false
       bIsEditable=false
       bIsStaticSwitch=false
       bIsVisible=false
       type {classStructOrEnum {refPath=&NiagaraFloat}}
      }
      Slab Axis {
       bIsDynamic=false
       bIsEditable=false
       bIsStaticSwitch=false
       bIsVisible=false
       type {classStructOrEnum {refPath=&Vector3f}}
      }
      Cone Angle {
       bIsDynamic=false
       bIsEditable=false
       bIsStaticSwitch=false
       bIsVisible=false
       type {classStructOrEnum {refPath=&NiagaraFloat}}
      }
      Cone Axis {
       bIsDynamic=false
       bIsEditable=false
       bIsStaticSwitch=false
       bIsVisible=false
       type {classStructOrEnum {refPath=&Vector3f}}
      }
      Cone Length {
       bIsDynamic=false
       bIsEditable=false
       bIsStaticSwitch=false
       bIsVisible=false
       type {classStructOrEnum {refPath=&NiagaraFloat}}
      }
      Kill Volume Enabled {bIsDynamic=false, bIsEditable=true, bIsStaticSwitch=false, bIsVisible=true, type {classStructOrEnum {refPath=&NiagaraBool}}}
      Invert Volume {bIsDynamic=false, bIsEditable=true, bIsStaticSwitch=false, bIsVisible=true, type {classStructOrEnum {refPath=&NiagaraBool}}}
      Volume Origin {bIsDynamic=false, bIsEditable=true, bIsStaticSwitch=false, bIsVisible=true, type {classStructOrEnum {refPath=&Vector3f}}}
      Origin Offset {bIsDynamic=false, bIsEditable=true, bIsStaticSwitch=false, bIsVisible=true, type {classStructOrEnum {refPath=&Vector3f}}}
      Position {
       bIsDynamic=false
       bIsEditable=true
       bIsStaticSwitch=false
       bIsVisible=true
       type {classStructOrEnum {refPath=&NiagaraPosition}}
      }
     ]
     moduleScript {refPath=&KillParticlesInVolume}
    }
    {
     bIsSetParametersModule=false
     enabled=true
     moduleName=ScaleColor
     inputs [
      Scale Mode {bIsDynamic=false, bIsEditable=true, bIsStaticSwitch=true, bIsVisible=true, type {classStructOrEnum {refPath=&NiagaraInt32}}}
      Linear Color Curve {
       bIsDynamic=false
       bIsEditable=false
       bIsStaticSwitch=false
       bIsVisible=false
       type {classStructOrEnum {refPath=&NiagaraDataInterfaceColorCurve}}
      }
      ScaleRGB {bIsDynamic=false, bIsEditable=true, bIsStaticSwitch=false, bIsVisible=true, type {classStructOrEnum {refPath=&NiagaraBool}}}
      Scale RGB {bIsDynamic=false, bIsEditable=true, bIsStaticSwitch=false, bIsVisible=true, type {classStructOrEnum {refPath=&Vector3f}}}
      ScaleRGBA {bIsDynamic=false, bIsEditable=false, bIsStaticSwitch=false, bIsVisible=false, type {classStructOrEnum {refPath=&NiagaraBool}}}
      Scale RGBA {
       bIsDynamic=false
       bIsEditable=false
       bIsStaticSwitch=false
       bIsVisible=false
       type {classStructOrEnum {refPath=&Vector4f}}
      }
      Curve Index {
       bIsDynamic=false
       bIsEditable=false
       bIsStaticSwitch=false
       bIsVisible=false
       type {classStructOrEnum {refPath=&NiagaraFloat}}
      }
      ScaleA {bIsDynamic=false, bIsEditable=true, bIsStaticSwitch=false, bIsVisible=true, type {classStructOrEnum {refPath=&NiagaraBool}}}
      Scale Alpha {bIsDynamic=false, bIsEditable=false, bIsStaticSwitch=false, bIsVisible=true, type {classStructOrEnum {refPath=&NiagaraFloat}}}
      Color Value To Scale {
       bIsDynamic=false
       bIsEditable=true
       bIsStaticSwitch=false
       bIsVisible=true
       type {classStructOrEnum {refPath=&LinearColor}}
      }
      Color Scale {
       bIsDynamic=false
       bIsEditable=false
       bIsStaticSwitch=false
       bIsVisible=false
       type {classStructOrEnum {refPath=&LinearColor}}
      }
     ]
     moduleScript {refPath=&ScaleColor}
    }
   ]
  }
  rendererClasses [{refPath=&NiagaraMeshRendererProperties}, {refPath=&NiagaraSpriteRendererProperties}]
  renderers [
   {rendererIndex=0, rendererClass {refPath=&NiagaraMeshRendererProperties}}
   {rendererIndex=1, rendererClass {refPath=&NiagaraSpriteRendererProperties}}
  ]
 }
 Rocks001 {
  emitterSpawnScript {scriptName=EmitterSpawnScript, modules []}
  emitterUpdateScript {
   scriptName=EmitterUpdateScript
   modules [
    {
     bIsSetParametersModule=false
     enabled=true
     moduleName=EmitterState
     inputs [
      Life Cycle Mode {
       bIsDynamic=false
       bIsEditable=true
       bIsStaticSwitch=true
       bIsVisible=true
       type {classStructOrEnum {refPath=&ENiagaraEmitterLifeCycleMode}}
      }
      Inactive Response {
       bIsDynamic=false
       bIsEditable=true
       bIsStaticSwitch=true
       bIsVisible=true
       type {classStructOrEnum {refPath=&ENiagaraInactiveMode}}
      }
      Loop Behavior {
       bIsDynamic=false
       bIsEditable=true
       bIsStaticSwitch=true
       bIsVisible=true
       type {classStructOrEnum {refPath=&ENiagara_EmitterStateOptions}}
      }
      Loop Duration Mode {
       bIsDynamic=false
       bIsEditable=false
       bIsStaticSwitch=true
       bIsVisible=false
       type {classStructOrEnum {refPath=&ENiagara_InfiniteLoopDuration}}
      }
      Loop Count {
       bIsDynamic=false
       bIsEditable=false
       bIsStaticSwitch=false
       bIsVisible=false
       type {classStructOrEnum {refPath=&NiagaraInt32}}
      }
      Loop Duration {bIsDynamic=false, bIsEditable=true, bIsStaticSwitch=false, bIsVisible=true, type {classStructOrEnum {refPath=&NiagaraFloat}}}
      Recalculate Duration Each Loop {bIsDynamic=false, bIsEditable=true, bIsStaticSwitch=false, bIsVisible=true, type {classStructOrEnum {refPath=&NiagaraBool}}}
      UseLoopDelay {bIsDynamic=false, bIsEditable=true, bIsStaticSwitch=false, bIsVisible=true, type {classStructOrEnum {refPath=&NiagaraBool}}}
      Loop Delay {bIsDynamic=false, bIsEditable=false, bIsStaticSwitch=false, bIsVisible=true, type {classStructOrEnum {refPath=&NiagaraFloat}}}
      Delay First Loop Only {bIsDynamic=false, bIsEditable=false, bIsStaticSwitch=false, bIsVisible=false, type {classStructOrEnum {refPath=&NiagaraBool}}}
      Scalability Mode {
       bIsDynamic=false
       bIsEditable=true
       bIsStaticSwitch=true
       bIsVisible=true
       type {classStructOrEnum {refPath=&ENiagaraEmitterScalabilityMode_Limited}}
      }
      Enable Distance Culling {bIsDynamic=false, bIsEditable=false, bIsStaticSwitch=true, bIsVisible=false, type {classStructOrEnum {refPath=&NiagaraBool}}}
      MinDistance {
       bIsDynamic=false
       bIsEditable=false
       bIsStaticSwitch=false
       bIsVisible=false
       type {classStructOrEnum {refPath=&NiagaraFloat}}
      }
      Min Distance Response {
       bIsDynamic=false
       bIsEditable=false
       bIsStaticSwitch=false
       bIsVisible=false
       type {classStructOrEnum {refPath=&ENiagaraExecutionStateManagement}}
      }
      MaxDistance {
       bIsDynamic=false
       bIsEditable=false
       bIsStaticSwitch=false
       bIsVisible=false
       type {classStructOrEnum {refPath=&NiagaraFloat}}
      }
      Max Distance Response {
       bIsDynamic=false
       bIsEditable=false
       bIsStaticSwitch=false
       bIsVisible=false
       type {classStructOrEnum {refPath=&ENiagaraExecutionStateManagement}}
      }
      Scale Spawn Count {bIsDynamic=false, bIsEditable=false, bIsStaticSwitch=true, bIsVisible=false, type {classStructOrEnum {refPath=&NiagaraBool}}}
      Spawn Count Scale {
       bIsDynamic=false
       bIsEditable=false
       bIsStaticSwitch=false
       bIsVisible=false
       type {classStructOrEnum {refPath=&NiagaraFloat}}
      }
      Spawn Count Scale By Distance {
       bIsDynamic=false
       bIsEditable=false
       bIsStaticSwitch=false
       bIsVisible=false
       type {classStructOrEnum {refPath=&NiagaraDataInterfaceCurve}}
      }
      Enable Visibility Culling {bIsDynamic=false, bIsEditable=false, bIsStaticSwitch=true, bIsVisible=false, type {classStructOrEnum {refPath=&NiagaraBool}}}
      Visibility Cull Response {
       bIsDynamic=false
       bIsEditable=false
       bIsStaticSwitch=false
       bIsVisible=false
       type {classStructOrEnum {refPath=&ENiagaraExecutionStateManagement}}
      }
      Visibility Cull Delay {
       bIsDynamic=false
       bIsEditable=false
       bIsStaticSwitch=false
       bIsVisible=false
       type {classStructOrEnum {refPath=&NiagaraFloat}}
      }
      Reset Age On Awaken {bIsDynamic=false, bIsEditable=false, bIsStaticSwitch=false, bIsVisible=false, type {classStructOrEnum {refPath=&NiagaraBool}}}
     ]
     moduleScript {refPath=&EmitterState}
    }
    {
     bIsSetParametersModule=false
     enabled=true
     moduleName=SpawnRate
     inputs [
      SpawnRate {bIsDynamic=false, bIsEditable=true, bIsStaticSwitch=false, bIsVisible=true, type {classStructOrEnum {refPath=&NiagaraFloat}}}
      Use Spawn Probability {bIsDynamic=false, bIsEditable=true, bIsStaticSwitch=false, bIsVisible=true, type {classStructOrEnum {refPath=&NiagaraBool}}}
      Spawn Probability {bIsDynamic=false, bIsEditable=true, bIsStaticSwitch=false, bIsVisible=true, type {classStructOrEnum {refPath=&NiagaraFloat}}}
      SpawnGroup {bIsDynamic=false, bIsEditable=true, bIsStaticSwitch=false, bIsVisible=true, type {classStructOrEnum {refPath=&NiagaraInt32}}}
     ]
     moduleScript {refPath=&SpawnRate}
    }
    {
     bIsSetParametersModule=false
     enabled=true
     moduleName=SpawnRate001
     inputs [
      SpawnRate {bIsDynamic=false, bIsEditable=true, bIsStaticSwitch=false, bIsVisible=true, type {classStructOrEnum {refPath=&NiagaraFloat}}}
      Use Spawn Probability {bIsDynamic=false, bIsEditable=true, bIsStaticSwitch=false, bIsVisible=true, type {classStructOrEnum {refPath=&NiagaraBool}}}
      Spawn Probability {bIsDynamic=false, bIsEditable=true, bIsStaticSwitch=false, bIsVisible=true, type {classStructOrEnum {refPath=&NiagaraFloat}}}
      SpawnGroup {bIsDynamic=false, bIsEditable=true, bIsStaticSwitch=false, bIsVisible=true, type {classStructOrEnum {refPath=&NiagaraInt32}}}
     ]
     moduleScript {refPath=&SpawnRate}
    }
   ]
  }
  particleSpawnScript {
   scriptName=ParticleSpawnScript
   modules [
    {
     bIsSetParametersModule=false
     enabled=true
     moduleName=InitializeParticle001
     inputs [
      Lifetime Mode {
       bIsDynamic=false
       bIsEditable=true
       bIsStaticSwitch=true
       bIsVisible=true
       type {classStructOrEnum {refPath=&ENiagara_LifetimeMode}}
      }
      Lifetime Min {bIsDynamic=true, bIsEditable=true, bIsStaticSwitch=false, bIsVisible=true, type {classStructOrEnum {refPath=&NiagaraFloat}}}
      Lifetime {bIsDynamic=true, bIsEditable=false, bIsStaticSwitch=false, bIsVisible=false, type {classStructOrEnum {refPath=&NiagaraFloat}}}
      Lifetime Max {bIsDynamic=true, bIsEditable=true, bIsStaticSwitch=false, bIsVisible=true, type {classStructOrEnum {refPath=&NiagaraFloat}}}
      Lifetime Randomness Mode {
       bIsDynamic=false
       bIsEditable=true
       bIsStaticSwitch=true
       bIsVisible=true
       type {classStructOrEnum {refPath=&ENiagaraRandomnessMode}}
      }
      Color Mode {
       bIsDynamic=false
       bIsEditable=true
       bIsStaticSwitch=true
       bIsVisible=true
       type {classStructOrEnum {refPath=&ENiagara_ColorInitializationMode}}
      }
      Color Minimum {
       bIsDynamic=false
       bIsEditable=false
       bIsStaticSwitch=false
       bIsVisible=false
       type {classStructOrEnum {refPath=&LinearColor}}
      }
      Color {
       bIsDynamic=true
       bIsEditable=true
       bIsStaticSwitch=false
       bIsVisible=true
       type {classStructOrEnum {refPath=&LinearColor}}
      }
      AdjustHue {bIsDynamic=false, bIsEditable=false, bIsStaticSwitch=false, bIsVisible=false, type {classStructOrEnum {refPath=&NiagaraBool}}}
      Hue Shift Range {
       bIsDynamic=false
       bIsEditable=false
       bIsStaticSwitch=false
       bIsVisible=false
       type {classStructOrEnum {refPath=&Vector2f}}
      }
      Color Maximum {
       bIsDynamic=false
       bIsEditable=false
       bIsStaticSwitch=false
       bIsVisible=false
       type {classStructOrEnum {refPath=&LinearColor}}
      }
      AdjustSaturation {bIsDynamic=false, bIsEditable=false, bIsStaticSwitch=false, bIsVisible=false, type {classStructOrEnum {refPath=&NiagaraBool}}}
      Saturation Range {
       bIsDynamic=false
       bIsEditable=false
       bIsStaticSwitch=false
       bIsVisible=false
       type {classStructOrEnum {refPath=&Vector2f}}
      }
      AdjustValue {bIsDynamic=false, bIsEditable=false, bIsStaticSwitch=false, bIsVisible=false, type {classStructOrEnum {refPath=&NiagaraBool}}}
      Value Range {
       bIsDynamic=false
       bIsEditable=false
       bIsStaticSwitch=false
       bIsVisible=false
       type {classStructOrEnum {refPath=&Vector2f}}
      }
      AdjustAlpha {bIsDynamic=false, bIsEditable=false, bIsStaticSwitch=false, bIsVisible=false, type {classStructOrEnum {refPath=&NiagaraBool}}}
      Alpha Scale Range {
       bIsDynamic=false
       bIsEditable=false
       bIsStaticSwitch=false
       bIsVisible=false
       type {classStructOrEnum {refPath=&Vector2f}}
      }
      Color Channel Mode {
       bIsDynamic=false
       bIsEditable=false
       bIsStaticSwitch=true
       bIsVisible=false
       type {classStructOrEnum {refPath=&ENiagaraChannelCorrelation}}
      }
      Color Randomness Mode {
       bIsDynamic=false
       bIsEditable=false
       bIsStaticSwitch=true
       bIsVisible=false
       type {classStructOrEnum {refPath=&ENiagaraRandomnessMode}}
      }
      Position Mode {
       bIsDynamic=false
       bIsEditable=true
       bIsStaticSwitch=true
       bIsVisible=true
       type {classStructOrEnum {refPath=&ENiagara_PositionInitializationMode}}
      }
      Position {
       bIsDynamic=true
       bIsEditable=false
       bIsStaticSwitch=false
       bIsVisible=false
       type {classStructOrEnum {refPath=&NiagaraPosition}}
      }
      UsePositionOffset {bIsDynamic=false, bIsEditable=true, bIsStaticSwitch=false, bIsVisible=true, type {classStructOrEnum {refPath=&NiagaraBool}}}
      Position Offset {bIsDynamic=true, bIsEditable=true, bIsStaticSwitch=false, bIsVisible=true, type {classStructOrEnum {refPath=&Vector3f}}}
      Position Offset Coordinate Space {
       bIsDynamic=false
       bIsEditable=true
       bIsStaticSwitch=false
       bIsVisible=true
       type {classStructOrEnum {refPath=&ENiagaraCoordinateSpace}}
      }
      Mass Mode {
       bIsDynamic=false
       bIsEditable=true
       bIsStaticSwitch=true
       bIsVisible=true
       type {classStructOrEnum {refPath=&ENiagara_MassInitializationMode}}
      }
      Mass Min {
       bIsDynamic=false
       bIsEditable=false
       bIsStaticSwitch=false
       bIsVisible=false
       type {classStructOrEnum {refPath=&NiagaraFloat}}
      }
      Mass {
       bIsDynamic=false
       bIsEditable=false
       bIsStaticSwitch=false
       bIsVisible=false
       type {classStructOrEnum {refPath=&NiagaraFloat}}
      }
      Mass Max {
       bIsDynamic=false
       bIsEditable=false
       bIsStaticSwitch=false
       bIsVisible=false
       type {classStructOrEnum {refPath=&NiagaraFloat}}
      }
      Mass Randomness Mode {
       bIsDynamic=false
       bIsEditable=false
       bIsStaticSwitch=true
       bIsVisible=false
       type {classStructOrEnum {refPath=&ENiagaraRandomnessMode}}
      }
      Write Material Random {bIsDynamic=false, bIsEditable=true, bIsStaticSwitch=false, bIsVisible=true, type {classStructOrEnum {refPath=&NiagaraBool}}}
      Material Random {bIsDynamic=false, bIsEditable=false, bIsStaticSwitch=false, bIsVisible=true, type {classStructOrEnum {refPath=&NiagaraFloat}}}
      Sprite Size Mode {
       bIsDynamic=false
       bIsEditable=true
       bIsStaticSwitch=true
       bIsVisible=true
       type {classStructOrEnum {refPath=&ENiagara_SizeScaleMode}}
      }
      Sprite Size {
       bIsDynamic=false
       bIsEditable=false
       bIsStaticSwitch=false
       bIsVisible=false
       type {classStructOrEnum {refPath=&Vector2f}}
      }
      Uniform Sprite Size {bIsDynamic=true, bIsEditable=true, bIsStaticSwitch=false, bIsVisible=true, type {classStructOrEnum {refPath=&NiagaraFloat}}}
      Uniform Sprite Size Min {
       bIsDynamic=false
       bIsEditable=false
       bIsStaticSwitch=false
       bIsVisible=false
       type {classStructOrEnum {refPath=&NiagaraFloat}}
      }
      Sprite Size Min {
       bIsDynamic=false
       bIsEditable=false
       bIsStaticSwitch=false
       bIsVisible=false
       type {classStructOrEnum {refPath=&Vector2f}}
      }
      Sprite Size Max {
       bIsDynamic=false
       bIsEditable=false
       bIsStaticSwitch=false
       bIsVisible=false
       type {classStructOrEnum {refPath=&Vector2f}}
      }
      Uniform Sprite Size Max {
       bIsDynamic=false
       bIsEditable=false
       bIsStaticSwitch=false
       bIsVisible=false
       type {classStructOrEnum {refPath=&NiagaraFloat}}
      }
      Sprite Size Randomness Mode {
       bIsDynamic=false
       bIsEditable=false
       bIsStaticSwitch=true
       bIsVisible=false
       type {classStructOrEnum {refPath=&ENiagaraRandomnessMode}}
      }
      Sprite Rotation Mode {
       bIsDynamic=false
       bIsEditable=true
       bIsStaticSwitch=true
       bIsVisible=true
       type {classStructOrEnum {refPath=&ENiagara_SpriteRotationMode}}
      }
      Sprite Rotation Angle Min {
       bIsDynamic=false
       bIsEditable=false
       bIsStaticSwitch=false
       bIsVisible=false
       type {classStructOrEnum {refPath=&NiagaraFloat}}
      }
      Sprite Rotation Angle {
       bIsDynamic=false
       bIsEditable=false
       bIsStaticSwitch=false
       bIsVisible=false
       type {classStructOrEnum {refPath=&NiagaraFloat}}
      }
      Sprite Rotation Angle Max {
       bIsDynamic=false
       bIsEditable=false
       bIsStaticSwitch=false
       bIsVisible=false
       type {classStructOrEnum {refPath=&NiagaraFloat}}
      }
      Sprite Rotation Randomness Mode {
       bIsDynamic=false
       bIsEditable=false
       bIsStaticSwitch=true
       bIsVisible=false
       type {classStructOrEnum {refPath=&ENiagaraRandomnessMode}}
      }
      Sprite UV Mode {
       bIsDynamic=false
       bIsEditable=true
       bIsStaticSwitch=true
       bIsVisible=true
       type {classStructOrEnum {refPath=&ENiagara_UVFlippingMode}}
      }
      Sprite UV Randomness Mode {
       bIsDynamic=false
       bIsEditable=false
       bIsStaticSwitch=true
       bIsVisible=false
       type {classStructOrEnum {refPath=&ENiagaraRandomnessMode}}
      }
      Sprite UV Scale {
       bIsDynamic=false
       bIsEditable=false
       bIsStaticSwitch=false
       bIsVisible=false
       type {classStructOrEnum {refPath=&Vector2f}}
      }
      Mesh Scale Mode {
       bIsDynamic=false
       bIsEditable=true
       bIsStaticSwitch=true
       bIsVisible=true
       type {classStructOrEnum {refPath=&ENiagara_SizeScaleMode}}
      }
      Mesh Uniform Scale Min {bIsDynamic=false, bIsEditable=true, bIsStaticSwitch=false, bIsVisible=true, type {classStructOrEnum {refPath=&NiagaraFloat}}}
      Mesh Uniform Scale {
       bIsDynamic=false
       bIsEditable=false
       bIsStaticSwitch=false
       bIsVisible=false
       type {classStructOrEnum {refPath=&NiagaraFloat}}
      }
      Mesh Scale Min {
       bIsDynamic=false
       bIsEditable=false
       bIsStaticSwitch=false
       bIsVisible=false
       type {classStructOrEnum {refPath=&Vector3f}}
      }
      Mesh Scale {bIsDynamic=true, bIsEditable=false, bIsStaticSwitch=false, bIsVisible=false, type {classStructOrEnum {refPath=&Vector3f}}}
      Mesh Scale Max {
       bIsDynamic=false
       bIsEditable=false
       bIsStaticSwitch=false
       bIsVisible=false
       type {classStructOrEnum {refPath=&Vector3f}}
      }
      Mesh Uniform Scale Max {bIsDynamic=false, bIsEditable=true, bIsStaticSwitch=false, bIsVisible=true, type {classStructOrEnum {refPath=&NiagaraFloat}}}
      Mesh Scale Randomness Mode {
       bIsDynamic=false
       bIsEditable=true
       bIsStaticSwitch=true
       bIsVisible=true
       type {classStructOrEnum {refPath=&ENiagaraRandomnessMode}}
      }
      Mesh Renderer Array Visibility Mode {
       bIsDynamic=false
       bIsEditable=true
       bIsStaticSwitch=true
       bIsVisible=true
       type {classStructOrEnum {refPath=&ENiagara_UnsetDirectSetRandom}}
      }
      Mesh Renderer Mesh Index {
       bIsDynamic=false
       bIsEditable=false
       bIsStaticSwitch=false
       bIsVisible=false
       type {classStructOrEnum {refPath=&NiagaraInt32}}
      }
      Wrap Index By Mesh Count {bIsDynamic=false, bIsEditable=false, bIsStaticSwitch=true, bIsVisible=false, type {classStructOrEnum {refPath=&NiagaraBool}}}
      Mesh Renderer Info {
       bIsDynamic=false
       bIsEditable=true
       bIsStaticSwitch=false
       bIsVisible=true
       type {classStructOrEnum {refPath=&NiagaraDataInterfaceMeshRendererInfo}}
      }
      Mesh Renderer Random Seed {bIsDynamic=false, bIsEditable=true, bIsStaticSwitch=false, bIsVisible=true, type {classStructOrEnum {refPath=&NiagaraInt32}}}
      Ribbon Width Mode {
       bIsDynamic=false
       bIsEditable=true
       bIsStaticSwitch=true
       bIsVisible=true
       type {classStructOrEnum {refPath=&ENiagara_UnsetDirectSet}}
      }
      Ribbon Width {
       bIsDynamic=false
       bIsEditable=false
       bIsStaticSwitch=false
       bIsVisible=false
       type {classStructOrEnum {refPath=&NiagaraFloat}}
      }
      Ribbon Facing Mode {
       bIsDynamic=false
       bIsEditable=true
       bIsStaticSwitch=true
       bIsVisible=true
       type {classStructOrEnum {refPath=&ENiagara_UnsetDirectSet}}
      }
      Ribbon Facing Vector {
       bIsDynamic=false
       bIsEditable=false
       bIsStaticSwitch=false
       bIsVisible=false
       type {classStructOrEnum {refPath=&Vector3f}}
      }
      Ribbon Twist Mode {
       bIsDynamic=false
       bIsEditable=true
       bIsStaticSwitch=true
       bIsVisible=true
       type {classStructOrEnum {refPath=&ENiagara_UnsetDirectSet}}
      }
      "Ribbon Twist " {
       bIsDynamic=false
       bIsEditable=false
       bIsStaticSwitch=false
       bIsVisible=false
       type {classStructOrEnum {refPath=&NiagaraFloat}}
      }
     ]
     moduleScript {refPath=&InitializeParticle}
    }
    {
     bIsSetParametersModule=false
     enabled=true
     moduleName=InitialMeshOrientation
     inputs [
      Mesh Orientation Mode {
       bIsDynamic=false
       bIsEditable=true
       bIsStaticSwitch=true
       bIsVisible=true
       type {classStructOrEnum {refPath=&ENiagara_MeshOrientationOptions}}
      }
      Orientation Quaternion {bIsDynamic=false, bIsEditable=false, bIsStaticSwitch=false, bIsVisible=false, type {classStructOrEnum {refPath=&Quat4f}}}
      Orientation Matrix {
       bIsDynamic=false
       bIsEditable=false
       bIsStaticSwitch=false
       bIsVisible=false
       type {classStructOrEnum {refPath=&NiagaraMatrix}}
      }
      Orientation Vector {
       bIsDynamic=false
       bIsEditable=false
       bIsStaticSwitch=false
       bIsVisible=false
       type {classStructOrEnum {refPath=&Vector3f}}
      }
      Orientation Coordinate Space {
       bIsDynamic=false
       bIsEditable=false
       bIsStaticSwitch=true
       bIsVisible=false
       type {classStructOrEnum {refPath=&ENiagaraCoordinateSpace}}
      }
      Mesh Axis To Orient {
       bIsDynamic=false
       bIsEditable=false
       bIsStaticSwitch=false
       bIsVisible=false
       type {classStructOrEnum {refPath=&Vector3f}}
      }
      Random Seed {bIsDynamic=false, bIsEditable=true, bIsStaticSwitch=false, bIsVisible=true, type {classStructOrEnum {refPath=&NiagaraInt32}}}
      Use Rotation Vector {bIsDynamic=false, bIsEditable=true, bIsStaticSwitch=false, bIsVisible=true, type {classStructOrEnum {refPath=&NiagaraBool}}}
      Rotation {bIsDynamic=false, bIsEditable=false, bIsStaticSwitch=false, bIsVisible=true, type {classStructOrEnum {refPath=&Vector3f}}}
      Rotation Coordinate Space {
       bIsDynamic=false
       bIsEditable=false
       bIsStaticSwitch=true
       bIsVisible=true
       type {classStructOrEnum {refPath=&ENiagaraMeshTransforms}}
      }
     ]
     moduleScript {refPath=&InitialMeshOrientation}
    }
    {
     bIsSetParametersModule=false
     enabled=true
     moduleName=ShapeLocation
     inputs [
      Shape Primitive {
       bIsDynamic=false
       bIsEditable=true
       bIsStaticSwitch=true
       bIsVisible=true
       type {classStructOrEnum {refPath=&ENiagara_LocationShapes}}
      }
      Box / Plane Mode {
       bIsDynamic=false
       bIsEditable=false
       bIsStaticSwitch=true
       bIsVisible=false
       type {classStructOrEnum {refPath=&ENiagara_BoxPlaneMode}}
      }
      Cone Mode {
       bIsDynamic=false
       bIsEditable=false
       bIsStaticSwitch=true
       bIsVisible=false
       type {classStructOrEnum {refPath=&ENiagara_ConeMode}}
      }
      Cylinder Height {
       bIsDynamic=false
       bIsEditable=false
       bIsStaticSwitch=false
       bIsVisible=false
       type {classStructOrEnum {refPath=&NiagaraFloat}}
      }
      Sphere Radius {
       bIsDynamic=false
       bIsEditable=false
       bIsStaticSwitch=false
       bIsVisible=false
       type {classStructOrEnum {refPath=&NiagaraFloat}}
      }
      Ring / Disc Mode {
       bIsDynamic=false
       bIsEditable=true
       bIsStaticSwitch=true
       bIsVisible=true
       type {classStructOrEnum {refPath=&ENiagaraRingDiscMode}}
      }
      Torus Mode {
       bIsDynamic=false
       bIsEditable=false
       bIsStaticSwitch=true
       bIsVisible=false
       type {classStructOrEnum {refPath=&ENiagaraShapeTorusMode}}
      }
      Knot Ring Loops {
       bIsDynamic=false
       bIsEditable=false
       bIsStaticSwitch=false
       bIsVisible=false
       type {classStructOrEnum {refPath=&NiagaraInt32}}
      }
      Knot Helix Loops {
       bIsDynamic=false
       bIsEditable=false
       bIsStaticSwitch=false
       bIsVisible=false
       type {classStructOrEnum {refPath=&NiagaraInt32}}
      }
      Enable Torus Knot Radius {bIsDynamic=false, bIsEditable=false, bIsStaticSwitch=false, bIsVisible=false, type {classStructOrEnum {refPath=&NiagaraBool}}}
      Torus Knot Tube Radius {
       bIsDynamic=false
       bIsEditable=false
       bIsStaticSwitch=false
       bIsVisible=false
       type {classStructOrEnum {refPath=&NiagaraFloat}}
      }
      Cylinder Radius {
       bIsDynamic=false
       bIsEditable=false
       bIsStaticSwitch=false
       bIsVisible=false
       type {classStructOrEnum {refPath=&NiagaraFloat}}
      }
      Cylinder Height Midpoint {
       bIsDynamic=false
       bIsEditable=false
       bIsStaticSwitch=false
       bIsVisible=false
       type {classStructOrEnum {refPath=&NiagaraFloat}}
      }
      Ring Radius {bIsDynamic=false, bIsEditable=true, bIsStaticSwitch=false, bIsVisible=true, type {classStructOrEnum {refPath=&NiagaraFloat}}}
      Box Size {
       bIsDynamic=false
       bIsEditable=false
       bIsStaticSwitch=false
       bIsVisible=false
       type {classStructOrEnum {refPath=&Vector3f}}
      }
      Box Midpoint {
       bIsDynamic=false
       bIsEditable=false
       bIsStaticSwitch=false
       bIsVisible=false
       type {classStructOrEnum {refPath=&Vector3f}}
      }
      Cone Length {
       bIsDynamic=false
       bIsEditable=false
       bIsStaticSwitch=false
       bIsVisible=false
       type {classStructOrEnum {refPath=&NiagaraFloat}}
      }
      Plane Size {
       bIsDynamic=false
       bIsEditable=false
       bIsStaticSwitch=false
       bIsVisible=false
       type {classStructOrEnum {refPath=&Vector2f}}
      }
      Wedge Horizontal Angle {
       bIsDynamic=false
       bIsEditable=false
       bIsStaticSwitch=false
       bIsVisible=false
       type {classStructOrEnum {refPath=&NiagaraFloat}}
      }
      Large Radius {
       bIsDynamic=false
       bIsEditable=false
       bIsStaticSwitch=false
       bIsVisible=false
       type {classStructOrEnum {refPath=&NiagaraFloat}}
      }
      Plane Midpoint {
       bIsDynamic=false
       bIsEditable=false
       bIsStaticSwitch=false
       bIsVisible=false
       type {classStructOrEnum {refPath=&Vector2f}}
      }
      Wedge Vertical Angle {
       bIsDynamic=false
       bIsEditable=false
       bIsStaticSwitch=false
       bIsVisible=false
       type {classStructOrEnum {refPath=&NiagaraFloat}}
      }
      Cone Angle {
       bIsDynamic=false
       bIsEditable=false
       bIsStaticSwitch=false
       bIsVisible=false
       type {classStructOrEnum {refPath=&NiagaraFloat}}
      }
      Handle Radius {
       bIsDynamic=false
       bIsEditable=false
       bIsStaticSwitch=false
       bIsVisible=false
       type {classStructOrEnum {refPath=&NiagaraFloat}}
      }
      Cone Inner Angle {
       bIsDynamic=false
       bIsEditable=false
       bIsStaticSwitch=false
       bIsVisible=false
       type {classStructOrEnum {refPath=&NiagaraFloat}}
      }
      Cone Radial Angle {
       bIsDynamic=false
       bIsEditable=false
       bIsStaticSwitch=false
       bIsVisible=false
       type {classStructOrEnum {refPath=&NiagaraFloat}}
      }
      Cone Angle Type {
       bIsDynamic=false
       bIsEditable=false
       bIsStaticSwitch=true
       bIsVisible=false
       type {classStructOrEnum {refPath=&ENiagara_AngleInput}}
      }
      Flatten Endcaps {bIsDynamic=false, bIsEditable=false, bIsStaticSwitch=true, bIsVisible=false, type {classStructOrEnum {refPath=&NiagaraBool}}}
      Apply To Particle Position {
       bIsDynamic=false
       bIsEditable=true
       bIsStaticSwitch=true
       bIsVisible=true
       type {classStructOrEnum {refPath=&ENiagara_AttributeSamplingApplyOutput}}
      }
      Cone Surface Distribution {
       bIsDynamic=false
       bIsEditable=false
       bIsStaticSwitch=false
       bIsVisible=false
       type {classStructOrEnum {refPath=&NiagaraFloat}}
      }
      Ring / Disc Distribution Mode {
       bIsDynamic=false
       bIsEditable=true
       bIsStaticSwitch=true
       bIsVisible=true
       type {classStructOrEnum {refPath=&ENiagaraSphereDistributionMode}}
      }
      Sphere Distribution {
       bIsDynamic=false
       bIsEditable=false
       bIsStaticSwitch=true
       bIsVisible=false
       type {classStructOrEnum {refPath=&ENiagaraSphereDistributionMode}}
      }
      Torus Distribution Mode {
       bIsDynamic=false
       bIsEditable=false
       bIsStaticSwitch=true
       bIsVisible=false
       type {classStructOrEnum {refPath=&ENiagaraTorusDistributionMode}}
      }
      Sphere Surface Distribution {
       bIsDynamic=false
       bIsEditable=false
       bIsStaticSwitch=false
       bIsVisible=false
       type {classStructOrEnum {refPath=&NiagaraFloat}}
      }
      Hemisphere Distribution {
       bIsDynamic=false
       bIsEditable=false
       bIsStaticSwitch=false
       bIsVisible=false
       type {classStructOrEnum {refPath=&Vector2f}}
      }
      Cylinder Distribution {
       bIsDynamic=false
       bIsEditable=false
       bIsStaticSwitch=true
       bIsVisible=false
       type {classStructOrEnum {refPath=&ENiagara_CylinderMode}}
      }
      Surface Only {bIsDynamic=false, bIsEditable=false, bIsStaticSwitch=false, bIsVisible=false, type {classStructOrEnum {refPath=&NiagaraBool}}}
      Surface Only Band Thickness {
       bIsDynamic=false
       bIsEditable=false
       bIsStaticSwitch=false
       bIsVisible=false
       type {classStructOrEnum {refPath=&NiagaraFloat}}
      }
      Use Endcaps In Surface Only Mode {bIsDynamic=false, bIsEditable=false, bIsStaticSwitch=true, bIsVisible=false, type {classStructOrEnum {refPath=&NiagaraBool}}}
      Hemicircle X {bIsDynamic=false, bIsEditable=false, bIsStaticSwitch=true, bIsVisible=false, type {classStructOrEnum {refPath=&NiagaraBool}}}
      Hemicircle Y {bIsDynamic=false, bIsEditable=false, bIsStaticSwitch=true, bIsVisible=false, type {classStructOrEnum {refPath=&NiagaraBool}}}
      Disc Coverage {bIsDynamic=false, bIsEditable=true, bIsStaticSwitch=false, bIsVisible=true, type {classStructOrEnum {refPath=&NiagaraFloat}}}
      Hemicircle Internal Cap {bIsDynamic=false, bIsEditable=false, bIsStaticSwitch=true, bIsVisible=false, type {classStructOrEnum {refPath=&NiagaraBool}}}
      U Position {
       bIsDynamic=false
       bIsEditable=false
       bIsStaticSwitch=false
       bIsVisible=false
       type {classStructOrEnum {refPath=&NiagaraFloat}}
      }
      Enable Lathe Profile {bIsDynamic=false, bIsEditable=false, bIsStaticSwitch=true, bIsVisible=false, type {classStructOrEnum {refPath=&NiagaraBool}}}
      Lathe Profile {
       bIsDynamic=false
       bIsEditable=false
       bIsStaticSwitch=false
       bIsVisible=false
       type {classStructOrEnum {refPath=&NiagaraDataInterfaceCurve}}
      }
      Surface Distribution {
       bIsDynamic=false
       bIsEditable=false
       bIsStaticSwitch=false
       bIsVisible=false
       type {classStructOrEnum {refPath=&NiagaraFloat}}
      }
      Surface Expansion Mode {
       bIsDynamic=false
       bIsEditable=false
       bIsStaticSwitch=false
       bIsVisible=false
       type {classStructOrEnum {refPath=&ENiagaraExpansionMode}}
      }
      U Distribution {bIsDynamic=false, bIsEditable=true, bIsStaticSwitch=false, bIsVisible=true, type {classStructOrEnum {refPath=&NiagaraFloat}}}
      V Position {
       bIsDynamic=false
       bIsEditable=false
       bIsStaticSwitch=false
       bIsVisible=false
       type {classStructOrEnum {refPath=&NiagaraFloat}}
      }
      Radius Position {
       bIsDynamic=false
       bIsEditable=false
       bIsStaticSwitch=false
       bIsVisible=false
       type {classStructOrEnum {refPath=&NiagaraFloat}}
      }
      Sphere Radius Position {
       bIsDynamic=false
       bIsEditable=false
       bIsStaticSwitch=false
       bIsVisible=false
       type {classStructOrEnum {refPath=&NiagaraFloat}}
      }
      V Distribution {
       bIsDynamic=false
       bIsEditable=false
       bIsStaticSwitch=false
       bIsVisible=false
       type {classStructOrEnum {refPath=&NiagaraFloat}}
      }
      Evaluation Type {
       bIsDynamic=false
       bIsEditable=false
       bIsStaticSwitch=true
       bIsVisible=false
       type {classStructOrEnum {refPath=&ENiagaraRandomnessEvaluation}}
      }
      Uniform Distribution {
       bIsDynamic=false
       bIsEditable=false
       bIsStaticSwitch=false
       bIsVisible=false
       type {classStructOrEnum {refPath=&NiagaraFloat}}
      }
      Uniform Spiral Amount {
       bIsDynamic=false
       bIsEditable=false
       bIsStaticSwitch=false
       bIsVisible=false
       type {classStructOrEnum {refPath=&NiagaraFloat}}
      }
      Uniform Spiral Falloff {
       bIsDynamic=false
       bIsEditable=false
       bIsStaticSwitch=false
       bIsVisible=false
       type {classStructOrEnum {refPath=&NiagaraFloat}}
      }
      Uniform Count {
       bIsDynamic=false
       bIsEditable=false
       bIsStaticSwitch=false
       bIsVisible=false
       type {classStructOrEnum {refPath=&NiagaraInt32}}
      }
      Uniform Particle Index {
       bIsDynamic=false
       bIsEditable=false
       bIsStaticSwitch=false
       bIsVisible=false
       type {classStructOrEnum {refPath=&NiagaraInt32}}
      }
      Hemisphere Angle Type {
       bIsDynamic=false
       bIsEditable=false
       bIsStaticSwitch=true
       bIsVisible=false
       type {classStructOrEnum {refPath=&ENiagara_AngleInput}}
      }
      Uniform Endpoint Offset {
       bIsDynamic=false
       bIsEditable=false
       bIsStaticSwitch=false
       bIsVisible=false
       type {classStructOrEnum {refPath=&NiagaraFloat}}
      }
      MaskSpawnGroup {bIsDynamic=false, bIsEditable=true, bIsStaticSwitch=false, bIsVisible=true, type {classStructOrEnum {refPath=&NiagaraBool}}}
      Spawn Group Mask {bIsDynamic=false, bIsEditable=false, bIsStaticSwitch=false, bIsVisible=true, type {classStructOrEnum {refPath=&NiagaraInt32}}}
      Spawn Group {
       bIsDynamic=false
       bIsEditable=false
       bIsStaticSwitch=false
       bIsVisible=false
       type {classStructOrEnum {refPath=&NiagaraInt32}}
      }
      Randomness Mode {
       bIsDynamic=false
       bIsEditable=true
       bIsStaticSwitch=true
       bIsVisible=true
       type {classStructOrEnum {refPath=&ENiagaraRandomnessMode}}
      }
      Override Seed {bIsDynamic=false, bIsEditable=true, bIsStaticSwitch=false, bIsVisible=true, type {classStructOrEnum {refPath=&NiagaraBool}}}
      Random Seed {bIsDynamic=false, bIsEditable=false, bIsStaticSwitch=false, bIsVisible=true, type {classStructOrEnum {refPath=&NiagaraInt32}}}
      Fixed Random Seed {bIsDynamic=false, bIsEditable=false, bIsStaticSwitch=true, bIsVisible=false, type {classStructOrEnum {refPath=&NiagaraBool}}}
      Transform Order {
       bIsDynamic=false
       bIsEditable=true
       bIsStaticSwitch=true
       bIsVisible=true
       type {classStructOrEnum {refPath=&ENiagara_TransformOrder}}
      }
      Scale Mode {
       bIsDynamic=false
       bIsEditable=true
       bIsStaticSwitch=true
       bIsVisible=true
       type {classStructOrEnum {refPath=&ENiagara_ScaleMode}}
      }
      Non Uniform Scale {bIsDynamic=false, bIsEditable=true, bIsStaticSwitch=false, bIsVisible=true, type {classStructOrEnum {refPath=&Vector3f}}}
      UseOwnerScale {bIsDynamic=false, bIsEditable=true, bIsStaticSwitch=false, bIsVisible=true, type {classStructOrEnum {refPath=&NiagaraBool}}}
      Apply Owner Scale {bIsDynamic=false, bIsEditable=true, bIsStaticSwitch=false, bIsVisible=true, type {classStructOrEnum {refPath=&Vector3f}}}
      Rotation Mode {
       bIsDynamic=false
       bIsEditable=true
       bIsStaticSwitch=true
       bIsVisible=true
       type {classStructOrEnum {refPath=&ENiagara_RotationMode}}
      }
      Rotation Axis {
       bIsDynamic=false
       bIsEditable=false
       bIsStaticSwitch=false
       bIsVisible=false
       type {classStructOrEnum {refPath=&Vector3f}}
      }
      Rotation Matrix {
       bIsDynamic=false
       bIsEditable=false
       bIsStaticSwitch=false
       bIsVisible=false
       type {classStructOrEnum {refPath=&NiagaraMatrix}}
      }
      Rotation Quaternion {bIsDynamic=false, bIsEditable=false, bIsStaticSwitch=false, bIsVisible=false, type {classStructOrEnum {refPath=&Quat4f}}}
      Yaw / Pitch / Roll {bIsDynamic=false, bIsEditable=true, bIsStaticSwitch=false, bIsVisible=true, type {classStructOrEnum {refPath=&Vector3f}}}
      Rotation Angle {
       bIsDynamic=false
       bIsEditable=false
       bIsStaticSwitch=false
       bIsVisible=false
       type {classStructOrEnum {refPath=&NiagaraFloat}}
      }
      Invert Rotation Quaternion {bIsDynamic=false, bIsEditable=false, bIsStaticSwitch=false, bIsVisible=false, type {classStructOrEnum {refPath=&NiagaraBool}}}
      Additional Quaternion Rotation {bIsDynamic=false, bIsEditable=false, bIsStaticSwitch=false, bIsVisible=false, type {classStructOrEnum {refPath=&NiagaraBool}}}
      Additional Yaw / Pitch / Roll {
       bIsDynamic=false
       bIsEditable=false
       bIsStaticSwitch=false
       bIsVisible=false
       type {classStructOrEnum {refPath=&Vector3f}}
      }
      Rotation Angle Type {
       bIsDynamic=false
       bIsEditable=true
       bIsStaticSwitch=true
       bIsVisible=true
       type {classStructOrEnum {refPath=&ENiagara_AngleInput}}
      }
      Rotation Coordinate Space {
       bIsDynamic=false
       bIsEditable=true
       bIsStaticSwitch=false
       bIsVisible=true
       type {classStructOrEnum {refPath=&ENiagaraCoordinateSpace}}
      }
      Offset Mode {
       bIsDynamic=false
       bIsEditable=true
       bIsStaticSwitch=true
       bIsVisible=true
       type {classStructOrEnum {refPath=&ENiagara_OffsetMode}}
      }
      Shape Origin {
       bIsDynamic=false
       bIsEditable=true
       bIsStaticSwitch=false
       bIsVisible=true
       type {classStructOrEnum {refPath=&NiagaraPosition}}
      }
      Offset {bIsDynamic=true, bIsEditable=true, bIsStaticSwitch=false, bIsVisible=true, type {classStructOrEnum {refPath=&Vector3f}}}
      Offset Coordinate Space {
       bIsDynamic=false
       bIsEditable=true
       bIsStaticSwitch=false
       bIsVisible=true
       type {classStructOrEnum {refPath=&ENiagaraCoordinateSpace}}
      }
      Custom Transform Matrix {
       bIsDynamic=false
       bIsEditable=false
       bIsStaticSwitch=false
       bIsVisible=false
       type {classStructOrEnum {refPath=&NiagaraMatrix}}
      }
      Transform Method {
       bIsDynamic=false
       bIsEditable=true
       bIsStaticSwitch=true
       bIsVisible=true
       type {classStructOrEnum {refPath=&ENiagara_TransformType}}
      }
     ]
     moduleScript {refPath=&ShapeLocation}
    }
    {
     bIsSetParametersModule=false
     enabled=true
     moduleName=AddVelocity
     inputs [
      Velocity Mode {
       bIsDynamic=false
       bIsEditable=true
       bIsStaticSwitch=true
       bIsVisible=true
       type {classStructOrEnum {refPath=&ENiagara_VelocityMode}}
      }
      Velocity {bIsDynamic=false, bIsEditable=true, bIsStaticSwitch=false, bIsVisible=true, type {classStructOrEnum {refPath=&Vector3f}}}
      Velocity Speed {
       bIsDynamic=false
       bIsEditable=false
       bIsStaticSwitch=false
       bIsVisible=false
       type {classStructOrEnum {refPath=&NiagaraFloat}}
      }
      Velocity Speed Scale {bIsDynamic=true, bIsEditable=true, bIsStaticSwitch=false, bIsVisible=true, type {classStructOrEnum {refPath=&NiagaraFloat}}}
      Constrain To Radius {
       bIsDynamic=false
       bIsEditable=false
       bIsStaticSwitch=false
       bIsVisible=false
       type {classStructOrEnum {refPath=&NiagaraFloat}}
      }
      Distribution Along Cone Axis {
       bIsDynamic=false
       bIsEditable=false
       bIsStaticSwitch=false
       bIsVisible=false
       type {classStructOrEnum {refPath=&NiagaraFloat}}
      }
      Radius Falloff Near / Far {
       bIsDynamic=false
       bIsEditable=false
       bIsStaticSwitch=false
       bIsVisible=false
       type {classStructOrEnum {refPath=&Vector2f}}
      }
      Radius Falloff Exponent {
       bIsDynamic=false
       bIsEditable=false
       bIsStaticSwitch=false
       bIsVisible=false
       type {classStructOrEnum {refPath=&NiagaraFloat}}
      }
      Speed Falloff From Cone Axis {
       bIsDynamic=false
       bIsEditable=false
       bIsStaticSwitch=false
       bIsVisible=false
       type {classStructOrEnum {refPath=&NiagaraFloat}}
      }
      Invert Velocity Falloff {bIsDynamic=false, bIsEditable=false, bIsStaticSwitch=false, bIsVisible=false, type {classStructOrEnum {refPath=&NiagaraBool}}}
      Override Seed {bIsDynamic=false, bIsEditable=false, bIsStaticSwitch=false, bIsVisible=false, type {classStructOrEnum {refPath=&NiagaraBool}}}
      Random Seed {
       bIsDynamic=false
       bIsEditable=false
       bIsStaticSwitch=false
       bIsVisible=false
       type {classStructOrEnum {refPath=&NiagaraInt32}}
      }
      Fixed Random Seed {bIsDynamic=false, bIsEditable=false, bIsStaticSwitch=true, bIsVisible=false, type {classStructOrEnum {refPath=&NiagaraBool}}}
      Randomness Mode {
       bIsDynamic=false
       bIsEditable=false
       bIsStaticSwitch=true
       bIsVisible=false
       type {classStructOrEnum {refPath=&ENiagaraRandomnessMode}}
      }
      Cone Axis {
       bIsDynamic=false
       bIsEditable=false
       bIsStaticSwitch=false
       bIsVisible=false
       type {classStructOrEnum {refPath=&Vector3f}}
      }
      Cone Angle {
       bIsDynamic=false
       bIsEditable=false
       bIsStaticSwitch=false
       bIsVisible=false
       type {classStructOrEnum {refPath=&NiagaraFloat}}
      }
      Inner Cone Angle {
       bIsDynamic=false
       bIsEditable=false
       bIsStaticSwitch=false
       bIsVisible=false
       type {classStructOrEnum {refPath=&NiagaraFloat}}
      }
      Cone Angle Mode {
       bIsDynamic=false
       bIsEditable=false
       bIsStaticSwitch=true
       bIsVisible=false
       type {classStructOrEnum {refPath=&ENiagara_AngleInput}}
      }
      Rotation Mode {
       bIsDynamic=false
       bIsEditable=true
       bIsStaticSwitch=true
       bIsVisible=true
       type {classStructOrEnum {refPath=&ENiagara_RotationMode}}
      }
      Rotation Axis {
       bIsDynamic=false
       bIsEditable=false
       bIsStaticSwitch=false
       bIsVisible=false
       type {classStructOrEnum {refPath=&Vector3f}}
      }
      Yaw / Pitch / Roll {
       bIsDynamic=false
       bIsEditable=false
       bIsStaticSwitch=false
       bIsVisible=false
       type {classStructOrEnum {refPath=&Vector3f}}
      }
      Rotation Quaternion {bIsDynamic=false, bIsEditable=false, bIsStaticSwitch=false, bIsVisible=false, type {classStructOrEnum {refPath=&Quat4f}}}
      Rotation Matrix {
       bIsDynamic=false
       bIsEditable=false
       bIsStaticSwitch=false
       bIsVisible=false
       type {classStructOrEnum {refPath=&NiagaraMatrix}}
      }
      Rotation Angle {
       bIsDynamic=false
       bIsEditable=false
       bIsStaticSwitch=false
       bIsVisible=false
       type {classStructOrEnum {refPath=&NiagaraFloat}}
      }
      Invert Rotation Quaternion {bIsDynamic=false, bIsEditable=false, bIsStaticSwitch=false, bIsVisible=false, type {classStructOrEnum {refPath=&NiagaraBool}}}
      Additional Quaternion Rotation {bIsDynamic=false, bIsEditable=false, bIsStaticSwitch=false, bIsVisible=false, type {classStructOrEnum {refPath=&NiagaraBool}}}
      Additional Yaw / Pitch / Roll {
       bIsDynamic=false
       bIsEditable=false
       bIsStaticSwitch=false
       bIsVisible=false
       type {classStructOrEnum {refPath=&Vector3f}}
      }
      Rotation Angle Type {
       bIsDynamic=false
       bIsEditable=false
       bIsStaticSwitch=true
       bIsVisible=false
       type {classStructOrEnum {refPath=&ENiagara_AngleInput}}
      }
      Rotation Coordinate Space {
       bIsDynamic=false
       bIsEditable=true
       bIsStaticSwitch=false
       bIsVisible=true
       type {classStructOrEnum {refPath=&ENiagaraCoordinateSpace}}
      }
      Velocity Origin {
       bIsDynamic=false
       bIsEditable=false
       bIsStaticSwitch=false
       bIsVisible=false
       type {classStructOrEnum {refPath=&NiagaraPosition}}
      }
      Origin Offset {
       bIsDynamic=false
       bIsEditable=false
       bIsStaticSwitch=false
       bIsVisible=false
       type {classStructOrEnum {refPath=&Vector3f}}
      }
      Origin Offset Coordinate Space {
       bIsDynamic=false
       bIsEditable=false
       bIsStaticSwitch=false
       bIsVisible=false
       type {classStructOrEnum {refPath=&ENiagaraCoordinateSpace}}
      }
      Default Position {
       bIsDynamic=false
       bIsEditable=false
       bIsStaticSwitch=false
       bIsVisible=false
       type {classStructOrEnum {refPath=&NiagaraPosition}}
      }
     ]
     moduleScript {refPath=&AddVelocity}
    }
    {
     bIsSetParametersModule=false
     enabled=true
     moduleName=MeshRotationForce001
     inputs [
      Rotation {bIsDynamic=true, bIsEditable=true, bIsStaticSwitch=false, bIsVisible=true, type {classStructOrEnum {refPath=&Vector3f}}}
      Coordinate Space {
       bIsDynamic=false
       bIsEditable=true
       bIsStaticSwitch=true
       bIsVisible=true
       type {classStructOrEnum {refPath=&ENiagaraMeshTransforms}}
      }
      "Lever Radius (cm)" {bIsDynamic=false, bIsEditable=true, bIsStaticSwitch=false, bIsVisible=true, type {classStructOrEnum {refPath=&NiagaraFloat}}}
     ]
     moduleScript {refPath=&MeshRotationForce}
    }
    {
     bIsSetParametersModule=false
     enabled=true
     moduleName=ApplyInitialForces
     inputs [
      Apply Force to Velocity {bIsDynamic=false, bIsEditable=true, bIsStaticSwitch=false, bIsVisible=true, type {classStructOrEnum {refPath=&NiagaraBool}}}
      Apply Rotational Force to Rotational Velocity {bIsDynamic=false, bIsEditable=true, bIsStaticSwitch=false, bIsVisible=true, type {classStructOrEnum {refPath=&NiagaraBool}}}
      Apply Force to Position {bIsDynamic=false, bIsEditable=true, bIsStaticSwitch=false, bIsVisible=true, type {classStructOrEnum {refPath=&NiagaraBool}}}
      Positional Force Warmup Time {bIsDynamic=false, bIsEditable=false, bIsStaticSwitch=false, bIsVisible=true, type {classStructOrEnum {refPath=&NiagaraFloat}}}
     ]
     moduleScript {refPath=&ApplyInitialForces}
    }
   ]
  }
  particleUpdateScript {
   scriptName=ParticleUpdateScript
   modules [
    {
     bIsSetParametersModule=false
     enabled=true
     moduleName=ParticleState
     inputs [
      Kill Particles When Lifetime Has Elapsed {bIsDynamic=false, bIsEditable=true, bIsStaticSwitch=true, bIsVisible=true, type {classStructOrEnum {refPath=&NiagaraBool}}}
      Loop Particles Lifetime {bIsDynamic=false, bIsEditable=false, bIsStaticSwitch=false, bIsVisible=false, type {classStructOrEnum {refPath=&NiagaraBool}}}
      Let Infinitely Lived Particles Die When Emitter Deactivates {bIsDynamic=false, bIsEditable=false, bIsStaticSwitch=false, bIsVisible=false, type {classStructOrEnum {refPath=&NiagaraBool}}}
      Lifetime {bIsDynamic=false, bIsEditable=true, bIsStaticSwitch=false, bIsVisible=true, type {classStructOrEnum {refPath=&NiagaraFloat}}}
      DeltaTime {bIsDynamic=false, bIsEditable=true, bIsStaticSwitch=false, bIsVisible=true, type {classStructOrEnum {refPath=&NiagaraFloat}}}
     ]
     moduleScript {refPath=&ParticleState}
    }
    {
     bIsSetParametersModule=false
     enabled=true
     moduleName=ScaleMeshSize001
     inputs [
      Scale Factor {bIsDynamic=true, bIsEditable=true, bIsStaticSwitch=false, bIsVisible=true, type {classStructOrEnum {refPath=&Vector3f}}}
      Initial Mesh Scale {bIsDynamic=false, bIsEditable=true, bIsStaticSwitch=false, bIsVisible=true, type {classStructOrEnum {refPath=&Vector3f}}}
     ]
     moduleScript {refPath=&ScaleMeshSize}
    }
    {
     bIsSetParametersModule=false
     enabled=true
     moduleName=ScaleMeshSize
     inputs [
      Scale Factor {bIsDynamic=true, bIsEditable=true, bIsStaticSwitch=false, bIsVisible=true, type {classStructOrEnum {refPath=&Vector3f}}}
      Initial Mesh Scale {bIsDynamic=false, bIsEditable=true, bIsStaticSwitch=false, bIsVisible=true, type {classStructOrEnum {refPath=&Vector3f}}}
     ]
     moduleScript {refPath=&ScaleMeshSize}
    }
    {
     bIsSetParametersModule=false
     enabled=true
     moduleName=OffsetPosition
     inputs [
      Position to Offset {
       bIsDynamic=false
       bIsEditable=true
       bIsStaticSwitch=false
       bIsVisible=true
       type {classStructOrEnum {refPath=&NiagaraPosition}}
      }
      Position Output {bIsDynamic=false, bIsEditable=true, bIsStaticSwitch=true, bIsVisible=true, type {classStructOrEnum {refPath=&NiagaraInt32}}}
      Offset Mode {bIsDynamic=false, bIsEditable=true, bIsStaticSwitch=true, bIsVisible=true, type {classStructOrEnum {refPath=&NiagaraInt32}}}
      Jitter Mode {bIsDynamic=false, bIsEditable=false, bIsStaticSwitch=true, bIsVisible=false, type {classStructOrEnum {refPath=&NiagaraInt32}}}
      Jitter Offset {
       bIsDynamic=false
       bIsEditable=false
       bIsStaticSwitch=false
       bIsVisible=false
       type {classStructOrEnum {refPath=&Vector3f}}
      }
      Jitter Offset Coordinate Space {
       bIsDynamic=false
       bIsEditable=false
       bIsStaticSwitch=false
       bIsVisible=false
       type {classStructOrEnum {refPath=&ENiagaraCoordinateSpace}}
      }
      Jitter Scale {
       bIsDynamic=false
       bIsEditable=false
       bIsStaticSwitch=false
       bIsVisible=false
       type {classStructOrEnum {refPath=&Vector3f}}
      }
      Jitter Amount {
       bIsDynamic=false
       bIsEditable=false
       bIsStaticSwitch=false
       bIsVisible=false
       type {classStructOrEnum {refPath=&NiagaraFloat}}
      }
      Jitter Delay {
       bIsDynamic=false
       bIsEditable=false
       bIsStaticSwitch=false
       bIsVisible=false
       type {classStructOrEnum {refPath=&NiagaraFloat}}
      }
      DeltaTime {
       bIsDynamic=false
       bIsEditable=false
       bIsStaticSwitch=false
       bIsVisible=false
       type {classStructOrEnum {refPath=&NiagaraFloat}}
      }
      Jitter Time Offset {
       bIsDynamic=false
       bIsEditable=false
       bIsStaticSwitch=false
       bIsVisible=false
       type {classStructOrEnum {refPath=&NiagaraFloat}}
      }
      Jitter Interpolation {bIsDynamic=false, bIsEditable=false, bIsStaticSwitch=true, bIsVisible=false, type {classStructOrEnum {refPath=&NiagaraInt32}}}
      Interpolation Curve {
       bIsDynamic=false
       bIsEditable=false
       bIsStaticSwitch=false
       bIsVisible=false
       type {classStructOrEnum {refPath=&NiagaraDataInterfaceCurve}}
      }
      Interpolation Scale {
       bIsDynamic=false
       bIsEditable=false
       bIsStaticSwitch=false
       bIsVisible=false
       type {classStructOrEnum {refPath=&NiagaraFloat}}
      }
      Position Offset {bIsDynamic=true, bIsEditable=true, bIsStaticSwitch=false, bIsVisible=true, type {classStructOrEnum {refPath=&Vector3f}}}
      Position Offset Coordinate Space {
       bIsDynamic=false
       bIsEditable=true
       bIsStaticSwitch=false
       bIsVisible=true
       type {classStructOrEnum {refPath=&ENiagaraCoordinateSpace}}
      }
      Debug Draw {
       bIsDynamic=false
       bIsEditable=false
       bIsStaticSwitch=false
       bIsVisible=false
       type {classStructOrEnum {refPath=&NiagaraDataInterfaceDebugDraw}}
      }
      SkipID {
       bIsDynamic=false
       bIsEditable=false
       bIsStaticSwitch=false
       bIsVisible=false
       type {classStructOrEnum {refPath=&NiagaraInt32}}
      }
      Skip {
       bIsDynamic=false
       bIsEditable=false
       bIsStaticSwitch=false
       bIsVisible=false
       type {classStructOrEnum {refPath=&NiagaraInt32}}
      }
      Gizmo Scale {
       bIsDynamic=false
       bIsEditable=false
       bIsStaticSwitch=false
       bIsVisible=false
       type {classStructOrEnum {refPath=&NiagaraFloat}}
      }
     ]
     moduleScript {refPath=&OffsetPosition}
    }
    {
     bIsSetParametersModule=false
     enabled=true
     moduleName=MeshRotationForce
     inputs [
      Rotation {bIsDynamic=true, bIsEditable=true, bIsStaticSwitch=false, bIsVisible=true, type {classStructOrEnum {refPath=&Vector3f}}}
      Coordinate Space {
       bIsDynamic=false
       bIsEditable=true
       bIsStaticSwitch=true
       bIsVisible=true
       type {classStructOrEnum {refPath=&ENiagaraMeshTransforms}}
      }
      "Lever Radius (cm)" {bIsDynamic=false, bIsEditable=true, bIsStaticSwitch=false, bIsVisible=true, type {classStructOrEnum {refPath=&NiagaraFloat}}}
     ]
     moduleScript {refPath=&MeshRotationForce}
    }
    {
     bIsSetParametersModule=false
     enabled=true
     moduleName=ScaleSpriteSize
     inputs [
      Scale Sprite Size Mode {
       bIsDynamic=false
       bIsEditable=true
       bIsStaticSwitch=true
       bIsVisible=true
       type {classStructOrEnum {refPath=&ENiagara_ScaleSpriteSize}}
      }
      Uniform Curve Sprite Scale {
       bIsDynamic=false
       bIsEditable=true
       bIsStaticSwitch=false
       bIsVisible=true
       type {classStructOrEnum {refPath=&NiagaraDataInterfaceCurve}}
      }
      Non-Uniform Curve Sprite Scale {
       bIsDynamic=false
       bIsEditable=false
       bIsStaticSwitch=false
       bIsVisible=false
       type {classStructOrEnum {refPath=&NiagaraDataInterfaceVector2DCurve}}
      }
      Uniform Curve Index {bIsDynamic=false, bIsEditable=true, bIsStaticSwitch=false, bIsVisible=true, type {classStructOrEnum {refPath=&NiagaraFloat}}}
      Non-Uniform Curve Index {
       bIsDynamic=false
       bIsEditable=false
       bIsStaticSwitch=false
       bIsVisible=false
       type {classStructOrEnum {refPath=&NiagaraFloat}}
      }
      Non-Uniform Curve Scale {
       bIsDynamic=false
       bIsEditable=false
       bIsStaticSwitch=false
       bIsVisible=false
       type {classStructOrEnum {refPath=&Vector2f}}
      }
      Uniform Curve Scale {bIsDynamic=false, bIsEditable=true, bIsStaticSwitch=false, bIsVisible=true, type {classStructOrEnum {refPath=&NiagaraFloat}}}
      Scale Factor {
       bIsDynamic=false
       bIsEditable=false
       bIsStaticSwitch=false
       bIsVisible=false
       type {classStructOrEnum {refPath=&Vector2f}}
      }
      Uniform Scale Factor {
       bIsDynamic=false
       bIsEditable=false
       bIsStaticSwitch=false
       bIsVisible=false
       type {classStructOrEnum {refPath=&NiagaraFloat}}
      }
      Initial Sprite Size {bIsDynamic=false, bIsEditable=true, bIsStaticSwitch=false, bIsVisible=true, type {classStructOrEnum {refPath=&Vector2f}}}
     ]
     moduleScript {refPath=&ScaleSpriteSize}
    }
    {
     bIsSetParametersModule=false
     enabled=true
     moduleName=ScaleSpriteSize001
     inputs [
      Scale Sprite Size Mode {
       bIsDynamic=false
       bIsEditable=true
       bIsStaticSwitch=true
       bIsVisible=true
       type {classStructOrEnum {refPath=&ENiagara_ScaleSpriteSize}}
      }
      Uniform Curve Sprite Scale {
       bIsDynamic=false
       bIsEditable=true
       bIsStaticSwitch=false
       bIsVisible=true
       type {classStructOrEnum {refPath=&NiagaraDataInterfaceCurve}}
      }
      Non-Uniform Curve Sprite Scale {
       bIsDynamic=false
       bIsEditable=false
       bIsStaticSwitch=false
       bIsVisible=false
       type {classStructOrEnum {refPath=&NiagaraDataInterfaceVector2DCurve}}
      }
      Uniform Curve Index {bIsDynamic=true, bIsEditable=true, bIsStaticSwitch=false, bIsVisible=true, type {classStructOrEnum {refPath=&NiagaraFloat}}}
      Non-Uniform Curve Index {
       bIsDynamic=false
       bIsEditable=false
       bIsStaticSwitch=false
       bIsVisible=false
       type {classStructOrEnum {refPath=&NiagaraFloat}}
      }
      Non-Uniform Curve Scale {
       bIsDynamic=false
       bIsEditable=false
       bIsStaticSwitch=false
       bIsVisible=false
       type {classStructOrEnum {refPath=&Vector2f}}
      }
      Uniform Curve Scale {bIsDynamic=false, bIsEditable=true, bIsStaticSwitch=false, bIsVisible=true, type {classStructOrEnum {refPath=&NiagaraFloat}}}
      Scale Factor {
       bIsDynamic=false
       bIsEditable=false
       bIsStaticSwitch=false
       bIsVisible=false
       type {classStructOrEnum {refPath=&Vector2f}}
      }
      Uniform Scale Factor {
       bIsDynamic=false
       bIsEditable=false
       bIsStaticSwitch=false
       bIsVisible=false
       type {classStructOrEnum {refPath=&NiagaraFloat}}
      }
      Initial Sprite Size {bIsDynamic=false, bIsEditable=true, bIsStaticSwitch=false, bIsVisible=true, type {classStructOrEnum {refPath=&Vector2f}}}
     ]
     moduleScript {refPath=&ScaleSpriteSize}
    }
    {
     bIsSetParametersModule=false
     enabled=true
     moduleName=WindForce
     inputs [
      Debug Scale {
       bIsDynamic=false
       bIsEditable=false
       bIsStaticSwitch=false
       bIsVisible=false
       type {classStructOrEnum {refPath=&NiagaraFloat}}
      }
      Debug {
       bIsDynamic=false
       bIsEditable=false
       bIsStaticSwitch=false
       bIsVisible=false
       type {classStructOrEnum {refPath=&NiagaraDataInterfaceDebugDraw}}
      }
      Wind Speed {bIsDynamic=true, bIsEditable=true, bIsStaticSwitch=false, bIsVisible=true, type {classStructOrEnum {refPath=&Vector3f}}}
      Wind Speed Scale {bIsDynamic=false, bIsEditable=true, bIsStaticSwitch=false, bIsVisible=true, type {classStructOrEnum {refPath=&NiagaraFloat}}}
      Turbulence Mode {
       bIsDynamic=false
       bIsEditable=true
       bIsStaticSwitch=true
       bIsVisible=true
       type {classStructOrEnum {refPath=&ENiagaraWindTurbulenceMode}}
      }
      Turbulence Scale Mode {
       bIsDynamic=false
       bIsEditable=false
       bIsStaticSwitch=true
       bIsVisible=false
       type {classStructOrEnum {refPath=&ENiagaraWindTurbulenceContributionMode}}
      }
      Scale {
       bIsDynamic=false
       bIsEditable=false
       bIsStaticSwitch=false
       bIsVisible=false
       type {classStructOrEnum {refPath=&NiagaraFloat}}
      }
      Scale Min / Max {
       bIsDynamic=false
       bIsEditable=false
       bIsStaticSwitch=false
       bIsVisible=false
       type {classStructOrEnum {refPath=&Vector2f}}
      }
      Speed Min / Max {
       bIsDynamic=false
       bIsEditable=false
       bIsStaticSwitch=false
       bIsVisible=false
       type {classStructOrEnum {refPath=&Vector2f}}
      }
      Scale Curve {
       bIsDynamic=false
       bIsEditable=false
       bIsStaticSwitch=false
       bIsVisible=false
       type {classStructOrEnum {refPath=&NiagaraDataInterfaceCurve}}
      }
      Turbulence Frequency Mode {
       bIsDynamic=false
       bIsEditable=false
       bIsStaticSwitch=true
       bIsVisible=false
       type {classStructOrEnum {refPath=&ENiagaraWindTurbulenceFrequencyMode}}
      }
      Frequency Min / Max {
       bIsDynamic=false
       bIsEditable=false
       bIsStaticSwitch=false
       bIsVisible=false
       type {classStructOrEnum {refPath=&Vector2f}}
      }
      Frequency Blend {
       bIsDynamic=false
       bIsEditable=false
       bIsStaticSwitch=false
       bIsVisible=false
       type {classStructOrEnum {refPath=&NiagaraFloat}}
      }
      Frequency {
       bIsDynamic=false
       bIsEditable=false
       bIsStaticSwitch=false
       bIsVisible=false
       type {classStructOrEnum {refPath=&NiagaraFloat}}
      }
      Turbulence Offset Mode {
       bIsDynamic=false
       bIsEditable=false
       bIsStaticSwitch=true
       bIsVisible=false
       type {classStructOrEnum {refPath=&ENiagaraWindOffsetMode}}
      }
      Offset By Time {
       bIsDynamic=false
       bIsEditable=false
       bIsStaticSwitch=false
       bIsVisible=false
       type {classStructOrEnum {refPath=&Vector3f}}
      }
      Offset {
       bIsDynamic=false
       bIsEditable=false
       bIsStaticSwitch=false
       bIsVisible=false
       type {classStructOrEnum {refPath=&Vector3f}}
      }
      Turbulence Flow Loop Duration {
       bIsDynamic=false
       bIsEditable=false
       bIsStaticSwitch=false
       bIsVisible=false
       type {classStructOrEnum {refPath=&NiagaraFloat}}
      }
      Noise Quality / Cost {
       bIsDynamic=false
       bIsEditable=false
       bIsStaticSwitch=true
       bIsVisible=false
       type {classStructOrEnum {refPath=&ENiagaraCurlNoiseQuality}}
      }
      Curl Noise {
       bIsDynamic=false
       bIsEditable=false
       bIsStaticSwitch=false
       bIsVisible=false
       type {classStructOrEnum {refPath=&NiagaraDataInterfaceCurlNoise}}
      }
      Time Variable {
       bIsDynamic=false
       bIsEditable=false
       bIsStaticSwitch=false
       bIsVisible=false
       type {classStructOrEnum {refPath=&NiagaraFloat}}
      }
      Surface Alignment Mode {
       bIsDynamic=false
       bIsEditable=false
       bIsStaticSwitch=true
       bIsVisible=false
       type {classStructOrEnum {refPath=&ENiagaraWindCombingMode}}
      }
      Alignment Falloff Start / End {
       bIsDynamic=false
       bIsEditable=false
       bIsStaticSwitch=false
       bIsVisible=false
       type {classStructOrEnum {refPath=&Vector2f}}
      }
      Friction Mode {
       bIsDynamic=false
       bIsEditable=false
       bIsStaticSwitch=true
       bIsVisible=false
       type {classStructOrEnum {refPath=&ENiagaraWindFrictionMode}}
      }
      Friction Distance Mode {
       bIsDynamic=false
       bIsEditable=false
       bIsStaticSwitch=true
       bIsVisible=false
       type {classStructOrEnum {refPath=&ENiagaraWindFrictionDistanceMode}}
      }
      Friction Distance Limit {
       bIsDynamic=false
       bIsEditable=false
       bIsStaticSwitch=false
       bIsVisible=false
       type {classStructOrEnum {refPath=&NiagaraFloat}}
      }
      Friction {
       bIsDynamic=false
       bIsEditable=false
       bIsStaticSwitch=false
       bIsVisible=false
       type {classStructOrEnum {refPath=&NiagaraFloat}}
      }
      Friction Falloff Start / End {
       bIsDynamic=false
       bIsEditable=false
       bIsStaticSwitch=false
       bIsVisible=false
       type {classStructOrEnum {refPath=&Vector2f}}
      }
      Normal Friction Mask {
       bIsDynamic=false
       bIsEditable=false
       bIsStaticSwitch=false
       bIsVisible=false
       type {classStructOrEnum {refPath=&NiagaraFloat}}
      }
      Use Depth Buffer {bIsDynamic=false, bIsEditable=false, bIsStaticSwitch=true, bIsVisible=false, type {classStructOrEnum {refPath=&NiagaraBool}}}
      Use Landscape {bIsDynamic=false, bIsEditable=false, bIsStaticSwitch=true, bIsVisible=false, type {classStructOrEnum {refPath=&NiagaraBool}}}
      Use Ground Mask {bIsDynamic=false, bIsEditable=true, bIsStaticSwitch=true, bIsVisible=true, type {classStructOrEnum {refPath=&NiagaraBool}}}
      Ground Mask Elevation Mode {
       bIsDynamic=false
       bIsEditable=false
       bIsStaticSwitch=true
       bIsVisible=false
       type {classStructOrEnum {refPath=&ENiagaraWindGroundMaskMode}}
      }
      Minimum Ground Elevation {
       bIsDynamic=false
       bIsEditable=false
       bIsStaticSwitch=false
       bIsVisible=false
       type {classStructOrEnum {refPath=&NiagaraFloat}}
      }
      Ground Elevation {
       bIsDynamic=false
       bIsEditable=false
       bIsStaticSwitch=false
       bIsVisible=false
       type {classStructOrEnum {refPath=&NiagaraFloat}}
      }
      Ground Distance Min / Max {
       bIsDynamic=false
       bIsEditable=false
       bIsStaticSwitch=false
       bIsVisible=false
       type {classStructOrEnum {refPath=&Vector2f}}
      }
      Ground Mask Min / Max {
       bIsDynamic=false
       bIsEditable=false
       bIsStaticSwitch=false
       bIsVisible=false
       type {classStructOrEnum {refPath=&Vector2f}}
      }
      Use Physics Field {bIsDynamic=false, bIsEditable=true, bIsStaticSwitch=true, bIsVisible=true, type {classStructOrEnum {refPath=&NiagaraBool}}}
      Physics Field Scale {
       bIsDynamic=false
       bIsEditable=false
       bIsStaticSwitch=false
       bIsVisible=false
       type {classStructOrEnum {refPath=&NiagaraFloat}}
      }
      Position {
       bIsDynamic=false
       bIsEditable=true
       bIsStaticSwitch=false
       bIsVisible=true
       type {classStructOrEnum {refPath=&NiagaraPosition}}
      }
      CollisionQuery {
       bIsDynamic=false
       bIsEditable=false
       bIsStaticSwitch=false
       bIsVisible=false
       type {classStructOrEnum {refPath=&NiagaraDataInterfaceCollisionQuery}}
      }
      Physics Field {
       bIsDynamic=false
       bIsEditable=false
       bIsStaticSwitch=false
       bIsVisible=false
       type {classStructOrEnum {refPath=&NiagaraDataInterfacePhysicsField}}
      }
      Landscape Sample {
       bIsDynamic=false
       bIsEditable=false
       bIsStaticSwitch=false
       bIsVisible=false
       type {classStructOrEnum {refPath=&NiagaraDataInterfaceLandscape}}
      }
     ]
     moduleScript {refPath=&WindForce}
    }
    {
     bIsSetParametersModule=false
     enabled=true
     moduleName=VortexForce
     inputs [
      Vortex Force Amount {bIsDynamic=true, bIsEditable=true, bIsStaticSwitch=false, bIsVisible=true, type {classStructOrEnum {refPath=&NiagaraFloat}}}
      Vortex Axis {bIsDynamic=false, bIsEditable=true, bIsStaticSwitch=false, bIsVisible=true, type {classStructOrEnum {refPath=&Vector3f}}}
      Vortex Axis Coordinate Space {
       bIsDynamic=false
       bIsEditable=true
       bIsStaticSwitch=false
       bIsVisible=true
       type {classStructOrEnum {refPath=&ENiagaraCoordinateSpace}}
      }
      Vortex Influence Position {
       bIsDynamic=false
       bIsEditable=true
       bIsStaticSwitch=false
       bIsVisible=true
       type {classStructOrEnum {refPath=&NiagaraPosition}}
      }
      Write to Intrinsic Parameters {bIsDynamic=false, bIsEditable=true, bIsStaticSwitch=true, bIsVisible=true, type {classStructOrEnum {refPath=&NiagaraBool}}}
      Vortex Origin {
       bIsDynamic=false
       bIsEditable=true
       bIsStaticSwitch=false
       bIsVisible=true
       type {classStructOrEnum {refPath=&NiagaraPosition}}
      }
      Origin Pull Amount {bIsDynamic=false, bIsEditable=true, bIsStaticSwitch=false, bIsVisible=true, type {classStructOrEnum {refPath=&NiagaraFloat}}}
      Origin Pull Only On Vortex Axis {bIsDynamic=false, bIsEditable=true, bIsStaticSwitch=true, bIsVisible=true, type {classStructOrEnum {refPath=&NiagaraBool}}}
      UseOriginOffset {bIsDynamic=false, bIsEditable=true, bIsStaticSwitch=false, bIsVisible=true, type {classStructOrEnum {refPath=&NiagaraBool}}}
      Vortex Origin Offset {bIsDynamic=false, bIsEditable=false, bIsStaticSwitch=false, bIsVisible=true, type {classStructOrEnum {refPath=&Vector3f}}}
      Vortex Origin Offset Coordinate Space {
       bIsDynamic=false
       bIsEditable=false
       bIsStaticSwitch=false
       bIsVisible=false
       type {classStructOrEnum {refPath=&ENiagaraCoordinateSpace}}
      }
      Use Influence Falloff {bIsDynamic=false, bIsEditable=true, bIsStaticSwitch=true, bIsVisible=true, type {classStructOrEnum {refPath=&NiagaraBool}}}
      Influence Falloff Radius {
       bIsDynamic=false
       bIsEditable=false
       bIsStaticSwitch=false
       bIsVisible=false
       type {classStructOrEnum {refPath=&NiagaraFloat}}
      }
      Influence Falloff Exponent {
       bIsDynamic=false
       bIsEditable=false
       bIsStaticSwitch=false
       bIsVisible=false
       type {classStructOrEnum {refPath=&NiagaraFloat}}
      }
      Invert Influence Falloff {bIsDynamic=false, bIsEditable=false, bIsStaticSwitch=false, bIsVisible=false, type {classStructOrEnum {refPath=&NiagaraBool}}}
      Debug Every N Particles {
       bIsDynamic=false
       bIsEditable=false
       bIsStaticSwitch=false
       bIsVisible=false
       type {classStructOrEnum {refPath=&NiagaraInt32}}
      }
      Debug Force Line Color {
       bIsDynamic=false
       bIsEditable=false
       bIsStaticSwitch=false
       bIsVisible=false
       type {classStructOrEnum {refPath=&LinearColor}}
      }
      Debug Force Line Length {
       bIsDynamic=false
       bIsEditable=false
       bIsStaticSwitch=false
       bIsVisible=false
       type {classStructOrEnum {refPath=&NiagaraFloat}}
      }
      Debug Influence Falloff Color {
       bIsDynamic=false
       bIsEditable=false
       bIsStaticSwitch=false
       bIsVisible=false
       type {classStructOrEnum {refPath=&LinearColor}}
      }
     ]
     moduleScript {refPath=&VortexForce}
    }
    {
     bIsSetParametersModule=false
     enabled=true
     moduleName=Drag
     inputs [
      Use Linear Drag {bIsDynamic=false, bIsEditable=true, bIsStaticSwitch=false, bIsVisible=true, type {classStructOrEnum {refPath=&NiagaraBool}}}
      Drag {bIsDynamic=true, bIsEditable=true, bIsStaticSwitch=false, bIsVisible=true, type {classStructOrEnum {refPath=&NiagaraFloat}}}
      Use Rotational Drag {bIsDynamic=false, bIsEditable=true, bIsStaticSwitch=false, bIsVisible=true, type {classStructOrEnum {refPath=&NiagaraBool}}}
      Rotational Drag {bIsDynamic=false, bIsEditable=false, bIsStaticSwitch=false, bIsVisible=true, type {classStructOrEnum {refPath=&NiagaraFloat}}}
      Ignore Mass {bIsDynamic=false, bIsEditable=true, bIsStaticSwitch=false, bIsVisible=true, type {classStructOrEnum {refPath=&NiagaraBool}}}
     ]
     moduleScript {refPath=&Drag}
    }
    {
     bIsSetParametersModule=true
     enabled=true
     moduleName=SetVariables_#4
     inputs [Particles.CameraOffset_LensFlare {bIsDynamic=false, bIsEditable=true, bIsStaticSwitch=false, bIsVisible=true, type {classStructOrEnum {refPath=&NiagaraFloat}}}]
     moduleScript {refPath=&self:Rocks001_0.NiagaraScriptSource_0.NiagaraGraph_0.NiagaraNodeAssignment_0.SetVariables_#4}
    }
    {
     bIsSetParametersModule=false
     enabled=true
     moduleName=SolveForcesAndVelocity
     inputs [
      Clamp Velocity {bIsDynamic=false, bIsEditable=true, bIsStaticSwitch=false, bIsVisible=true, type {classStructOrEnum {refPath=&NiagaraBool}}}
      Speed Limit {bIsDynamic=false, bIsEditable=true, bIsStaticSwitch=false, bIsVisible=true, type {classStructOrEnum {refPath=&NiagaraFloat}}}
      Limit Acceleration {bIsDynamic=false, bIsEditable=true, bIsStaticSwitch=false, bIsVisible=true, type {classStructOrEnum {refPath=&NiagaraBool}}}
      Acceleration Limit {bIsDynamic=false, bIsEditable=false, bIsStaticSwitch=false, bIsVisible=true, type {classStructOrEnum {refPath=&NiagaraFloat}}}
      Manually Enable Rotational Solver {bIsDynamic=false, bIsEditable=true, bIsStaticSwitch=false, bIsVisible=true, type {classStructOrEnum {refPath=&NiagaraBool}}}
      Write to Intrinsic Properties {bIsDynamic=false, bIsEditable=true, bIsStaticSwitch=false, bIsVisible=true, type {classStructOrEnum {refPath=&NiagaraBool}}}
      Write to Presolve Properties {bIsDynamic=false, bIsEditable=true, bIsStaticSwitch=true, bIsVisible=true, type {classStructOrEnum {refPath=&NiagaraBool}}}
      DeltaTime {bIsDynamic=false, bIsEditable=true, bIsStaticSwitch=false, bIsVisible=true, type {classStructOrEnum {refPath=&NiagaraFloat}}}
      Mass {bIsDynamic=false, bIsEditable=true, bIsStaticSwitch=false, bIsVisible=true, type {classStructOrEnum {refPath=&NiagaraFloat}}}
      Position {
       bIsDynamic=false
       bIsEditable=true
       bIsStaticSwitch=false
       bIsVisible=true
       type {classStructOrEnum {refPath=&NiagaraPosition}}
      }
      Velocity {bIsDynamic=false, bIsEditable=true, bIsStaticSwitch=false, bIsVisible=true, type {classStructOrEnum {refPath=&Vector3f}}}
      PreviousVelocity {bIsDynamic=false, bIsEditable=true, bIsStaticSwitch=false, bIsVisible=true, type {classStructOrEnum {refPath=&Vector3f}}}
      Force {bIsDynamic=false, bIsEditable=true, bIsStaticSwitch=false, bIsVisible=true, type {classStructOrEnum {refPath=&Vector3f}}}
      Mesh Orientation {bIsDynamic=false, bIsEditable=true, bIsStaticSwitch=false, bIsVisible=true, type {classStructOrEnum {refPath=&Quat4f}}}
      Rotational Velocity {bIsDynamic=false, bIsEditable=true, bIsStaticSwitch=false, bIsVisible=true, type {classStructOrEnum {refPath=&Vector3f}}}
      Rotational Force {bIsDynamic=false, bIsEditable=true, bIsStaticSwitch=false, bIsVisible=true, type {classStructOrEnum {refPath=&Vector3f}}}
      Rotational Inertia {bIsDynamic=false, bIsEditable=true, bIsStaticSwitch=false, bIsVisible=true, type {classStructOrEnum {refPath=&NiagaraFloat}}}
     ]
     moduleScript {refPath=&SolveForcesAndVelocity}
    }
    {
     bIsSetParametersModule=false
     enabled=true
     moduleName=DynamicMaterialParameters
     inputs [
      Write Parameter Index 0 {
       bIsDynamic=false
       bIsEditable=true
       bIsStaticSwitch=true
       bIsVisible=true
       type {classStructOrEnum {refPath=&ENiagara_DynamicMaterialParametersInputMode}}
      }
      Index 0 Vector 1 {
       bIsDynamic=false
       bIsEditable=false
       bIsStaticSwitch=false
       bIsVisible=false
       type {classStructOrEnum {refPath=&Vector3f}}
      }
      Index 0 Param 1 {
       bIsDynamic=false
       bIsEditable=false
       bIsStaticSwitch=false
       bIsVisible=false
       type {classStructOrEnum {refPath=&NiagaraFloat}}
      }
      Index 0 Param 2 {
       bIsDynamic=false
       bIsEditable=false
       bIsStaticSwitch=false
       bIsVisible=false
       type {classStructOrEnum {refPath=&NiagaraFloat}}
      }
      Index 0 Param 3 {
       bIsDynamic=false
       bIsEditable=false
       bIsStaticSwitch=false
       bIsVisible=false
       type {classStructOrEnum {refPath=&NiagaraFloat}}
      }
      Index 0 Param 4 {
       bIsDynamic=false
       bIsEditable=false
       bIsStaticSwitch=false
       bIsVisible=false
       type {classStructOrEnum {refPath=&NiagaraFloat}}
      }
      Index 0 Vector2D 1 {
       bIsDynamic=false
       bIsEditable=false
       bIsStaticSwitch=false
       bIsVisible=false
       type {classStructOrEnum {refPath=&Vector2f}}
      }
      Index 0 Vector2D 2 {
       bIsDynamic=false
       bIsEditable=false
       bIsStaticSwitch=false
       bIsVisible=false
       type {classStructOrEnum {refPath=&Vector2f}}
      }
      Index 0 Vector 4 {
       bIsDynamic=false
       bIsEditable=false
       bIsStaticSwitch=false
       bIsVisible=false
       type {classStructOrEnum {refPath=&Vector4f}}
      }
      Index 0 Linear Color {
       bIsDynamic=true
       bIsEditable=true
       bIsStaticSwitch=false
       bIsVisible=true
       type {classStructOrEnum {refPath=&LinearColor}}
      }
      Write Parameter Index 1 {
       bIsDynamic=false
       bIsEditable=true
       bIsStaticSwitch=true
       bIsVisible=true
       type {classStructOrEnum {refPath=&ENiagara_DynamicMaterialParametersInputMode}}
      }
      Index 1 Vector 1 {
       bIsDynamic=false
       bIsEditable=false
       bIsStaticSwitch=false
       bIsVisible=false
       type {classStructOrEnum {refPath=&Vector3f}}
      }
      Index 1 Param 1 {
       bIsDynamic=false
       bIsEditable=false
       bIsStaticSwitch=false
       bIsVisible=false
       type {classStructOrEnum {refPath=&NiagaraFloat}}
      }
      Index 1 Param 2 {
       bIsDynamic=false
       bIsEditable=false
       bIsStaticSwitch=false
       bIsVisible=false
       type {classStructOrEnum {refPath=&NiagaraFloat}}
      }
      Index 1 Param 3 {
       bIsDynamic=false
       bIsEditable=false
       bIsStaticSwitch=false
       bIsVisible=false
       type {classStructOrEnum {refPath=&NiagaraFloat}}
      }
      Index 1 Param 4 {
       bIsDynamic=false
       bIsEditable=false
       bIsStaticSwitch=false
       bIsVisible=false
       type {classStructOrEnum {refPath=&NiagaraFloat}}
      }
      Index 1 Vector2D 1 {
       bIsDynamic=false
       bIsEditable=false
       bIsStaticSwitch=false
       bIsVisible=false
       type {classStructOrEnum {refPath=&Vector2f}}
      }
      Index 1 Vector2D 2 {
       bIsDynamic=false
       bIsEditable=false
       bIsStaticSwitch=false
       bIsVisible=false
       type {classStructOrEnum {refPath=&Vector2f}}
      }
      Index 1 Vector 4 {
       bIsDynamic=false
       bIsEditable=false
       bIsStaticSwitch=false
       bIsVisible=false
       type {classStructOrEnum {refPath=&Vector4f}}
      }
      Index 1 Linear Color {
       bIsDynamic=false
       bIsEditable=false
       bIsStaticSwitch=false
       bIsVisible=false
       type {classStructOrEnum {refPath=&LinearColor}}
      }
      Write Parameter Index 2 {
       bIsDynamic=false
       bIsEditable=true
       bIsStaticSwitch=true
       bIsVisible=true
       type {classStructOrEnum {refPath=&ENiagara_DynamicMaterialParametersInputMode}}
      }
      Index 2 Vector 1 {
       bIsDynamic=false
       bIsEditable=false
       bIsStaticSwitch=false
       bIsVisible=false
       type {classStructOrEnum {refPath=&Vector3f}}
      }
      Index 2 Param 1 {
       bIsDynamic=false
       bIsEditable=false
       bIsStaticSwitch=false
       bIsVisible=false
       type {classStructOrEnum {refPath=&NiagaraFloat}}
      }
      Index 2 Param 2 {
       bIsDynamic=false
       bIsEditable=false
       bIsStaticSwitch=false
       bIsVisible=false
       type {classStructOrEnum {refPath=&NiagaraFloat}}
      }
      Index 2 Param 3 {
       bIsDynamic=false
       bIsEditable=false
       bIsStaticSwitch=false
       bIsVisible=false
       type {classStructOrEnum {refPath=&NiagaraFloat}}
      }
      Index 2 Param 4 {
       bIsDynamic=false
       bIsEditable=false
       bIsStaticSwitch=false
       bIsVisible=false
       type {classStructOrEnum {refPath=&NiagaraFloat}}
      }
      Index 2 Vector2D 1 {
       bIsDynamic=false
       bIsEditable=false
       bIsStaticSwitch=false
       bIsVisible=false
       type {classStructOrEnum {refPath=&Vector2f}}
      }
      Index 2 Vector2D 2 {
       bIsDynamic=false
       bIsEditable=false
       bIsStaticSwitch=false
       bIsVisible=false
       type {classStructOrEnum {refPath=&Vector2f}}
      }
      Index 2 Vector 4 {
       bIsDynamic=false
       bIsEditable=false
       bIsStaticSwitch=false
       bIsVisible=false
       type {classStructOrEnum {refPath=&Vector4f}}
      }
      Index 2 Linear Color {
       bIsDynamic=false
       bIsEditable=false
       bIsStaticSwitch=false
       bIsVisible=false
       type {classStructOrEnum {refPath=&LinearColor}}
      }
      Write Parameter Index 3 {
       bIsDynamic=false
       bIsEditable=true
       bIsStaticSwitch=true
       bIsVisible=true
       type {classStructOrEnum {refPath=&ENiagara_DynamicMaterialParametersInputMode}}
      }
      Index 3 Vector 1 {
       bIsDynamic=false
       bIsEditable=false
       bIsStaticSwitch=false
       bIsVisible=false
       type {classStructOrEnum {refPath=&Vector3f}}
      }
      Index 3 Param 1 {
       bIsDynamic=false
       bIsEditable=false
       bIsStaticSwitch=false
       bIsVisible=false
       type {classStructOrEnum {refPath=&NiagaraFloat}}
      }
      Index 3 Param 2 {
       bIsDynamic=false
       bIsEditable=false
       bIsStaticSwitch=false
       bIsVisible=false
       type {classStructOrEnum {refPath=&NiagaraFloat}}
      }
      Index 3 Param 3 {
       bIsDynamic=false
       bIsEditable=false
       bIsStaticSwitch=false
       bIsVisible=false
       type {classStructOrEnum {refPath=&NiagaraFloat}}
      }
      Index 3 Param 4 {
       bIsDynamic=false
       bIsEditable=false
       bIsStaticSwitch=false
       bIsVisible=false
       type {classStructOrEnum {refPath=&NiagaraFloat}}
      }
      Index 3 Vector2D 1 {
       bIsDynamic=false
       bIsEditable=false
       bIsStaticSwitch=false
       bIsVisible=false
       type {classStructOrEnum {refPath=&Vector2f}}
      }
      Index 3 Vector2D 2 {
       bIsDynamic=false
       bIsEditable=false
       bIsStaticSwitch=false
       bIsVisible=false
       type {classStructOrEnum {refPath=&Vector2f}}
      }
      Index 3 Vector 4 {
       bIsDynamic=false
       bIsEditable=false
       bIsStaticSwitch=false
       bIsVisible=false
       type {classStructOrEnum {refPath=&Vector4f}}
      }
      Index 3 Linear Color {
       bIsDynamic=false
       bIsEditable=false
       bIsStaticSwitch=false
       bIsVisible=false
       type {classStructOrEnum {refPath=&LinearColor}}
      }
      Dynamic Parameter Output {
       bIsDynamic=false
       bIsEditable=true
       bIsStaticSwitch=true
       bIsVisible=true
       type {classStructOrEnum {refPath=&ENiagara_AttributeSamplingApplyOutput}}
      }
     ]
     moduleScript {refPath=&DynamicMaterialParameters}
    }
    {
     bIsSetParametersModule=false
     enabled=true
     moduleName=BlendSpace
     inputs [
      Position Mode {bIsDynamic=false, bIsEditable=true, bIsStaticSwitch=true, bIsVisible=true, type {classStructOrEnum {refPath=&NiagaraInt32}}}
      Position Coordinate Space {bIsDynamic=false, bIsEditable=true, bIsStaticSwitch=true, bIsVisible=true, type {classStructOrEnum {refPath=&NiagaraInt32}}}
      "Position Blend (World <-> Local)" {bIsDynamic=true, bIsEditable=true, bIsStaticSwitch=false, bIsVisible=true, type {classStructOrEnum {refPath=&NiagaraFloat}}}
      Mesh Orientation Mode {bIsDynamic=false, bIsEditable=true, bIsStaticSwitch=true, bIsVisible=true, type {classStructOrEnum {refPath=&NiagaraInt32}}}
      Mesh Orientation Coordinate Space {bIsDynamic=false, bIsEditable=false, bIsStaticSwitch=true, bIsVisible=false, type {classStructOrEnum {refPath=&NiagaraInt32}}}
      "Mesh Orientation Blend (World <-> Local)" {
       bIsDynamic=false
       bIsEditable=false
       bIsStaticSwitch=false
       bIsVisible=false
       type {classStructOrEnum {refPath=&NiagaraFloat}}
      }
      Velocity Mode {bIsDynamic=false, bIsEditable=true, bIsStaticSwitch=true, bIsVisible=true, type {classStructOrEnum {refPath=&NiagaraInt32}}}
      Velocity Coordinate Space {bIsDynamic=false, bIsEditable=false, bIsStaticSwitch=true, bIsVisible=false, type {classStructOrEnum {refPath=&NiagaraInt32}}}
      "Velocity Blend (World <-> Local)" {
       bIsDynamic=false
       bIsEditable=false
       bIsStaticSwitch=false
       bIsVisible=false
       type {classStructOrEnum {refPath=&NiagaraFloat}}
      }
     ]
     moduleScript {refPath=&BlendSpace}
    }
    {
     bIsSetParametersModule=false
     enabled=true
     moduleName=KillParticlesInVolume001
     inputs [
      Kill Shape {
       bIsDynamic=false
       bIsEditable=true
       bIsStaticSwitch=true
       bIsVisible=true
       type {classStructOrEnum {refPath=&ENiagaraKillVolumeOptions}}
      }
      Box Size {bIsDynamic=false, bIsEditable=true, bIsStaticSwitch=false, bIsVisible=true, type {classStructOrEnum {refPath=&Vector3f}}}
      Plane Normal {
       bIsDynamic=false
       bIsEditable=false
       bIsStaticSwitch=false
       bIsVisible=false
       type {classStructOrEnum {refPath=&Vector3f}}
      }
      Slab Width {
       bIsDynamic=false
       bIsEditable=false
       bIsStaticSwitch=false
       bIsVisible=false
       type {classStructOrEnum {refPath=&NiagaraFloat}}
      }
      Sphere Radius {
       bIsDynamic=false
       bIsEditable=false
       bIsStaticSwitch=false
       bIsVisible=false
       type {classStructOrEnum {refPath=&NiagaraFloat}}
      }
      Slab Axis {
       bIsDynamic=false
       bIsEditable=false
       bIsStaticSwitch=false
       bIsVisible=false
       type {classStructOrEnum {refPath=&Vector3f}}
      }
      Cone Angle {
       bIsDynamic=false
       bIsEditable=false
       bIsStaticSwitch=false
       bIsVisible=false
       type {classStructOrEnum {refPath=&NiagaraFloat}}
      }
      Cone Axis {
       bIsDynamic=false
       bIsEditable=false
       bIsStaticSwitch=false
       bIsVisible=false
       type {classStructOrEnum {refPath=&Vector3f}}
      }
      Cone Length {
       bIsDynamic=false
       bIsEditable=false
       bIsStaticSwitch=false
       bIsVisible=false
       type {classStructOrEnum {refPath=&NiagaraFloat}}
      }
      Kill Volume Enabled {bIsDynamic=false, bIsEditable=true, bIsStaticSwitch=false, bIsVisible=true, type {classStructOrEnum {refPath=&NiagaraBool}}}
      Invert Volume {bIsDynamic=false, bIsEditable=true, bIsStaticSwitch=false, bIsVisible=true, type {classStructOrEnum {refPath=&NiagaraBool}}}
      Volume Origin {bIsDynamic=false, bIsEditable=true, bIsStaticSwitch=false, bIsVisible=true, type {classStructOrEnum {refPath=&Vector3f}}}
      Origin Offset {bIsDynamic=false, bIsEditable=true, bIsStaticSwitch=false, bIsVisible=true, type {classStructOrEnum {refPath=&Vector3f}}}
      Position {
       bIsDynamic=false
       bIsEditable=true
       bIsStaticSwitch=false
       bIsVisible=true
       type {classStructOrEnum {refPath=&NiagaraPosition}}
      }
     ]
     moduleScript {refPath=&KillParticlesInVolume}
    }
   ]
  }
  rendererClasses [{refPath=&NiagaraMeshRendererProperties}, {refPath=&NiagaraSpriteRendererProperties}]
  renderers [
   {rendererIndex=0, rendererClass {refPath=&NiagaraMeshRendererProperties}}
   {rendererIndex=1, rendererClass {refPath=&NiagaraSpriteRendererProperties}}
  ]
 }
 water_square_plane_vertical {
  emitterSpawnScript {scriptName=EmitterSpawnScript, modules []}
  emitterUpdateScript {
   scriptName=EmitterUpdateScript
   modules [
    {
     bIsSetParametersModule=false
     enabled=true
     moduleName=EmitterState
     inputs [
      Life Cycle Mode {
       bIsDynamic=false
       bIsEditable=true
       bIsStaticSwitch=true
       bIsVisible=true
       type {classStructOrEnum {refPath=&ENiagaraEmitterLifeCycleMode}}
      }
      Inactive Response {
       bIsDynamic=false
       bIsEditable=true
       bIsStaticSwitch=true
       bIsVisible=true
       type {classStructOrEnum {refPath=&ENiagaraInactiveMode}}
      }
      Loop Behavior {
       bIsDynamic=false
       bIsEditable=true
       bIsStaticSwitch=true
       bIsVisible=true
       type {classStructOrEnum {refPath=&ENiagara_EmitterStateOptions}}
      }
      Loop Duration Mode {
       bIsDynamic=false
       bIsEditable=false
       bIsStaticSwitch=true
       bIsVisible=false
       type {classStructOrEnum {refPath=&ENiagara_InfiniteLoopDuration}}
      }
      Loop Count {
       bIsDynamic=false
       bIsEditable=false
       bIsStaticSwitch=false
       bIsVisible=false
       type {classStructOrEnum {refPath=&NiagaraInt32}}
      }
      Loop Duration {bIsDynamic=false, bIsEditable=true, bIsStaticSwitch=false, bIsVisible=true, type {classStructOrEnum {refPath=&NiagaraFloat}}}
      Recalculate Duration Each Loop {bIsDynamic=false, bIsEditable=true, bIsStaticSwitch=false, bIsVisible=true, type {classStructOrEnum {refPath=&NiagaraBool}}}
      UseLoopDelay {bIsDynamic=false, bIsEditable=true, bIsStaticSwitch=false, bIsVisible=true, type {classStructOrEnum {refPath=&NiagaraBool}}}
      Loop Delay {bIsDynamic=false, bIsEditable=false, bIsStaticSwitch=false, bIsVisible=true, type {classStructOrEnum {refPath=&NiagaraFloat}}}
      Delay First Loop Only {bIsDynamic=false, bIsEditable=false, bIsStaticSwitch=false, bIsVisible=false, type {classStructOrEnum {refPath=&NiagaraBool}}}
      Scalability Mode {
       bIsDynamic=false
       bIsEditable=true
       bIsStaticSwitch=true
       bIsVisible=true
       type {classStructOrEnum {refPath=&ENiagaraEmitterScalabilityMode_Limited}}
      }
      Enable Distance Culling {bIsDynamic=false, bIsEditable=false, bIsStaticSwitch=true, bIsVisible=false, type {classStructOrEnum {refPath=&NiagaraBool}}}
      MinDistance {
       bIsDynamic=false
       bIsEditable=false
       bIsStaticSwitch=false
       bIsVisible=false
       type {classStructOrEnum {refPath=&NiagaraFloat}}
      }
      Min Distance Response {
       bIsDynamic=false
       bIsEditable=false
       bIsStaticSwitch=false
       bIsVisible=false
       type {classStructOrEnum {refPath=&ENiagaraExecutionStateManagement}}
      }
      MaxDistance {
       bIsDynamic=false
       bIsEditable=false
       bIsStaticSwitch=false
       bIsVisible=false
       type {classStructOrEnum {refPath=&NiagaraFloat}}
      }
      Max Distance Response {
       bIsDynamic=false
       bIsEditable=false
       bIsStaticSwitch=false
       bIsVisible=false
       type {classStructOrEnum {refPath=&ENiagaraExecutionStateManagement}}
      }
      Scale Spawn Count {bIsDynamic=false, bIsEditable=false, bIsStaticSwitch=true, bIsVisible=false, type {classStructOrEnum {refPath=&NiagaraBool}}}
      Spawn Count Scale {
       bIsDynamic=false
       bIsEditable=false
       bIsStaticSwitch=false
       bIsVisible=false
       type {classStructOrEnum {refPath=&NiagaraFloat}}
      }
      Spawn Count Scale By Distance {
       bIsDynamic=false
       bIsEditable=false
       bIsStaticSwitch=false
       bIsVisible=false
       type {classStructOrEnum {refPath=&NiagaraDataInterfaceCurve}}
      }
      Enable Visibility Culling {bIsDynamic=false, bIsEditable=false, bIsStaticSwitch=true, bIsVisible=false, type {classStructOrEnum {refPath=&NiagaraBool}}}
      Visibility Cull Response {
       bIsDynamic=false
       bIsEditable=false
       bIsStaticSwitch=false
       bIsVisible=false
       type {classStructOrEnum {refPath=&ENiagaraExecutionStateManagement}}
      }
      Visibility Cull Delay {
       bIsDynamic=false
       bIsEditable=false
       bIsStaticSwitch=false
       bIsVisible=false
       type {classStructOrEnum {refPath=&NiagaraFloat}}
      }
      Reset Age On Awaken {bIsDynamic=false, bIsEditable=false, bIsStaticSwitch=false, bIsVisible=false, type {classStructOrEnum {refPath=&NiagaraBool}}}
     ]
     moduleScript {refPath=&EmitterState}
    }
    {
     bIsSetParametersModule=false
     enabled=true
     moduleName=SpawnRate
     inputs [
      SpawnRate {bIsDynamic=false, bIsEditable=true, bIsStaticSwitch=false, bIsVisible=true, type {classStructOrEnum {refPath=&NiagaraFloat}}}
      Use Spawn Probability {bIsDynamic=false, bIsEditable=true, bIsStaticSwitch=false, bIsVisible=true, type {classStructOrEnum {refPath=&NiagaraBool}}}
      Spawn Probability {bIsDynamic=false, bIsEditable=false, bIsStaticSwitch=false, bIsVisible=true, type {classStructOrEnum {refPath=&NiagaraFloat}}}
      SpawnGroup {bIsDynamic=false, bIsEditable=true, bIsStaticSwitch=false, bIsVisible=true, type {classStructOrEnum {refPath=&NiagaraInt32}}}
     ]
     moduleScript {refPath=&SpawnRate}
    }
   ]
  }
  particleSpawnScript {
   scriptName=ParticleSpawnScript
   modules [
    {
     bIsSetParametersModule=false
     enabled=true
     moduleName=InitializeParticle
     inputs [
      Lifetime Mode {
       bIsDynamic=false
       bIsEditable=true
       bIsStaticSwitch=true
       bIsVisible=true
       type {classStructOrEnum {refPath=&ENiagara_LifetimeMode}}
      }
      Lifetime Min {
       bIsDynamic=false
       bIsEditable=false
       bIsStaticSwitch=false
       bIsVisible=false
       type {classStructOrEnum {refPath=&NiagaraFloat}}
      }
      Lifetime {bIsDynamic=true, bIsEditable=true, bIsStaticSwitch=false, bIsVisible=true, type {classStructOrEnum {refPath=&NiagaraFloat}}}
      Lifetime Max {
       bIsDynamic=false
       bIsEditable=false
       bIsStaticSwitch=false
       bIsVisible=false
       type {classStructOrEnum {refPath=&NiagaraFloat}}
      }
      Lifetime Randomness Mode {
       bIsDynamic=false
       bIsEditable=false
       bIsStaticSwitch=true
       bIsVisible=false
       type {classStructOrEnum {refPath=&ENiagaraRandomnessMode}}
      }
      Color Mode {
       bIsDynamic=false
       bIsEditable=true
       bIsStaticSwitch=true
       bIsVisible=true
       type {classStructOrEnum {refPath=&ENiagara_ColorInitializationMode}}
      }
      Color Minimum {
       bIsDynamic=false
       bIsEditable=false
       bIsStaticSwitch=false
       bIsVisible=false
       type {classStructOrEnum {refPath=&LinearColor}}
      }
      Color {
       bIsDynamic=true
       bIsEditable=true
       bIsStaticSwitch=false
       bIsVisible=true
       type {classStructOrEnum {refPath=&LinearColor}}
      }
      AdjustHue {bIsDynamic=false, bIsEditable=false, bIsStaticSwitch=false, bIsVisible=false, type {classStructOrEnum {refPath=&NiagaraBool}}}
      Hue Shift Range {
       bIsDynamic=false
       bIsEditable=false
       bIsStaticSwitch=false
       bIsVisible=false
       type {classStructOrEnum {refPath=&Vector2f}}
      }
      Color Maximum {
       bIsDynamic=false
       bIsEditable=false
       bIsStaticSwitch=false
       bIsVisible=false
       type {classStructOrEnum {refPath=&LinearColor}}
      }
      AdjustSaturation {bIsDynamic=false, bIsEditable=false, bIsStaticSwitch=false, bIsVisible=false, type {classStructOrEnum {refPath=&NiagaraBool}}}
      Saturation Range {
       bIsDynamic=false
       bIsEditable=false
       bIsStaticSwitch=false
       bIsVisible=false
       type {classStructOrEnum {refPath=&Vector2f}}
      }
      AdjustValue {bIsDynamic=false, bIsEditable=false, bIsStaticSwitch=false, bIsVisible=false, type {classStructOrEnum {refPath=&NiagaraBool}}}
      Value Range {
       bIsDynamic=false
       bIsEditable=false
       bIsStaticSwitch=false
       bIsVisible=false
       type {classStructOrEnum {refPath=&Vector2f}}
      }
      AdjustAlpha {bIsDynamic=false, bIsEditable=false, bIsStaticSwitch=false, bIsVisible=false, type {classStructOrEnum {refPath=&NiagaraBool}}}
      Alpha Scale Range {
       bIsDynamic=false
       bIsEditable=false
       bIsStaticSwitch=false
       bIsVisible=false
       type {classStructOrEnum {refPath=&Vector2f}}
      }
      Color Channel Mode {
       bIsDynamic=false
       bIsEditable=false
       bIsStaticSwitch=true
       bIsVisible=false
       type {classStructOrEnum {refPath=&ENiagaraChannelCorrelation}}
      }
      Color Randomness Mode {
       bIsDynamic=false
       bIsEditable=false
       bIsStaticSwitch=true
       bIsVisible=false
       type {classStructOrEnum {refPath=&ENiagaraRandomnessMode}}
      }
      Position Mode {
       bIsDynamic=false
       bIsEditable=true
       bIsStaticSwitch=true
       bIsVisible=true
       type {classStructOrEnum {refPath=&ENiagara_PositionInitializationMode}}
      }
      Position {
       bIsDynamic=false
       bIsEditable=false
       bIsStaticSwitch=false
       bIsVisible=false
       type {classStructOrEnum {refPath=&NiagaraPosition}}
      }
      UsePositionOffset {bIsDynamic=false, bIsEditable=true, bIsStaticSwitch=false, bIsVisible=true, type {classStructOrEnum {refPath=&NiagaraBool}}}
      Position Offset {bIsDynamic=true, bIsEditable=false, bIsStaticSwitch=false, bIsVisible=true, type {classStructOrEnum {refPath=&Vector3f}}}
      Position Offset Coordinate Space {
       bIsDynamic=false
       bIsEditable=false
       bIsStaticSwitch=false
       bIsVisible=false
       type {classStructOrEnum {refPath=&ENiagaraCoordinateSpace}}
      }
      Mass Mode {
       bIsDynamic=false
       bIsEditable=true
       bIsStaticSwitch=true
       bIsVisible=true
       type {classStructOrEnum {refPath=&ENiagara_MassInitializationMode}}
      }
      Mass Min {
       bIsDynamic=false
       bIsEditable=false
       bIsStaticSwitch=false
       bIsVisible=false
       type {classStructOrEnum {refPath=&NiagaraFloat}}
      }
      Mass {
       bIsDynamic=false
       bIsEditable=false
       bIsStaticSwitch=false
       bIsVisible=false
       type {classStructOrEnum {refPath=&NiagaraFloat}}
      }
      Mass Max {
       bIsDynamic=false
       bIsEditable=false
       bIsStaticSwitch=false
       bIsVisible=false
       type {classStructOrEnum {refPath=&NiagaraFloat}}
      }
      Mass Randomness Mode {
       bIsDynamic=false
       bIsEditable=false
       bIsStaticSwitch=true
       bIsVisible=false
       type {classStructOrEnum {refPath=&ENiagaraRandomnessMode}}
      }
      Write Material Random {bIsDynamic=false, bIsEditable=true, bIsStaticSwitch=false, bIsVisible=true, type {classStructOrEnum {refPath=&NiagaraBool}}}
      Material Random {bIsDynamic=false, bIsEditable=true, bIsStaticSwitch=false, bIsVisible=true, type {classStructOrEnum {refPath=&NiagaraFloat}}}
      Sprite Size Mode {
       bIsDynamic=false
       bIsEditable=true
       bIsStaticSwitch=true
       bIsVisible=true
       type {classStructOrEnum {refPath=&ENiagara_SizeScaleMode}}
      }
      Sprite Size {
       bIsDynamic=false
       bIsEditable=false
       bIsStaticSwitch=false
       bIsVisible=false
       type {classStructOrEnum {refPath=&Vector2f}}
      }
      Uniform Sprite Size {
       bIsDynamic=false
       bIsEditable=false
       bIsStaticSwitch=false
       bIsVisible=false
       type {classStructOrEnum {refPath=&NiagaraFloat}}
      }
      Uniform Sprite Size Min {
       bIsDynamic=false
       bIsEditable=false
       bIsStaticSwitch=false
       bIsVisible=false
       type {classStructOrEnum {refPath=&NiagaraFloat}}
      }
      Sprite Size Min {
       bIsDynamic=false
       bIsEditable=false
       bIsStaticSwitch=false
       bIsVisible=false
       type {classStructOrEnum {refPath=&Vector2f}}
      }
      Sprite Size Max {
       bIsDynamic=false
       bIsEditable=false
       bIsStaticSwitch=false
       bIsVisible=false
       type {classStructOrEnum {refPath=&Vector2f}}
      }
      Uniform Sprite Size Max {
       bIsDynamic=false
       bIsEditable=false
       bIsStaticSwitch=false
       bIsVisible=false
       type {classStructOrEnum {refPath=&NiagaraFloat}}
      }
      Sprite Size Randomness Mode {
       bIsDynamic=false
       bIsEditable=false
       bIsStaticSwitch=true
       bIsVisible=false
       type {classStructOrEnum {refPath=&ENiagaraRandomnessMode}}
      }
      Sprite Rotation Mode {
       bIsDynamic=false
       bIsEditable=true
       bIsStaticSwitch=true
       bIsVisible=true
       type {classStructOrEnum {refPath=&ENiagara_SpriteRotationMode}}
      }
      Sprite Rotation Angle Min {
       bIsDynamic=false
       bIsEditable=false
       bIsStaticSwitch=false
       bIsVisible=false
       type {classStructOrEnum {refPath=&NiagaraFloat}}
      }
      Sprite Rotation Angle {
       bIsDynamic=false
       bIsEditable=false
       bIsStaticSwitch=false
       bIsVisible=false
       type {classStructOrEnum {refPath=&NiagaraFloat}}
      }
      Sprite Rotation Angle Max {
       bIsDynamic=false
       bIsEditable=false
       bIsStaticSwitch=false
       bIsVisible=false
       type {classStructOrEnum {refPath=&NiagaraFloat}}
      }
      Sprite Rotation Randomness Mode {
       bIsDynamic=false
       bIsEditable=false
       bIsStaticSwitch=true
       bIsVisible=false
       type {classStructOrEnum {refPath=&ENiagaraRandomnessMode}}
      }
      Sprite UV Mode {
       bIsDynamic=false
       bIsEditable=true
       bIsStaticSwitch=true
       bIsVisible=true
       type {classStructOrEnum {refPath=&ENiagara_UVFlippingMode}}
      }
      Sprite UV Randomness Mode {
       bIsDynamic=false
       bIsEditable=false
       bIsStaticSwitch=true
       bIsVisible=false
       type {classStructOrEnum {refPath=&ENiagaraRandomnessMode}}
      }
      Sprite UV Scale {
       bIsDynamic=false
       bIsEditable=false
       bIsStaticSwitch=false
       bIsVisible=false
       type {classStructOrEnum {refPath=&Vector2f}}
      }
      Mesh Scale Mode {
       bIsDynamic=false
       bIsEditable=true
       bIsStaticSwitch=true
       bIsVisible=true
       type {classStructOrEnum {refPath=&ENiagara_SizeScaleMode}}
      }
      Mesh Uniform Scale Min {
       bIsDynamic=false
       bIsEditable=false
       bIsStaticSwitch=false
       bIsVisible=false
       type {classStructOrEnum {refPath=&NiagaraFloat}}
      }
      Mesh Uniform Scale {
       bIsDynamic=false
       bIsEditable=false
       bIsStaticSwitch=false
       bIsVisible=false
       type {classStructOrEnum {refPath=&NiagaraFloat}}
      }
      Mesh Scale Min {
       bIsDynamic=false
       bIsEditable=false
       bIsStaticSwitch=false
       bIsVisible=false
       type {classStructOrEnum {refPath=&Vector3f}}
      }
      Mesh Scale {bIsDynamic=true, bIsEditable=true, bIsStaticSwitch=false, bIsVisible=true, type {classStructOrEnum {refPath=&Vector3f}}}
      Mesh Scale Max {
       bIsDynamic=false
       bIsEditable=false
       bIsStaticSwitch=false
       bIsVisible=false
       type {classStructOrEnum {refPath=&Vector3f}}
      }
      Mesh Uniform Scale Max {
       bIsDynamic=false
       bIsEditable=false
       bIsStaticSwitch=false
       bIsVisible=false
       type {classStructOrEnum {refPath=&NiagaraFloat}}
      }
      Mesh Scale Randomness Mode {
       bIsDynamic=false
       bIsEditable=false
       bIsStaticSwitch=true
       bIsVisible=false
       type {classStructOrEnum {refPath=&ENiagaraRandomnessMode}}
      }
      Mesh Renderer Array Visibility Mode {
       bIsDynamic=false
       bIsEditable=true
       bIsStaticSwitch=true
       bIsVisible=true
       type {classStructOrEnum {refPath=&ENiagara_UnsetDirectSetRandom}}
      }
      Mesh Renderer Mesh Index {
       bIsDynamic=false
       bIsEditable=false
       bIsStaticSwitch=false
       bIsVisible=false
       type {classStructOrEnum {refPath=&NiagaraInt32}}
      }
      Wrap Index By Mesh Count {bIsDynamic=false, bIsEditable=false, bIsStaticSwitch=true, bIsVisible=false, type {classStructOrEnum {refPath=&NiagaraBool}}}
      Mesh Renderer Info {
       bIsDynamic=false
       bIsEditable=false
       bIsStaticSwitch=false
       bIsVisible=false
       type {classStructOrEnum {refPath=&NiagaraDataInterfaceMeshRendererInfo}}
      }
      Mesh Renderer Random Seed {
       bIsDynamic=false
       bIsEditable=false
       bIsStaticSwitch=false
       bIsVisible=false
       type {classStructOrEnum {refPath=&NiagaraInt32}}
      }
      Ribbon Width Mode {
       bIsDynamic=false
       bIsEditable=true
       bIsStaticSwitch=true
       bIsVisible=true
       type {classStructOrEnum {refPath=&ENiagara_UnsetDirectSet}}
      }
      Ribbon Width {
       bIsDynamic=false
       bIsEditable=false
       bIsStaticSwitch=false
       bIsVisible=false
       type {classStructOrEnum {refPath=&NiagaraFloat}}
      }
      Ribbon Facing Mode {
       bIsDynamic=false
       bIsEditable=true
       bIsStaticSwitch=true
       bIsVisible=true
       type {classStructOrEnum {refPath=&ENiagara_UnsetDirectSet}}
      }
      Ribbon Facing Vector {
       bIsDynamic=false
       bIsEditable=false
       bIsStaticSwitch=false
       bIsVisible=false
       type {classStructOrEnum {refPath=&Vector3f}}
      }
      Ribbon Twist Mode {
       bIsDynamic=false
       bIsEditable=true
       bIsStaticSwitch=true
       bIsVisible=true
       type {classStructOrEnum {refPath=&ENiagara_UnsetDirectSet}}
      }
      "Ribbon Twist " {
       bIsDynamic=false
       bIsEditable=false
       bIsStaticSwitch=false
       bIsVisible=false
       type {classStructOrEnum {refPath=&NiagaraFloat}}
      }
     ]
     moduleScript {refPath=&InitializeParticle}
    }
    {
     bIsSetParametersModule=false
     enabled=true
     moduleName=ShapeLocation
     inputs [
      Shape Primitive {
       bIsDynamic=false
       bIsEditable=true
       bIsStaticSwitch=true
       bIsVisible=true
       type {classStructOrEnum {refPath=&ENiagara_LocationShapes}}
      }
      Box / Plane Mode {
       bIsDynamic=false
       bIsEditable=false
       bIsStaticSwitch=true
       bIsVisible=false
       type {classStructOrEnum {refPath=&ENiagara_BoxPlaneMode}}
      }
      Cone Mode {
       bIsDynamic=false
       bIsEditable=false
       bIsStaticSwitch=true
       bIsVisible=false
       type {classStructOrEnum {refPath=&ENiagara_ConeMode}}
      }
      Cylinder Height {
       bIsDynamic=false
       bIsEditable=false
       bIsStaticSwitch=false
       bIsVisible=false
       type {classStructOrEnum {refPath=&NiagaraFloat}}
      }
      Sphere Radius {
       bIsDynamic=false
       bIsEditable=false
       bIsStaticSwitch=false
       bIsVisible=false
       type {classStructOrEnum {refPath=&NiagaraFloat}}
      }
      Ring / Disc Mode {
       bIsDynamic=false
       bIsEditable=true
       bIsStaticSwitch=true
       bIsVisible=true
       type {classStructOrEnum {refPath=&ENiagaraRingDiscMode}}
      }
      Torus Mode {
       bIsDynamic=false
       bIsEditable=false
       bIsStaticSwitch=true
       bIsVisible=false
       type {classStructOrEnum {refPath=&ENiagaraShapeTorusMode}}
      }
      Knot Ring Loops {
       bIsDynamic=false
       bIsEditable=false
       bIsStaticSwitch=false
       bIsVisible=false
       type {classStructOrEnum {refPath=&NiagaraInt32}}
      }
      Knot Helix Loops {
       bIsDynamic=false
       bIsEditable=false
       bIsStaticSwitch=false
       bIsVisible=false
       type {classStructOrEnum {refPath=&NiagaraInt32}}
      }
      Enable Torus Knot Radius {bIsDynamic=false, bIsEditable=false, bIsStaticSwitch=false, bIsVisible=false, type {classStructOrEnum {refPath=&NiagaraBool}}}
      Torus Knot Tube Radius {
       bIsDynamic=false
       bIsEditable=false
       bIsStaticSwitch=false
       bIsVisible=false
       type {classStructOrEnum {refPath=&NiagaraFloat}}
      }
      Cylinder Radius {
       bIsDynamic=false
       bIsEditable=false
       bIsStaticSwitch=false
       bIsVisible=false
       type {classStructOrEnum {refPath=&NiagaraFloat}}
      }
      Cylinder Height Midpoint {
       bIsDynamic=false
       bIsEditable=false
       bIsStaticSwitch=false
       bIsVisible=false
       type {classStructOrEnum {refPath=&NiagaraFloat}}
      }
      Ring Radius {bIsDynamic=false, bIsEditable=true, bIsStaticSwitch=false, bIsVisible=true, type {classStructOrEnum {refPath=&NiagaraFloat}}}
      Box Size {
       bIsDynamic=false
       bIsEditable=false
       bIsStaticSwitch=false
       bIsVisible=false
       type {classStructOrEnum {refPath=&Vector3f}}
      }
      Box Midpoint {
       bIsDynamic=false
       bIsEditable=false
       bIsStaticSwitch=false
       bIsVisible=false
       type {classStructOrEnum {refPath=&Vector3f}}
      }
      Cone Length {
       bIsDynamic=false
       bIsEditable=false
       bIsStaticSwitch=false
       bIsVisible=false
       type {classStructOrEnum {refPath=&NiagaraFloat}}
      }
      Plane Size {
       bIsDynamic=false
       bIsEditable=false
       bIsStaticSwitch=false
       bIsVisible=false
       type {classStructOrEnum {refPath=&Vector2f}}
      }
      Wedge Horizontal Angle {
       bIsDynamic=false
       bIsEditable=false
       bIsStaticSwitch=false
       bIsVisible=false
       type {classStructOrEnum {refPath=&NiagaraFloat}}
      }
      Large Radius {
       bIsDynamic=false
       bIsEditable=false
       bIsStaticSwitch=false
       bIsVisible=false
       type {classStructOrEnum {refPath=&NiagaraFloat}}
      }
      Plane Midpoint {
       bIsDynamic=false
       bIsEditable=false
       bIsStaticSwitch=false
       bIsVisible=false
       type {classStructOrEnum {refPath=&Vector2f}}
      }
      Wedge Vertical Angle {
       bIsDynamic=false
       bIsEditable=false
       bIsStaticSwitch=false
       bIsVisible=false
       type {classStructOrEnum {refPath=&NiagaraFloat}}
      }
      Cone Angle {
       bIsDynamic=false
       bIsEditable=false
       bIsStaticSwitch=false
       bIsVisible=false
       type {classStructOrEnum {refPath=&NiagaraFloat}}
      }
      Handle Radius {
       bIsDynamic=false
       bIsEditable=false
       bIsStaticSwitch=false
       bIsVisible=false
       type {classStructOrEnum {refPath=&NiagaraFloat}}
      }
      Cone Inner Angle {
       bIsDynamic=false
       bIsEditable=false
       bIsStaticSwitch=false
       bIsVisible=false
       type {classStructOrEnum {refPath=&NiagaraFloat}}
      }
      Cone Radial Angle {
       bIsDynamic=false
       bIsEditable=false
       bIsStaticSwitch=false
       bIsVisible=false
       type {classStructOrEnum {refPath=&NiagaraFloat}}
      }
      Cone Angle Type {
       bIsDynamic=false
       bIsEditable=false
       bIsStaticSwitch=true
       bIsVisible=false
       type {classStructOrEnum {refPath=&ENiagara_AngleInput}}
      }
      Flatten Endcaps {bIsDynamic=false, bIsEditable=false, bIsStaticSwitch=true, bIsVisible=false, type {classStructOrEnum {refPath=&NiagaraBool}}}
      Apply To Particle Position {
       bIsDynamic=false
       bIsEditable=true
       bIsStaticSwitch=true
       bIsVisible=true
       type {classStructOrEnum {refPath=&ENiagara_AttributeSamplingApplyOutput}}
      }
      Cone Surface Distribution {
       bIsDynamic=false
       bIsEditable=false
       bIsStaticSwitch=false
       bIsVisible=false
       type {classStructOrEnum {refPath=&NiagaraFloat}}
      }
      Ring / Disc Distribution Mode {
       bIsDynamic=false
       bIsEditable=true
       bIsStaticSwitch=true
       bIsVisible=true
       type {classStructOrEnum {refPath=&ENiagaraSphereDistributionMode}}
      }
      Sphere Distribution {
       bIsDynamic=false
       bIsEditable=false
       bIsStaticSwitch=true
       bIsVisible=false
       type {classStructOrEnum {refPath=&ENiagaraSphereDistributionMode}}
      }
      Torus Distribution Mode {
       bIsDynamic=false
       bIsEditable=false
       bIsStaticSwitch=true
       bIsVisible=false
       type {classStructOrEnum {refPath=&ENiagaraTorusDistributionMode}}
      }
      Sphere Surface Distribution {
       bIsDynamic=false
       bIsEditable=false
       bIsStaticSwitch=false
       bIsVisible=false
       type {classStructOrEnum {refPath=&NiagaraFloat}}
      }
      Hemisphere Distribution {
       bIsDynamic=false
       bIsEditable=false
       bIsStaticSwitch=false
       bIsVisible=false
       type {classStructOrEnum {refPath=&Vector2f}}
      }
      Cylinder Distribution {
       bIsDynamic=false
       bIsEditable=false
       bIsStaticSwitch=true
       bIsVisible=false
       type {classStructOrEnum {refPath=&ENiagara_CylinderMode}}
      }
      Surface Only {bIsDynamic=false, bIsEditable=false, bIsStaticSwitch=false, bIsVisible=false, type {classStructOrEnum {refPath=&NiagaraBool}}}
      Surface Only Band Thickness {
       bIsDynamic=false
       bIsEditable=false
       bIsStaticSwitch=false
       bIsVisible=false
       type {classStructOrEnum {refPath=&NiagaraFloat}}
      }
      Use Endcaps In Surface Only Mode {bIsDynamic=false, bIsEditable=false, bIsStaticSwitch=true, bIsVisible=false, type {classStructOrEnum {refPath=&NiagaraBool}}}
      Hemicircle X {bIsDynamic=false, bIsEditable=false, bIsStaticSwitch=true, bIsVisible=false, type {classStructOrEnum {refPath=&NiagaraBool}}}
      Hemicircle Y {bIsDynamic=false, bIsEditable=false, bIsStaticSwitch=true, bIsVisible=false, type {classStructOrEnum {refPath=&NiagaraBool}}}
      Disc Coverage {bIsDynamic=false, bIsEditable=true, bIsStaticSwitch=false, bIsVisible=true, type {classStructOrEnum {refPath=&NiagaraFloat}}}
      Hemicircle Internal Cap {bIsDynamic=false, bIsEditable=false, bIsStaticSwitch=true, bIsVisible=false, type {classStructOrEnum {refPath=&NiagaraBool}}}
      U Position {
       bIsDynamic=false
       bIsEditable=false
       bIsStaticSwitch=false
       bIsVisible=false
       type {classStructOrEnum {refPath=&NiagaraFloat}}
      }
      Enable Lathe Profile {bIsDynamic=false, bIsEditable=false, bIsStaticSwitch=true, bIsVisible=false, type {classStructOrEnum {refPath=&NiagaraBool}}}
      Lathe Profile {
       bIsDynamic=false
       bIsEditable=false
       bIsStaticSwitch=false
       bIsVisible=false
       type {classStructOrEnum {refPath=&NiagaraDataInterfaceCurve}}
      }
      Surface Distribution {
       bIsDynamic=false
       bIsEditable=false
       bIsStaticSwitch=false
       bIsVisible=false
       type {classStructOrEnum {refPath=&NiagaraFloat}}
      }
      Surface Expansion Mode {
       bIsDynamic=false
       bIsEditable=false
       bIsStaticSwitch=false
       bIsVisible=false
       type {classStructOrEnum {refPath=&ENiagaraExpansionMode}}
      }
      U Distribution {bIsDynamic=false, bIsEditable=true, bIsStaticSwitch=false, bIsVisible=true, type {classStructOrEnum {refPath=&NiagaraFloat}}}
      V Position {
       bIsDynamic=false
       bIsEditable=false
       bIsStaticSwitch=false
       bIsVisible=false
       type {classStructOrEnum {refPath=&NiagaraFloat}}
      }
      Radius Position {
       bIsDynamic=false
       bIsEditable=false
       bIsStaticSwitch=false
       bIsVisible=false
       type {classStructOrEnum {refPath=&NiagaraFloat}}
      }
      Sphere Radius Position {
       bIsDynamic=false
       bIsEditable=false
       bIsStaticSwitch=false
       bIsVisible=false
       type {classStructOrEnum {refPath=&NiagaraFloat}}
      }
      V Distribution {
       bIsDynamic=false
       bIsEditable=false
       bIsStaticSwitch=false
       bIsVisible=false
       type {classStructOrEnum {refPath=&NiagaraFloat}}
      }
      Evaluation Type {
       bIsDynamic=false
       bIsEditable=false
       bIsStaticSwitch=true
       bIsVisible=false
       type {classStructOrEnum {refPath=&ENiagaraRandomnessEvaluation}}
      }
      Uniform Distribution {
       bIsDynamic=false
       bIsEditable=false
       bIsStaticSwitch=false
       bIsVisible=false
       type {classStructOrEnum {refPath=&NiagaraFloat}}
      }
      Uniform Spiral Amount {
       bIsDynamic=false
       bIsEditable=false
       bIsStaticSwitch=false
       bIsVisible=false
       type {classStructOrEnum {refPath=&NiagaraFloat}}
      }
      Uniform Spiral Falloff {
       bIsDynamic=false
       bIsEditable=false
       bIsStaticSwitch=false
       bIsVisible=false
       type {classStructOrEnum {refPath=&NiagaraFloat}}
      }
      Uniform Count {
       bIsDynamic=false
       bIsEditable=false
       bIsStaticSwitch=false
       bIsVisible=false
       type {classStructOrEnum {refPath=&NiagaraInt32}}
      }
      Uniform Particle Index {
       bIsDynamic=false
       bIsEditable=false
       bIsStaticSwitch=false
       bIsVisible=false
       type {classStructOrEnum {refPath=&NiagaraInt32}}
      }
      Hemisphere Angle Type {
       bIsDynamic=false
       bIsEditable=false
       bIsStaticSwitch=true
       bIsVisible=false
       type {classStructOrEnum {refPath=&ENiagara_AngleInput}}
      }
      Uniform Endpoint Offset {
       bIsDynamic=false
       bIsEditable=false
       bIsStaticSwitch=false
       bIsVisible=false
       type {classStructOrEnum {refPath=&NiagaraFloat}}
      }
      MaskSpawnGroup {bIsDynamic=false, bIsEditable=true, bIsStaticSwitch=false, bIsVisible=true, type {classStructOrEnum {refPath=&NiagaraBool}}}
      Spawn Group Mask {bIsDynamic=false, bIsEditable=false, bIsStaticSwitch=false, bIsVisible=true, type {classStructOrEnum {refPath=&NiagaraInt32}}}
      Spawn Group {
       bIsDynamic=false
       bIsEditable=false
       bIsStaticSwitch=false
       bIsVisible=false
       type {classStructOrEnum {refPath=&NiagaraInt32}}
      }
      Randomness Mode {
       bIsDynamic=false
       bIsEditable=true
       bIsStaticSwitch=true
       bIsVisible=true
       type {classStructOrEnum {refPath=&ENiagaraRandomnessMode}}
      }
      Override Seed {bIsDynamic=false, bIsEditable=true, bIsStaticSwitch=false, bIsVisible=true, type {classStructOrEnum {refPath=&NiagaraBool}}}
      Random Seed {bIsDynamic=false, bIsEditable=false, bIsStaticSwitch=false, bIsVisible=true, type {classStructOrEnum {refPath=&NiagaraInt32}}}
      Fixed Random Seed {bIsDynamic=false, bIsEditable=false, bIsStaticSwitch=true, bIsVisible=false, type {classStructOrEnum {refPath=&NiagaraBool}}}
      Transform Order {
       bIsDynamic=false
       bIsEditable=true
       bIsStaticSwitch=true
       bIsVisible=true
       type {classStructOrEnum {refPath=&ENiagara_TransformOrder}}
      }
      Scale Mode {
       bIsDynamic=false
       bIsEditable=true
       bIsStaticSwitch=true
       bIsVisible=true
       type {classStructOrEnum {refPath=&ENiagara_ScaleMode}}
      }
      Non Uniform Scale {bIsDynamic=false, bIsEditable=true, bIsStaticSwitch=false, bIsVisible=true, type {classStructOrEnum {refPath=&Vector3f}}}
      UseOwnerScale {bIsDynamic=false, bIsEditable=true, bIsStaticSwitch=false, bIsVisible=true, type {classStructOrEnum {refPath=&NiagaraBool}}}
      Apply Owner Scale {bIsDynamic=false, bIsEditable=true, bIsStaticSwitch=false, bIsVisible=true, type {classStructOrEnum {refPath=&Vector3f}}}
      Rotation Mode {
       bIsDynamic=false
       bIsEditable=true
       bIsStaticSwitch=true
       bIsVisible=true
       type {classStructOrEnum {refPath=&ENiagara_RotationMode}}
      }
      Rotation Axis {
       bIsDynamic=false
       bIsEditable=false
       bIsStaticSwitch=false
       bIsVisible=false
       type {classStructOrEnum {refPath=&Vector3f}}
      }
      Rotation Matrix {
       bIsDynamic=false
       bIsEditable=false
       bIsStaticSwitch=false
       bIsVisible=false
       type {classStructOrEnum {refPath=&NiagaraMatrix}}
      }
      Rotation Quaternion {bIsDynamic=false, bIsEditable=false, bIsStaticSwitch=false, bIsVisible=false, type {classStructOrEnum {refPath=&Quat4f}}}
      Yaw / Pitch / Roll {bIsDynamic=false, bIsEditable=true, bIsStaticSwitch=false, bIsVisible=true, type {classStructOrEnum {refPath=&Vector3f}}}
      Rotation Angle {
       bIsDynamic=false
       bIsEditable=false
       bIsStaticSwitch=false
       bIsVisible=false
       type {classStructOrEnum {refPath=&NiagaraFloat}}
      }
      Invert Rotation Quaternion {bIsDynamic=false, bIsEditable=false, bIsStaticSwitch=false, bIsVisible=false, type {classStructOrEnum {refPath=&NiagaraBool}}}
      Additional Quaternion Rotation {bIsDynamic=false, bIsEditable=false, bIsStaticSwitch=false, bIsVisible=false, type {classStructOrEnum {refPath=&NiagaraBool}}}
      Additional Yaw / Pitch / Roll {
       bIsDynamic=false
       bIsEditable=false
       bIsStaticSwitch=false
       bIsVisible=false
       type {classStructOrEnum {refPath=&Vector3f}}
      }
      Rotation Angle Type {
       bIsDynamic=false
       bIsEditable=true
       bIsStaticSwitch=true
       bIsVisible=true
       type {classStructOrEnum {refPath=&ENiagara_AngleInput}}
      }
      Rotation Coordinate Space {
       bIsDynamic=false
       bIsEditable=true
       bIsStaticSwitch=false
       bIsVisible=true
       type {classStructOrEnum {refPath=&ENiagaraCoordinateSpace}}
      }
      Offset Mode {
       bIsDynamic=false
       bIsEditable=true
       bIsStaticSwitch=true
       bIsVisible=true
       type {classStructOrEnum {refPath=&ENiagara_OffsetMode}}
      }
      Shape Origin {
       bIsDynamic=false
       bIsEditable=true
       bIsStaticSwitch=false
       bIsVisible=true
       type {classStructOrEnum {refPath=&NiagaraPosition}}
      }
      Offset {bIsDynamic=false, bIsEditable=true, bIsStaticSwitch=false, bIsVisible=true, type {classStructOrEnum {refPath=&Vector3f}}}
      Offset Coordinate Space {
       bIsDynamic=false
       bIsEditable=true
       bIsStaticSwitch=false
       bIsVisible=true
       type {classStructOrEnum {refPath=&ENiagaraCoordinateSpace}}
      }
      Custom Transform Matrix {
       bIsDynamic=false
       bIsEditable=false
       bIsStaticSwitch=false
       bIsVisible=false
       type {classStructOrEnum {refPath=&NiagaraMatrix}}
      }
      Transform Method {
       bIsDynamic=false
       bIsEditable=true
       bIsStaticSwitch=true
       bIsVisible=true
       type {classStructOrEnum {refPath=&ENiagara_TransformType}}
      }
     ]
     moduleScript {refPath=&ShapeLocation}
    }
    {
     bIsSetParametersModule=false
     enabled=true
     moduleName=AddVelocity
     inputs [
      Velocity Mode {
       bIsDynamic=false
       bIsEditable=true
       bIsStaticSwitch=true
       bIsVisible=true
       type {classStructOrEnum {refPath=&ENiagara_VelocityMode}}
      }
      Velocity {
       bIsDynamic=false
       bIsEditable=false
       bIsStaticSwitch=false
       bIsVisible=false
       type {classStructOrEnum {refPath=&Vector3f}}
      }
      Velocity Speed {bIsDynamic=true, bIsEditable=true, bIsStaticSwitch=false, bIsVisible=true, type {classStructOrEnum {refPath=&NiagaraFloat}}}
      Velocity Speed Scale {
       bIsDynamic=false
       bIsEditable=false
       bIsStaticSwitch=false
       bIsVisible=false
       type {classStructOrEnum {refPath=&NiagaraFloat}}
      }
      Module.Use Falloff Distance {bIsDynamic=false, bIsEditable=true, bIsStaticSwitch=false, bIsVisible=true, type {classStructOrEnum {refPath=&NiagaraBool}}}
      Constrain To Radius {bIsDynamic=false, bIsEditable=false, bIsStaticSwitch=false, bIsVisible=true, type {classStructOrEnum {refPath=&NiagaraFloat}}}
      Distribution Along Cone Axis {
       bIsDynamic=false
       bIsEditable=false
       bIsStaticSwitch=false
       bIsVisible=false
       type {classStructOrEnum {refPath=&NiagaraFloat}}
      }
      Radius Falloff Near / Far {
       bIsDynamic=false
       bIsEditable=false
       bIsStaticSwitch=false
       bIsVisible=false
       type {classStructOrEnum {refPath=&Vector2f}}
      }
      Radius Falloff Exponent {
       bIsDynamic=false
       bIsEditable=false
       bIsStaticSwitch=false
       bIsVisible=false
       type {classStructOrEnum {refPath=&NiagaraFloat}}
      }
      Speed Falloff From Cone Axis {
       bIsDynamic=false
       bIsEditable=false
       bIsStaticSwitch=false
       bIsVisible=false
       type {classStructOrEnum {refPath=&NiagaraFloat}}
      }
      Invert Velocity Falloff {bIsDynamic=false, bIsEditable=false, bIsStaticSwitch=false, bIsVisible=false, type {classStructOrEnum {refPath=&NiagaraBool}}}
      Override Seed {bIsDynamic=false, bIsEditable=true, bIsStaticSwitch=false, bIsVisible=true, type {classStructOrEnum {refPath=&NiagaraBool}}}
      Random Seed {bIsDynamic=false, bIsEditable=false, bIsStaticSwitch=false, bIsVisible=true, type {classStructOrEnum {refPath=&NiagaraInt32}}}
      Fixed Random Seed {bIsDynamic=false, bIsEditable=false, bIsStaticSwitch=true, bIsVisible=false, type {classStructOrEnum {refPath=&NiagaraBool}}}
      Randomness Mode {
       bIsDynamic=false
       bIsEditable=true
       bIsStaticSwitch=true
       bIsVisible=true
       type {classStructOrEnum {refPath=&ENiagaraRandomnessMode}}
      }
      Cone Axis {
       bIsDynamic=false
       bIsEditable=false
       bIsStaticSwitch=false
       bIsVisible=false
       type {classStructOrEnum {refPath=&Vector3f}}
      }
      Cone Angle {
       bIsDynamic=false
       bIsEditable=false
       bIsStaticSwitch=false
       bIsVisible=false
       type {classStructOrEnum {refPath=&NiagaraFloat}}
      }
      Inner Cone Angle {
       bIsDynamic=false
       bIsEditable=false
       bIsStaticSwitch=false
       bIsVisible=false
       type {classStructOrEnum {refPath=&NiagaraFloat}}
      }
      Cone Angle Mode {
       bIsDynamic=false
       bIsEditable=false
       bIsStaticSwitch=true
       bIsVisible=false
       type {classStructOrEnum {refPath=&ENiagara_AngleInput}}
      }
      Rotation Mode {
       bIsDynamic=false
       bIsEditable=false
       bIsStaticSwitch=true
       bIsVisible=false
       type {classStructOrEnum {refPath=&ENiagara_RotationMode}}
      }
      Rotation Axis {
       bIsDynamic=false
       bIsEditable=false
       bIsStaticSwitch=false
       bIsVisible=false
       type {classStructOrEnum {refPath=&Vector3f}}
      }
      Yaw / Pitch / Roll {
       bIsDynamic=false
       bIsEditable=false
       bIsStaticSwitch=false
       bIsVisible=false
       type {classStructOrEnum {refPath=&Vector3f}}
      }
      Rotation Quaternion {bIsDynamic=false, bIsEditable=false, bIsStaticSwitch=false, bIsVisible=false, type {classStructOrEnum {refPath=&Quat4f}}}
      Rotation Matrix {
       bIsDynamic=false
       bIsEditable=false
       bIsStaticSwitch=false
       bIsVisible=false
       type {classStructOrEnum {refPath=&NiagaraMatrix}}
      }
      Rotation Angle {
       bIsDynamic=false
       bIsEditable=false
       bIsStaticSwitch=false
       bIsVisible=false
       type {classStructOrEnum {refPath=&NiagaraFloat}}
      }
      Invert Rotation Quaternion {bIsDynamic=false, bIsEditable=false, bIsStaticSwitch=false, bIsVisible=false, type {classStructOrEnum {refPath=&NiagaraBool}}}
      Additional Quaternion Rotation {bIsDynamic=false, bIsEditable=false, bIsStaticSwitch=false, bIsVisible=false, type {classStructOrEnum {refPath=&NiagaraBool}}}
      Additional Yaw / Pitch / Roll {
       bIsDynamic=false
       bIsEditable=false
       bIsStaticSwitch=false
       bIsVisible=false
       type {classStructOrEnum {refPath=&Vector3f}}
      }
      Rotation Angle Type {
       bIsDynamic=false
       bIsEditable=false
       bIsStaticSwitch=true
       bIsVisible=false
       type {classStructOrEnum {refPath=&ENiagara_AngleInput}}
      }
      Rotation Coordinate Space {
       bIsDynamic=false
       bIsEditable=false
       bIsStaticSwitch=false
       bIsVisible=false
       type {classStructOrEnum {refPath=&ENiagaraCoordinateSpace}}
      }
      Velocity Origin {
       bIsDynamic=false
       bIsEditable=true
       bIsStaticSwitch=false
       bIsVisible=true
       type {classStructOrEnum {refPath=&NiagaraPosition}}
      }
      Origin Offset {bIsDynamic=true, bIsEditable=true, bIsStaticSwitch=false, bIsVisible=true, type {classStructOrEnum {refPath=&Vector3f}}}
      Origin Offset Coordinate Space {
       bIsDynamic=false
       bIsEditable=true
       bIsStaticSwitch=false
       bIsVisible=true
       type {classStructOrEnum {refPath=&ENiagaraCoordinateSpace}}
      }
      Default Position {
       bIsDynamic=false
       bIsEditable=true
       bIsStaticSwitch=false
       bIsVisible=true
       type {classStructOrEnum {refPath=&NiagaraPosition}}
      }
     ]
     moduleScript {refPath=&AddVelocity}
    }
   ]
  }
  particleUpdateScript {
   scriptName=ParticleUpdateScript
   modules [
    {
     bIsSetParametersModule=false
     enabled=true
     moduleName=ParticleState
     inputs [
      Kill Particles When Lifetime Has Elapsed {bIsDynamic=false, bIsEditable=true, bIsStaticSwitch=true, bIsVisible=true, type {classStructOrEnum {refPath=&NiagaraBool}}}
      Loop Particles Lifetime {bIsDynamic=false, bIsEditable=false, bIsStaticSwitch=false, bIsVisible=false, type {classStructOrEnum {refPath=&NiagaraBool}}}
      Let Infinitely Lived Particles Die When Emitter Deactivates {bIsDynamic=false, bIsEditable=false, bIsStaticSwitch=false, bIsVisible=false, type {classStructOrEnum {refPath=&NiagaraBool}}}
      Lifetime {bIsDynamic=false, bIsEditable=true, bIsStaticSwitch=false, bIsVisible=true, type {classStructOrEnum {refPath=&NiagaraFloat}}}
      DeltaTime {bIsDynamic=false, bIsEditable=true, bIsStaticSwitch=false, bIsVisible=true, type {classStructOrEnum {refPath=&NiagaraFloat}}}
     ]
     moduleScript {refPath=&ParticleState}
    }
    {
     bIsSetParametersModule=false
     enabled=true
     moduleName=ScaleMeshSize
     inputs [
      Scale Factor {bIsDynamic=true, bIsEditable=true, bIsStaticSwitch=false, bIsVisible=true, type {classStructOrEnum {refPath=&Vector3f}}}
      Initial Mesh Scale {bIsDynamic=false, bIsEditable=true, bIsStaticSwitch=false, bIsVisible=true, type {classStructOrEnum {refPath=&Vector3f}}}
     ]
     moduleScript {refPath=&ScaleMeshSize}
    }
    {
     bIsSetParametersModule=false
     enabled=true
     moduleName=ScaleColor
     inputs [
      Scale Mode {bIsDynamic=false, bIsEditable=true, bIsStaticSwitch=true, bIsVisible=true, type {classStructOrEnum {refPath=&NiagaraInt32}}}
      Linear Color Curve {
       bIsDynamic=false
       bIsEditable=false
       bIsStaticSwitch=false
       bIsVisible=false
       type {classStructOrEnum {refPath=&NiagaraDataInterfaceColorCurve}}
      }
      ScaleRGB {bIsDynamic=false, bIsEditable=true, bIsStaticSwitch=false, bIsVisible=true, type {classStructOrEnum {refPath=&NiagaraBool}}}
      Scale RGB {bIsDynamic=false, bIsEditable=false, bIsStaticSwitch=false, bIsVisible=true, type {classStructOrEnum {refPath=&Vector3f}}}
      ScaleRGBA {bIsDynamic=false, bIsEditable=false, bIsStaticSwitch=false, bIsVisible=false, type {classStructOrEnum {refPath=&NiagaraBool}}}
      Scale RGBA {
       bIsDynamic=false
       bIsEditable=false
       bIsStaticSwitch=false
       bIsVisible=false
       type {classStructOrEnum {refPath=&Vector4f}}
      }
      Curve Index {
       bIsDynamic=false
       bIsEditable=false
       bIsStaticSwitch=false
       bIsVisible=false
       type {classStructOrEnum {refPath=&NiagaraFloat}}
      }
      ScaleA {bIsDynamic=false, bIsEditable=true, bIsStaticSwitch=false, bIsVisible=true, type {classStructOrEnum {refPath=&NiagaraBool}}}
      Scale Alpha {bIsDynamic=true, bIsEditable=true, bIsStaticSwitch=false, bIsVisible=true, type {classStructOrEnum {refPath=&NiagaraFloat}}}
      Color Value To Scale {
       bIsDynamic=false
       bIsEditable=true
       bIsStaticSwitch=false
       bIsVisible=true
       type {classStructOrEnum {refPath=&LinearColor}}
      }
      Color Scale {
       bIsDynamic=false
       bIsEditable=false
       bIsStaticSwitch=false
       bIsVisible=false
       type {classStructOrEnum {refPath=&LinearColor}}
      }
     ]
     moduleScript {refPath=&ScaleColor}
    }
    {
     bIsSetParametersModule=false
     enabled=true
     moduleName=ScaleColor001
     inputs [
      Scale Mode {bIsDynamic=false, bIsEditable=true, bIsStaticSwitch=true, bIsVisible=true, type {classStructOrEnum {refPath=&NiagaraInt32}}}
      Linear Color Curve {
       bIsDynamic=false
       bIsEditable=false
       bIsStaticSwitch=false
       bIsVisible=false
       type {classStructOrEnum {refPath=&NiagaraDataInterfaceColorCurve}}
      }
      ScaleRGB {bIsDynamic=false, bIsEditable=true, bIsStaticSwitch=false, bIsVisible=true, type {classStructOrEnum {refPath=&NiagaraBool}}}
      Scale RGB {bIsDynamic=false, bIsEditable=false, bIsStaticSwitch=false, bIsVisible=true, type {classStructOrEnum {refPath=&Vector3f}}}
      ScaleRGBA {bIsDynamic=false, bIsEditable=false, bIsStaticSwitch=false, bIsVisible=false, type {classStructOrEnum {refPath=&NiagaraBool}}}
      Scale RGBA {
       bIsDynamic=false
       bIsEditable=false
       bIsStaticSwitch=false
       bIsVisible=false
       type {classStructOrEnum {refPath=&Vector4f}}
      }
      Curve Index {
       bIsDynamic=false
       bIsEditable=false
       bIsStaticSwitch=false
       bIsVisible=false
       type {classStructOrEnum {refPath=&NiagaraFloat}}
      }
      ScaleA {bIsDynamic=false, bIsEditable=true, bIsStaticSwitch=false, bIsVisible=true, type {classStructOrEnum {refPath=&NiagaraBool}}}
      Scale Alpha {bIsDynamic=true, bIsEditable=true, bIsStaticSwitch=false, bIsVisible=true, type {classStructOrEnum {refPath=&NiagaraFloat}}}
      Color Value To Scale {
       bIsDynamic=false
       bIsEditable=true
       bIsStaticSwitch=false
       bIsVisible=true
       type {classStructOrEnum {refPath=&LinearColor}}
      }
      Color Scale {
       bIsDynamic=false
       bIsEditable=false
       bIsStaticSwitch=false
       bIsVisible=false
       type {classStructOrEnum {refPath=&LinearColor}}
      }
     ]
     moduleScript {refPath=&ScaleColor}
    }
    {
     bIsSetParametersModule=false
     enabled=true
     moduleName=Drag
     inputs [
      Use Linear Drag {bIsDynamic=false, bIsEditable=true, bIsStaticSwitch=false, bIsVisible=true, type {classStructOrEnum {refPath=&NiagaraBool}}}
      Drag {bIsDynamic=false, bIsEditable=true, bIsStaticSwitch=false, bIsVisible=true, type {classStructOrEnum {refPath=&NiagaraFloat}}}
      Use Rotational Drag {bIsDynamic=false, bIsEditable=true, bIsStaticSwitch=false, bIsVisible=true, type {classStructOrEnum {refPath=&NiagaraBool}}}
      Rotational Drag {bIsDynamic=false, bIsEditable=false, bIsStaticSwitch=false, bIsVisible=true, type {classStructOrEnum {refPath=&NiagaraFloat}}}
      Ignore Mass {bIsDynamic=false, bIsEditable=true, bIsStaticSwitch=false, bIsVisible=true, type {classStructOrEnum {refPath=&NiagaraBool}}}
     ]
     moduleScript {refPath=&Drag}
    }
    {
     bIsSetParametersModule=false
     enabled=true
     moduleName=SolveForcesAndVelocity
     inputs [
      Clamp Velocity {bIsDynamic=false, bIsEditable=true, bIsStaticSwitch=false, bIsVisible=true, type {classStructOrEnum {refPath=&NiagaraBool}}}
      Speed Limit {bIsDynamic=false, bIsEditable=false, bIsStaticSwitch=false, bIsVisible=true, type {classStructOrEnum {refPath=&NiagaraFloat}}}
      Limit Acceleration {bIsDynamic=false, bIsEditable=true, bIsStaticSwitch=false, bIsVisible=true, type {classStructOrEnum {refPath=&NiagaraBool}}}
      Acceleration Limit {bIsDynamic=false, bIsEditable=false, bIsStaticSwitch=false, bIsVisible=true, type {classStructOrEnum {refPath=&NiagaraFloat}}}
      Manually Enable Rotational Solver {bIsDynamic=false, bIsEditable=true, bIsStaticSwitch=false, bIsVisible=true, type {classStructOrEnum {refPath=&NiagaraBool}}}
      Write to Intrinsic Properties {bIsDynamic=false, bIsEditable=true, bIsStaticSwitch=false, bIsVisible=true, type {classStructOrEnum {refPath=&NiagaraBool}}}
      Write to Presolve Properties {bIsDynamic=false, bIsEditable=true, bIsStaticSwitch=true, bIsVisible=true, type {classStructOrEnum {refPath=&NiagaraBool}}}
      DeltaTime {bIsDynamic=false, bIsEditable=true, bIsStaticSwitch=false, bIsVisible=true, type {classStructOrEnum {refPath=&NiagaraFloat}}}
      Mass {bIsDynamic=false, bIsEditable=true, bIsStaticSwitch=false, bIsVisible=true, type {classStructOrEnum {refPath=&NiagaraFloat}}}
      Position {
       bIsDynamic=false
       bIsEditable=true
       bIsStaticSwitch=false
       bIsVisible=true
       type {classStructOrEnum {refPath=&NiagaraPosition}}
      }
      Velocity {bIsDynamic=false, bIsEditable=true, bIsStaticSwitch=false, bIsVisible=true, type {classStructOrEnum {refPath=&Vector3f}}}
      PreviousVelocity {bIsDynamic=false, bIsEditable=true, bIsStaticSwitch=false, bIsVisible=true, type {classStructOrEnum {refPath=&Vector3f}}}
      Force {bIsDynamic=false, bIsEditable=true, bIsStaticSwitch=false, bIsVisible=true, type {classStructOrEnum {refPath=&Vector3f}}}
      Mesh Orientation {bIsDynamic=false, bIsEditable=false, bIsStaticSwitch=false, bIsVisible=false, type {classStructOrEnum {refPath=&Quat4f}}}
      Rotational Velocity {
       bIsDynamic=false
       bIsEditable=false
       bIsStaticSwitch=false
       bIsVisible=false
       type {classStructOrEnum {refPath=&Vector3f}}
      }
      Rotational Force {
       bIsDynamic=false
       bIsEditable=false
       bIsStaticSwitch=false
       bIsVisible=false
       type {classStructOrEnum {refPath=&Vector3f}}
      }
      Rotational Inertia {
       bIsDynamic=false
       bIsEditable=false
       bIsStaticSwitch=false
       bIsVisible=false
       type {classStructOrEnum {refPath=&NiagaraFloat}}
      }
     ]
     moduleScript {refPath=&SolveForcesAndVelocity}
    }
    {
     bIsSetParametersModule=false
     enabled=true
     moduleName=BlendSpace
     inputs [
      Position Mode {bIsDynamic=false, bIsEditable=true, bIsStaticSwitch=true, bIsVisible=true, type {classStructOrEnum {refPath=&NiagaraInt32}}}
      Position Coordinate Space {bIsDynamic=false, bIsEditable=true, bIsStaticSwitch=true, bIsVisible=true, type {classStructOrEnum {refPath=&NiagaraInt32}}}
      "Position Blend (World <-> Local)" {bIsDynamic=true, bIsEditable=true, bIsStaticSwitch=false, bIsVisible=true, type {classStructOrEnum {refPath=&NiagaraFloat}}}
      Mesh Orientation Mode {bIsDynamic=false, bIsEditable=true, bIsStaticSwitch=true, bIsVisible=true, type {classStructOrEnum {refPath=&NiagaraInt32}}}
      Mesh Orientation Coordinate Space {bIsDynamic=false, bIsEditable=false, bIsStaticSwitch=true, bIsVisible=false, type {classStructOrEnum {refPath=&NiagaraInt32}}}
      "Mesh Orientation Blend (World <-> Local)" {
       bIsDynamic=false
       bIsEditable=false
       bIsStaticSwitch=false
       bIsVisible=false
       type {classStructOrEnum {refPath=&NiagaraFloat}}
      }
      Velocity Mode {bIsDynamic=false, bIsEditable=true, bIsStaticSwitch=true, bIsVisible=true, type {classStructOrEnum {refPath=&NiagaraInt32}}}
      Velocity Coordinate Space {bIsDynamic=false, bIsEditable=false, bIsStaticSwitch=true, bIsVisible=false, type {classStructOrEnum {refPath=&NiagaraInt32}}}
      "Velocity Blend (World <-> Local)" {
       bIsDynamic=false
       bIsEditable=false
       bIsStaticSwitch=false
       bIsVisible=false
       type {classStructOrEnum {refPath=&NiagaraFloat}}
      }
     ]
     moduleScript {refPath=&BlendSpace}
    }
   ]
  }
  rendererClasses [{refPath=&NiagaraMeshRendererProperties}]
  renderers [
   {rendererIndex=0, rendererClass {refPath=&NiagaraMeshRendererProperties}}
   {rendererIndex=1, rendererClass {refPath=&NiagaraMeshRendererProperties}}
  ]
 }
}
summary {
 systemName=NS_Emote_FunkyFeather_energyWater
 emitters [
  {
   emitterName=InfiniteParticleLifetime002
   rendererClasses [{refPath=&NiagaraMeshRendererProperties}, {refPath=&NiagaraSpriteRendererProperties}]
  }
  {emitterName=water_square_plane_vertical, rendererClasses [{refPath=&NiagaraMeshRendererProperties}]}
  {emitterName=PuddleRadial_Trail, rendererClasses [{refPath=&NiagaraMeshRendererProperties}]}
  {
   emitterName=Rocks
   simTarget=GPUComputeSim
   rendererClasses [{refPath=&NiagaraMeshRendererProperties}, {refPath=&NiagaraSpriteRendererProperties}]
  }
  {
   emitterName=Rocks001
   rendererClasses [{refPath=&NiagaraMeshRendererProperties}, {refPath=&NiagaraSpriteRendererProperties}]
  }
  {emitterName=Decal_Caustics, rendererClasses [{refPath=&NiagaraDecalRendererProperties}]}
  {emitterName=Light, rendererClasses [{refPath=&NiagaraLightRendererProperties}]}
  {emitterName=LW_Fog_Axis, rendererClasses []}
  {emitterName=LW_FogHalo, rendererClasses []}
  {emitterName=PuddleRadial001, rendererClasses [{refPath=&NiagaraMeshRendererProperties}]}
 ]
 userVariables [
  User.bIsFrontEnd {
   description=""
   defaultValue {struct {refPath=&NiagaraBool}, value {value=0}}
   type {classStructOrEnum {refPath=&NiagaraBool}}
  }
  User.bIsFrontEndPreview {
   description=""
   defaultValue {struct {refPath=&NiagaraBool}, value {value=0}}
   type {classStructOrEnum {refPath=&NiagaraBool}}
  }
 ]
}
